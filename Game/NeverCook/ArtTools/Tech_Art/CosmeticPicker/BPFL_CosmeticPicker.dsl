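; Function library for the Cosmetic Picker editor utility tool that finds the mannequin actor matching a given cosmetic type.
; For most types it finds the parent picker widget of the supplied widget and reads the mannequin from its mannequin sub-widget.
; The companion type is a special case, resolved through the Advanced picker widget's companion mannequin sub-widget.
; The result is stored in a local variable and returned.
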(fn FindMannequinActorForCosmeticType
  (params (CosmeticType :type CosmeticPicker_CosmeticTypes%20Enum :default "NewEnumerator0") (WidgetObject :type Widget%20Object%20Reference))
  (Utilities|FlowControl|Sequence
    (:then_0
      (if (Utilities|Enum|NotEqual(Enum) CosmeticType "NewEnumerator9")
        (bind _returnvalue (CommonUI|FindParentWidgetOfType (Variables|GetWidgetObject) "/Game/NeverCook/ArtTools/Tech_Art/CosmeticPicker/EUW_CosmeticPickerv2.EUW_CosmeticPickerv2_C"))
        (bind _aseuw_cosmetic_pickerv_2 (Utilities|Casting|CastToEUW_CosmeticPickerv2 _returnvalue))
        (LocalVariables|SetLocal_MannequinActor (Variables|/Game/NeverCook/ArtTools/Tech_Art/CosmeticPicker/EUW_CosmeticPickerv2_Mannequin.EUW_CosmeticPickerv2_Mannequin_C|GetMannequin (Variables|/Game/NeverCook/ArtTools/Tech_Art/CosmeticPicker/EUW_CosmeticPickerv2.EUW_CosmeticPickerv2_C|GetEUW_CosmeticPickerv2_Mannequin _aseuw_cosmetic_pickerv_2)))
        (else
          ;[0.9542 1 0.9694 0.5] Special case for the companion mannequin
          (bind _returnvalue_2 (CommonUI|FindParentWidgetOfType (Variables|GetWidgetObject) "/Game/NeverCook/ArtTools/Tech_Art/CosmeticPicker/EUW_CosmeticPickerv2_Advanced.EUW_CosmeticPickerv2_Advanced_C"))
          (bind _aseuw_cosmetic_pickerv_2_advanced (Utilities|Casting|CastToEUW_CosmeticPickerv2_Advanced _returnvalue_2))
          (LocalVariables|SetLocal_MannequinActor (Variables|/Game/NeverCook/ArtTools/Tech_Art/CosmeticPicker/EUW_CosmeticPickerv2_CompanionMannequin.EUW_CosmeticPickerv2_CompanionMannequin_C|GetCompanionMannequin (Variables|/Game/NeverCook/ArtTools/Tech_Art/CosmeticPicker/EUW_CosmeticPickerv2_Advanced.EUW_CosmeticPickerv2_Advanced_C|GetEUW_CosmeticPickerv2_CompanionMannequin _aseuw_cosmetic_pickerv_2_advanced))))))
    (:then_1
      (return (LocalVariables|GetLocal_MannequinActor)))))
