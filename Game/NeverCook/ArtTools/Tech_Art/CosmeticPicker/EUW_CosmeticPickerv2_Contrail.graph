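# Editor utility widget for picking an Athena skydive contrail in the Cosmetic Picker tech-art tool.
# On construct it reads the current contrail from the Python CosmeticPickerModule into a variable and shows only its Contrail category in a DetailsView of the widget itself.
# A property-change delegate is bound to the view, and the contrailChanged handler passes the selected contrail back to Python via setContrail.
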EditorUtilityWidgetBlueprint /Game/NeverCook/ArtTools/Tech_Art/CosmeticPicker/EUW_CosmeticPickerv2_Contrail
graphs [
 EventGraph {nodes [
   Event Construct()
   > PrintString(InString=HelloTrail)
   > n1 = getContrail!(execute, self) -> (then, name:Athena Sky Dive Contrail Item Definition Object Reference)
   > Set contrail(contrail:Athena Sky Dive Contrail Item Definition Object Reference=@n1) -> (Output_Get:Athena Sky Dive Contrail Item Definition Object Reference)
   > SetEditorProperty(
    Object=Get DetailsView_contrail() -> (DetailsView_contrail:Details View Object Reference)
    PropertyName=CategoriesToShow
    PropertyValue:Array of Names=Make Array("[0]":Name=String To Name(InString=Contrail)) -> (Array:Array of Names)
   )
   n2 = Get DetailsView_contrail()
   > SetObject(self=@n2, NewObject=Self-Reference())
   > Bind Event to On Property Changed(self:Property View Base Object Reference=@n2, Delegate=Create Event()) {is=title K2Node_AddDelegate}
   Event PreConstruct()
   Event Tick()
  ]}
 contrailChanged {nodes [
   contrailChanged!() -> (then, PropertyName:Name) {is=title K2Node_FunctionEntry}
   > PrintString(InString=ContrailChanged)
   > setContrail!(
    execute
    self
    inContrail:Athena Sky Dive Contrail Item Definition Object Reference=Get contrail() -> (contrail:Athena Sky Dive Contrail Item Definition Object Reference)
   ) -> (then)
  ]}
 getContrail {nodes [
   getContrail() {is=title K2Node_FunctionEntry}
   > n1 = Execute Python Script!(execute, PythonScript="import TechArt.CosmeticPicker.CosmeticPickerModule as CosmeticPicker\r\ncontrail = CosmeticPicker.getContrail()") -> (then, ReturnValue, contrail:Athena Sky Dive Contrail Item Definition Object Reference) {outputs [contrail]}
   > Return Node!(execute, name:Athena Sky Dive Contrail Item Definition Object Reference=@n1.contrail)
  ]}
 setContrail {nodes [
   n1 = setContrail!() -> (then, inContrail:Athena Sky Dive Contrail Item Definition Object Reference) {is=title K2Node_FunctionEntry}
   > Execute Python Script!(
    execute
    PythonScript="import TechArt.CosmeticPicker.CosmeticPickerModule as CosmeticPicker\r\nCosmeticPicker.setContrail(contrail)"
    contrail:Athena Sky Dive Contrail Item Definition Object Reference=@n1
   ) -> (then, ReturnValue) {inputs [contrail]}
   > Return Node!(execute)
  ]}
]
variables [contrail {type=AthenaSkyDiveContrailItemDefinition, default=null}]
widgets {
 info {rootWidgetClass=&CanvasPanel, widgetCount=3}
 widgets [
  CanvasPanel_0 {widgetClassPath=&CanvasPanel, widget=CanvasPanel_0}
  VerticalBox_26 {
   widgetClassPath=&VerticalBox
   parent=CanvasPanel_0
   widget=VerticalBox_26
   properties {clipping=ClipToBounds}
   slot {
    class=CanvasPanelSlot
    properties {bAutoSize=true, layoutData {alignment=<0 0>, anchors {maximum=<1 1>, minimum=<0 0>}, offsets=<ltrb 0 0 0 0>}}
   }
  }
  DetailsView_contrail {
   widgetClassPath=&DetailsView
   parent=VerticalBox_26
   widget=DetailsView_contrail
   slot {class=VerticalBoxSlot}
  }
 ]
}
