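# Editor utility widget: the Cosmetic Picker V2 Global tab, for artist-facing editor-wide settings.
# Combo boxes set the global Time of Day Manager class and the r.SkeletalMeshLODBias CVar via console command.
# Construct syncs both boxes to current state, storing an unlisted manager in slot 0; IsInitialSetup? suppresses change handlers meanwhile.
# A button runs inline Python to synchronously compile uncompiled shoe CustomizableObjects, then shows a completion dialog.
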EditorUtilityWidgetBlueprint /Game/NeverCook/ArtTools/Tech_Art/CosmeticPicker/EUW_CosmeticPickerV2_GlobalTab
graphs [
 EventGraph {nodes [
   "On Selection Changed (TODComboBox)"!() -> (OutputDelegate, then, SelectedItem:String, SelectionType:ESelectInfo Enum) {is=title K2Node_ComponentBoundEvent}
   > n1 = Branch(Condition=Get IsInitialSetup?() -> (IsInitialSetup?:bool))
   n2 = "Get (a copy)"(
    Array:Array of Object Class References=Get TOD_Mangers() -> (TOD_Mangers:Array of Object Class References)
    Dimension 1=GetSelectedIndex!(self:"Combo Box (String) Object Reference"=Get TODComboBox() -> (TODComboBox:Editor Utility Combo Box String Object Reference)) -> (ReturnValue:int=0)
   ) -> (Output:Object Class Reference)
   n3 = Branch(execute=@n1.else, Condition=IsValidClass!(Class:Object Class Reference=@n2) -> (ReturnValue:bool))
   > n4 = Cast To Info Class!(execute:Exec, Class:Object Class Reference=@n2) -> (then:Exec, CastFailed:Exec, AsInfo:Info Class Reference) {is=title K2Node_ClassDynamicCast}
   > SetGlobalTimeOfDayManager!(
    execute
    WorldContextObject:Object Reference
    InTimeOfDayManager:Info Soft Class Reference=Make Soft Reference!(Input:Info Class Reference=@n4) -> (Output:Info Soft Class Reference)
   ) -> (then)
   PrintString(execute=@n3.else, InString=Bad TODM Class Reference)
   n5 = "On Selection Changed (LODBiasComboBox)"!() -> (OutputDelegate, then, SelectedItem:String, SelectionType:ESelectInfo Enum) {is=title K2Node_ComponentBoundEvent}
   > n6 = Branch(Condition=Get IsInitialSetup?())
   n7 = Switch on String!(execute:Exec=@n6.else, Selection:String=@n5.SelectedItem) -> (Default:Exec, Auto:Exec, Force 0:Exec, 1:Exec, 2:Exec, 3:Exec, 4:Exec) {pinNames [Auto, Force 0, 1, 2, 3, 4]}
   SetLODBias!(execute=@n7.1, self, LODBias:int=1) -> (then)
   SetLODBias(execute=@n7.2, LODBias=2)
   SetLODBias(execute=@n7.3, LODBias=3)
   SetLODBias(execute=@n7.4, LODBias=4)
   SetLODBias(execute=@n7.Auto, LODBias=0)
   SetLODBias(execute=@n7.Force 0, LODBias=-10)
   "On Clicked (EditorUtilityButton_CompileShoes)"() {is=title K2Node_ComponentBoundEvent}
   > Execute Python Script(PythonScript="import unreal\r\n\r\nasset_registry = unreal.AssetRegistryHelpers.get_asset_registry()\r\nshoe_path = \"/CosmeticShoes/Assets/Mutable/COs\"\r\n\r\nassets = asset_registry.get_assets_by_path(unreal.Name(shoe_path))\r\n\r\nfor asset in assets:\r\n    if asset.asset_class_path.asset_name == unreal.Name(\"CustomizableObject\"):       \r\n        co_asset = asset.get_asset()\r\n        if not co_asset.is_compiled():\r\n            unreal.CustomizableObjectEditorFunctionLibrary.compile_customizable_object_synchronously(co_asset)\r\n            print(f\"Customizable Object|CO is now Compiled: {co_asset.get_name()}\")\r\n        else:\r\n            print(f\"Customizable Object|CO is Already Compiled: {co_asset.get_name()}\")")
   > Show Message Dialog(
    Title=MakeLiteralText(Value="LOCTEXT(\"Compile Shoes\")")
    Message=MakeLiteralText(Value="LOCTEXT(\"Finished Compiling Shoes.\\r\\n\\r\\nCheck the Output Log for Details.\")")
    MessageType=Ok
    DefaultValue=Ok
    MessageCategory=Info
   )
   Event Construct()
   > n8 = Sequence!(execute) -> (then_0, then_1, then_2:Exec)
   n9 = Get Class(Object="To Object (Interface)"!(Interface:"Interface Interface (by ref)"=GetGlobalTimeOfDayManager!(WorldContextObject:Object Reference) -> (ReturnValue:Fort Time Of Day Manager Interface Interface)) -> (ReturnValue:Object Reference)) -> (ReturnValue:Object Class Reference)
   n10 = Get TOD_Mangers()
   n11 = Find Item!(TargetArray:Array of Object Class References=@n10, ItemToFind:"Object Class Reference (by ref)"=@n9) -> (ReturnValue)
   > n12 = Branch(Condition="integer >= integer"(A=@n11))
   > SetSelectedIndex!(execute, self:"Combo Box (String) Object Reference"=Get TODComboBox(), Index:int=@n11) -> (then)
   Set Array Elem!(
    execute=@n12.else
    TargetArray:Array of Object Class References=@n10
    Index:int=0
    Item:"Object Class Reference (by ref)"=@n9
    bSizeToFit:bool=false
   ) -> (then)
   Set IsInitialSetup?(execute=@n8.then_2, IsInitialSetup?:bool=false) -> (Output_Get:bool)
   n13 = Truncate(A=GetNumericCVar!(CVar:String=r.SkeletalMeshLODBias) -> (OutValue:float, ReturnValue:bool))
   n14 = Branch(execute=@n8.then_1, Condition="integer < integer"(A=@n13))
   n15 = Get LODBiasComboBox() -> (LODBiasComboBox:Editor Utility Combo Box String Object Reference)
   > SetSelectedIndex(self=@n15, Index=1)
   n16 = Switch on Int!(execute:Exec=@n14.else, Selection:int=@n13) -> (Default:Exec, 0:Exec, 1:Exec, 2:Exec, 3:Exec, 4:Exec)
   SetSelectedIndex(execute=[@n16.0, @n16.Default], self=@n15, Index=0)
   SetSelectedIndex(execute=@n16.1, self=@n15, Index=2)
   SetSelectedIndex(execute=@n16.2, self=@n15, Index=3)
   SetSelectedIndex(execute=@n16.3, self=@n15, Index=4)
   SetSelectedIndex(execute=@n16.4, self=@n15, Index=5)
   Event PreConstruct()
   Event Tick()
  ]}
 SetLODBias {nodes [
   n1 = SetLODBias!() -> (then, LODBias:int) {is=title K2Node_FunctionEntry}
   > ExecuteConsoleCommand(Command=Append(A="r.SkeletalMeshLODBias ", B="To String (Integer)"(InInt=@n1)))
  ]}
]
variables [
 IsInitialSetup? {type=boolean, default=true}
 TOD_Mangers {
  type="array<Class>"
  default [
   null
   /Game/TimeOfDay/TODM_Sequence/Blueprint/BP_DaySequence_Current.BP_DaySequence_Current_C
   /Game/TimeOfDay/TODM_Sequence/Blueprint/BP_DaySequence_CharacterReview.BP_DaySequence_CharacterReview_C
  ]
 }
]
widgets {
 info {rootWidgetClass=&Overlay, widgetCount=15}
 widgets [
  Overlay_0 {widgetClassPath=&Overlay, widget=Overlay_0}
  VerticalBox_27 {
   widgetClassPath=&VerticalBox
   parent=Overlay_0
   widget=VerticalBox_27
   slot {class=OverlaySlot, properties {horizontalAlignment=HAlign_Fill, verticalAlignment=VAlign_Fill}}
  }
  TODSelector {widgetClassPath=&VerticalBox, parent=VerticalBox_27, widget=TODSelector, slot {class=VerticalBoxSlot}}
  TextBlock_43 {
   widgetClassPath=&TextBlock
   parent=TODSelector
   widget=TextBlock_43
   properties {
    text=TODM:
    font {
     bForceMonospaced=false
     bMaterialIsStencil=false
     fontMaterial=null
     fontObject=&Roboto
     letterSpacing=0
     monospacedWidth=1
     size=15
     skewAmount=0
     typefaceFontName=Bold
     outlineSettings {
      bApplyOutlineToDropShadows=false
      bMiteredCorners=false
      bSeparateFillAlpha=false
      outlineMaterial=null
      outlineSize=0
      outlineColor=<rgba 0 0 0 1>
     }
    }
   }
   slot {class=VerticalBoxSlot}
  }
  TODComboBox {
   widgetClassPath=/Script/Blutility.EditorUtilityComboBoxString
   bIsVariable=true
   parent=TODSelector
   widget=TODComboBox
   properties {
    selectedOption=Unknown Initial TODM Loaded
    defaultOptions [Unknown Initial TODM Loaded, DaySequence_Current, Character]
    foregroundColor {colorUseRule=UseColor_Specified, specifiedColor=<rgba 0.5271 0.5271 0.5271 1>}
   }
   slot {class=VerticalBoxSlot}
  }
  Spacer_95 {
   widgetClassPath=&Spacer
   parent=VerticalBox_27
   widget=Spacer_95
   slot {class=VerticalBoxSlot, properties {padding=<ltrb 1 10 1 1>}}
  }
  LODToggle {widgetClassPath=&VerticalBox, parent=VerticalBox_27, widget=LODToggle, slot {class=VerticalBoxSlot}}
  TextBlock {
   widgetClassPath=&TextBlock
   parent=LODToggle
   widget=TextBlock
   properties {
    text=Skel Mesh LOD Bias Toggle:
    font {
     bForceMonospaced=false
     bMaterialIsStencil=false
     fontMaterial=null
     fontObject=&Roboto
     letterSpacing=0
     monospacedWidth=1
     size=15
     skewAmount=0
     typefaceFontName=Bold
     outlineSettings {
      bApplyOutlineToDropShadows=false
      bMiteredCorners=false
      bSeparateFillAlpha=false
      outlineMaterial=null
      outlineSize=0
      outlineColor=<rgba 0 0 0 1>
     }
    }
   }
   slot {class=VerticalBoxSlot}
  }
  LODBiasComboBox {
   widgetClassPath=&EditorUtilityComboBoxString
   bIsVariable=true
   parent=LODToggle
   widget=LODBiasComboBox
   properties {
    selectedOption=Unknown Initial Loaded
    defaultOptions [Auto, Force 0, 1, 2, 3, 4]
    foregroundColor {colorUseRule=UseColor_Specified, specifiedColor=<rgba 0.5271 0.5271 0.5271 1>}
   }
   slot {class=VerticalBoxSlot}
  }
  Spacer {widgetClassPath=&Spacer, parent=VerticalBox_27, widget=Spacer, slot {class=VerticalBoxSlot, properties {padding=<ltrb 1 10 1 1>}}}
  CompileShoes {widgetClassPath=&VerticalBox, parent=VerticalBox_27, widget=CompileShoes, slot {class=VerticalBoxSlot}}
  TextBlock_1 {
   widgetClassPath=&TextBlock
   parent=CompileShoes
   widget=TextBlock_1
   properties {
    text=Shoes:
    font {
     bForceMonospaced=false
     bMaterialIsStencil=false
     fontMaterial=null
     fontObject=&Roboto
     letterSpacing=0
     monospacedWidth=1
     size=15
     skewAmount=0
     typefaceFontName=Bold
     outlineSettings {
      bApplyOutlineToDropShadows=false
      bMiteredCorners=false
      bSeparateFillAlpha=false
      outlineMaterial=null
      outlineSize=0
      outlineColor=<rgba 0 0 0 1>
     }
    }
   }
   slot {class=VerticalBoxSlot}
  }
  HorizontalBox_87 {widgetClassPath=&HorizontalBox, parent=CompileShoes, widget=HorizontalBox_87, slot {class=VerticalBoxSlot}}
  EditorUtilityButton_CompileShoes {
   widgetClassPath=&EditorUtilityButton
   parent=HorizontalBox_87
   widget=EditorUtilityButton_CompileShoes
   properties {toolTipText=Compile All Shoes}
   slot {class=HorizontalBoxSlot}
  }
  TextBlock_2 {
   widgetClassPath=&TextBlock
   parent=EditorUtilityButton_CompileShoes
   widget=TextBlock_2
   properties {
    text=Compile All Shoes
    font {
     bForceMonospaced=false
     bMaterialIsStencil=false
     fontMaterial=null
     fontObject=&Roboto
     letterSpacing=0
     monospacedWidth=1
     size=15
     skewAmount=0
     typefaceFontName=Bold
     outlineSettings {
      bApplyOutlineToDropShadows=false
      bMiteredCorners=false
      bSeparateFillAlpha=false
      outlineMaterial=null
      outlineSize=0
      outlineColor=<rgba 0 0 0 1>
     }
    }
   }
   slot {class=ButtonSlot, properties {horizontalAlignment=HAlign_Fill, verticalAlignment=VAlign_Fill, padding=<ltrb 0 0 0 0>}}
  }
 ]
}
