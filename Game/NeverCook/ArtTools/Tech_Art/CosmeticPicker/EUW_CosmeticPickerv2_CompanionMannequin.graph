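# Editor utility widget panel of the Cosmetic Picker v2 tool for picking and updating a Fort Companion Mannequin in the level.
# A details view exposes the widget's own Companions properties; on a mannequin change it clears the parent picker's variant channels and passes the companion definition to its companion ID picker.
# Updating applies variant overrides through the Python CosmeticPicker module, to either the selected mannequin actors or the stored mannequin.
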EditorUtilityWidgetBlueprint /Game/NeverCook/ArtTools/Tech_Art/CosmeticPicker/EUW_CosmeticPickerv2_CompanionMannequin
graphs [
 EventGraph {nodes [
   Event Construct()
   > SetEditorProperty(
    Object=Get DetailsView_CompanionMannequin() -> (DetailsView_CompanionMannequin:Details View Object Reference)
    PropertyName=CategoriesToShow
    PropertyValue:Array of Names=Make Array("[0]":Name=String To Name(InString=Companions)) -> (Array:Array of Names)
   )
   n1 = Get DetailsView_CompanionMannequin()
   > SetObject(self=@n1, NewObject=Self-Reference())
   > Bind Event to On Property Changed(self:Property View Base Object Reference=@n1, Delegate=Create Event()) {is=title K2Node_AddDelegate}
  ]}
 MannequinChanged {nodes [
   MannequinChanged!() -> (then, PropertyName:Name) {is=title K2Node_FunctionEntry}
   > PrintString(InString=MannequinChanged)
   > n1 = FindParentWidgetOfType!(
    execute
    StartingWidget:Widget Object Reference=Self-Reference()
    Type:Widget Class Reference=&/EUW_CosmeticPickerv2_Advanced.EUW_CosmeticPickerv2_Advanced_C
   ) -> (then, ReturnValue:EUW Cosmetic Pickerv 2 Advanced Object Reference)
   > n2 = Cast To EUW_CosmeticPickerv2_Advanced!(execute, Object=@n1) -> (then, CastFailed, AsEUW Cosmetic Pickerv 2 Advanced:EUW Cosmetic Pickerv 2 Advanced Object Reference) {is=title K2Node_DynamicCast}
   > ClearCompanionVariantChannels(self:EUW Cosmetic Pickerv 2 Advanced Object Reference=@n1)
   > n3 = Get!(execute:Exec) -> (then:Exec, else:Exec, CompanionMannequin:Fort Companion Mannequin Object Reference)
   n4 = Get CompanionItemDefinition!(self:Fort Companion Mannequin Object Reference=@n3) -> (CompanionItemDefinition:Cosmetic Companion Item Definition Object Reference)
   n5 = GetDisplayName!(self:Item Definition Base Object Reference=@n4) -> (ReturnValue:Text)
   > PrintString(InString=Append(A=mannequin companion:, B="To String (Text)"(InText=@n5)))
   n6 = Get EUW_CosmeticPicker_ID_Companions!(self:EUW Cosmetic Pickerv 2 Advanced Object Reference=@n2) -> (EUW_CosmeticPicker_ID_Companions:EUW Cosmetic Pickerv 2 ID Object Reference)
   > Set Companion!(
    execute
    Companion:Cosmetic Companion Item Definition Soft Object Reference=Make Soft Reference!(Input:Cosmetic Companion Item Definition Object Reference=@n4) -> (Output:Cosmetic Companion Item Definition Soft Object Reference)
    self:EUW Cosmetic Pickerv 2 ID Object Reference=@n6
   ) -> (then, Output_Get:Cosmetic Companion Item Definition Soft Object Reference)
   > IDChanged!(execute, self:EUW Cosmetic Pickerv 2 ID Object Reference=@n6, PropertyName:Name=None) -> (then)
  ]}
 SetCompanionMannequin {nodes [
   n1 = SetCompanionMannequin!() -> (then, CompanionMannequin:Fort Companion Mannequin Object Reference) {is=title K2Node_FunctionEntry}
   > Set CompanionMannequin(CompanionMannequin:Fort Companion Mannequin Object Reference=@n1) -> (Output_Get:Fort Companion Mannequin Object Reference)
   > MannequinChanged!(execute, self, PropertyName:Name=None) -> (then) {is=title K2Node_CallFunction}
  ]}
 SetMannequinOverrideVariants {nodes [
   n1 = SetMannequinOverrideVariants!() -> (then, mannequin:Fort Companion Mannequin Object Reference) {is=title K2Node_FunctionEntry}
   > Execute Python Script!(
    execute
    PythonScript="import TechArt.CosmeticPicker.CosmeticPickerModule as CosmeticPicker\r\n\r\nfrom importlib import *\r\nreload(CosmeticPicker)\r\n\r\nCosmeticPicker.setMannequinOverrideVariants(mannequin)"
    mannequin:Fort Companion Mannequin Object Reference=@n1
   ) -> (then, ReturnValue) {inputs [mannequin]}
   > Return Node()
  ]}
 clearMannequinSlot {nodes [
   clearMannequinSlot() {is=title K2Node_FunctionEntry}
   > Set CompanionMannequin(CompanionMannequin=Get Mannequin_clear() -> (Mannequin_clear:Fort Companion Mannequin Object Reference))
  ]}
 updateMannequin {nodes [
   updateMannequin() {is=title K2Node_FunctionEntry}
   > PrintString(InString=UpdateCompanionMannequin)
   > n1 = Branch(Condition=Get UpdateSelected?() -> (UpdateSelected?:bool))
   > n2 = GetSelectedActors()
   > n3 = For Each Loop(Array:Array of Actor Object References=@n2) -> (Array Element:Actor Object Reference)
   > n4 = Cast To FortCompanionMannequin!(execute, Object=@n3) -> (then, CastFailed, AsFort Companion Mannequin:Fort Companion Mannequin Object Reference) {is=title K2Node_DynamicCast}
   > SetMannequinOverrideVariants!(execute, self, mannequin:Fort Companion Mannequin Object Reference=@n4) -> (then) {is=title K2Node_CallFunction}
   n5 = Get CompanionMannequin() -> (CompanionMannequin:Fort Companion Mannequin Object Reference)
   Is Valid(exec=@n1.else, InputObject=@n5)
   > SetMannequinOverrideVariants(mannequin=@n5)
  ]}
]
variables [
 CompanionMannequin {type=FortCompanionMannequin, default=null}
 Mannequin_clear {type=FortCompanionMannequin, default=null}
 UpdateSelected? {type=boolean}
]
widgets {
 info {rootWidgetClass=&CanvasPanel, widgetCount=3}
 widgets [
  CanvasPanel_0 {widgetClassPath=&CanvasPanel, widget=CanvasPanel_0}
  VerticalBox_26 {
   widgetClassPath=&VerticalBox
   parent=CanvasPanel_0
   widget=VerticalBox_26
   properties {clipping=ClipToBounds}
   slot {
    class=CanvasPanelSlot
    properties {bAutoSize=true, layoutData {alignment=<0 0>, anchors {maximum=<1 0>, minimum=<0 0>}, offsets=<ltrb 0 0 0 100>}}
   }
  }
  DetailsView_CompanionMannequin {
   widgetClassPath=&DetailsView
   parent=VerticalBox_26
   widget=DetailsView_CompanionMannequin
   properties {renderTransform {angle=0, scale=<1 1>, shear=<0 0>, translation=<0 -25>}}
   slot {class=VerticalBoxSlot}
  }
 ]
}
