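# Editor utility widget for the tech art Cosmetic Picker that chooses the emotes to test.
# Two buttons or a details view set the current data table, whose rows become the emote list, also editable in a details view.
# List changes go to the Python CosmeticPicker module and are saved as soft references.
# Post-load restores table and list, or uses the default table without a save, then binds details-view events.
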EditorUtilityWidgetBlueprint /Game/NeverCook/ArtTools/Tech_Art/CosmeticPicker/EUW_CosmeticPickerv2_Emotes
graphs [
 EventGraph {nodes [
   "On Clicked (btnSetEmoteTestDefaults)"!() -> (OutputDelegate, then, widget:User Widget Object Reference, event:Pointer Event Structure) {is=title K2Node_ComponentBoundEvent}
   > SetCurrentDataTable!(execute, self, EmoteTableCurrent:Data Table Object Reference=Get EmoteTable_Default() -> (EmoteTable_Default:Data Table Object Reference)) -> (then)
   "On Clicked (btnSetEmoteTestTechAnim)"!() -> (OutputDelegate, then, widget:User Widget Object Reference, event:Pointer Event Structure) {is=title K2Node_ComponentBoundEvent}
   > SetCurrentDataTable(EmoteTableCurrent=Get EmoteTable_TechAnim() -> (EmoteTable_TechAnim:Data Table Object Reference))
   "On Property Changed (DetailsView_Table)"!() -> (OutputDelegate, then, PropertyName:Name=None) {is=title K2Node_ComponentBoundEvent}
   > HandleDetailsViewChanged_DataTable!(execute, self, property:Name=None) -> (then)
   Event Construct()
   > Parent: Construct()
   > InitDetailsView()
  ]}
 HandleDetailsViewChanged_DataTable {nodes [
   HandleDetailsViewChanged_DataTable!() -> (then, property:Name) {is=title K2Node_FunctionEntry}
   > LoadEmotesFromTable()
  ]}
 HandleDetailsViewChanged_Emotes {nodes [
   HandleDetailsViewChanged_Emotes!() -> (then, PropertyName:Name) {is=title K2Node_FunctionEntry}
   > SaveGame_Save(self:TA Base Editor Widget Object Reference)
   > UpdateCosmeticEmotes()
  ]}
 InitDetailsView {nodes [
   InitDetailsView() {is=title K2Node_FunctionEntry}
   > SetObject(
    self=[
     Get DetailsView_Emotes() -> (DetailsView_Emotes:Details View Object Reference)
     Get DetailsView_Table() -> (DetailsView_Table:Details View Object Reference)
    ]
    NewObject=Self-Reference()
   )
  ]}
 InitDetailsViewEvents {nodes [
   InitDetailsViewEvents() {is=title K2Node_FunctionEntry}
   > Bind Event to On Property Changed(self:Property View Base Object Reference=Get DetailsView_Table(), Delegate=Create Event()) {is=title K2Node_AddDelegate}
   > Bind Event to On Property Changed(self=Get DetailsView_Emotes(), Delegate=Create Event())
  ]}
 LoadEmotesFromTable {nodes [
   LoadEmotesFromTable() {is=title K2Node_FunctionEntry}
   n1 = Get EmoteTableCurrent() -> (EmoteTableCurrent:Data Table Object Reference)
   > n2 = GetDataTableRowNames(Table=@n1)
   > n3 = For Each Loop(Array:Array of Names=@n2) -> (Array Element:Name)
   SetEmoteList!(
    execute=@n3.Completed
    self
    emotes:Array of Athena Dance Item Definition Object References=Get l_newEmotes() -> (l_newEmotes:Array of Athena Dance Item Definition Object References)
   ) -> (then)
   n4 = Get Data Table Row!(execute:Exec=@n3.LoopBody, DataTable:Data Table Object Reference=@n1, RowName:Name=@n3) -> (then:Exec, RowNotFound:Exec, ReturnValue:Emote List Item Structure)
   > n5 = LoadAsset_Blocking(Asset=Break Emote List Item!(EmoteListItem:"Emote List Item Structure (by ref)"=@n4) -> (Emote_2_#1:Athena Dance Item Definition Soft Object Reference) {is=title K2Node_BreakStruct})
   > n6 = Cast To AthenaDanceItemDefinition!(execute, Object=@n5) -> (then, CastFailed, AsAthena Dance Item Definition:Athena Dance Item Definition Object Reference) {is=title K2Node_DynamicCast}
   > Add(
    TargetArray:Array of Athena Dance Item Definition Object References=Get l_newEmotes()
    NewItem:"Athena Dance Item Definition Object Reference (by ref)"=@n6
   ) {is=title K2Node_CallArrayFunction}
  ]}
 SaveGame_PostLoad {nodes [
   n1 = SaveGame_PostLoad!() -> (then, saveGameExisted:bool, saveGame_Object:Save Game Object Reference) {is=title K2Node_FunctionEntry}
   > Parent: SaveGame_PostLoad!(execute, saveGameExisted:bool=@n1, saveGame_Object:Save Game Object Reference=@n1.saveGame_Object) -> (then, " ":bool) {is=title K2Node_CallParentFunction}
   > n2 = Branch(Condition=@n1)
   > n3 = Cast To SG_CosmeticPickerEmote!(execute, Object=@n1.saveGame_Object) -> (then, CastFailed, AsSG Cosmetic Picker Emote:SG Cosmetic Picker Emote Object Reference) {is=title K2Node_DynamicCast}
   > n4 = Sequence!(execute) -> (then_0, then_1, then_2:Exec)
   > n5 = LoadAsset_Blocking(Asset=Get EmoteTable_TechAnim!(self:SG Cosmetic Picker Emote Object Reference=@n3) -> (EmoteTable_TechAnim:Data Table Soft Object Reference))
   > n6 = Cast To DataTable!(execute, Object=@n5) -> (then, CastFailed, AsData Table:Data Table Object Reference) {is=title K2Node_DynamicCast}
   > Set EmoteTableCurrent(EmoteTableCurrent:Data Table Object Reference=@n6) -> (Output_Get:Data Table Object Reference)
   n7 = For Each Loop(
    Exec=@n4.then_1
    Array:Array of Athena Dance Item Definition Soft Object References=Get emotes!(self:SG Cosmetic Picker Emote Object Reference=@n3) -> (emotes:Array of Athena Dance Item Definition Soft Object References)
   ) -> (Array Element:Athena Dance Item Definition Soft Object Reference)
   SetEmoteList(execute=@n7.Completed, emotes=Get l_emoteList() -> (l_emoteList:Array of Athena Dance Item Definition Object References))
   n8 = LoadAsset_Blocking(execute=@n7.LoopBody, Asset=@n7)
   > n9 = Cast To AthenaDanceItemDefinition(Object=@n8)
   > Add(TargetArray=Get l_emoteList(), NewItem=@n9)
   InitDetailsViewEvents(execute=@n4.then_2) {is=title K2Node_CallFunction}
   SetCurrentDataTable(execute=@n2.else, EmoteTableCurrent=Get EmoteTable_Default())
   > InitDetailsViewEvents()
   Return Node!(execute, " ":bool=false)
  ]}
 SaveGame_PrepareToSave {nodes [
   n1 = SaveGame_PrepareToSave!() -> (then, SaveGameObject:Save Game Object Reference) {is=title K2Node_FunctionEntry}
   > n2 = Cast To SG_CosmeticPickerEmote(Object=@n1)
   > Set l_save(l_save:SG Cosmetic Picker Emote Object Reference=@n2) -> (Output_Get:SG Cosmetic Picker Emote Object Reference)
   > n3 = Sequence()
   > n4 = For Each Loop(Array:Array of Athena Dance Item Definition Object References=Get emotes!() -> (emotes:Array of Athena Dance Item Definition Object References)) -> (Array Element:Athena Dance Item Definition Object Reference)
   Set emotes!(
    execute=@n4.Completed
    emotes:Array of Athena Dance Item Definition Soft Object References=Get l_emoteList() -> (l_emoteList:Array of Athena Dance Item Definition Soft Object References)
    self:SG Cosmetic Picker Emote Object Reference=Get l_save() -> (l_save:SG Cosmetic Picker Emote Object Reference)
   ) -> (then, Output_Get:Array of Athena Dance Item Definition Soft Object References)
   Add(
    execute=@n4.LoopBody
    TargetArray:Array of Athena Dance Item Definition Soft Object References=Get l_emoteList()
    NewItem:"Athena Dance Item Definition Soft Object Reference (by ref)"=Make Soft Reference!(Input:Athena Dance Item Definition Object Reference=@n4) -> (Output:Athena Dance Item Definition Soft Object Reference)
   )
   Set with Broadcast EmoteTable_TechAnim!(
    execute=@n3.then_1
    EmoteTable_TechAnim:Data Table Soft Object Reference=Make Soft Reference(Input:Data Table Object Reference=Get EmoteTableCurrent()) -> (Output:Data Table Soft Object Reference)
    self:SG Cosmetic Picker Emote Object Reference=Get l_save()
   ) -> (then, Output_Get:Data Table Soft Object Reference)
   Return Node(" "=false)
  ]}
 SetCurrentDataTable {nodes [
   n1 = SetCurrentDataTable!() -> (then, EmoteTableCurrent:Data Table Object Reference) {is=title K2Node_FunctionEntry}
   > Set EmoteTableCurrent(EmoteTableCurrent=@n1)
   > HandleDetailsViewChanged_DataTable(property=None) {is=title K2Node_CallFunction}
  ]}
 SetEmoteList {nodes [
   n1 = SetEmoteList!() -> (then, emotes:Array of Athena Dance Item Definition Object References) {is=title K2Node_FunctionEntry}
   > Set emotes!(execute, emotes:Array of Athena Dance Item Definition Object References=@n1) -> (then, Output_Get:Array of Athena Dance Item Definition Object References)
   > UpdateCosmeticEmotes()
   > SaveGame_Save()
  ]}
 UpdateCosmeticEmotes {nodes [
   UpdateCosmeticEmotes() {is=title K2Node_FunctionEntry}
   > Execute Python Script!(
    execute
    PythonScript="import TechArt.CosmeticPicker.CosmeticPickerModule as CosmeticPicker\r\nCosmeticPicker.setEmotes(emotes)"
    emotes:Array of Athena Dance Item Definition Object References=Get emotes()
   ) -> (then, ReturnValue) {inputs [emotes]}
  ]}
]
variables [
 emotes {type="array<AthenaDanceItemDefinition>", default []}
 EmoteTable_Default {type=DataTable, default=&/EmotesForTesting_TechArt}
 EmoteTable_TechAnim {type=DataTable, default=&/EmotesForTesting_TechAnim}
 EmoteTableCurrent {type=DataTable, default=null}
]
widgets {
 info {
  parentClass=/Game/Blueprints/Blutilities/ForTechArt/Widgets/TA_Base_EditorWidget.TA_Base_EditorWidget_C
  rootWidgetClass=&CanvasPanel
  widgetCount=7
 }
 widgets [
  CanvasPanel_0 {widgetClassPath=&CanvasPanel, widget=CanvasPanel_0}
  VerticalBox_26 {
   widgetClassPath=&VerticalBox
   parent=CanvasPanel_0
   widget=VerticalBox_26
   properties {clipping=ClipToBounds}
   slot {
    class=CanvasPanelSlot
    properties {bAutoSize=true, layoutData {alignment=<0 0>, anchors {maximum=<1 1>, minimum=<0 0>}, offsets=<ltrb 0 0 0 0>}}
   }
  }
  DetailsView_Emotes {
   widgetClassPath=&DetailsView
   parent=VerticalBox_26
   widget=DetailsView_Emotes
   properties {propertiesToShow [emotes]}
   slot {class=VerticalBoxSlot}
  }
  DetailsView_Table {
   widgetClassPath=&DetailsView
   parent=VerticalBox_26
   widget=DetailsView_Table
   properties {propertiesToShow [EmoteTableCurrent]}
   slot {class=VerticalBoxSlot}
  }
  HorizontalBox_60 {
   widgetClassPath=&HorizontalBox
   parent=VerticalBox_26
   widget=HorizontalBox_60
   slot {class=VerticalBoxSlot, properties {horizontalAlignment=HAlign_Center}}
  }
  btnSetEmoteTestDefaults {
   widgetClassPath=&TA_Generic01_Button
   parent=HorizontalBox_60
   widget=btnSetEmoteTestDefaults
   properties {label=Set Emote Testing Defaults, buttonColor_Default=<rgba 0.1302 0.1302 0.1302 1>}
   slot {class=HorizontalBoxSlot, properties {horizontalAlignment=HAlign_Center, padding=<ltrb 0 0 5 0>}}
  }
  btnSetEmoteTestTechAnim {
   widgetClassPath=&TA_Generic01_Button
   parent=HorizontalBox_60
   widget=btnSetEmoteTestTechAnim
   properties {label=Set Emotes Testing TechAnim, buttonColor_Default=<rgba 0.1302 0.1302 0.1302 1>}
   slot {class=HorizontalBoxSlot, properties {horizontalAlignment=HAlign_Center, padding=<ltrb 5 0 0 0>}}
  }
 ]
}
