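# Editor utility panel, the advanced part of Cosmetic Picker v2, for picking wraps, emotes, contrails and companions and debugging companions.
# Child picker widgets handle selection, with SetContrail and setWrap forwarding items to them.
# Buttons send console commands, mostly to the first PIE world, for contrail visibility, companion state and mannequin refresh.
# Freeze instead toggles actor tick and mesh animation pause, while helper functions recolor labels and Python clears variant channels.
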EditorUtilityWidgetBlueprint /Game/NeverCook/ArtTools/Tech_Art/CosmeticPicker/EUW_CosmeticPickerv2_Advanced
graphs [
 ClearCompanionVariantChannels {nodes [
   ClearCompanionVariantChannels() {is=title K2Node_FunctionEntry}
   > Execute Python Script(PythonScript="import TechArt.CosmeticPicker.CosmeticPickerModule as CosmeticPicker\r\nCosmeticPicker.clearCharacterVariantChannels()\r\n")
  ]}
 EventGraph {nodes [
   "On Clicked (EditorUtilityButton_CompanionMood)"() {is=title K2Node_ComponentBoundEvent}
   > n1 = GetPIEWorlds!(execute, bIncludeDedicatedServer:bool=false) -> (then, ReturnValue:Array of World Object References)
   > Branch(Condition=Is Not Empty!(TargetArray:Array of World Object References=@n1) -> (ReturnValue:bool))
   n2 = GetSelectedOption!(self:"Combo Box (String) Object Reference"=Get EditorUtilityComboBoxString_CompanionMood() -> (EditorUtilityComboBoxString_CompanionMood:Editor Utility Combo Box String Object Reference)) -> (ReturnValue:String)
   > ExecuteConsoleCommand(
    WorldContextObject="Get (a copy)"(Array:Array of World Object References=@n1) -> (Output:World Object Reference)
    Command=Append(A="Cheat Fort.PlayerCompanion.ForcedMood ", B=@n2)
   )
   > PrintString(InString=Append(A="Companion Mood ", B=@n2))
   "On Clicked (EditorUtilityButton_95)"() {is=title K2Node_ComponentBoundEvent}
   > n3 = GetPIEWorlds(bIncludeDedicatedServer=false)
   > n4 = Flip Flop!(None:Exec) -> (A:Exec, B:Exec, IsA:bool=~) {is=title K2Node_MacroInstance}
   n5 = "Get (a copy)"(Array=@n3)
   > ExecuteConsoleCommand(WorldContextObject=@n5, Command=SetAlwaysShowContrails 1)
   ExecuteConsoleCommand(execute=@n4.B, WorldContextObject=@n5, Command=SetAlwaysShowContrails 0)
   "On Clicked (EditorUtilityButton_CompanionFreeze)"() {is=title K2Node_ComponentBoundEvent}
   > n6 = GetPIEWorlds(bIncludeDedicatedServer=false)
   > Branch(Condition=Is Not Empty(TargetArray=@n6))
   > n7 = GetAllActorsOfClass!(execute, WorldContextObject:Object Reference="Get (a copy)"(Array=@n6), ActorClass:Actor Class Reference) -> (then, OutActors:Array of Actor Object References)
   n8 = "Get (a copy)"(Array:Array of Actor Object References=@n7) -> (Output:Actor Object Reference)
   > n9 = Branch(Condition=IsActorTickEnabled!(self:Actor Object Reference=@n8) -> (ReturnValue:bool))
   > SetActorTickEnabled!(execute, self:Actor Object Reference=@n8, bEnabled:bool=false) -> (then)
   n10 = Get Mesh(self:Object Reference)
   > Set bPauseAnims!(execute, bPauseAnims:bool=true, self:Skeletal Mesh Component Object Reference=@n10) -> (then, Output_Get:bool)
   > ToggleCompanionFreeze!(execute, self, Freeze:bool=true) -> (then)
   > PrintString(InString=Companion FROZEN)
   SetActorTickEnabled(execute=@n9.else, self=@n8, bEnabled=true)
   > Set bPauseAnims(bPauseAnims=false, self=@n10)
   > ToggleCompanionFreeze(Freeze=false)
   > PrintString(InString=Companion UNFROZEN)
   "On Clicked (EditorUtilityButton_CompanionForceStay)"() {is=title K2Node_ComponentBoundEvent}
   > n11 = GetPIEWorlds(bIncludeDedicatedServer=false)
   > Branch(Condition=Is Not Empty(TargetArray=@n11))
   > n12 = Flip Flop() -> (IsA=~)
   n13 = "Get (a copy)"(Array:Array of World Object References=@n11) -> (Output:World Object Reference)
   > ExecuteConsoleCommand(WorldContextObject=@n13, Command=CompanionComeAndStay 1)
   > PrintString(InString=Companion Come and Stay ENABLED)
   ExecuteConsoleCommand(execute=@n12.B, WorldContextObject=@n13, Command=CompanionComeAndStay 0)
   > PrintString(InString=Companion Come and Stay DISABLED)
   "On Clicked (EditorUtilityButton_CompanionToggle)"() {is=title K2Node_ComponentBoundEvent}
   > n14 = GetPIEWorlds(bIncludeDedicatedServer=false)
   > Branch(Condition=Is Not Empty(TargetArray=@n14))
   > n15 = Branch(Condition="Equal (Float)"(A=GetNumericCVar!(CVar:String=Fort.PlayerCompanion.Enabled) -> (OutValue:float, ReturnValue:bool)))
   n16 = "Get (a copy)"(Array=@n14)
   > ExecuteConsoleCommand(WorldContextObject=@n16, Command=Cheat Fort.PlayerCompanion.Enabled 1)
   > ToggleCompanionEnable!(execute, self, isEnabled:bool=true) -> (then)
   > PrintString(InString=Companions ENABLED)
   ExecuteConsoleCommand(execute=@n15.else, WorldContextObject=@n16, Command=Cheat Fort.PlayerCompanion.Enabled 0)
   > ToggleCompanionEnable(isEnabled=false)
   > PrintString(InString=Companions DISABLED)
   "On Clicked (EditorUtilityButton_RefreshAllMannequins)"() {is=title K2Node_ComponentBoundEvent}
   > ExecuteConsoleCommand(Command=Fort.CompanionMannequin.RefreshAllMannequinsInEditorWorld)
   > PrintString(InString=RefreshedAllMannequinsInEditorWorld)
   "On Clicked (EditorUtilityButton_CompanionForceInteract)"() {is=title K2Node_ComponentBoundEvent}
   > n17 = GetPIEWorlds(bIncludeDedicatedServer=false)
   > Branch(Condition=Is Not Empty(TargetArray=@n17))
   > ExecuteConsoleCommand(WorldContextObject="Get (a copy)"(Array=@n17), Command=cheat Fort.PlayerCompanion.ForcePlayInteract)
   "On Clicked (EditorUtilityButton_CompanionForceEmote)"() {is=title K2Node_ComponentBoundEvent}
   > n18 = GetPIEWorlds(bIncludeDedicatedServer=false)
   > Branch(Condition=Is Not Empty(TargetArray=@n18))
   > ExecuteConsoleCommand(WorldContextObject="Get (a copy)"(Array=@n18), Command=cheat Fort.PlayerCompanion.ForcePlayEmote)
   Event Construct()
   > SetupCompanionWidgets()
  ]}
 Get_TextBlock_ForceMood_ColorAndOpacity {nodes [
   Get_TextBlock_ForceMood_ColorAndOpacity() {is=title K2Node_FunctionEntry}
   > n1 = GetConsoleVariableStringValue!(execute, VariableName:String=Fort.PlayerCompanion.ForcedMood) -> (then, ReturnValue:String)
   > n2 = Branch(Condition=OR Boolean(A="Equal, Case Insensitive (String)"(A=@n1), B="Equal, Case Insensitive (String)"(A=@n1, B=None)))
   > Return Node!(execute, ReturnValue:Slate Color Structure=Make Slate Color(SpecifiedColor=(R=1,G=0,B=1,A=1), ColorUseRule=UseColor_Foreground))
   Return Node(execute=@n2.else, ReturnValue=Make Slate Color(SpecifiedColor=(R=0.1204,G=1,B=0,A=1)))
  ]}
 NewEventDispatcher_0 {nodes [NewEventDispatcher_0() {is=title K2Node_FunctionEntry}]}
 SetContrail {nodes [
   n1 = SetContrail!() -> (then, Contrail:Object Reference) {is=title K2Node_FunctionEntry}
   > n2 = Cast To AthenaSkyDiveContrailItemDefinition!(execute, Object=@n1) -> (then, CastFailed, AsAthena Sky Dive Contrail Item Definition:Athena Sky Dive Contrail Item Definition Object Reference) {is=title K2Node_DynamicCast}
   n3 = Get EUW_CosmeticPicker_ID_Contrail() -> (EUW_CosmeticPicker_ID_Contrail:EUW Cosmetic Pickerv 2 ID Object Reference)
   > n4 = Set contrail!(execute, contrail:Athena Sky Dive Contrail Item Definition Object Reference=@n2, self:EUW Cosmetic Pickerv 2 ID Object Reference=@n3) -> (then, Output_Get:Athena Sky Dive Contrail Item Definition Object Reference)
   > SetOverride!(
    execute
    self:EUW Cosmetic Pickerv 2 ID Object Reference=@n3
    id:Fort Item Definition Object Reference=@n4
    cosmeticType:CosmeticPicker_CosmeticTypes Enum=NewEnumerator7
   ) -> (then)
   > updateVariants!(execute, self:EUW Cosmetic Pickerv 2 ID Object Reference=@n3, id:Fort Item Definition Object Reference=@n4) -> (then, Variants:EUW Cosmetic Pickerv 2 Variants Object Reference)
  ]}
 SetupCompanionWidgets {nodes [
   SetupCompanionWidgets() {is=title K2Node_FunctionEntry}
   > ToggleCompanionEnable(isEnabled="Equal (Float)"(A=GetNumericCVar(CVar=Fort.PlayerCompanion.Enabled), B=1))
  ]}
 ToggleCompanionEnable {nodes [
   n1 = ToggleCompanionEnable!() -> (then, isEnabled:bool=~) {is=title K2Node_FunctionEntry}
   > n2 = Branch(Condition=@n1)
   n3 = Get TextBlock_ToggleCompanion() -> (TextBlock_ToggleCompanion:Text Object Reference)
   > Set Text(Text="LOCTEXT(\"(PIE) Companion ENABLED\")", self=@n3)
   > SetColorAndOpacity(self=@n3, InColorAndOpacity=Make Slate Color(SpecifiedColor=(R=0,G=1,B=0,A=1)))
   Set Text(execute=@n2.else, Text="LOCTEXT(\"(PIE) Companion DISABLED\")", self=@n3)
   > SetColorAndOpacity(self=@n3, InColorAndOpacity=Make Slate Color(SpecifiedColor=(R=1,G=0,B=0,A=1)))
  ]}
 ToggleCompanionFreeze {nodes [
   n1 = ToggleCompanionFreeze!() -> (then, Freeze:bool=~) {is=title K2Node_FunctionEntry}
   > n2 = Branch(Condition=@n1)
   n3 = Get TextBlock_ToggleFreeze() -> (TextBlock_ToggleFreeze:Text Object Reference)
   > Set Text(Text="LOCTEXT(\"(PIE) UNFREEZE\")", self=@n3)
   > SetColorAndOpacity(self=@n3, InColorAndOpacity=Make Slate Color(SpecifiedColor=(R=0.2707,G=0.968,B=1,A=1)))
   n4 = Get EditorUtilityButton_CompanionFreeze() -> (EditorUtilityButton_CompanionFreeze:Editor Utility Button Object Reference)
   > SetBackgroundColor!(execute, self:Button Object Reference=@n4, InBackgroundColor:Linear Color Structure=(R=0,G=0.1042,B=0.5565,A=1)) -> (then)
   Set Text(execute=@n2.else, Text="LOCTEXT(\"(PIE) FREEZE\")", self=@n3)
   > SetColorAndOpacity(self=@n3, InColorAndOpacity=Make Slate Color(SpecifiedColor=(R=1,G=1,B=1,A=1)))
   > SetBackgroundColor(self=@n4, InBackgroundColor=(R=1,G=1,B=1,A=1))
  ]}
 setWrap {nodes [
   n1 = setWrap!() -> (then, inWrap:Athena Item Wrap Definition Object Reference) {is=title K2Node_FunctionEntry}
   > PrintString(InString=Append(A="SetWrap  ", B=GetDisplayName(Object=@n1)))
   > setWrapVariable!(
    execute
    self:EUW Cosmetic Pickerv 2 Wrap Object Reference=Get EUW_CosmeticPickerv2_Wrap() -> (EUW_CosmeticPickerv2_Wrap:EUW Cosmetic Pickerv 2 Wrap Object Reference)
    inWrap:Athena Item Wrap Definition Object Reference=@n1
   ) -> (then)
  ]}
]
variables [
 emotes {type="array<AthenaDanceItemDefinition>", default []}
 NewEventDispatcher_0 {default=()}
 ShouldTick? {type=boolean, default=true}
]
widgets {
 info {rootWidgetClass=&CanvasPanel, widgetCount=42}
 widgets [
  CanvasPanel_0 {widgetClassPath=&CanvasPanel, bIsVariable=true, widget=CanvasPanel_0}
  ScrollBox_176 {
   widgetClassPath=&ScrollBox
   parent=CanvasPanel_0
   widget=ScrollBox_176
   properties {
    widgetBarStyle {
     thickness=8
     draggedThumbImage {
      bIsDynamicallyLoaded=false
      drawAs=RoundedBox
      imageType=NoImage
      mirroring=NoMirror
      resourceName=None
      resourceObject=null
      tiling=NoTile
      imageSize=<0 0>
      margin=<ltrb 0 0 0 0>
      outlineSettings {
       bUseBrushTransparency=false
       roundingType=FixedRadius
       width=0
       color {colorUseRule=UseColor_Specified, specifiedColor=<rgba 0 0 0 0>}
       cornerRadii=<4 4 4 4>
      }
      tintColor {colorUseRule=UseColor_Specified, specifiedColor=<rgba 0.2159 0.2159 0.2159 1>}
      uVRegion {bIsValid=false, max=<0 0>, min=<0 0>}
     }
     horizontalBackgroundImage {
      bIsDynamicallyLoaded=false
      drawAs=NoDrawType
      imageType=NoImage
      mirroring=NoMirror
      resourceName=None
      resourceObject=null
      tiling=NoTile
      imageSize=<0 0>
      margin=<ltrb 0 0 0 0>
      outlineSettings {
       bUseBrushTransparency=false
       roundingType=HalfHeightRadius
       width=0
       color {colorUseRule=UseColor_Specified, specifiedColor=<rgba 0 0 0 0>}
       cornerRadii=<0 0 0 0>
      }
      tintColor {colorUseRule=UseColor_Specified, specifiedColor=<rgba 1 1 1 1>}
      uVRegion {bIsValid=false, max=<0 0>, min=<0 0>}
     }
     horizontalBottomSlotImage {
      bIsDynamicallyLoaded=false
      drawAs=NoDrawType
      imageType=NoImage
      mirroring=NoMirror
      resourceName=None
      resourceObject=null
      tiling=NoTile
      imageSize=<0 0>
      margin=<ltrb 0 0 0 0>
      outlineSettings {
       bUseBrushTransparency=false
       roundingType=HalfHeightRadius
       width=0
       color {colorUseRule=UseColor_Specified, specifiedColor=<rgba 0 0 0 0>}
       cornerRadii=<0 0 0 0>
      }
      tintColor {colorUseRule=UseColor_Specified, specifiedColor=<rgba 1 1 1 1>}
      uVRegion {bIsValid=false, max=<0 0>, min=<0 0>}
     }
     horizontalTopSlotImage {
      bIsDynamicallyLoaded=false
      drawAs=NoDrawType
      imageType=NoImage
      mirroring=NoMirror
      resourceName=None
      resourceObject=null
      tiling=NoTile
      imageSize=<0 0>
      margin=<ltrb 0 0 0 0>
      outlineSettings {
       bUseBrushTransparency=false
       roundingType=HalfHeightRadius
       width=0
       color {colorUseRule=UseColor_Specified, specifiedColor=<rgba 0 0 0 0>}
       cornerRadii=<0 0 0 0>
      }
      tintColor {colorUseRule=UseColor_Specified, specifiedColor=<rgba 1 1 1 1>}
      uVRegion {bIsValid=false, max=<0 0>, min=<0 0>}
     }
     hoveredThumbImage {
      bIsDynamicallyLoaded=false
      drawAs=RoundedBox
      imageType=NoImage
      mirroring=NoMirror
      resourceName=None
      resourceObject=null
      tiling=NoTile
      imageSize=<0 0>
      margin=<ltrb 0 0 0 0>
      outlineSettings {
       bUseBrushTransparency=false
       roundingType=FixedRadius
       width=0
       color {colorUseRule=UseColor_Specified, specifiedColor=<rgba 0 0 0 0>}
       cornerRadii=<4 4 4 4>
      }
      tintColor {colorUseRule=UseColor_Specified, specifiedColor=<rgba 0.2159 0.2159 0.2159 1>}
      uVRegion {bIsValid=false, max=<0 0>, min=<0 0>}
     }
     normalThumbImage {
      bIsDynamicallyLoaded=false
      drawAs=RoundedBox
      imageType=NoImage
      mirroring=NoMirror
      resourceName=None
      resourceObject=null
      tiling=NoTile
      imageSize=<0 0>
      margin=<ltrb 0 0 0 0>
      outlineSettings {
       bUseBrushTransparency=false
       roundingType=FixedRadius
       width=0
       color {colorUseRule=UseColor_Specified, specifiedColor=<rgba 0 0 0 0>}
       cornerRadii=<4 4 4 4>
      }
      tintColor {colorUseRule=UseColor_Specified, specifiedColor=<rgba 0.0953 0.0953 0.0953 1>}
      uVRegion {bIsValid=false, max=<0 0>, min=<0 0>}
     }
     verticalBackgroundImage {
      bIsDynamicallyLoaded=false
      drawAs=NoDrawType
      imageType=NoImage
      mirroring=NoMirror
      resourceName=None
      resourceObject=null
      tiling=NoTile
      imageSize=<0 0>
      margin=<ltrb 0 0 0 0>
      outlineSettings {
       bUseBrushTransparency=false
       roundingType=HalfHeightRadius
       width=0
       color {colorUseRule=UseColor_Specified, specifiedColor=<rgba 0 0 0 0>}
       cornerRadii=<0 0 0 0>
      }
      tintColor {colorUseRule=UseColor_Specified, specifiedColor=<rgba 1 1 1 1>}
      uVRegion {bIsValid=false, max=<0 0>, min=<0 0>}
     }
     verticalBottomSlotImage {
      bIsDynamicallyLoaded=false
      drawAs=NoDrawType
      imageType=NoImage
      mirroring=NoMirror
      resourceName=None
      resourceObject=null
      tiling=NoTile
      imageSize=<0 0>
      margin=<ltrb 0 0 0 0>
      outlineSettings {
       bUseBrushTransparency=false
       roundingType=HalfHeightRadius
       width=0
       color {colorUseRule=UseColor_Specified, specifiedColor=<rgba 0 0 0 0>}
       cornerRadii=<0 0 0 0>
      }
      tintColor {colorUseRule=UseColor_Specified, specifiedColor=<rgba 1 1 1 1>}
      uVRegion {bIsValid=false, max=<0 0>, min=<0 0>}
     }
     verticalTopSlotImage {
      bIsDynamicallyLoaded=false
      drawAs=NoDrawType
      imageType=NoImage
      mirroring=NoMirror
      resourceName=None
      resourceObject=null
      tiling=NoTile
      imageSize=<0 0>
      margin=<ltrb 0 0 0 0>
      outlineSettings {
       bUseBrushTransparency=false
       roundingType=HalfHeightRadius
       width=0
       color {colorUseRule=UseColor_Specified, specifiedColor=<rgba 0 0 0 0>}
       cornerRadii=<0 0 0 0>
      }
      tintColor {colorUseRule=UseColor_Specified, specifiedColor=<rgba 1 1 1 1>}
      uVRegion {bIsValid=false, max=<0 0>, min=<0 0>}
     }
    }
    widgetStyle {
     barThickness=8
     bottomShadowBrush {
      bIsDynamicallyLoaded=false
      drawAs=Box
      imageType=FullColor
      mirroring=NoMirror
      resourceName=../../../Engine/Content/Slate/Common/ScrollBoxShadowBottom.png
      resourceObject=null
      tiling=NoTile
      imageSize=<16 8>
      margin=<ltrb 0.5 0 0.5 1>
      outlineSettings {
       bUseBrushTransparency=false
       roundingType=HalfHeightRadius
       width=0
       color {colorUseRule=UseColor_Specified, specifiedColor=<rgba 0 0 0 0>}
       cornerRadii=<0 0 0 0>
      }
      tintColor {colorUseRule=UseColor_Specified, specifiedColor=<rgba 1 1 1 1>}
      uVRegion {bIsValid=false, max=<0 0>, min=<0 0>}
     }
     horizontalScrolledContentPadding=<ltrb 0 0 1 0>
     leftShadowBrush {
      bIsDynamicallyLoaded=false
      drawAs=Box
      imageType=FullColor
      mirroring=NoMirror
      resourceName=../../../Engine/Content/Slate/Common/ScrollBoxShadowLeft.png
      resourceObject=null
      tiling=NoTile
      imageSize=<8 16>
      margin=<ltrb 1 0.5 0 0.5>
      outlineSettings {
       bUseBrushTransparency=false
       roundingType=HalfHeightRadius
       width=0
       color {colorUseRule=UseColor_Specified, specifiedColor=<rgba 0 0 0 0>}
       cornerRadii=<0 0 0 0>
      }
      tintColor {colorUseRule=UseColor_Specified, specifiedColor=<rgba 1 1 1 1>}
      uVRegion {bIsValid=false, max=<0 0>, min=<0 0>}
     }
     rightShadowBrush {
      bIsDynamicallyLoaded=false
      drawAs=Box
      imageType=FullColor
      mirroring=NoMirror
      resourceName=../../../Engine/Content/Slate/Common/ScrollBoxShadowRight.png
      resourceObject=null
      tiling=NoTile
      imageSize=<8 16>
      margin=<ltrb 0 0.5 1 0.5>
      outlineSettings {
       bUseBrushTransparency=false
       roundingType=HalfHeightRadius
       width=0
       color {colorUseRule=UseColor_Specified, specifiedColor=<rgba 0 0 0 0>}
       cornerRadii=<0 0 0 0>
      }
      tintColor {colorUseRule=UseColor_Specified, specifiedColor=<rgba 1 1 1 1>}
      uVRegion {bIsValid=false, max=<0 0>, min=<0 0>}
     }
     topShadowBrush {
      bIsDynamicallyLoaded=false
      drawAs=Box
      imageType=FullColor
      mirroring=NoMirror
      resourceName=../../../Engine/Content/Slate/Common/ScrollBoxShadowTop.png
      resourceObject=null
      tiling=NoTile
      imageSize=<16 8>
      margin=<ltrb 0.5 1 0.5 0>
      outlineSettings {
       bUseBrushTransparency=false
       roundingType=HalfHeightRadius
       width=0
       color {colorUseRule=UseColor_Specified, specifiedColor=<rgba 0 0 0 0>}
       cornerRadii=<0 0 0 0>
      }
      tintColor {colorUseRule=UseColor_Specified, specifiedColor=<rgba 1 1 1 1>}
      uVRegion {bIsValid=false, max=<0 0>, min=<0 0>}
     }
     verticalScrolledContentPadding=<ltrb 0 0 0 1>
    }
   }
   slot {class=CanvasPanelSlot, properties {layoutData {alignment=<0 0>, anchors {maximum=<1 1>, minimum=<0 0>}, offsets=<ltrb 4 0 4 0>}}}
  }
  EUW_CosmeticPickerv2_Wrap {
   widgetClassPath=&/EUW_CosmeticPickerv2_Wrap
   bIsVariable=true
   parent=ScrollBox_176
   widget=EUW_CosmeticPickerv2_Wrap
   slot {class=ScrollBoxSlot}
  }
  Spacer_4 {widgetClassPath=&Spacer, parent=ScrollBox_176, widget=Spacer_4, properties {size=<1 0>}, slot {class=ScrollBoxSlot}}
  EUW_CosmeticPickerv2_Emotes {
   widgetClassPath=&/EUW_CosmeticPickerv2_Emotes
   bIsVariable=true
   parent=ScrollBox_176
   widget=EUW_CosmeticPickerv2_Emotes
   slot {class=ScrollBoxSlot}
  }
  Spacer {widgetClassPath=&Spacer, parent=ScrollBox_176, widget=Spacer, properties {size=<1 40>}, slot {class=ScrollBoxSlot}}
  VerticalBox_0 {widgetClassPath=&VerticalBox, parent=ScrollBox_176, widget=VerticalBox_0, slot {class=ScrollBoxSlot}}
  EditorUtilityButton_95 {
   widgetClassPath=&EditorUtilityButton
   parent=VerticalBox_0
   widget=EditorUtilityButton_95
   slot {class=VerticalBoxSlot, properties {horizontalAlignment=HAlign_Center}}
  }
  TextBlock_127 {
   widgetClassPath=&TextBlock
   parent=EditorUtilityButton_95
   widget=TextBlock_127
   properties {
    text="Toggle Contrail Visibility (PIE only)"
    font {
     bForceMonospaced=false
     bMaterialIsStencil=false
     fontMaterial=null
     fontObject=&Roboto
     letterSpacing=0
     monospacedWidth=1
     size=14.25
     skewAmount=0
     typefaceFontName=Bold
     outlineSettings {
      bApplyOutlineToDropShadows=false
      bMiteredCorners=false
      bSeparateFillAlpha=false
      outlineMaterial=null
      outlineSize=0
      outlineColor=<rgba 0 0 0 1>
     }
    }
   }
   slot {class=ButtonSlot, properties {verticalAlignment=VAlign_Fill, padding=<ltrb 0 0 0 0>}}
  }
  EUW_CosmeticPicker_ID_Contrail {
   widgetClassPath=&/EUW_CosmeticPickerv2_ID
   bIsVariable=true
   parent=VerticalBox_0
   widget=EUW_CosmeticPicker_ID_Contrail
   properties {cosmeticType=Contrail}
   slot {class=VerticalBoxSlot}
  }
  VerticalBox_1 {
   widgetClassPath=&VerticalBox
   parent=ScrollBox_176
   widget=VerticalBox_1
   slot {class=ScrollBoxSlot, properties {verticalAlignment=VAlign_Top, padding=<ltrb 10 0 10 0>}}
  }
  EUW_CosmeticPickerv2_CompanionMannequin {
   widgetClassPath=&/EUW_CosmeticPickerv2_CompanionMannequin
   bIsVariable=true
   parent=VerticalBox_1
   widget=EUW_CosmeticPickerv2_CompanionMannequin
   slot {class=VerticalBoxSlot}
  }
  EUW_CosmeticPicker_ID_Companions {
   widgetClassPath=&EUW_CosmeticPickerv2_ID
   bIsVariable=true
   parent=VerticalBox_1
   widget=EUW_CosmeticPicker_ID_Companions
   properties {cosmeticType=Companion}
   slot {class=VerticalBoxSlot}
  }
  VerticalBox_2 {
   widgetClassPath=&VerticalBox
   parent=VerticalBox_1
   widget=VerticalBox_2
   slot {class=VerticalBoxSlot, properties {horizontalAlignment=HAlign_Center, verticalAlignment=VAlign_Center, padding=<ltrb 0 -10 0 0>}}
  }
  HorizontalBox_0 {widgetClassPath=&HorizontalBox, parent=VerticalBox_2, widget=HorizontalBox_0, slot {class=VerticalBoxSlot}}
  Spacer_10 {
   widgetClassPath=&Spacer
   parent=HorizontalBox_0
   widget=Spacer_10
   properties {size=<20 1>}
   slot {class=HorizontalBoxSlot, properties {size {sizeRule=Fill, value=1}}}
  }
  EditorUtilityButton_CompanionToggle {
   widgetClassPath=&EditorUtilityButton
   parent=HorizontalBox_0
   widget=EditorUtilityButton_CompanionToggle
   slot {class=HorizontalBoxSlot, properties {horizontalAlignment=HAlign_Center, verticalAlignment=VAlign_Center}}
  }
  TextBlock_ToggleCompanion {
   widgetClassPath=&TextBlock
   bIsVariable=true
   parent=EditorUtilityButton_CompanionToggle
   widget=TextBlock_ToggleCompanion
   properties {
    text=Toggle Companion
    font {
     bForceMonospaced=false
     bMaterialIsStencil=false
     fontMaterial=null
     fontObject=&Roboto
     letterSpacing=0
     monospacedWidth=1
     size=10.5
     skewAmount=0
     typefaceFontName=Bold
     outlineSettings {
      bApplyOutlineToDropShadows=false
      bMiteredCorners=false
      bSeparateFillAlpha=false
      outlineMaterial=null
      outlineSize=0
      outlineColor=<rgba 0 0 0 1>
     }
    }
   }
   slot {class=ButtonSlot}
  }
  Spacer_11 {widgetClassPath=&Spacer, parent=HorizontalBox_0, widget=Spacer_11, properties {size=<20 1>}, slot {class=HorizontalBoxSlot}}
  EditorUtilityButton_CompanionMood {
   widgetClassPath=&EditorUtilityButton
   parent=HorizontalBox_0
   widget=EditorUtilityButton_CompanionMood
   slot {class=HorizontalBoxSlot, properties {horizontalAlignment=HAlign_Center, verticalAlignment=VAlign_Center}}
  }
  TextBlock_ForceMood {
   widgetClassPath=&TextBlock
   parent=EditorUtilityButton_CompanionMood
   widget=TextBlock_ForceMood
   properties {
    text="(PIE) Force Mood:"
    font {
     bForceMonospaced=false
     bMaterialIsStencil=false
     fontMaterial=null
     fontObject=&Roboto
     letterSpacing=0
     monospacedWidth=1
     size=10.5
     skewAmount=0
     typefaceFontName=Bold
     outlineSettings {
      bApplyOutlineToDropShadows=false
      bMiteredCorners=false
      bSeparateFillAlpha=false
      outlineMaterial=null
      outlineSize=0
      outlineColor=<rgba 0 0 0 1>
     }
    }
   }
   slot {class=ButtonSlot}
  }
  EditorUtilityComboBoxString_CompanionMood {
   widgetClassPath=/Script/Blutility.EditorUtilityComboBoxString
   bIsVariable=true
   parent=HorizontalBox_0
   widget=EditorUtilityComboBoxString_CompanionMood
   properties {
    selectedOption=Bored
    defaultOptions [None,Default,Combat,Worried,Bored,Wander]
    font {
     bForceMonospaced=false
     bMaterialIsStencil=false
     fontMaterial=null
     fontObject=&Roboto
     letterSpacing=0
     monospacedWidth=1
     size=12
     skewAmount=0
     typefaceFontName=Bold
     outlineSettings {
      bApplyOutlineToDropShadows=false
      bMiteredCorners=false
      bSeparateFillAlpha=false
      outlineMaterial=null
      outlineSize=0
      outlineColor=<rgba 0 0 0 1>
     }
    }
    foregroundColor {colorUseRule=UseColor_Specified, specifiedColor=<rgba 0.5271 0.5271 0.5271 1>}
   }
   slot {class=HorizontalBoxSlot, properties {horizontalAlignment=HAlign_Center, verticalAlignment=VAlign_Center}}
  }
  Spacer_9 {
   widgetClassPath=&Spacer
   parent=HorizontalBox_0
   widget=Spacer_9
   properties {size=<20 1>}
   slot {class=HorizontalBoxSlot, properties {size {sizeRule=Fill, value=1}}}
  }
  HorizontalBox_1 {
   widgetClassPath=&HorizontalBox
   parent=VerticalBox_2
   widget=HorizontalBox_1
   slot {class=VerticalBoxSlot, properties {padding=<ltrb 0 5 0 0>}}
  }
  Spacer_3 {
   widgetClassPath=&Spacer
   parent=HorizontalBox_1
   widget=Spacer_3
   properties {size=<20 1>}
   slot {class=HorizontalBoxSlot, properties {size {sizeRule=Fill, value=1}}}
  }
  EditorUtilityButton_CompanionForceInteract {
   widgetClassPath=&EditorUtilityButton
   parent=HorizontalBox_1
   widget=EditorUtilityButton_CompanionForceInteract
   slot {class=HorizontalBoxSlot, properties {horizontalAlignment=HAlign_Center, verticalAlignment=VAlign_Center}}
  }
  UEFN_TextBlock_C_74 {
   widgetClassPath=&UEFN_TextBlock
   parent=EditorUtilityButton_CompanionForceInteract
   widget=UEFN_TextBlock_C_74
   properties {
    text="(PIE) Force Interact"
    font {
     bForceMonospaced=false
     bMaterialIsStencil=false
     fontMaterial=null
     fontObject=&Roboto
     letterSpacing=0
     monospacedWidth=1
     size=10.5
     skewAmount=0
     typefaceFontName=Bold
     outlineSettings {
      bApplyOutlineToDropShadows=false
      bMiteredCorners=false
      bSeparateFillAlpha=false
      outlineMaterial=null
      outlineSize=0
      outlineColor=<rgba 0 0 0 1>
     }
    }
   }
   slot {class=ButtonSlot, properties {horizontalAlignment=HAlign_Fill, verticalAlignment=VAlign_Fill}}
  }
  Spacer_5 {widgetClassPath=&Spacer, parent=HorizontalBox_1, widget=Spacer_5, properties {size=<20 1>}, slot {class=HorizontalBoxSlot}}
  EditorUtilityButton_CompanionForceEmote {
   widgetClassPath=&EditorUtilityButton
   parent=HorizontalBox_1
   widget=EditorUtilityButton_CompanionForceEmote
   slot {class=HorizontalBoxSlot, properties {horizontalAlignment=HAlign_Center, verticalAlignment=VAlign_Center}}
  }
  UEFN_TextBlock_C {
   widgetClassPath=&UEFN_TextBlock
   parent=EditorUtilityButton_CompanionForceEmote
   widget=UEFN_TextBlock_C
   properties {
    text="(PIE) Force Emote"
    font {
     bForceMonospaced=false
     bMaterialIsStencil=false
     fontMaterial=null
     fontObject=&Roboto
     letterSpacing=0
     monospacedWidth=1
     size=10.5
     skewAmount=0
     typefaceFontName=Bold
     outlineSettings {
      bApplyOutlineToDropShadows=false
      bMiteredCorners=false
      bSeparateFillAlpha=false
      outlineMaterial=null
      outlineSize=0
      outlineColor=<rgba 0 0 0 1>
     }
    }
   }
   slot {class=ButtonSlot, properties {horizontalAlignment=HAlign_Fill, verticalAlignment=VAlign_Fill}}
  }
  Spacer_1 {widgetClassPath=&Spacer, parent=HorizontalBox_1, widget=Spacer_1, properties {size=<20 1>}, slot {class=HorizontalBoxSlot}}
  EditorUtilityButton_CompanionFreeze {
   widgetClassPath=&EditorUtilityButton
   parent=HorizontalBox_1
   widget=EditorUtilityButton_CompanionFreeze
   slot {class=HorizontalBoxSlot, properties {horizontalAlignment=HAlign_Center, verticalAlignment=VAlign_Center}}
  }
  TextBlock_ToggleFreeze {
   widgetClassPath=&TextBlock
   bIsVariable=true
   parent=EditorUtilityButton_CompanionFreeze
   widget=TextBlock_ToggleFreeze
   properties {
    text="(PIE) FREEZE"
    font {
     bForceMonospaced=false
     bMaterialIsStencil=false
     fontMaterial=null
     fontObject=&Roboto
     letterSpacing=0
     monospacedWidth=1
     size=10.5
     skewAmount=0
     typefaceFontName=Bold
     outlineSettings {
      bApplyOutlineToDropShadows=false
      bMiteredCorners=false
      bSeparateFillAlpha=false
      outlineMaterial=null
      outlineSize=0
      outlineColor=<rgba 0 0 0 1>
     }
    }
   }
   slot {class=ButtonSlot}
  }
  Spacer_2 {
   widgetClassPath=&Spacer
   parent=HorizontalBox_1
   widget=Spacer_2
   properties {size=<20 1>}
   slot {class=HorizontalBoxSlot, properties {size {sizeRule=Fill, value=1}}}
  }
  HorizontalBox_103 {
   widgetClassPath=&HorizontalBox
   parent=VerticalBox_2
   widget=HorizontalBox_103
   slot {class=VerticalBoxSlot, properties {padding=<ltrb 0 5 0 0>}}
  }
  Spacer_7 {
   widgetClassPath=&Spacer
   parent=HorizontalBox_103
   widget=Spacer_7
   properties {size=<20 1>}
   slot {class=HorizontalBoxSlot, properties {size {sizeRule=Fill, value=1}}}
  }
  EditorUtilityButton_CompanionForceStay {
   widgetClassPath=&EditorUtilityButton
   parent=HorizontalBox_103
   widget=EditorUtilityButton_CompanionForceStay
   slot {class=HorizontalBoxSlot, properties {horizontalAlignment=HAlign_Center, verticalAlignment=VAlign_Center}}
  }
  TextBlock_ToggleStay {
   widgetClassPath=&UEFN_TextBlock
   parent=EditorUtilityButton_CompanionForceStay
   widget=TextBlock_ToggleStay
   properties {
    text="(PIE) Toggle Stay"
    font {
     bForceMonospaced=false
     bMaterialIsStencil=false
     fontMaterial=null
     fontObject=&Roboto
     letterSpacing=0
     monospacedWidth=1
     size=10.5
     skewAmount=0
     typefaceFontName=Bold
     outlineSettings {
      bApplyOutlineToDropShadows=false
      bMiteredCorners=false
      bSeparateFillAlpha=false
      outlineMaterial=null
      outlineSize=0
      outlineColor=<rgba 0 0 0 1>
     }
    }
   }
   slot {class=ButtonSlot, properties {horizontalAlignment=HAlign_Fill, verticalAlignment=VAlign_Fill}}
  }
  Spacer_6 {widgetClassPath=&Spacer, parent=HorizontalBox_103, widget=Spacer_6, properties {size=<20 1>}, slot {class=HorizontalBoxSlot}}
  EditorUtilityButton_RefreshAllMannequins {
   widgetClassPath=&EditorUtilityButton
   parent=HorizontalBox_103
   widget=EditorUtilityButton_RefreshAllMannequins
   slot {class=HorizontalBoxSlot}
  }
  TextBlock_2 {
   widgetClassPath=&TextBlock
   parent=EditorUtilityButton_RefreshAllMannequins
   widget=TextBlock_2
   properties {
    text="Refresh All Mannequins (Editor World)"
    font {
     bForceMonospaced=false
     bMaterialIsStencil=false
     fontMaterial=null
     fontObject=&Roboto
     letterSpacing=0
     monospacedWidth=1
     size=10.5
     skewAmount=0
     typefaceFontName=Bold
     outlineSettings {
      bApplyOutlineToDropShadows=false
      bMiteredCorners=false
      bSeparateFillAlpha=false
      outlineMaterial=null
      outlineSize=0
      outlineColor=<rgba 0 0 0 1>
     }
    }
   }
   slot {class=ButtonSlot}
  }
  Spacer_8 {
   widgetClassPath=&Spacer
   parent=HorizontalBox_103
   widget=Spacer_8
   properties {size=<20 1>}
   slot {class=HorizontalBoxSlot, properties {size {sizeRule=Fill, value=1}}}
  }
 ]
}
