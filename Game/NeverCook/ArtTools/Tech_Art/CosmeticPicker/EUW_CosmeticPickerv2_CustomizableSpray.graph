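# Editor utility widget for Cosmetic Picker v2 that edits a customizable spray variant channel on a cosmetic item.
# Setup stores the item and variant index, fetches the default spray through Python, labels itself and joins the supplied wrap box.
# A self-bound Details View exposes the spray and a LinearColor packing rotation, scale, wear and saturation.
# Edits send these values to the CosmeticPicker Python module, which also backs an add-channel function.
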EditorUtilityWidgetBlueprint /Game/NeverCook/ArtTools/Tech_Art/CosmeticPicker/EUW_CosmeticPickerv2_CustomizableSpray
graphs [
 Add Custom Variant Channel {nodes [
   Add Custom Variant Channel() {is=title K2Node_FunctionEntry}
   n1 = GetDisplayName(Object=Get spray() -> (spray:Athena Spray Item Definition Object Reference))
   > PrintString(InString=@n1)
   n2 = Get!() -> (RotScaleWearSat_R:float=~, RotScaleWearSat_G:float=~, RotScaleWearSat_B:float=~, RotScaleWearSat_A:float=~)
   > Execute Python Script!(
    execute
    PythonScript="import TechArt.CosmeticPicker.CosmeticPickerModule as CosmeticPicker\r\ntype = CosmeticPicker.addCharacterVariantChannel_spray(id, item_variant_index, spray, rotation, scale, wear, saturation)"
    id:Fort Item Definition Object Reference=Get Id() -> (Id:Fort Item Definition Object Reference)
    itemVariantIndex:int=Get itemVariantIndex() -> (itemVariantIndex:int=0)
    spray:String=@n1
    rotation:float=@n2
    scale:float=@n2.RotScaleWearSat_G
    wear:float=@n2.RotScaleWearSat_B
    saturation:float=@n2.RotScaleWearSat_A
   ) -> (then, ReturnValue) {inputs [id, itemVariantIndex, spray, rotation, scale, wear, saturation]}
  ]}
 EventGraph {nodes [
   n1 = FindParentWidgetOfType!(
    execute
    StartingWidget:Widget Object Reference=Self-Reference()
    Type:Widget Class Reference=&/EUW_CosmeticPickerv2_Variants.EUW_CosmeticPickerv2_Variants_C
   ) -> (then, ReturnValue:EUW Cosmetic Pickerv 2 Variants Object Reference)
   > Update Custom Variant Channels(self:EUW Cosmetic Pickerv 2 Variants Object Reference=@n1)
   "On Property Changed (DetailsView_Spray)"!() -> (OutputDelegate, then, PropertyName:Name=None) {is=title K2Node_ComponentBoundEvent}
   > PrintString(InString=SprayChangedEvent)
   > Update Custom Variant Channel()
   Event Construct()
   > SetEditorProperty(
    Object=Get DetailsView_Spray() -> (DetailsView_Spray:Details View Object Reference)
    PropertyName=CategoriesToShow
    PropertyValue:Array of Names=Make Array("[0]":Name=String To Name(InString=nocategory)) -> (Array:Array of Names)
   )
   > SetObject(self=Get DetailsView_Spray(), NewObject=Self-Reference())
   Event PreConstruct()
   Event Tick()
  ]}
 Update Custom Variant Channel {nodes [
   Update Custom Variant Channel() {is=title K2Node_FunctionEntry}
   n1 = GetDisplayName(Object=Get spray())
   > PrintString(InString=@n1)
   n2 = Get() -> (RotScaleWearSat_R=~, RotScaleWearSat_G=~, RotScaleWearSat_B=~, RotScaleWearSat_A=~)
   > Execute Python Script(
    PythonScript="import TechArt.CosmeticPicker.CosmeticPickerModule as CosmeticPicker\r\ntype = CosmeticPicker.updateCharacterVariantChannel_spray(id, item_variant_index, spray, rotation, scale, wear, saturation)"
    id=Get Id()
    itemVariantIndex=Get itemVariantIndex() -> (itemVariantIndex=0)
    spray=@n1
    rotation=@n2
    scale=@n2.RotScaleWearSat_G
    wear=@n2.RotScaleWearSat_B
    saturation=@n2.RotScaleWearSat_A
   ) {inputs [id, itemVariantIndex, spray, rotation, scale, wear, saturation]}
  ]}
 getDefaultValue {nodes [
   n1 = getDefaultValue!() -> (then, item:Fort Item Definition Object Reference, itemVariantIndex:int) {is=title K2Node_FunctionEntry}
   > n2 = Execute Python Script!(
    execute
    PythonScript="import TechArt.CosmeticPicker.CosmeticPickerModule as CosmeticPicker\r\nspray = CosmeticPicker.getDefaultValue_spray(id, item_variant_index)"
    id=@n1
    itemVariantIndex=@n1.itemVariantIndex
   ) -> (then, ReturnValue, spray:Primary Asset Id Structure) {inputs [id, itemVariantIndex], outputs [spray]}
   > n3 = Cast To AthenaSprayItemDefinition!(execute, Object=GetObjectFromPrimaryAssetId!(PrimaryAssetId:Primary Asset Id Structure=@n2.spray) -> (ReturnValue:Object Reference)) -> (then, CastFailed, AsAthena Spray Item Definition:Athena Spray Item Definition Object Reference) {is=title K2Node_DynamicCast}
   > Return Node!(execute, spray:Athena Spray Item Definition Object Reference=@n3)
  ]}
 getMannequinOrDefaultValue {nodes [
   getMannequinOrDefaultValue!() -> (then, defaultIndex:int) {is=title K2Node_FunctionEntry}
   > Return Node!(execute, index:int=0)
  ]}
 setup {nodes [
   n1 = setup!() -> (then, cosmeticID:Fort Item Definition Object Reference, itemVariantIndex:int, wrapBox:Wrap Box Object Reference, name:String) {is=title K2Node_FunctionEntry}
   > Set Id(Id:Fort Item Definition Object Reference=@n1) -> (Output_Get:Fort Item Definition Object Reference)
   > Set itemVariantIndex(itemVariantIndex:int=@n1.itemVariantIndex) -> (Output_Get:int=0)
   > Set localName(localName:String=@n1.name) -> (Output_Get:String)
   > PrintString(InString="spray->Setup")
   > n2 = getDefaultValue!(execute, self, item:Fort Item Definition Object Reference=Get Id(), itemVariantIndex:int=Get itemVariantIndex() -> (itemVariantIndex=0)) -> (then, spray:Athena Spray Item Definition Object Reference) {is=title K2Node_CallFunction}
   > Set spray(spray:Athena Spray Item Definition Object Reference=@n2) -> (Output_Get:Athena Spray Item Definition Object Reference)
   n3 = Self-Reference()
   > AddChild(self=@n1.wrapBox, Content=@n3)
   > SetVerticalAlignment(
    self:Wrap Box Slot Object Reference=SlotAsWrapBoxSlot!(Widget:Widget Object Reference=@n3) -> (ReturnValue:Wrap Box Slot Object Reference)
    InVerticalAlignment=VAlign_Bottom
   )
   > "SetText (Text)"(
    self=Get variantText_name() -> (variantText_name:Text Object Reference)
    InText="To Text (String)"(InString=Get localName() -> (localName:String))
   )
  ]}
]
variables [
 float {type=number}
 Id {type=FortItemDefinition, default=null}
 itemVariantIndex {type=integer}
 primary {type=PrimaryAssetId, default=""}
 RotScaleWearSat {type=LinearColor, default=<rgba 0 1 0 1>}
 spray {type=AthenaSprayItemDefinition, default=/Game/Athena/Items/Cosmetics/Sprays/SPID_002_XMark}
]
widgets {
 info {rootWidgetClass=&CanvasPanel, widgetCount=5}
 widgets [
  CanvasPanel_0 {widgetClassPath=&CanvasPanel, widget=CanvasPanel_0}
  VerticalBox_64 {
   widgetClassPath=&VerticalBox
   parent=CanvasPanel_0
   widget=VerticalBox_64
   slot {
    class=CanvasPanelSlot
    properties {bAutoSize=true, layoutData {alignment=<0 0>, anchors {maximum=<1 0>, minimum=<0 0>}, offsets=<ltrb 0 0 0 100>}}
   }
  }
  variantText_name {
   widgetClassPath=&TextBlock
   bIsVariable=true
   parent=VerticalBox_64
   widget=variantText_name
   properties {
    text=-
    font {
     bForceMonospaced=false
     bMaterialIsStencil=false
     fontMaterial=null
     fontObject=&Roboto
     letterSpacing=0
     monospacedWidth=1
     size=10
     skewAmount=0
     typefaceFontName=Bold
     outlineSettings {
      bApplyOutlineToDropShadows=false
      bMiteredCorners=false
      bSeparateFillAlpha=false
      outlineMaterial=null
      outlineSize=0
      outlineColor=<rgba 0 0 0 1>
     }
    }
   }
   slot {class=VerticalBoxSlot}
  }
  DetailsView_Spray {
   widgetClassPath=&DetailsView
   parent=VerticalBox_64
   widget=DetailsView_Spray
   properties {bShowScrollBar=false, propertiesToShow [spray, RotScaleWearSat, r]}
   slot {class=VerticalBoxSlot}
  }
  SizeBox {
   widgetClassPath=&SizeBox
   parent=VerticalBox_64
   widget=SizeBox
   properties {bOverride_HeightOverride=true, bOverride_WidthOverride=true, heightOverride=5, widthOverride=500}
   slot {class=VerticalBoxSlot}
  }
 ]
}
