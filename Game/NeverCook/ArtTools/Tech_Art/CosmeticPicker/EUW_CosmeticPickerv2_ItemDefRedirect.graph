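# Cosmetic Picker editor utility widget: a labeled dropdown for choosing the redirect target of an item-def redirect variant on a cosmetic.
# Setup lists primary assets of the variant's ItemDefClass after a None entry, preselecting the cheats value, overridden when a matching mannequin exists.
# Mouse-click selection stores the index and updates the parent Variants widget's custom channels.
# Cheat and mannequin lookups and channel adding run through CosmeticPickerModule Python scripts.
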EditorUtilityWidgetBlueprint /Game/NeverCook/ArtTools/Tech_Art/CosmeticPicker/EUW_CosmeticPickerv2_ItemDefRedirect
graphs [
 EventGraph {nodes [
   n1 = CustomEvent!() -> (OutputDelegate, then, SelectedItem:String, SelectionType:ESelectInfo Enum) {is=title K2Node_CustomEvent}
   > Branch(Condition="Equal (Enum)"(A:ESelectInfo Enum=@n1.SelectionType, B:ESelectInfo Enum=OnMouseClick))
   > PrintString(InString=Append!(A=@n1.SelectedItem, B=" ", C:String=Enum to String!(Enumerator:ESelectInfo Enum=@n1.SelectionType) -> (ReturnValue:String)) -> (ReturnValue))
   > n2 = FindParentWidgetOfType!(
    execute
    StartingWidget:Widget Object Reference=Self-Reference()
    Type:Widget Class Reference=&/EUW_CosmeticPickerv2_Variants.EUW_CosmeticPickerv2_Variants_C
   ) -> (then, ReturnValue:EUW Cosmetic Pickerv 2 Variants Object Reference)
   > n3 = Cast To EUW_CosmeticPickerv2_Variants!(execute, Object=@n2) -> (then, CastFailed, AsEUW Cosmetic Pickerv 2 Variants:EUW Cosmetic Pickerv 2 Variants Object Reference) {is=title K2Node_DynamicCast}
   > Set optionIndex(optionIndex:int=GetSelectedIndex!(self:"Combo Box (String) Object Reference"=Get ItemDefsComboBox() -> (ItemDefsComboBox:"Combo Box (String) Object Reference")) -> (ReturnValue:int=0)) -> (Output_Get:int=0)
   > Update Custom Variant Channels(self:EUW Cosmetic Pickerv 2 Variants Object Reference=@n3)
   Event Construct()
   > Bind Event to On Selection Changed(self:"Combo Box (String) Object Reference"=Get ItemDefsComboBox(), Delegate=@n1) {is=title K2Node_AddDelegate}
  ]}
 FindSelectedIndexByPrimaryAssetId {nodes [
   n1 = FindSelectedIndexByPrimaryAssetId!() -> (then, PrimaryAssetId:Primary Asset Id Structure) {is=title K2Node_FunctionEntry}
   > Set FoundIndex(FoundIndex:int=0) -> (Output_Get:int=0)
   > n2 = For Each Loop with Break!(
    Exec:Exec
    Array:Array of ST Item Def Redirect Option Structures=Get selectableOptions() -> (selectableOptions:Array of ST Item Def Redirect Option Structures)
    Break:Exec=@n3.then
   ) -> (LoopBody:Exec, Array Element:ST Item Def Redirect Option Structure, Array Index:int, Completed:Exec)
   Return Node!(execute=@n2.Completed, SelectedIndex:int=Get FoundIndex() -> (FoundIndex:int=0))
   > Branch(Condition="Equal (PrimaryAssetId)"(
     A:Primary Asset Id Structure=Break ST Item Def Redirect Option!(ST_ItemDefRedirectOption:"ST Item Def Redirect Option Structure (by ref)"=@n2) -> (OptionName_2_#1:String, PrimaryAssetID_5_#2:Primary Asset Id Structure) {is=title K2Node_BreakStruct}.PrimaryAssetID_5_#2
     B:Primary Asset Id Structure=@n1
    ) -> (ReturnValue:bool=~) {is=title K2Node_PromotableOperator})
   > n3 = Set FoundIndex(FoundIndex=@n2.Array Index) -> (Output_Get=0)
  ]}
 add Custom Variant Channel {nodes [
   add Custom Variant Channel() {is=title K2Node_FunctionEntry}
   n1 = Get optionIndex() -> (optionIndex:int=0)
   n2 = Get selectableOptions()
   > Branch(Condition=Is Valid Index!(TargetArray:Array of ST Item Def Redirect Option Structures=@n2, IndexToTest:int=@n1) -> (ReturnValue:bool))
   n3 = Break ST Item Def Redirect Option(ST_ItemDefRedirectOption="Get (a ref)"(Array:Array of ST Item Def Redirect Option Structures=@n2, Dimension 1=@n1) -> (Output:"ST Item Def Redirect Option Structure (by ref)"))
   > Branch(Condition="Not Equal, Case Insenstive (String)"!(A:String=@n3, B:String=None) -> (ReturnValue:bool))
   > Execute Python Script!(
    execute
    PythonScript="import TechArt.CosmeticPicker.CosmeticPickerModule as CosmeticPicker\r\ntype = CosmeticPicker.addCharacterVariantChannel_ItemDefRedirect(id, item_variant_index, primary_asset_id)"
    id:Fort Item Definition Object Reference=Get Item() -> (Item:Fort Item Definition Object Reference)
    itemVariantIndex:int=Get itemVariantIndex() -> (itemVariantIndex:int=0)
    primary_asset_Id:Primary Asset Id Structure=@n3.PrimaryAssetID_5_#2
   ) -> (then, ReturnValue) {inputs [id, itemVariantIndex, primary_asset_Id]}
  ]}
 getCheatsValue {nodes [
   n1 = getCheatsValue!() -> (then, Item:Fort Item Definition Object Reference, ItemVariantIndex:int) {is=title K2Node_FunctionEntry}
   > PrintString(InString=Cheats check)
   > n2 = Execute Python Script!(
    execute
    PythonScript="import TechArt.CosmeticPicker.CosmeticPickerModule as CosmeticPicker\r\nprimary_asset_id = CosmeticPicker.GetItemRedirectUsedByCheats(item, item_variant_index)"
    item:Fort Item Definition Object Reference=@n1
    itemVariantIndex=@n1.ItemVariantIndex
   ) -> (then, ReturnValue, primary_asset_id:Primary Asset Id Structure) {inputs [item, itemVariantIndex], outputs [primary_asset_id]}
   > Return Node!(execute, primary_asset_id:Primary Asset Id Structure=@n2.primary_asset_id)
  ]}
 getMannequinOrDefaultValue {nodes [
   n1 = getMannequinOrDefaultValue!() -> (then, defaultIndex:int) {is=title K2Node_FunctionEntry}
   n2 = Self-Reference()
   > n3 = FindMannequinActorForCosmeticType!(
    execute
    CosmeticType:CosmeticPicker_CosmeticTypes Enum=Get cosmetic Type() -> (cosmetic Type:CosmeticPicker_CosmeticTypes Enum=NewEnumerator0)
    WidgetObject:Widget Object Reference=@n2
    __WorldContext:Object Reference=@n2
   ) -> (then, MannequinActor:Skeletal Mesh Actor Object Reference)
   > n4 = Is Valid(InputObject=@n3)
   > n5 = getMannequinValue!(
    execute
    self
    mannequinActor:Skeletal Mesh Actor Object Reference=@n3
    item:Fort Item Definition Object Reference=Get item() -> (item:Fort Item Definition Object Reference)
    VariantIndex:int=Get itemVariantIndex() -> (itemVariantIndex=0)
   ) -> (then, PrimaryAssetId:Primary Asset Id Structure)
   > n6 = FindSelectedIndexByPrimaryAssetId!(execute, self, PrimaryAssetId:Primary Asset Id Structure=@n5) -> (then, SelectedIndex:int=0) {is=title K2Node_CallFunction}
   > Return Node!(execute, index:int=@n6)
   Return Node(execute=@n4.Is Not Valid, index=@n1)
  ]}
 getMannequinValue {nodes [
   n1 = getMannequinValue!() -> (then, mannequinActor:Skeletal Mesh Actor Object Reference, item:Fort Item Definition Object Reference, VariantIndex:int) {is=title K2Node_FunctionEntry}
   > PrintString(InString=Mannequin check)
   > n2 = Execute Python Script!(
    execute
    PythonScript="import TechArt.CosmeticPicker.CosmeticPickerModule as CosmeticPicker\r\nprimary_asset_id = CosmeticPicker.checkMannequinForItemDefRedirectVariant(mannequin_actor, id, item_variant_index)"
    mannequinActor:Skeletal Mesh Actor Object Reference=@n1
    id:Fort Item Definition Object Reference=@n1.item
    itemVariantIndex=@n1.VariantIndex
   ) -> (then, ReturnValue, primary_asset_id) {inputs [mannequinActor, id, itemVariantIndex], outputs [primary_asset_id]}
   > Return Node!(execute, PrimaryAssetId:Primary Asset Id Structure=@n2.primary_asset_id)
  ]}
 getValidItemDefRedirectOptions {nodes [
   n1 = getValidItemDefRedirectOptions!() -> (then, item:Fort Item Definition Object Reference, index:int) {is=title K2Node_FunctionEntry}
   > n2 = Cast To AthenaCosmeticItemDefinition!(execute, Object=@n1) -> (then, CastFailed, AsAthena Cosmetic Item Definition:Athena Cosmetic Item Definition Object Reference) {is=title K2Node_DynamicCast}
   > n3 = GetItemCosmeticVariantByIndex!(execute, ItemDef:Athena Cosmetic Item Definition Object Reference=@n2, Index:int=@n1.index) -> (then, ReturnValue:Fort Cosmetic Variant Object Reference)
   > n4 = Cast To FortCosmeticItemDefRedirectVariant!(execute, Object=@n3) -> (then, CastFailed, AsFort Cosmetic Item Def Redirect Variant:Fort Cosmetic Item Def Redirect Variant Object Reference) {is=title K2Node_DynamicCast}
   > n5 = GetPrimaryAssetPrimaryAssetIdListByClass!(
    execute
    ItemDefClass:"Athena Cosmetic Item Definition Class Reference (by ref)"=Get ItemDefClass!(self:Fort Cosmetic Item Def Redirect Variant Object Reference=@n4) -> (ItemDefClass:Athena Cosmetic Item Definition Class Reference)
   ) -> (then, OutPrimaryAssetIdList:Array of Primary Asset Id Structures)
   > Add(
    TargetArray:Array of ST Item Def Redirect Option Structures=Get LocalSelectableOptions() -> (LocalSelectableOptions:Array of ST Item Def Redirect Option Structures)
    NewItem:"ST Item Def Redirect Option Structure (by ref)"=Make ST Item Def Redirect Option!(OptionName_2_#1:String=None, PrimaryAssetID_5_#2:Primary Asset Id Structure) -> (ST_ItemDefRedirectOption:ST Item Def Redirect Option Structure) {is=title K2Node_MakeStruct}
   ) {is=title K2Node_CallArrayFunction}
   > n6 = For Each Loop(Array:Array of Primary Asset Id Structures=@n5) -> (Array Element:Primary Asset Id Structure)
   Return Node!(execute=@n6.Completed, selectable_options:Array of ST Item Def Redirect Option Structures=Get LocalSelectableOptions())
   n7 = Break Primary Asset Id!(PrimaryAssetId:"Primary Asset Id Structure (by ref)"=@n6) -> (PrimaryAssetType:Primary Asset Type Structure, PrimaryAssetName:Name) {is=title K2Node_BreakStruct}
   > Add(
    TargetArray=Get LocalSelectableOptions()
    NewItem=Make ST Item Def Redirect Option(
     OptionName_2_#1="To String (Name)"(InName=@n7.PrimaryAssetName)
     PrimaryAssetID_5_#2=@n6
    )
   )
  ]}
 setup {nodes [
   n1 = setup!() -> (
    then
    cosmeticID:Fort Item Definition Object Reference
    itemVariantIndex:int
    wrapBox:Wrap Box Object Reference
    name:String
    cosmeticType:CosmeticPicker_CosmeticTypes Enum=NewEnumerator0
   ) {is=title K2Node_FunctionEntry}
   > Set Item(Item:Fort Item Definition Object Reference=@n1) -> (Output_Get:Fort Item Definition Object Reference)
   > Set itemVariantIndex(itemVariantIndex:int=@n1.itemVariantIndex) -> (Output_Get:int=0)
   > Set cosmetic Type(cosmetic Type:CosmeticPicker_CosmeticTypes Enum=@n1.cosmeticType) -> (Output_Get:CosmeticPicker_CosmeticTypes Enum=NewEnumerator0)
   > Set name(name:String=@n1.name) -> (Output_Get:String)
   > PrintString(InString="ItemRedirect->Setup")
   n2 = Self-Reference()
   > AddChild(self=@n1.wrapBox, Content=@n2)
   > SetVerticalAlignment(
    self:Wrap Box Slot Object Reference=SlotAsWrapBoxSlot!(Widget:Widget Object Reference=@n2) -> (ReturnValue:Wrap Box Slot Object Reference)
    InVerticalAlignment=VAlign_Bottom
   )
   > "SetText (Text)"(self=Get variantText_name() -> (variantText_name:Text Object Reference), InText="To Text (String)"(InString=Get name() -> (name:String)))
   > n3 = getValidItemDefRedirectOptions!(execute, self, item:Fort Item Definition Object Reference=Get Item(), index:int=Get itemVariantIndex() -> (itemVariantIndex=0)) -> (then, selectable_options:Array of ST Item Def Redirect Option Structures) {is=title K2Node_CallFunction}
   > Set selectableOptions(selectableOptions:Array of ST Item Def Redirect Option Structures=@n3) -> (Output_Get:Array of ST Item Def Redirect Option Structures)
   > Set optionIndex(optionIndex=0) -> (Output_Get=0)
   > n4 = Sequence!(execute) -> (then_0, then_1, then_2:Exec)
   setupWidget!(
    execute=@n4.then_2
    self
    name:String=Get name()
    options:Array of ST Item Def Redirect Option Structures=Get selectableOptions()
    selected:int=Get optionIndex() -> (optionIndex=0)
   ) -> (then)
   n5 = getCheatsValue!(
    execute=@n4.then_0
    self
    Item:Fort Item Definition Object Reference=Get Item()
    ItemVariantIndex:int=Get itemVariantIndex() -> (itemVariantIndex=0)
   ) -> (then, primary_asset_id:Primary Asset Id Structure) {is=title K2Node_CallFunction}
   > n6 = FindSelectedIndexByPrimaryAssetId(PrimaryAssetId=@n5) -> (SelectedIndex=0)
   > Set optionIndex(optionIndex=@n6) -> (Output_Get=0)
   n7 = getMannequinOrDefaultValue!(execute=@n4.then_1, self, defaultIndex:int=Get optionIndex() -> (optionIndex=0)) -> (then, index:int=0) {is=title K2Node_CallFunction}
   > Set optionIndex(optionIndex=@n7) -> (Output_Get=0)
  ]}
 setupWidget {nodes [
   n1 = setupWidget!() -> (then, name:String, options:Array of ST Item Def Redirect Option Structures, selected:int) {is=title K2Node_FunctionEntry}
   > ClearOptions(self:"Combo Box (String) Object Reference"=Get ItemDefsComboBox())
   > n2 = Set local_options(local_options:Array of ST Item Def Redirect Option Structures=@n1.options) -> (Output_Get:Array of ST Item Def Redirect Option Structures)
   n3 = Length(TargetArray:Array of ST Item Def Redirect Option Structures=@n2)
   > SetVisibility(self=Get ItemDefsComboBox(), InVisibility=Select(Option 0=Visible, Option 1=Collapsed, Index="Equal (Integer)"(A=@n3)))
   > n4 = For Each Loop(Array:Array of ST Item Def Redirect Option Structures=Get local_options() -> (local_options:Array of ST Item Def Redirect Option Structures)) -> (Array Element:ST Item Def Redirect Option Structure)
   SetSelectedIndex!(execute=@n4.Completed, self:"Combo Box (String) Object Reference"=Get ItemDefsComboBox(), Index:int=@n1.selected) -> (then)
   AddOption!(
    execute=@n4.LoopBody
    self:"Combo Box (String) Object Reference"=Get ItemDefsComboBox()
    Option:String=Break ST Item Def Redirect Option(ST_ItemDefRedirectOption=@n4)
   ) -> (then)
  ]}
]
variables [
 cosmetic Type {type=CosmeticPicker_CosmeticTypes, default=Character}
 Item {type=FortItemDefinition, default=null}
 itemVariantIndex {type=integer}
 name {type=string, default=""}
 optionIndex {type=integer}
 selectableOptions {type="array<ST_ItemDefRedirectOption>", default []}
]
widgets {
 info {rootWidgetClass=&CanvasPanel, widgetCount=5}
 widgets [
  CanvasPanel_0 {widgetClassPath=&CanvasPanel, widget=CanvasPanel_0}
  VerticalBox_64 {
   widgetClassPath=&VerticalBox
   parent=CanvasPanel_0
   widget=VerticalBox_64
   slot {
    class=CanvasPanelSlot
    properties {bAutoSize=true, layoutData {alignment=<0 0>, anchors {maximum=<0 0>, minimum=<0 0>}, offsets=<ltrb 0 0 300 30>}}
   }
  }
  variantText_name {
   widgetClassPath=&TextBlock
   bIsVariable=true
   parent=VerticalBox_64
   widget=variantText_name
   properties {
    text=-
    font {
     bForceMonospaced=false
     bMaterialIsStencil=false
     fontMaterial=null
     fontObject=&Roboto
     letterSpacing=0
     monospacedWidth=1
     size=10
     skewAmount=0
     typefaceFontName=Bold
     outlineSettings {
      bApplyOutlineToDropShadows=false
      bMiteredCorners=false
      bSeparateFillAlpha=false
      outlineMaterial=null
      outlineSize=0
      outlineColor=<rgba 0 0 0 1>
     }
    }
   }
   slot {class=VerticalBoxSlot}
  }
  ItemDefsComboBox {
   widgetClassPath=/Script/UMG.ComboBoxString
   bIsVariable=true
   parent=VerticalBox_64
   widget=ItemDefsComboBox
   properties {
    font {
     bForceMonospaced=false
     bMaterialIsStencil=false
     fontMaterial=null
     fontObject=&Roboto
     letterSpacing=0
     monospacedWidth=1
     size=8.5
     skewAmount=0
     typefaceFontName=Bold
     outlineSettings {
      bApplyOutlineToDropShadows=false
      bMiteredCorners=false
      bSeparateFillAlpha=false
      outlineMaterial=null
      outlineSize=0
      outlineColor=<rgba 0 0 0 1>
     }
    }
    foregroundColor {colorUseRule=UseColor_Specified, specifiedColor=<rgba 0.5271 0.5271 0.5271 1>}
    itemStyle {
     bUseParentRowBrush=false
     activeBrush {
      bIsDynamicallyLoaded=false
      drawAs=Image
      imageType=NoImage
      mirroring=NoMirror
      resourceName=None
      resourceObject=null
      tiling=NoTile
      imageSize=<0 0>
      margin=<ltrb 0 0 0 0>
      outlineSettings {
       bUseBrushTransparency=false
       roundingType=HalfHeightRadius
       width=0
       color {colorUseRule=UseColor_Specified, specifiedColor=<rgba 0 0 0 0>}
       cornerRadii=<0 0 0 0>
      }
      tintColor {colorUseRule=UseColor_Specified, specifiedColor=<rgba 0 0.162 0.7454 1>}
      uVRegion {bIsValid=false, max=<0 0>, min=<0 0>}
     }
     activeHighlightedBrush {
      bIsDynamicallyLoaded=false
      drawAs=Image
      imageType=NoImage
      mirroring=NoMirror
      resourceName=None
      resourceObject=null
      tiling=NoTile
      imageSize=<0 0>
      margin=<ltrb 0 0 0 0>
      outlineSettings {
       bUseBrushTransparency=false
       roundingType=HalfHeightRadius
       width=0
       color {colorUseRule=UseColor_Specified, specifiedColor=<rgba 0 0 0 0>}
       cornerRadii=<0 0 0 0>
      }
      tintColor {colorUseRule=UseColor_Specified, specifiedColor=<rgba 0.0252 0.0319 0.0423 1>}
      uVRegion {bIsValid=false, max=<0 0>, min=<0 0>}
     }
     activeHoveredBrush {
      bIsDynamicallyLoaded=false
      drawAs=Image
      imageType=NoImage
      mirroring=NoMirror
      resourceName=None
      resourceObject=null
      tiling=NoTile
      imageSize=<0 0>
      margin=<ltrb 0 0 0 0>
      outlineSettings {
       bUseBrushTransparency=false
       roundingType=HalfHeightRadius
       width=0
       color {colorUseRule=UseColor_Specified, specifiedColor=<rgba 0 0 0 0>}
       cornerRadii=<0 0 0 0>
      }
      tintColor {colorUseRule=UseColor_Specified, specifiedColor=<rgba 0 0.162 0.7454 1>}
      uVRegion {bIsValid=false, max=<0 0>, min=<0 0>}
     }
     dropIndicator_Above {
      bIsDynamicallyLoaded=false
      drawAs=Box
      imageType=FullColor
      mirroring=NoMirror
      resourceName=../../../Engine/Content/Slate/Common/DropZoneIndicator_Above.png
      resourceObject=null
      tiling=NoTile
      imageSize=<0 0>
      margin=<ltrb 0.625 0.625 0 0>
      outlineSettings {
       bUseBrushTransparency=false
       roundingType=HalfHeightRadius
       width=0
       color {colorUseRule=UseColor_Specified, specifiedColor=<rgba 0 0 0 0>}
       cornerRadii=<0 0 0 0>
      }
      tintColor {colorUseRule=UseColor_Specified, specifiedColor=<rgba 0.0091 0.3515 0.9647 1>}
      uVRegion {bIsValid=false, max=<0 0>, min=<0 0>}
     }
     dropIndicator_Below {
      bIsDynamicallyLoaded=false
      drawAs=Box
      imageType=FullColor
      mirroring=NoMirror
      resourceName=../../../Engine/Content/Slate/Common/DropZoneIndicator_Below.png
      resourceObject=null
      tiling=NoTile
      imageSize=<0 0>
      margin=<ltrb 0.625 0 0 0.625>
      outlineSettings {
       bUseBrushTransparency=false
       roundingType=HalfHeightRadius
       width=0
       color {colorUseRule=UseColor_Specified, specifiedColor=<rgba 0 0 0 0>}
       cornerRadii=<0 0 0 0>
      }
      tintColor {colorUseRule=UseColor_Specified, specifiedColor=<rgba 0.0091 0.3515 0.9647 1>}
      uVRegion {bIsValid=false, max=<0 0>, min=<0 0>}
     }
     dropIndicator_Onto {
      bIsDynamicallyLoaded=false
      drawAs=Box
      imageType=FullColor
      mirroring=NoMirror
      resourceName=../../../Engine/Content/Slate/Common/DropZoneIndicator_Onto.png
      resourceObject=null
      tiling=NoTile
      imageSize=<0 0>
      margin=<ltrb 0.25 0.25 0.25 0.25>
      outlineSettings {
       bUseBrushTransparency=false
       roundingType=HalfHeightRadius
       width=0
       color {colorUseRule=UseColor_Specified, specifiedColor=<rgba 0 0 0 0>}
       cornerRadii=<0 0 0 0>
      }
      tintColor {colorUseRule=UseColor_Specified, specifiedColor=<rgba 0.0091 0.3515 0.9647 1>}
      uVRegion {bIsValid=false, max=<0 0>, min=<0 0>}
     }
     evenRowBackgroundBrush {
      bIsDynamicallyLoaded=false
      drawAs=Image
      imageType=NoImage
      mirroring=NoMirror
      resourceName=None
      resourceObject=null
      tiling=NoTile
      imageSize=<0 0>
      margin=<ltrb 0 0 0 0>
      outlineSettings {
       bUseBrushTransparency=false
       roundingType=HalfHeightRadius
       width=0
       color {colorUseRule=UseColor_Specified, specifiedColor=<rgba 0 0 0 0>}
       cornerRadii=<0 0 0 0>
      }
      tintColor {colorUseRule=UseColor_Specified, specifiedColor=<rgba 0.0103 0.0103 0.0103 1>}
      uVRegion {bIsValid=false, max=<0 0>, min=<0 0>}
     }
     evenRowBackgroundHoveredBrush {
      bIsDynamicallyLoaded=false
      drawAs=Image
      imageType=NoImage
      mirroring=NoMirror
      resourceName=None
      resourceObject=null
      tiling=NoTile
      imageSize=<0 0>
      margin=<ltrb 0 0 0 0>
      outlineSettings {
       bUseBrushTransparency=false
       roundingType=HalfHeightRadius
       width=0
       color {colorUseRule=UseColor_Specified, specifiedColor=<rgba 0 0 0 0>}
       cornerRadii=<0 0 0 0>
      }
      tintColor {colorUseRule=UseColor_Specified, specifiedColor=<rgba 0.0176 0.0176 0.0176 1>}
      uVRegion {bIsValid=false, max=<0 0>, min=<0 0>}
     }
     inactiveBrush {
      bIsDynamicallyLoaded=false
      drawAs=Image
      imageType=NoImage
      mirroring=NoMirror
      resourceName=None
      resourceObject=null
      tiling=NoTile
      imageSize=<0 0>
      margin=<ltrb 0 0 0 0>
      outlineSettings {
       bUseBrushTransparency=false
       roundingType=HalfHeightRadius
       width=0
       color {colorUseRule=UseColor_Specified, specifiedColor=<rgba 0 0 0 0>}
       cornerRadii=<0 0 0 0>
      }
      tintColor {colorUseRule=UseColor_Specified, specifiedColor=<rgba 0.0513 0.0953 0.159 1>}
      uVRegion {bIsValid=false, max=<0 0>, min=<0 0>}
     }
     inactiveHighlightedBrush {
      bIsDynamicallyLoaded=false
      drawAs=Image
      imageType=NoImage
      mirroring=NoMirror
      resourceName=None
      resourceObject=null
      tiling=NoTile
      imageSize=<0 0>
      margin=<ltrb 0 0 0 0>
      outlineSettings {
       bUseBrushTransparency=false
       roundingType=HalfHeightRadius
       width=0
       color {colorUseRule=UseColor_Specified, specifiedColor=<rgba 0 0 0 0>}
       cornerRadii=<0 0 0 0>
      }
      tintColor {colorUseRule=UseColor_Specified, specifiedColor=<rgba 0.0252 0.0319 0.0423 1>}
      uVRegion {bIsValid=false, max=<0 0>, min=<0 0>}
     }
     inactiveHoveredBrush {
      bIsDynamicallyLoaded=false
      drawAs=Image
      imageType=NoImage
      mirroring=NoMirror
      resourceName=None
      resourceObject=null
      tiling=NoTile
      imageSize=<0 0>
      margin=<ltrb 0 0 0 0>
      outlineSettings {
       bUseBrushTransparency=false
       roundingType=HalfHeightRadius
       width=0
       color {colorUseRule=UseColor_Specified, specifiedColor=<rgba 0 0 0 0>}
       cornerRadii=<0 0 0 0>
      }
      tintColor {colorUseRule=UseColor_Specified, specifiedColor=<rgba 0.0513 0.0953 0.159 1>}
      uVRegion {bIsValid=false, max=<0 0>, min=<0 0>}
     }
     oddRowBackgroundBrush {
      bIsDynamicallyLoaded=false
      drawAs=Image
      imageType=NoImage
      mirroring=NoMirror
      resourceName=None
      resourceObject=null
      tiling=NoTile
      imageSize=<0 0>
      margin=<ltrb 0 0 0 0>
      outlineSettings {
       bUseBrushTransparency=false
       roundingType=HalfHeightRadius
       width=0
       color {colorUseRule=UseColor_Specified, specifiedColor=<rgba 0 0 0 0>}
       cornerRadii=<0 0 0 0>
      }
      tintColor {colorUseRule=UseColor_Specified, specifiedColor=<rgba 0.0103 0.0103 0.0103 1>}
      uVRegion {bIsValid=false, max=<0 0>, min=<0 0>}
     }
     oddRowBackgroundHoveredBrush {
      bIsDynamicallyLoaded=false
      drawAs=Image
      imageType=NoImage
      mirroring=NoMirror
      resourceName=None
      resourceObject=null
      tiling=NoTile
      imageSize=<0 0>
      margin=<ltrb 0 0 0 0>
      outlineSettings {
       bUseBrushTransparency=false
       roundingType=HalfHeightRadius
       width=0
       color {colorUseRule=UseColor_Specified, specifiedColor=<rgba 0 0 0 0>}
       cornerRadii=<0 0 0 0>
      }
      tintColor {colorUseRule=UseColor_Specified, specifiedColor=<rgba 0.0176 0.0176 0.0176 1>}
      uVRegion {bIsValid=false, max=<0 0>, min=<0 0>}
     }
     parentRowBackgroundBrush {
      bIsDynamicallyLoaded=false
      drawAs=Image
      imageType=NoImage
      mirroring=NoMirror
      resourceName=None
      resourceObject=null
      tiling=NoTile
      imageSize=<32 32>
      margin=<ltrb 0 0 0 0>
      outlineSettings {
       bUseBrushTransparency=false
       roundingType=HalfHeightRadius
       width=0
       color {colorUseRule=UseColor_Specified, specifiedColor=<rgba 0 0 0 0>}
       cornerRadii=<0 0 0 0>
      }
      tintColor {colorUseRule=UseColor_Specified, specifiedColor=<rgba 1 1 1 1>}
      uVRegion {bIsValid=false, max=<0 0>, min=<0 0>}
     }
     parentRowBackgroundHoveredBrush {
      bIsDynamicallyLoaded=false
      drawAs=Image
      imageType=NoImage
      mirroring=NoMirror
      resourceName=None
      resourceObject=null
      tiling=NoTile
      imageSize=<32 32>
      margin=<ltrb 0 0 0 0>
      outlineSettings {
       bUseBrushTransparency=false
       roundingType=HalfHeightRadius
       width=0
       color {colorUseRule=UseColor_Specified, specifiedColor=<rgba 0 0 0 0>}
       cornerRadii=<0 0 0 0>
      }
      tintColor {colorUseRule=UseColor_Specified, specifiedColor=<rgba 1 1 1 1>}
      uVRegion {bIsValid=false, max=<0 0>, min=<0 0>}
     }
     selectedTextColor {colorUseRule=UseColor_Specified, specifiedColor=<rgba 1 1 1 1>}
     selectorFocusedBrush {
      bIsDynamicallyLoaded=false
      drawAs=Border
      imageType=FullColor
      mirroring=NoMirror
      resourceName=../../../Engine/Content/Slate/Common/Selector.png
      resourceObject=null
      tiling=Both
      imageSize=<0 0>
      margin=<ltrb 0.25 0.25 0.25 0.25>
      outlineSettings {
       bUseBrushTransparency=false
       roundingType=HalfHeightRadius
       width=0
       color {colorUseRule=UseColor_Specified, specifiedColor=<rgba 0 0 0 0>}
       cornerRadii=<0 0 0 0>
      }
      tintColor {colorUseRule=UseColor_Specified, specifiedColor=<rgba 0.701 0.225 0.003 1>}
      uVRegion {bIsValid=false, max=<0 0>, min=<0 0>}
     }
     textColor {colorUseRule=UseColor_Specified, specifiedColor=<rgba 0.5271 0.5271 0.5271 1>}
    }
    scrollBarStyle {
     thickness=8
     draggedThumbImage {
      bIsDynamicallyLoaded=false
      drawAs=RoundedBox
      imageType=NoImage
      mirroring=NoMirror
      resourceName=None
      resourceObject=null
      tiling=NoTile
      imageSize=<0 0>
      margin=<ltrb 0 0 0 0>
      outlineSettings {
       bUseBrushTransparency=false
       roundingType=FixedRadius
       width=0
       color {colorUseRule=UseColor_Specified, specifiedColor=<rgba 0 0 0 0>}
       cornerRadii=<4 4 4 4>
      }
      tintColor {colorUseRule=UseColor_Specified, specifiedColor=<rgba 0.2159 0.2159 0.2159 1>}
      uVRegion {bIsValid=false, max=<0 0>, min=<0 0>}
     }
     horizontalBackgroundImage {
      bIsDynamicallyLoaded=false
      drawAs=NoDrawType
      imageType=NoImage
      mirroring=NoMirror
      resourceName=None
      resourceObject=null
      tiling=NoTile
      imageSize=<0 0>
      margin=<ltrb 0 0 0 0>
      outlineSettings {
       bUseBrushTransparency=false
       roundingType=HalfHeightRadius
       width=0
       color {colorUseRule=UseColor_Specified, specifiedColor=<rgba 0 0 0 0>}
       cornerRadii=<0 0 0 0>
      }
      tintColor {colorUseRule=UseColor_Specified, specifiedColor=<rgba 1 1 1 1>}
      uVRegion {bIsValid=false, max=<0 0>, min=<0 0>}
     }
     horizontalBottomSlotImage {
      bIsDynamicallyLoaded=false
      drawAs=NoDrawType
      imageType=NoImage
      mirroring=NoMirror
      resourceName=None
      resourceObject=null
      tiling=NoTile
      imageSize=<0 0>
      margin=<ltrb 0 0 0 0>
      outlineSettings {
       bUseBrushTransparency=false
       roundingType=HalfHeightRadius
       width=0
       color {colorUseRule=UseColor_Specified, specifiedColor=<rgba 0 0 0 0>}
       cornerRadii=<0 0 0 0>
      }
      tintColor {colorUseRule=UseColor_Specified, specifiedColor=<rgba 1 1 1 1>}
      uVRegion {bIsValid=false, max=<0 0>, min=<0 0>}
     }
     horizontalTopSlotImage {
      bIsDynamicallyLoaded=false
      drawAs=NoDrawType
      imageType=NoImage
      mirroring=NoMirror
      resourceName=None
      resourceObject=null
      tiling=NoTile
      imageSize=<0 0>
      margin=<ltrb 0 0 0 0>
      outlineSettings {
       bUseBrushTransparency=false
       roundingType=HalfHeightRadius
       width=0
       color {colorUseRule=UseColor_Specified, specifiedColor=<rgba 0 0 0 0>}
       cornerRadii=<0 0 0 0>
      }
      tintColor {colorUseRule=UseColor_Specified, specifiedColor=<rgba 1 1 1 1>}
      uVRegion {bIsValid=false, max=<0 0>, min=<0 0>}
     }
     hoveredThumbImage {
      bIsDynamicallyLoaded=false
      drawAs=RoundedBox
      imageType=NoImage
      mirroring=NoMirror
      resourceName=None
      resourceObject=null
      tiling=NoTile
      imageSize=<0 0>
      margin=<ltrb 0 0 0 0>
      outlineSettings {
       bUseBrushTransparency=false
       roundingType=FixedRadius
       width=0
       color {colorUseRule=UseColor_Specified, specifiedColor=<rgba 0 0 0 0>}
       cornerRadii=<4 4 4 4>
      }
      tintColor {colorUseRule=UseColor_Specified, specifiedColor=<rgba 0.2159 0.2159 0.2159 1>}
      uVRegion {bIsValid=false, max=<0 0>, min=<0 0>}
     }
     normalThumbImage {
      bIsDynamicallyLoaded=false
      drawAs=RoundedBox
      imageType=NoImage
      mirroring=NoMirror
      resourceName=None
      resourceObject=null
      tiling=NoTile
      imageSize=<0 0>
      margin=<ltrb 0 0 0 0>
      outlineSettings {
       bUseBrushTransparency=false
       roundingType=FixedRadius
       width=0
       color {colorUseRule=UseColor_Specified, specifiedColor=<rgba 0 0 0 0>}
       cornerRadii=<4 4 4 4>
      }
      tintColor {colorUseRule=UseColor_Specified, specifiedColor=<rgba 0.0953 0.0953 0.0953 1>}
      uVRegion {bIsValid=false, max=<0 0>, min=<0 0>}
     }
     verticalBackgroundImage {
      bIsDynamicallyLoaded=false
      drawAs=NoDrawType
      imageType=NoImage
      mirroring=NoMirror
      resourceName=None
      resourceObject=null
      tiling=NoTile
      imageSize=<0 0>
      margin=<ltrb 0 0 0 0>
      outlineSettings {
       bUseBrushTransparency=false
       roundingType=HalfHeightRadius
       width=0
       color {colorUseRule=UseColor_Specified, specifiedColor=<rgba 0 0 0 0>}
       cornerRadii=<0 0 0 0>
      }
      tintColor {colorUseRule=UseColor_Specified, specifiedColor=<rgba 1 1 1 1>}
      uVRegion {bIsValid=false, max=<0 0>, min=<0 0>}
     }
     verticalBottomSlotImage {
      bIsDynamicallyLoaded=false
      drawAs=NoDrawType
      imageType=NoImage
      mirroring=NoMirror
      resourceName=None
      resourceObject=null
      tiling=NoTile
      imageSize=<0 0>
      margin=<ltrb 0 0 0 0>
      outlineSettings {
       bUseBrushTransparency=false
       roundingType=HalfHeightRadius
       width=0
       color {colorUseRule=UseColor_Specified, specifiedColor=<rgba 0 0 0 0>}
       cornerRadii=<0 0 0 0>
      }
      tintColor {colorUseRule=UseColor_Specified, specifiedColor=<rgba 1 1 1 1>}
      uVRegion {bIsValid=false, max=<0 0>, min=<0 0>}
     }
     verticalTopSlotImage {
      bIsDynamicallyLoaded=false
      drawAs=NoDrawType
      imageType=NoImage
      mirroring=NoMirror
      resourceName=None
      resourceObject=null
      tiling=NoTile
      imageSize=<0 0>
      margin=<ltrb 0 0 0 0>
      outlineSettings {
       bUseBrushTransparency=false
       roundingType=HalfHeightRadius
       width=0
       color {colorUseRule=UseColor_Specified, specifiedColor=<rgba 0 0 0 0>}
       cornerRadii=<0 0 0 0>
      }
      tintColor {colorUseRule=UseColor_Specified, specifiedColor=<rgba 1 1 1 1>}
      uVRegion {bIsValid=false, max=<0 0>, min=<0 0>}
     }
    }
    widgetStyle {
     comboButtonStyle {
      downArrowAlign=VAlign_Center
      buttonStyle {
       clickedSlateSound {resourceObject=null}
       disabled {
        bIsDynamicallyLoaded=false
        drawAs=RoundedBox
        imageType=NoImage
        mirroring=NoMirror
        resourceName=None
        resourceObject=null
        tiling=NoTile
        imageSize=<0 0>
        margin=<ltrb 0 0 0 0>
        outlineSettings {
         bUseBrushTransparency=false
         roundingType=FixedRadius
         width=1
         color {colorUseRule=UseColor_ColorTable, specifiedColor=<rgba 0 0 0 0>}
         cornerRadii=<4 4 4 4>
        }
        tintColor {colorUseRule=UseColor_Specified, specifiedColor=<rgba 0.0395 0.0395 0.0395 1>}
        uVRegion {bIsValid=false, max=<0 0>, min=<0 0>}
       }
       disabledForeground {colorUseRule=UseColor_ColorTable, specifiedColor=<rgba 0 0 0 0>}
       hovered {
        bIsDynamicallyLoaded=false
        drawAs=RoundedBox
        imageType=NoImage
        mirroring=NoMirror
        resourceName=None
        resourceObject=null
        tiling=NoTile
        imageSize=<0 0>
        margin=<ltrb 0 0 0 0>
        outlineSettings {
         bUseBrushTransparency=false
         roundingType=FixedRadius
         width=1
         color {colorUseRule=UseColor_ColorTable, specifiedColor=<rgba 0 0 0 0>}
         cornerRadii=<4 4 4 4>
        }
        tintColor {colorUseRule=UseColor_Specified, specifiedColor=<rgba 0.0048 0.0048 0.0048 1>}
        uVRegion {bIsValid=false, max=<0 0>, min=<0 0>}
       }
       hoveredForeground {colorUseRule=UseColor_ColorTable, specifiedColor=<rgba 0 0 0 0>}
       hoveredSlateSound {resourceObject=null}
       normal {
        bIsDynamicallyLoaded=false
        drawAs=RoundedBox
        imageType=NoImage
        mirroring=NoMirror
        resourceName=None
        resourceObject=null
        tiling=NoTile
        imageSize=<0 0>
        margin=<ltrb 0 0 0 0>
        outlineSettings {
         bUseBrushTransparency=false
         roundingType=FixedRadius
         width=1
         color {colorUseRule=UseColor_ColorTable, specifiedColor=<rgba 0 0 0 0>}
         cornerRadii=<4 4 4 4>
        }
        tintColor {colorUseRule=UseColor_Specified, specifiedColor=<rgba 0.0048 0.0048 0.0048 1>}
        uVRegion {bIsValid=false, max=<0 0>, min=<0 0>}
       }
       normalForeground {colorUseRule=UseColor_ColorTable, specifiedColor=<rgba 0 0 0 0>}
       normalPadding=<ltrb 8 1 8 1>
       pressed {
        bIsDynamicallyLoaded=false
        drawAs=RoundedBox
        imageType=NoImage
        mirroring=NoMirror
        resourceName=None
        resourceObject=null
        tiling=NoTile
        imageSize=<0 0>
        margin=<ltrb 0 0 0 0>
        outlineSettings {
         bUseBrushTransparency=false
         roundingType=FixedRadius
         width=1
         color {colorUseRule=UseColor_ColorTable, specifiedColor=<rgba 0 0 0 0>}
         cornerRadii=<4 4 4 4>
        }
        tintColor {colorUseRule=UseColor_Specified, specifiedColor=<rgba 0.0048 0.0048 0.0048 1>}
        uVRegion {bIsValid=false, max=<0 0>, min=<0 0>}
       }
       pressedForeground {colorUseRule=UseColor_ColorTable, specifiedColor=<rgba 0 0 0 0>}
       pressedPadding=<ltrb 8 2 8 0>
       pressedSlateSound {resourceObject=null}
      }
      contentPadding=<ltrb 0 0 0 0>
      downArrowImage {
       bIsDynamicallyLoaded=false
       drawAs=Image
       imageType=Vector
       mirroring=NoMirror
       resourceName=../../../Engine/Content/Slate/Starship/CoreWidgets/ComboBox/wide-chevron-down.svg
       resourceObject=null
       tiling=NoTile
       imageSize=<16 16>
       margin=<ltrb 0 0 0 0>
       outlineSettings {
        bUseBrushTransparency=false
        roundingType=HalfHeightRadius
        width=0
        color {colorUseRule=UseColor_Specified, specifiedColor=<rgba 0 0 0 0>}
        cornerRadii=<0 0 0 0>
       }
       tintColor {colorUseRule=UseColor_Specified, specifiedColor=<rgba 1 1 1 1>}
       uVRegion {bIsValid=false, max=<0 0>, min=<0 0>}
      }
      downArrowPadding=<ltrb 2 2 2 2>
      menuBorderBrush {
       bIsDynamicallyLoaded=false
       drawAs=RoundedBox
       imageType=NoImage
       mirroring=NoMirror
       resourceName=None
       resourceObject=null
       tiling=NoTile
       imageSize=<0 0>
       margin=<ltrb 0 0 0 0>
       outlineSettings {
        bUseBrushTransparency=false
        roundingType=FixedRadius
        width=1
        color {colorUseRule=UseColor_ColorTable, specifiedColor=<rgba 0 0 0 0>}
        cornerRadii=<0 0 0 0>
       }
       tintColor {colorUseRule=UseColor_Specified, specifiedColor=<rgba 0.0395 0.0395 0.0395 1>}
       uVRegion {bIsValid=false, max=<0 0>, min=<0 0>}
      }
      menuBorderPadding=<ltrb 0 0 0 0>
      shadowColorAndOpacity=<rgba 0 0 0 1>
      shadowOffset=<0 0>
     }
     contentPadding=<ltrb 0 0 0 0>
     menuRowPadding=<ltrb 8 3 8 3>
     pressedSlateSound {resourceObject=null}
     selectionChangeSlateSound {resourceObject=null}
    }
   }
   slot {class=VerticalBoxSlot, properties {verticalAlignment=VAlign_Bottom}}
  }
  SizeBox {
   widgetClassPath=&SizeBox
   parent=VerticalBox_64
   widget=SizeBox
   properties {bOverride_WidthOverride=true, heightOverride=5, widthOverride=150}
   slot {class=VerticalBoxSlot}
  }
 ]
}
