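# Editor utility widget for one clickable color swatch in the Cosmetic Picker v2 art tool.
# Setup takes a color and name, tints the circle button through both ColorAndOpacity and normal style, and sets a name-plus-color tooltip.
# On click it prints a debug string, finds its parent RichColor widget by type, and passes the button's ColorAndOpacity to colorSelected, so the button itself stores the color.
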EditorUtilityWidgetBlueprint /Game/NeverCook/ArtTools/Tech_Art/CosmeticPicker/EUW_CosmeticPickerv2_ColorSwatch
graphs [
 EventGraph {nodes [
   "On Clicked (Button_swatch)"() {is=title K2Node_ComponentBoundEvent}
   > PrintString(InString=Pressed)
   > n1 = FindParentWidgetOfType!(
    execute
    StartingWidget:Widget Object Reference=Self-Reference()
    Type:Widget Class Reference=&/EUW_CosmeticPickerv2_RichColor.EUW_CosmeticPickerv2_RichColor_C
   ) -> (then, ReturnValue:EUW Cosmetic Pickerv 2 Rich Color Object Reference)
   > n2 = Cast To EUW_CosmeticPickerv2_RichColor!(execute, Object=@n1) -> (then, CastFailed, AsEUW Cosmetic Pickerv 2 Rich Color:EUW Cosmetic Pickerv 2 Rich Color Object Reference) {is=title K2Node_DynamicCast}
   > colorSelected!(
    execute
    self:EUW Cosmetic Pickerv 2 Rich Color Object Reference=@n2
    color:Linear Color Structure=Get ColorAndOpacity!(self:Button Object Reference=Get Button_swatch() -> (Button_swatch:Button Object Reference)) -> (ColorAndOpacity:Linear Color Structure)
   ) -> (then)
   Event PreConstruct()
   Event Construct()
   Event Tick()
  ]}
 Setup {nodes [
   n1 = Setup!() -> (then, color:Linear Color Structure, name:Text) {is=title K2Node_FunctionEntry}
   n2 = Get Button_swatch()
   > SetColorAndOpacity(self:Button Object Reference=@n2, InColorAndOpacity:Linear Color Structure=@n1)
   > SetStyle!(
    execute
    self:Button Object Reference=@n2
    InStyle_Normal_TintColor:Slate Color Structure=Make Slate Color(SpecifiedColor=@n1)
    InStyle_Normal_DrawAs:ESlateBrushDrawType Enum=Image
    InStyle_Normal_Tiling:ESlateBrushTileType Enum=NoTile
    InStyle_Normal_Mirroring:ESlateBrushMirrorType Enum=NoMirror
    InStyle_Normal_ImageSize:Vector 2D Structure=(X=16,Y=16)
    InStyle_Normal_Margin:Margin Structure=(Left=0,Top=0,Right=0,Bottom=0)
    InStyle_Normal_ResourceObject:Object Reference=/Game/UI/Friends_UI/Social/circle
    InStyle_Normal_OutlineSettings:Slate Brush Outline Settings Structure=(CornerRadii=(X=0,Y=0,Z=0,W=0),Color=(SpecifiedColor=(R=0,G=0,B=0,A=0),ColorUseRule=UseColor_Specified),Width=0,RoundingType=HalfHeightRadius,bUseBrushTransparency=False)
    InStyle_Hovered_TintColor_SpecifiedColor:Linear Color Structure=(R=0.9333,G=1,B=0.957,A=1)
    InStyle_Hovered_TintColor_ColorUseRule:ESlateColorStylingMode Enum
    InStyle_Hovered_DrawAs:ESlateBrushDrawType Enum=Image
    InStyle_Hovered_Tiling:ESlateBrushTileType Enum=NoTile
    InStyle_Hovered_Mirroring:ESlateBrushMirrorType Enum=NoMirror
    InStyle_Hovered_ImageSize:Vector 2D Structure=(X=16,Y=16)
    InStyle_Hovered_Margin:Margin Structure=(Left=0,Top=0,Right=0,Bottom=0)
    InStyle_Hovered_ResourceObject:Object Reference=&circle
    InStyle_Hovered_OutlineSettings:Slate Brush Outline Settings Structure=(CornerRadii=(X=0,Y=0,Z=0,W=0),Color=(SpecifiedColor=(R=0,G=0,B=0,A=0),ColorUseRule=UseColor_Specified),Width=0,RoundingType=HalfHeightRadius,bUseBrushTransparency=False)
    InStyle_Pressed:Slate Brush Structure=(TintColor=(SpecifiedColor=(R=1,G=1,B=1,A=1),ColorUseRule=UseColor_Specified),DrawAs=Image,Tiling=NoTile,Mirroring=NoMirror,ImageType=NoImage,ImageSize=(X=32,Y=32),Margin=(Left=0,Top=0,Right=0,Bottom=0),ResourceObject=None,OutlineSettings=(CornerRadii=(X=0,Y=0,Z=0,W=1),Color=(SpecifiedColor=(R=0,G=0,B=0,A=0),ColorUseRule=UseColor_Specified),Width=0,RoundingType=HalfHeightRadius,bUseBrushTransparency=False),UVRegion=(Min=(X=0,Y=0),Max=(X=0,Y=0),bIsValid=False),bIsDynamicallyLoaded=False,ResourceName="")
    InStyle_Disabled:Slate Brush Structure=(TintColor=(SpecifiedColor=(R=1,G=1,B=1,A=1),ColorUseRule=UseColor_Specified),DrawAs=NoDrawType,Tiling=NoTile,Mirroring=NoMirror,ImageType=NoImage,ImageSize=(X=0,Y=0),Margin=(Left=0,Top=0,Right=0,Bottom=0),ResourceObject=None,OutlineSettings=(CornerRadii=(X=0,Y=0,Z=0,W=1),Color=(SpecifiedColor=(R=0,G=0,B=0,A=0),ColorUseRule=UseColor_Specified),Width=0,RoundingType=HalfHeightRadius,bUseBrushTransparency=False),UVRegion=(Min=(X=0,Y=0),Max=(X=0,Y=0),bIsValid=False),bIsDynamicallyLoaded=False,ResourceName="")
    InStyle_NormalForeground:Slate Color Structure=(SpecifiedColor=(R=1,G=0,B=1,A=1),ColorUseRule=UseColor_Foreground)
    InStyle_HoveredForeground:Slate Color Structure=(SpecifiedColor=(R=1,G=0,B=1,A=1),ColorUseRule=UseColor_Foreground)
    InStyle_PressedForeground:Slate Color Structure=(SpecifiedColor=(R=1,G=0,B=1,A=1),ColorUseRule=UseColor_Foreground)
    InStyle_DisabledForeground:Slate Color Structure=(SpecifiedColor=(R=1,G=0,B=1,A=1),ColorUseRule=UseColor_Foreground)
    InStyle_NormalPadding:Margin Structure=(Left=0,Top=0,Right=0,Bottom=0)
    InStyle_PressedPadding:Margin Structure=(Left=0,Top=0,Right=0,Bottom=0)
    InStyle_PressedSlateSound:Slate Sound Structure=(ResourceObject=None)
    InStyle_ClickedSlateSound:Slate Sound Structure=(ResourceObject=None)
    InStyle_HoveredSlateSound:Slate Sound Structure=(ResourceObject=None)
   ) -> (then)
   n3 = "To String (Text)"(InText=@n1.name)
   n4 = "To String (LinearColor)"!(InColor:Linear Color Structure=@n1) -> (ReturnValue:String)
   > SetToolTipText!(
    execute
    self:Widget Object Reference=@n2
    InToolTipText:"Text (by ref)"="To Text (String)"(InString=Append!(A=@n3, B=" ", C:String=@n4) -> (ReturnValue))
   ) -> (then)
  ]}
]
variables []
widgets {
 info {rootWidgetClass=&CanvasPanel, widgetCount=2}
 widgets [
  CanvasPanel_0 {widgetClassPath=&CanvasPanel, widget=CanvasPanel_0}
  Button_swatch {
   widgetClassPath=&Button
   parent=CanvasPanel_0
   widget=Button_swatch
   properties {
    colorAndOpacity=<rgba 0.9 0.9 0.9 1>
    widgetStyle {
     clickedSlateSound {resourceObject=null}
     disabled {
      bIsDynamicallyLoaded=false
      drawAs=RoundedBox
      imageType=NoImage
      mirroring=NoMirror
      resourceName=None
      resourceObject=null
      tiling=NoTile
      imageSize=<0 0>
      margin=<ltrb 0 0 0 0>
      outlineSettings {
       bUseBrushTransparency=false
       roundingType=FixedRadius
       width=1
       color {colorUseRule=UseColor_ColorTable, specifiedColor=<rgba 0 0 0 0>}
       cornerRadii=<4 4 4 4>
      }
      tintColor {colorUseRule=UseColor_Specified, specifiedColor=<rgba 0.0395 0.0395 0.0395 1>}
      uVRegion {bIsValid=false, max=<0 0>, min=<0 0>}
     }
     disabledForeground {colorUseRule=UseColor_ColorTable, specifiedColor=<rgba 0 0 0 0>}
     hovered {
      bIsDynamicallyLoaded=false
      drawAs=Image
      imageType=NoImage
      mirroring=NoMirror
      resourceName=None
      resourceObject=&circle
      tiling=NoTile
      imageSize=<16 16>
      margin=<ltrb 0 0 0 0>
      outlineSettings {
       bUseBrushTransparency=false
       roundingType=FixedRadius
       width=1
       color {colorUseRule=UseColor_ColorTable, specifiedColor=<rgba 0 0 0 0>}
       cornerRadii=<4 4 4 4>
      }
      tintColor {colorUseRule=UseColor_Specified, specifiedColor=<rgba 0.0953 0.0953 0.0953 1>}
      uVRegion {bIsValid=false, max=<0 0>, min=<0 0>}
     }
     hoveredForeground {colorUseRule=UseColor_ColorTable, specifiedColor=<rgba 0 0 0 0>}
     hoveredSlateSound {resourceObject=null}
     normal {
      bIsDynamicallyLoaded=false
      drawAs=Image
      imageType=NoImage
      mirroring=NoMirror
      resourceName=None
      resourceObject=&circle
      tiling=NoTile
      imageSize=<16 16>
      margin=<ltrb 0 0 0 0>
      outlineSettings {
       bUseBrushTransparency=false
       roundingType=FixedRadius
       width=1
       color {colorUseRule=UseColor_ColorTable, specifiedColor=<rgba 0 0 0 0>}
       cornerRadii=<4 4 4 4>
      }
      tintColor {colorUseRule=UseColor_Specified, specifiedColor=<rgba 0.8 0.8 0.8 1>}
      uVRegion {bIsValid=false, max=<0 0>, min=<0 0>}
     }
     normalForeground {colorUseRule=UseColor_ColorTable, specifiedColor=<rgba 0 0 0 0>}
     normalPadding=<ltrb 1 1 1 1>
     pressed {
      bIsDynamicallyLoaded=false
      drawAs=Image
      imageType=NoImage
      mirroring=NoMirror
      resourceName=None
      resourceObject=/LauncherSocial/circle
      tiling=NoTile
      imageSize=<16 16>
      margin=<ltrb 0 0 0 0>
      outlineSettings {
       bUseBrushTransparency=false
       roundingType=FixedRadius
       width=1
       color {colorUseRule=UseColor_ColorTable, specifiedColor=<rgba 0 0 0 0>}
       cornerRadii=<4 4 4 4>
      }
      tintColor {colorUseRule=UseColor_Specified, specifiedColor=<rgba 0.0284 0.0284 0.0284 1>}
      uVRegion {bIsValid=false, max=<0 0>, min=<0 0>}
     }
     pressedForeground {colorUseRule=UseColor_ColorTable, specifiedColor=<rgba 0 0 0 0>}
     pressedPadding=<ltrb 1 1 1 0>
     pressedSlateSound {resourceObject=null}
    }
   }
   slot {class=CanvasPanelSlot, properties {layoutData {alignment=<0 0>, anchors {maximum=<0 0>, minimum=<0 0>}, offsets=<ltrb 0 0 16 16>}}}
  }
 ]
}
