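# Editor utility widget for the tech-art Cosmetic Picker that edits one RichColor variant channel of a cosmetic item.
# Setup fetches the channel's default color and swatch palette through Python, prefers the color found on the preview mannequin, then builds a swatch button per palette entry.
# Selecting a color updates the preview images and tells the parent Variants widget to update its custom channels.
# Add Custom Variant Channel passes the color and backend string to Python.
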EditorUtilityWidgetBlueprint /Game/NeverCook/ArtTools/Tech_Art/CosmeticPicker/EUW_CosmeticPickerv2_RichColor
graphs [
 Add Custom Variant Channel {nodes [
   Add Custom Variant Channel() {is=title K2Node_FunctionEntry}
   n1 = Get color() -> (color:Linear Color Structure)
   > Execute Python Script!(
    execute
    PythonScript="import TechArt.CosmeticPicker.CosmeticPickerModule as CosmeticPicker\r\ntype = CosmeticPicker.addCharacterVariantChannel_RichColor(id, item_variant_index, color, backend_color)"
    id:Fort Item Definition Object Reference=Get Id() -> (Id:Fort Item Definition Object Reference)
    itemVariantIndex:int=Get itemVariantIndex() -> (itemVariantIndex:int=0)
    color:Linear Color Structure=@n1
    BackendColor:String=ConvertColorToBackendDataString!(Color:"Linear Color Structure (by ref)"=@n1) -> (ReturnValue:String)
   ) -> (then, ReturnValue) {inputs [id, itemVariantIndex, color, BackendColor]}
  ]}
 EventGraph {nodes [
   Event PreConstruct()
   Event Construct()
   Event Tick()
  ]}
 FindMannequinActorForCosmeticType {nodes [
   n1 = FindMannequinActorForCosmeticType!() -> (then, CosmeticType:CosmeticPicker_CosmeticTypes Enum=NewEnumerator0) {is=title K2Node_FunctionEntry}
   > n2 = Sequence()
   > n3 = Branch(Condition="Not Equal (Enum)"!(A:CosmeticPicker_CosmeticTypes Enum=@n1, B:CosmeticPicker_CosmeticTypes Enum=NewEnumerator9) -> (ReturnValue:bool))
   > n4 = FindParentWidgetOfType!(
    execute
    StartingWidget:Widget Object Reference=Self-Reference()
    Type:Widget Class Reference=&/EUW_CosmeticPickerv2.EUW_CosmeticPickerv2_C
   ) -> (then, ReturnValue:EUW Cosmetic Pickerv 2 Object Reference)
   > n5 = Cast To EUW_CosmeticPickerv2!(execute, Object=@n4) -> (then, CastFailed, AsEUW Cosmetic Pickerv 2:EUW Cosmetic Pickerv 2 Object Reference) {is=title K2Node_DynamicCast}
   > Set Local_MannequinActor(Local_MannequinActor:Skeletal Mesh Actor Object Reference=Get Mannequin!(self:EUW Cosmetic Pickerv 2 Mannequin Object Reference=Get EUW_CosmeticPickerv2_Mannequin!(self:EUW Cosmetic Pickerv 2 Object Reference=@n5) -> (EUW_CosmeticPickerv2_Mannequin:EUW Cosmetic Pickerv 2 Mannequin Object Reference)) -> (Mannequin:Fort Player Mannequin Object Reference)) -> (Output_Get:Skeletal Mesh Actor Object Reference)
   n6 = FindParentWidgetOfType(
    execute=@n3.else
    StartingWidget=Self-Reference()
    Type=&/EUW_CosmeticPickerv2_Advanced.EUW_CosmeticPickerv2_Advanced_C
   ) -> (ReturnValue:EUW Cosmetic Pickerv 2 Advanced Object Reference)
   > n7 = Cast To EUW_CosmeticPickerv2_Advanced!(execute, Object=@n6) -> (then, CastFailed, AsEUW Cosmetic Pickerv 2 Advanced:EUW Cosmetic Pickerv 2 Advanced Object Reference) {is=title K2Node_DynamicCast}
   > Set Local_MannequinActor(Local_MannequinActor=Get CompanionMannequin!(self:EUW Cosmetic Pickerv 2 Companion Mannequin Object Reference=Get EUW_CosmeticPickerv2_CompanionMannequin!(self:EUW Cosmetic Pickerv 2 Advanced Object Reference=@n7) -> (EUW_CosmeticPickerv2_CompanionMannequin:EUW Cosmetic Pickerv 2 Companion Mannequin Object Reference)) -> (CompanionMannequin:Fort Companion Mannequin Object Reference))
   Return Node!(
    execute=@n2.then_1
    MannequinActor:Skeletal Mesh Actor Object Reference=Get Local_MannequinActor() -> (Local_MannequinActor:Skeletal Mesh Actor Object Reference)
   )
  ]}
 colorSelected {nodes [
   n1 = colorSelected!() -> (then, color:Linear Color Structure) {is=title K2Node_FunctionEntry}
   > n2 = Set color(color:Linear Color Structure=@n1) -> (Output_Get:Linear Color Structure)
   n3 = Get itemVariantIndex() -> (itemVariantIndex=0)
   > PrintString(InString=Append!(A="To String (Integer)"(InInt=@n3), B=" ", C:String="To String (LinearColor)"!(InColor:Linear Color Structure=@n2) -> (ReturnValue:String)) -> (ReturnValue))
   > setColor!(execute, self, color:Linear Color Structure=Get color()) -> (then)
   > n4 = FindParentWidgetOfType(
    StartingWidget=Self-Reference()
    Type=&/EUW_CosmeticPickerv2_Variants.EUW_CosmeticPickerv2_Variants_C
   ) -> (ReturnValue:EUW Cosmetic Pickerv 2 Variants Object Reference)
   > n5 = Cast To EUW_CosmeticPickerv2_Variants!(execute, Object=@n4) -> (then, CastFailed, AsEUW Cosmetic Pickerv 2 Variants:EUW Cosmetic Pickerv 2 Variants Object Reference) {is=title K2Node_DynamicCast}
   > Update Custom Variant Channels(self:EUW Cosmetic Pickerv 2 Variants Object Reference=@n5)
  ]}
 getColor {nodes [
   getColor() {is=title K2Node_FunctionEntry}
   n1 = Get color()
   > PrintString(InString="To String (LinearColor)"(InColor=@n1), TextColor=(R=0,G=1,B=0.1039,A=1))
   > Return Node!(execute, color:Linear Color Structure=@n1)
  ]}
 getMannequinOrDefaultValue {nodes [
   n1 = getMannequinOrDefaultValue!() -> (then, defaultVariantColor:Linear Color Structure) {is=title K2Node_FunctionEntry}
   n2 = Get itemVariantIndex() -> (itemVariantIndex=0)
   > PrintString(InString=Append!(A="RichColor - GetInitialValue ", B="To String (Integer)"(InInt=@n2)) -> (ReturnValue))
   n3 = Self-Reference()
   > n4 = FindMannequinActorForCosmeticType!(
    execute
    CosmeticType:CosmeticPicker_CosmeticTypes Enum=Get CosmeticType() -> (CosmeticType:CosmeticPicker_CosmeticTypes Enum=NewEnumerator0)
    WidgetObject:Widget Object Reference=@n3
    __WorldContext:Object Reference=@n3
   ) -> (then, MannequinActor:Skeletal Mesh Actor Object Reference) {is=title K2Node_CallFunction}
   > n5 = Is Valid(InputObject=@n4)
   > PrintString(InString="   valid mannequin")
   > n6 = getMannequinValue!(
    execute
    self
    mannequin:Skeletal Mesh Actor Object Reference=@n4
    id:Fort Item Definition Object Reference=Get Id()
    itemVariantIndex:int=Get itemVariantIndex() -> (itemVariantIndex=0)
    color:Linear Color Structure=@n1
   ) -> (then, outColor:Linear Color Structure)
   > Return Node(color=@n6)
   PrintString(execute=@n5.Is Not Valid, InString="   no mannequin")
   > Return Node(color=@n1)
  ]}
 getMannequinValue {nodes [
   n1 = getMannequinValue!() -> (
    then
    mannequin:Skeletal Mesh Actor Object Reference
    id:Fort Item Definition Object Reference
    itemVariantIndex:int
    color:Linear Color Structure
   ) {is=title K2Node_FunctionEntry}
   > PrintString(InString=Mannequin check)
   > n2 = Execute Python Script!(
    execute
    PythonScript="import TechArt.CosmeticPicker.CosmeticPickerModule as CosmeticPicker\r\nout_color = CosmeticPicker.checkMannequinForRichColorVariant(mannequin_actor, id, item_variant_index, color)"
    mannequinActor:Skeletal Mesh Actor Object Reference=@n1
    id=@n1.id
    itemVariantIndex=@n1.itemVariantIndex
    color=@n1.color
   ) -> (then, ReturnValue, outColor:Linear Color Structure) {inputs [mannequinActor, id, itemVariantIndex, color], outputs [outColor]}
   > Return Node!(execute, outColor:Linear Color Structure=@n2.outColor)
  ]}
 getRichColorVariantOptions {nodes [
   n1 = getRichColorVariantOptions!() -> (then, id:Fort Item Definition Object Reference, index:int) {is=title K2Node_FunctionEntry}
   > n2 = Execute Python Script!(
    execute
    PythonScript="import TechArt.CosmeticPicker.CosmeticPickerModule as CosmeticPicker\r\nresult = CosmeticPicker.getRichColorVariantOptions(id, index)\r\ndefault_color = result[0]\r\ncolor_swatch = result[1]"
    id=@n1
    index:int=@n1.index
   ) -> (then, ReturnValue, default_color:Linear Color Structure, color_swatch:Custom Color Swatch Object Reference) {inputs [id,index], outputs [default_color, color_swatch]}
   > Return Node!(execute, color_swatch:Custom Color Swatch Object Reference=@n2.color_swatch, default_color:Linear Color Structure=@n2.default_color)
  ]}
 setColor {nodes [
   n1 = setColor!() -> (then, color:Linear Color Structure) {is=title K2Node_FunctionEntry}
   > n2 = Set color(color=@n1)
   > SetColorAndOpacity(self:Image Object Reference=Get Image_Swatch() -> (Image_Swatch:Image Object Reference), InColorAndOpacity:Linear Color Structure=@n2)
   > SetColorAndOpacity(
    self=Get Image_Swatch_bg() -> (Image_Swatch_bg:Image Object Reference)
    InColorAndOpacity=LinearColor * LinearColor(A:Linear Color Structure=@n2, B:Linear Color Structure=(R=0.5,G=0.5,B=0.5,A=1)) -> (ReturnValue:Linear Color Structure) {is=title K2Node_PromotableOperator}
   )
  ]}
 setup {nodes [
   n1 = setup!() -> (
    then
    cosmeticID:Fort Item Definition Object Reference
    itemVariantIndex:int
    wrapBox:Wrap Box Object Reference
    name:String
    CosmeticType:CosmeticPicker_CosmeticTypes Enum=NewEnumerator0
   ) {is=title K2Node_FunctionEntry}
   > Set Id(Id:Fort Item Definition Object Reference=@n1) -> (Output_Get:Fort Item Definition Object Reference)
   > Set itemVariantIndex(itemVariantIndex:int=@n1.itemVariantIndex) -> (Output_Get:int=0)
   > Set CosmeticType(CosmeticType:CosmeticPicker_CosmeticTypes Enum=@n1.CosmeticType) -> (Output_Get:CosmeticPicker_CosmeticTypes Enum=NewEnumerator0)
   > Set name(name:String=@n1.name) -> (Output_Get:String)
   n2 = Self-Reference()
   > AddChild(self=@n1.wrapBox, Content=@n2)
   > SetVerticalAlignment(
    self:Wrap Box Slot Object Reference=SlotAsWrapBoxSlot!(Widget:Widget Object Reference=@n2) -> (ReturnValue:Wrap Box Slot Object Reference)
    InVerticalAlignment=VAlign_Bottom
   )
   > n3 = getRichColorVariantOptions!(execute, self, id:Fort Item Definition Object Reference=Get Id(), index:int=Get itemVariantIndex() -> (itemVariantIndex=0)) -> (then, color_swatch:Custom Color Swatch Object Reference, default_color:Linear Color Structure) {is=title K2Node_CallFunction}
   > n4 = getMannequinOrDefaultValue!(execute, self, defaultVariantColor:Linear Color Structure=@n3.default_color) -> (then, color:Linear Color Structure) {is=title K2Node_CallFunction}
   > setupWidget!(
    execute
    self
    name:String=Get name() -> (name:String)
    colorSwatch:Custom Color Swatch Object Reference=@n3
    color:Linear Color Structure=@n4
   ) -> (then)
  ]}
 setupWidget {nodes [
   n1 = setupWidget!() -> (then, name:String, colorSwatch:Custom Color Swatch Object Reference, color:Linear Color Structure) {is=title K2Node_FunctionEntry}
   > "SetText (Text)"(self=Get variantText_name() -> (variantText_name:Text Object Reference), InText="To Text (String)"(InString=@n1))
   > n2 = Cast To CustomDynamicColorSwatch!(execute, Object=@n1.colorSwatch) -> (then, CastFailed, AsCustom Dynamic Color Swatch:Custom Dynamic Color Swatch Object Reference) {is=title K2Node_DynamicCast}
   > Is Valid(InputObject=@n2)
   > n3 = For Each Loop!(
    Exec
    Array:Array of Color Swatch Pair Structures=Get ColorPairs!(self:Custom Dynamic Color Swatch Object Reference=@n2) -> (ColorPairs:Array of Color Swatch Pair Structures)
   ) -> (
    LoopBody
    Array Element_ColorName:Name
    Array Element_ColorValue:Linear Color Structure
    Array Element_ColorDisplayName:Text
    Array Index
    Completed
   )
   setColor(execute=@n3.Completed, color=@n1.color) {is=title K2Node_CallFunction}
   PrintString(
    execute=@n3.LoopBody
    InString=Append!(
     A="color "
     B="To String (Integer)"(InInt=@n3.Array Index)
     C:String=" "
     D:String="To String (Text)"(InText=@n3.Array Element_ColorDisplayName)
     E:String="To String (LinearColor)"(InColor=@n3.Array Element_ColorValue)
    ) -> (ReturnValue)
   )
   > n4 = Create Widget!(
    execute:Exec
    WorldContextObject:Object Reference
    Class:User Widget Class Reference=&/EUW_CosmeticPickerv2_ColorSwatch.EUW_CosmeticPickerv2_ColorSwatch_C
    OwningPlayer:Player Controller Object Reference
   ) -> (then:Exec, ReturnValue:EUW Cosmetic Pickerv 2 Color Swatch Object Reference)
   > AddChild(self=Get WrapBox_swatches() -> (WrapBox_swatches:Wrap Box Object Reference), Content=@n4)
   > Setup!(
    execute
    self:EUW Cosmetic Pickerv 2 Color Swatch Object Reference=@n4
    color:Linear Color Structure=@n3.Array Element_ColorValue
    name:Text=@n3.Array Element_ColorDisplayName
   ) -> (then)
  ]}
]
variables [
 color {type=LinearColor, default=<rgba 1 1 1 0>}
 ColorSwatch {type=CustomDynamicColorSwatch, default=/Game/Characters/CharacterColorSwatches/Misc/CS_Elastic_Common}
 CosmeticType {type=CosmeticPicker_CosmeticTypes, default=Character}
 Id {type=FortItemDefinition, default=null}
 itemVariantIndex {type=integer}
 name {type=string, default=""}
]
widgets {
 info {rootWidgetClass=&CanvasPanel, widgetCount=8}
 widgets [
  CanvasPanel_0 {widgetClassPath=&CanvasPanel, widget=CanvasPanel_0}
  VerticalBox_45 {
   widgetClassPath=&VerticalBox
   parent=CanvasPanel_0
   widget=VerticalBox_45
   slot {
    class=CanvasPanelSlot
    properties {bAutoSize=true, layoutData {alignment=<0 0>, anchors {maximum=<0 0>, minimum=<0 0>}, offsets=<ltrb 0 0 300 300>}}
   }
  }
  variantText_name {
   widgetClassPath=&TextBlock
   bIsVariable=true
   parent=VerticalBox_45
   widget=variantText_name
   properties {
    text=-
    font {
     bForceMonospaced=false
     bMaterialIsStencil=false
     fontMaterial=null
     fontObject=&Roboto
     letterSpacing=0
     monospacedWidth=1
     size=10
     skewAmount=0
     typefaceFontName=Bold
     outlineSettings {
      bApplyOutlineToDropShadows=false
      bMiteredCorners=false
      bSeparateFillAlpha=false
      outlineMaterial=null
      outlineSize=0
      outlineColor=<rgba 0 0 0 1>
     }
    }
   }
   slot {class=VerticalBoxSlot}
  }
  Spacer_146 {
   widgetClassPath=&Spacer
   parent=VerticalBox_45
   widget=Spacer_146
   properties {size=<160 1>}
   slot {class=VerticalBoxSlot, properties {horizontalAlignment=HAlign_Left, verticalAlignment=VAlign_Top}}
  }
  WrapBox_swatches {
   widgetClassPath=&WrapBox
   bIsVariable=true
   parent=VerticalBox_45
   widget=WrapBox_swatches
   properties {bExplicitWrapSize=true, wrapSize=160, innerSlotPadding=<4 3>}
   slot {class=VerticalBoxSlot, properties {padding=<ltrb 0 0 0 4>}}
  }
  Overlay_60 {widgetClassPath=&Overlay, parent=VerticalBox_45, widget=Overlay_60, slot {class=VerticalBoxSlot}}
  Image_Swatch_bg {
   widgetClassPath=&Image
   bIsVariable=true
   parent=Overlay_60
   widget=Image_Swatch_bg
   properties {
    brush {
     bIsDynamicallyLoaded=false
     drawAs=RoundedBox
     imageType=NoImage
     mirroring=NoMirror
     resourceName=None
     resourceObject=null
     tiling=NoTile
     imageSize=<32 32>
     margin=<ltrb 0 0 0 0>
     outlineSettings {
      bUseBrushTransparency=false
      roundingType=HalfHeightRadius
      width=0
      color {colorUseRule=UseColor_Specified, specifiedColor=<rgba 0 0 0 0>}
      cornerRadii=<0 0 0 0>
     }
     tintColor {colorUseRule=UseColor_Specified, specifiedColor=<rgba 1 1 1 1>}
     uVRegion {bIsValid=false, max=<0 0>, min=<0 0>}
    }
    colorAndOpacity=<rgba 0.2 0.2 0.2 1>
   }
   slot {class=OverlaySlot, properties {horizontalAlignment=HAlign_Fill, verticalAlignment=VAlign_Fill}}
  }
  Image_Swatch {
   widgetClassPath=&Image
   bIsVariable=true
   parent=Overlay_60
   widget=Image_Swatch
   properties {
    brush {
     bIsDynamicallyLoaded=false
     drawAs=RoundedBox
     imageType=NoImage
     mirroring=NoMirror
     resourceName=None
     resourceObject=null
     tiling=NoTile
     imageSize=<32 32>
     margin=<ltrb 0 0 0 0>
     outlineSettings {
      bUseBrushTransparency=false
      roundingType=HalfHeightRadius
      width=0
      color {colorUseRule=UseColor_Specified, specifiedColor=<rgba 0 0 0 0>}
      cornerRadii=<0 0 0 0>
     }
     tintColor {colorUseRule=UseColor_Specified, specifiedColor=<rgba 1 1 1 1>}
     uVRegion {bIsValid=false, max=<0 0>, min=<0 0>}
    }
    renderTransform {angle=0, scale=<0.97 0.85>, shear=<0 0>, translation=<0 0>}
   }
   slot {class=OverlaySlot, properties {horizontalAlignment=HAlign_Fill, verticalAlignment=VAlign_Fill}}
  }
 ]
}
