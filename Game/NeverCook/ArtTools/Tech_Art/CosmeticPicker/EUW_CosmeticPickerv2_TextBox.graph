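# Editor utility widget for the Cosmetic Picker v2 tool that provides a labeled text box for one text-input variant channel of a cosmetic item.
# setup stores the item and variant index, adds itself to a supplied wrap box and sets the label.
# Text edits update InputText and trigger Update Custom Variant Channels on the parent Variants widget.
# Add Custom Variant Channel passes these values to Python; option/default getters are stubs, their setup calls disconnected.
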EditorUtilityWidgetBlueprint /Game/NeverCook/ArtTools/Tech_Art/CosmeticPicker/EUW_CosmeticPickerv2_TextBox
graphs [
 Add Custom Variant Channel {nodes [
   Add Custom Variant Channel() {is=title K2Node_FunctionEntry}
   > Execute Python Script!(
    execute
    PythonScript="import TechArt.CosmeticPicker.CosmeticPickerModule as CosmeticPicker\r\ntype = CosmeticPicker.addCharacterVariantChannel_inputText(id, item_variant_index, value)"
    id:Fort Item Definition Object Reference=Get Id() -> (Id:Fort Item Definition Object Reference)
    itemVariantIndex:int=Get itemVariantIndex() -> (itemVariantIndex:int=0)
    value:Text=Get InputText() -> (InputText:Text)
   ) -> (then, ReturnValue) {inputs [id, itemVariantIndex, value]}
  ]}
 EventGraph {nodes [
   n1 = "On Text Changed (TextBox)"!() -> (OutputDelegate, then, Text:"Text (by ref)") {is=title K2Node_ComponentBoundEvent}
   > changeInputText!(execute, self, text:Text=@n1.Text) -> (then)
   Event PreConstruct()
   Event Construct()
   Event Tick()
  ]}
 changeInputText {nodes [
   n1 = changeInputText!() -> (then, text:Text) {is=title K2Node_FunctionEntry}
   > Set InputText(InputText:Text=@n1) -> (Output_Get:Text)
   > n2 = FindParentWidgetOfType!(
    execute
    StartingWidget:Widget Object Reference=Self-Reference()
    Type:Widget Class Reference=&/EUW_CosmeticPickerv2_Variants.EUW_CosmeticPickerv2_Variants_C
   ) -> (then, ReturnValue:EUW Cosmetic Pickerv 2 Variants Object Reference)
   > n3 = Cast To EUW_CosmeticPickerv2_Variants!(execute, Object=@n2) -> (then, CastFailed, AsEUW Cosmetic Pickerv 2 Variants:EUW Cosmetic Pickerv 2 Variants Object Reference) {is=title K2Node_DynamicCast}
   > Update Custom Variant Channels(self:EUW Cosmetic Pickerv 2 Variants Object Reference=@n3)
  ]}
 getCheatsValue {nodes [
   getCheatsValue!() -> (then, item:Fort Item Definition Object Reference, itemVariantIndex:int, defaultIndex:int) {is=title K2Node_FunctionEntry}
   > Return Node!(execute, index:int=0)
  ]}
 getDefaultVariantOption {nodes [
   getDefaultVariantOption!() -> (then, item:Fort Item Definition Object Reference, index:int) {is=title K2Node_FunctionEntry}
   > Return Node!(execute, defaultIndex:int=0)
  ]}
 getMannequinOrDefaultValue {nodes [
   getMannequinOrDefaultValue!() -> (then, defaultIndex:int) {is=title K2Node_FunctionEntry}
   > Return Node!(execute, index:int=0)
  ]}
 getVariantOptions {nodes [
   getVariantOptions!() -> (then, item:Fort Item Definition Object Reference, index:int) {is=title K2Node_FunctionEntry}
   > Return Node!(execute, option_names:Array of Strings)
  ]}
 setup {nodes [
   n1 = getVariantOptions!(execute, self, item:Fort Item Definition Object Reference=Get Id(), index:int=0) -> (then, option_names:Array of Strings) {is=title K2Node_CallFunction}
   > Set optionNames(optionNames:Array of Strings=@n1) -> (Output_Get:Array of Strings)
   n2 = Get Id()
   > n3 = getDefaultVariantOption!(execute, self, item:Fort Item Definition Object Reference=@n2, index:int=0) -> (then, defaultIndex:int=0) {is=title K2Node_CallFunction}
   > Set ID_Default Index(ID_Default Index:int=@n3) -> (Output_Get:int=0)
   > n4 = getCheatsValue!(
    execute
    self
    item:Fort Item Definition Object Reference=Get item() -> (item:Fort Item Definition Object Reference)
    itemVariantIndex:int=0
    defaultIndex:int=@n3
   ) -> (then, index:int=0) {is=title K2Node_CallFunction}
   > PrintString(InString=Append!(
     A=**** Default Variant Option
     B=GetDisplayName(Object=@n2)
     C:String=" "
     D:String="To String (Integer)"(InInt=0)
     E:String=" "
     F:String="To String (Integer)"(InInt=@n3)
    ) -> (ReturnValue))
   > getMannequinOrDefaultValue!(execute, self, defaultIndex:int=@n4) -> (then, index:int=0) {is=title K2Node_CallFunction}
   n5 = setup!() -> (then, cosmeticID:Fort Item Definition Object Reference, itemVariantIndex:int, wrapBox:Wrap Box Object Reference, name:String) {is=title K2Node_FunctionEntry}
   > Set Id(Id:Fort Item Definition Object Reference=@n5) -> (Output_Get:Fort Item Definition Object Reference)
   > Set itemVariantIndex(itemVariantIndex:int=@n5.itemVariantIndex) -> (Output_Get:int=0)
   > PrintString(InString="TextBox->Setup")
   > Set localName(localName:String=@n5.name) -> (Output_Get:String)
   n6 = Self-Reference()
   > AddChild(self=@n5.wrapBox, Content=@n6)
   > SetVerticalAlignment(
    self:Wrap Box Slot Object Reference=SlotAsWrapBoxSlot!(Widget:Widget Object Reference=@n6) -> (ReturnValue:Wrap Box Slot Object Reference)
    InVerticalAlignment=VAlign_Bottom
   )
   > "SetText (Text)"(
    self=Get variantText_name() -> (variantText_name:Text Object Reference)
    InText="To Text (String)"(InString=Get localName() -> (localName:String))
   )
   > "SetText (Text Box)"!(
    execute
    self:Text Box Object Reference=Get TextBox() -> (TextBox:Editor Utility Editable Text Box Object Reference)
    InText:Text=Get InputText()
   ) -> (then)
   Reroute Node(InputPin:int=@n3) -> (OutputPin:int)
  ]}
]
variables [
 Id {type=FortItemDefinition, default=null}
 InputText {type=string, default=Text}
 itemVariantIndex {type=integer}
]
widgets {
 info {rootWidgetClass=&CanvasPanel, widgetCount=5}
 widgets [
  CanvasPanel_0 {widgetClassPath=&CanvasPanel, widget=CanvasPanel_0}
  VerticalBox_38 {
   widgetClassPath=&VerticalBox
   parent=CanvasPanel_0
   widget=VerticalBox_38
   slot {
    class=CanvasPanelSlot
    properties {bAutoSize=true, layoutData {alignment=<0 0>, anchors {maximum=<0 0>, minimum=<0 0>}, offsets=<ltrb 0 0 160 30>}}
   }
  }
  variantText_name {
   widgetClassPath=&TextBlock
   bIsVariable=true
   parent=VerticalBox_38
   widget=variantText_name
   properties {
    text=-
    font {
     bForceMonospaced=false
     bMaterialIsStencil=false
     fontMaterial=null
     fontObject=&Roboto
     letterSpacing=0
     monospacedWidth=1
     size=10
     skewAmount=0
     typefaceFontName=Bold
     outlineSettings {
      bApplyOutlineToDropShadows=false
      bMiteredCorners=false
      bSeparateFillAlpha=false
      outlineMaterial=null
      outlineSize=0
      outlineColor=<rgba 0 0 0 1>
     }
    }
   }
   slot {class=VerticalBoxSlot}
  }
  TextBox {
   widgetClassPath=&EditorUtilityEditableTextBox
   bIsVariable=true
   parent=VerticalBox_38
   widget=TextBox
   properties {text=Text}
   slot {class=VerticalBoxSlot}
  }
  SizeBox_83 {
   widgetClassPath=&SizeBox
   parent=VerticalBox_38
   widget=SizeBox_83
   properties {bOverride_HeightOverride=true, bOverride_WidthOverride=true, widthOverride=150}
   slot {class=VerticalBoxSlot}
  }
 ]
}
