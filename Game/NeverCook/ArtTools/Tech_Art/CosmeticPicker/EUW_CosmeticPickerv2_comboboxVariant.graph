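# Editor utility widget for the Cosmetic Picker v2 tool, showing one cosmetic item variant as a labeled combo box.
# Setup adds it to a wrap box and uses Python CosmeticPickerModule calls to get options and default, then resolves selection from cheats and any mannequin actor.
# A capped text list stars the default option.
# Mouse selection updates the parent Variants widget's custom channels; another function registers the chosen channel through Python.
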EditorUtilityWidgetBlueprint /Game/NeverCook/ArtTools/Tech_Art/CosmeticPicker/EUW_CosmeticPickerv2_comboboxVariant
graphs [
 Add Custom Variant Channel {nodes [
   Add Custom Variant Channel() {is=title K2Node_FunctionEntry}
   n1 = Get itemVariantIndex() -> (itemVariantIndex:int=0)
   n2 = Get optionIndex() -> (optionIndex:int=0)
   > PrintString(InString=Append!(A="SetCustomVariantChannel ", B="To String (Integer)"(InInt=@n1), C:String=" ", D:String="To String (Integer)"(InInt=@n2)) -> (ReturnValue))
   n3 = Get item() -> (item:Fort Item Definition Object Reference)
   n4 = Get Cosmetic Type() -> (Cosmetic Type:CosmeticPicker_CosmeticTypes Enum=NewEnumerator0)
   > PrintString(InString=Append!(A=Enum to String!(Enumerator:CosmeticPicker_CosmeticTypes Enum=@n4) -> (ReturnValue:String), B=" ", C=GetDisplayName(Object=@n3)) -> (ReturnValue))
   > Execute Python Script!(
    execute
    PythonScript="import TechArt.CosmeticPicker.CosmeticPickerModule as CosmeticPicker\r\ntype = CosmeticPicker.addCharacterVariantChannel(id, item_variant_index, option_index)"
    id:Fort Item Definition Object Reference=Get item()
    itemVariantIndex:int=Get itemVariantIndex() -> (itemVariantIndex=0)
    optionIndex:int=Get optionIndex() -> (optionIndex=0)
   ) -> (then, ReturnValue) {inputs [id, itemVariantIndex, optionIndex]}
  ]}
 EventGraph {nodes [
   n1 = OnSelectionChanged!() -> (OutputDelegate, then, SelectedItem:String, SelectionType:ESelectInfo Enum) {is=title K2Node_CustomEvent}
   > Branch(Condition="Equal (Enum)"(A:ESelectInfo Enum=@n1.SelectionType, B:ESelectInfo Enum=OnMouseClick))
   > PrintString(InString=Append(A=@n1.SelectedItem, B=" ", C=Enum to String(Enumerator:ESelectInfo Enum=@n1.SelectionType)))
   > n2 = FindParentWidgetOfType!(
    execute
    StartingWidget:Widget Object Reference=Self-Reference()
    Type:Widget Class Reference=&/EUW_CosmeticPickerv2_Variants.EUW_CosmeticPickerv2_Variants_C
   ) -> (then, ReturnValue:EUW Cosmetic Pickerv 2 Variants Object Reference)
   > n3 = Cast To EUW_CosmeticPickerv2_Variants!(execute, Object=@n2) -> (then, CastFailed, AsEUW Cosmetic Pickerv 2 Variants:EUW Cosmetic Pickerv 2 Variants Object Reference) {is=title K2Node_DynamicCast}
   > Set optionIndex(optionIndex:int=GetSelectedIndex!(self:"Combo Box (String) Object Reference"=Get variantComboBox() -> (variantComboBox:"Combo Box (String) Object Reference")) -> (ReturnValue:int=0)) -> (Output_Get:int=0)
   > Update Custom Variant Channels(self:EUW Cosmetic Pickerv 2 Variants Object Reference=@n3)
   Event Construct()
   > Bind Event to On Selection Changed(self:"Combo Box (String) Object Reference"=Get variantComboBox(), Delegate=@n1) {is=title K2Node_AddDelegate}
   Event PreConstruct()
   Event Tick()
   Get itemVariantIndex() -> (itemVariantIndex=0)
  ]}
 Get Current Variant {nodes [
   Get Current Variant() {is=title K2Node_FunctionEntry}
   n1 = Get optionNames() -> (optionNames:Array of Strings)
   n2 = Get optionIndex() -> (optionIndex=0)
   > Return Node!(
    execute
    name:String=Get name() -> (name:String)
    optionName:String="Get (a copy)"(Array=@n1, Dimension 1=@n2)
    optionIndex:int=@n2
    numOptions:int=Length(TargetArray:Array of Strings=@n1)
    defaultIndex:int=Get ID_Default Index() -> (ID_Default Index:int=0)
   )
  ]}
 getCheatsValue {nodes [
   n1 = getCheatsValue!() -> (then, item:Fort Item Definition Object Reference, itemVariantIndex:int, defaultIndex:int) {is=title K2Node_FunctionEntry}
   > PrintString(InString=Cheats check)
   > n2 = Execute Python Script!(
    execute
    PythonScript="import TechArt.CosmeticPicker.CosmeticPickerModule as CosmeticPicker\r\nindex = CosmeticPicker.checkCheatsForComboboxVariant(item, item_variant_index, default_index)"
    item:Fort Item Definition Object Reference=@n1
    itemVariantIndex=@n1.itemVariantIndex
    defaultIndex:int=@n1.defaultIndex
   ) -> (then, ReturnValue, index:int) {inputs [item, itemVariantIndex, defaultIndex], outputs [index]}
   > Return Node!(execute, index:int=@n2.index)
  ]}
 getDefaultVariantOption {nodes [
   n1 = getDefaultVariantOption!() -> (then, item:Fort Item Definition Object Reference, index:int) {is=title K2Node_FunctionEntry}
   > n2 = Execute Python Script!(
    execute
    PythonScript="import TechArt.CosmeticPicker.CosmeticPickerModule as CosmeticPicker\r\ndefault_index = CosmeticPicker.getDefaultVariantDefaultOption(id, index)"
    id:Fort Item Definition Object Reference=@n1
    index:int=@n1.index
   ) -> (then, ReturnValue, default_index:int) {inputs [id,index], outputs [default_index]}
   > Return Node!(execute, defaultIndex:int=@n2.default_index)
  ]}
 getMannequinOrDefaultValue {nodes [
   n1 = getMannequinOrDefaultValue!() -> (then, defaultIndex:int) {is=title K2Node_FunctionEntry}
   n2 = Self-Reference()
   > n3 = FindMannequinActorForCosmeticType!(
    execute
    CosmeticType:CosmeticPicker_CosmeticTypes Enum=Get Cosmetic Type() -> (Cosmetic Type=NewEnumerator0)
    WidgetObject:Widget Object Reference=@n2
    __WorldContext:Object Reference=@n2
   ) -> (then, MannequinActor:Skeletal Mesh Actor Object Reference)
   > n4 = Is Valid(InputObject=@n3)
   > n5 = getMannequinValue!(
    execute
    self
    mannequin:Skeletal Mesh Actor Object Reference=@n3
    id:Fort Item Definition Object Reference=Get item()
    itemVariantIndex:int=Get itemVariantIndex() -> (itemVariantIndex=0)
    defaultIndex:int=@n1
   ) -> (then, index:int=0)
   > Return Node!(execute, index:int=@n5)
   Return Node(execute=@n4.Is Not Valid, index=@n1)
  ]}
 getMannequinValue {nodes [
   n1 = getMannequinValue!() -> (then, mannequin:Skeletal Mesh Actor Object Reference, id:Fort Item Definition Object Reference, itemVariantIndex:int, defaultIndex:int) {is=title K2Node_FunctionEntry}
   > PrintString(InString=Mannequin check)
   > n2 = Execute Python Script!(
    execute
    PythonScript="import TechArt.CosmeticPicker.CosmeticPickerModule as CosmeticPicker\r\nindex = CosmeticPicker.checkMannequinForComboboxVariant(mannequin_actor, id, item_variant_index, default_index)"
    mannequinActor:Skeletal Mesh Actor Object Reference=@n1
    id=@n1.id
    itemVariantIndex:int=@n1.itemVariantIndex
    defaultIndex:int=@n1.defaultIndex
   ) -> (then, ReturnValue, index:int) {inputs [mannequinActor, id, itemVariantIndex, defaultIndex], outputs [index]}
   > Return Node(index=@n2.index)
  ]}
 getVariantOptions {nodes [
   n1 = getVariantOptions!() -> (then, item:Fort Item Definition Object Reference, index:int) {is=title K2Node_FunctionEntry}
   > n2 = Execute Python Script!(
    execute
    PythonScript="import TechArt.CosmeticPicker.CosmeticPickerModule as CosmeticPicker\r\noption_names = CosmeticPicker.getDefaultVariantOptions(id, index)"
    id=@n1
    index:int=@n1.index
   ) -> (then, ReturnValue, option_names:Array of Strings) {inputs [id,index], outputs [option_names]}
   > Return Node!(execute, option_names:Array of Strings=@n2.option_names)
  ]}
 getVariantsParentWidget {nodes [
   getVariantsParentWidget() {is=title K2Node_FunctionEntry}
   > n1 = FindParentWidgetOfType(
    StartingWidget=Self-Reference()
    Type=&EUW_CosmeticPickerv2_Variants_C
   )
   > n2 = Cast To EUW_CosmeticPickerv2_Variants(Object=@n1)
   > Return Node!(execute, AsEUW Cosmetic Pickerv 2 Variants:EUW Cosmetic Pickerv 2 Variants Object Reference=@n2)
  ]}
 setup {nodes [
   n1 = setup!() -> (
    then
    cosmeticID:Fort Item Definition Object Reference
    itemVariantIndex:int
    wrapBox:Wrap Box Object Reference
    name:String
    CosmeticType:CosmeticPicker_CosmeticTypes Enum=NewEnumerator0
   ) {is=title K2Node_FunctionEntry}
   > Set item(item:Fort Item Definition Object Reference=@n1) -> (Output_Get:Fort Item Definition Object Reference)
   > Set itemVariantIndex(itemVariantIndex:int=@n1.itemVariantIndex) -> (Output_Get:int=0)
   > Set Cosmetic Type(Cosmetic Type:CosmeticPicker_CosmeticTypes Enum=@n1.CosmeticType) -> (Output_Get:CosmeticPicker_CosmeticTypes Enum=NewEnumerator0)
   > Set name(name:String=@n1.name) -> (Output_Get:String)
   n2 = Self-Reference()
   > AddChild(self=@n1.wrapBox, Content=@n2)
   > SetVerticalAlignment(
    self:Wrap Box Slot Object Reference=SlotAsWrapBoxSlot!(Widget:Widget Object Reference=@n2) -> (ReturnValue:Wrap Box Slot Object Reference)
    InVerticalAlignment=VAlign_Bottom
   )
   > n3 = getVariantOptions!(execute, self, item:Fort Item Definition Object Reference=Get item(), index:int=Get itemVariantIndex() -> (itemVariantIndex=0)) -> (then, option_names:Array of Strings) {is=title K2Node_CallFunction}
   > Set optionNames(optionNames:Array of Strings=@n3) -> (Output_Get:Array of Strings)
   n4 = Get itemVariantIndex() -> (itemVariantIndex=0)
   n5 = Get item()
   > n6 = getDefaultVariantOption!(execute, self, item:Fort Item Definition Object Reference=@n5, index:int=@n4) -> (then, defaultIndex:int=0) {is=title K2Node_CallFunction}
   > Set ID_Default Index(ID_Default Index:int=@n6) -> (Output_Get:int=0)
   > n7 = getCheatsValue!(
    execute
    self
    item:Fort Item Definition Object Reference=Get item()
    itemVariantIndex:int=Get itemVariantIndex() -> (itemVariantIndex=0)
    defaultIndex:int=@n6
   ) -> (then, index:int=0) {is=title K2Node_CallFunction}
   > PrintString(InString=Append!(
     A=**** Default Variant Option
     B=GetDisplayName(Object=@n5)
     C=" "
     D:String="To String (Integer)"(InInt=@n4)
     E:String=" "
     F:String="To String (Integer)"(InInt=@n6)
    ) -> (ReturnValue))
   > n8 = getMannequinOrDefaultValue!(execute, self, defaultIndex:int=@n7) -> (then, index:int=0) {is=title K2Node_CallFunction}
   > n9 = Set defaultIndex(defaultIndex:int=@n8) -> (Output_Get:int=0)
   > Set optionIndex(optionIndex=@n9) -> (Output_Get=0)
   > setupWidget!(
    execute
    self
    name:String=Get name()
    options:Array of Strings=Get optionNames()
    default:int=@n6
    selected:int=Get defaultIndex() -> (defaultIndex:int=0)
    Index:int=0
   ) -> (then)
  ]}
 setupWidget {nodes [
   n1 = setupWidget!() -> (then, name:String, options:Array of Strings, default:int, selected:int, Index:int) {is=title K2Node_FunctionEntry}
   > Set defaultIndex(defaultIndex=@n1.default) -> (Output_Get=0)
   > "SetText (Text)"(self=Get variantText_name() -> (variantText_name:Text Object Reference), InText="To Text (String)"(InString=@n1))
   > n2 = Set local_options(local_options:Array of Strings=@n1.options) -> (Output_Get:Array of Strings)
   > n3 = Branch(Condition="Equal (Integer)"(A=Length(TargetArray=@n2)))
   > SetVisibility(self=Get variantComboBox(), InVisibility=Collapsed)
   > n4 = SetVisibility(self=Get variantText() -> (variantText:Text Object Reference), InVisibility=Collapsed)
   ClearOptions(execute=[@n13.then, @n4.then], self:"Combo Box (String) Object Reference"=Get variantComboBox())
   > Set variantString(variantString:String) -> (Output_Get:String)
   > n5 = For Each Loop(Array=Get local_options() -> (local_options:Array of Strings))
   n6 = Get local_options()
   n7 = Branch(
    execute=@n5.Completed
    Condition="integer >= integer"!(A:int=Length(TargetArray=@n6), B:int=Get maxOptions() -> (maxOptions:int=0)) -> (ReturnValue:bool) {is=title K2Node_CallFunction}
   )
   > n8 = Set variantString(variantString=Append!(A=Get variantString() -> (variantString:String), B="<more>") -> (ReturnValue))
   SetSelectedIndex!(execute=[@n8.then, @n9.then], self:"Combo Box (String) Object Reference"=Get variantComboBox(), Index:int=@n1.selected) -> (then)
   > "SetText (Text)"(self=Get variantText(), InText="To Text (String)"(InString=Get variantString()))
   n9 = Set variantString(execute=@n7.else, variantString=LeftChop(SourceString=Get variantString(), Count=2))
   AddOption!(execute=@n5.LoopBody, self:"Combo Box (String) Object Reference"=Get variantComboBox(), Option:String=@n5) -> (then)
   > Branch(Condition="integer < integer"!(A:int=@n5.Array Index, B:int=Get maxOptions() -> (maxOptions=0)) -> (ReturnValue:bool) {is=title K2Node_CallFunction})
   > n10 = Branch(Condition="Equal (Integer)"(A=@n5.Array Index, B=Get defaultIndex() -> (defaultIndex=0)))
   > n11 = Set postfix(postfix:String=" *") -> (Output_Get:String)
   Set variantString(
    execute=[@n12.then, @n11.then]
    variantString=Append!(
     A=Get variantString()
     B="To String (Integer)"(InInt=@n5.Array Index)
     C:String=" : "
     D:String=@n5
     E:String=Get postfix() -> (postfix:String)
     F:String="\r\n"
    ) -> (ReturnValue)
   )
   n12 = Set postfix(execute=@n10.else)
   SetVisibility(execute=@n3.else, self=Get variantComboBox(), InVisibility=Visible)
   > n13 = SetVisibility(self=Get variantText(), InVisibility=Visible)
  ]}
]
variables [
 Cosmetic Type {type=CosmeticPicker_CosmeticTypes, default=Character}
 ID_Default Index {type=integer}
 item {type=FortItemDefinition, default=null}
 itemVariantIndex {type=integer}
 name {type=string, default=""}
 optionIndex {type=integer}
 optionNames {type="array<string>", default []}
]
widgets {
 info {rootWidgetClass=&CanvasPanel, widgetCount=7}
 widgets [
  CanvasPanel_0 {widgetClassPath=&CanvasPanel, widget=CanvasPanel_0}
  VerticalBox_45 {
   widgetClassPath=&VerticalBox
   parent=CanvasPanel_0
   widget=VerticalBox_45
   slot {
    class=CanvasPanelSlot
    properties {bAutoSize=true, layoutData {alignment=<0 0>, anchors {maximum=<0 0>, minimum=<0 0>}, offsets=<ltrb 0 0 300 30>}}
   }
  }
  variantText_name {
   widgetClassPath=&TextBlock
   bIsVariable=true
   parent=VerticalBox_45
   widget=variantText_name
   properties {
    text=-
    font {
     bForceMonospaced=false
     bMaterialIsStencil=false
     fontMaterial=null
     fontObject=&Roboto
     letterSpacing=0
     monospacedWidth=1
     size=10
     skewAmount=0
     typefaceFontName=Bold
     outlineSettings {
      bApplyOutlineToDropShadows=false
      bMiteredCorners=false
      bSeparateFillAlpha=false
      outlineMaterial=null
      outlineSize=0
      outlineColor=<rgba 0 0 0 1>
     }
    }
   }
   slot {class=VerticalBoxSlot}
  }
  Spacer_146 {
   widgetClassPath=&Spacer
   parent=VerticalBox_45
   widget=Spacer_146
   properties {size=<160 1>}
   slot {class=VerticalBoxSlot, properties {horizontalAlignment=HAlign_Left, verticalAlignment=VAlign_Top}}
  }
  variantText {
   widgetClassPath=&TextBlock
   bIsVariable=true
   parent=VerticalBox_45
   widget=variantText
   properties {
    text=-
    font {
     bForceMonospaced=false
     bMaterialIsStencil=false
     fontMaterial=null
     fontObject=&Roboto
     letterSpacing=0
     monospacedWidth=1
     size=10
     skewAmount=0
     typefaceFontName=Regular
     outlineSettings {
      bApplyOutlineToDropShadows=false
      bMiteredCorners=false
      bSeparateFillAlpha=false
      outlineMaterial=null
      outlineSize=0
      outlineColor=<rgba 0 0 0 1>
     }
    }
   }
   slot {class=VerticalBoxSlot}
  }
  variantComboBox {
   widgetClassPath=/Script/UMG.ComboBoxString
   bIsVariable=true
   parent=VerticalBox_45
   widget=variantComboBox
   properties {
    font {
     bForceMonospaced=false
     bMaterialIsStencil=false
     fontMaterial=null
     fontObject=&Roboto
     letterSpacing=0
     monospacedWidth=1
     size=10
     skewAmount=0
     typefaceFontName=Bold
     outlineSettings {
      bApplyOutlineToDropShadows=false
      bMiteredCorners=false
      bSeparateFillAlpha=false
      outlineMaterial=null
      outlineSize=0
      outlineColor=<rgba 0 0 0 1>
     }
    }
    foregroundColor {colorUseRule=UseColor_Specified, specifiedColor=<rgba 0.5271 0.5271 0.5271 1>}
    itemStyle {
     bUseParentRowBrush=false
     activeBrush {
      bIsDynamicallyLoaded=false
      drawAs=Image
      imageType=NoImage
      mirroring=NoMirror
      resourceName=None
      resourceObject=null
      tiling=NoTile
      imageSize=<0 0>
      margin=<ltrb 0 0 0 0>
      outlineSettings {
       bUseBrushTransparency=false
       roundingType=HalfHeightRadius
       width=0
       color {colorUseRule=UseColor_Specified, specifiedColor=<rgba 0 0 0 0>}
       cornerRadii=<0 0 0 0>
      }
      tintColor {colorUseRule=UseColor_Specified, specifiedColor=<rgba 0 0.162 0.7454 1>}
      uVRegion {bIsValid=false, max=<0 0>, min=<0 0>}
     }
     activeHighlightedBrush {
      bIsDynamicallyLoaded=false
      drawAs=Image
      imageType=NoImage
      mirroring=NoMirror
      resourceName=None
      resourceObject=null
      tiling=NoTile
      imageSize=<0 0>
      margin=<ltrb 0 0 0 0>
      outlineSettings {
       bUseBrushTransparency=false
       roundingType=HalfHeightRadius
       width=0
       color {colorUseRule=UseColor_Specified, specifiedColor=<rgba 0 0 0 0>}
       cornerRadii=<0 0 0 0>
      }
      tintColor {colorUseRule=UseColor_Specified, specifiedColor=<rgba 0.0252 0.0319 0.0423 1>}
      uVRegion {bIsValid=false, max=<0 0>, min=<0 0>}
     }
     activeHoveredBrush {
      bIsDynamicallyLoaded=false
      drawAs=Image
      imageType=NoImage
      mirroring=NoMirror
      resourceName=None
      resourceObject=null
      tiling=NoTile
      imageSize=<0 0>
      margin=<ltrb 0 0 0 0>
      outlineSettings {
       bUseBrushTransparency=false
       roundingType=HalfHeightRadius
       width=0
       color {colorUseRule=UseColor_Specified, specifiedColor=<rgba 0 0 0 0>}
       cornerRadii=<0 0 0 0>
      }
      tintColor {colorUseRule=UseColor_Specified, specifiedColor=<rgba 0 0.162 0.7454 1>}
      uVRegion {bIsValid=false, max=<0 0>, min=<0 0>}
     }
     dropIndicator_Above {
      bIsDynamicallyLoaded=false
      drawAs=Box
      imageType=FullColor
      mirroring=NoMirror
      resourceName=../../../Engine/Content/Slate/Common/DropZoneIndicator_Above.png
      resourceObject=null
      tiling=NoTile
      imageSize=<0 0>
      margin=<ltrb 0.625 0.625 0 0>
      outlineSettings {
       bUseBrushTransparency=false
       roundingType=HalfHeightRadius
       width=0
       color {colorUseRule=UseColor_Specified, specifiedColor=<rgba 0 0 0 0>}
       cornerRadii=<0 0 0 0>
      }
      tintColor {colorUseRule=UseColor_Specified, specifiedColor=<rgba 0.0091 0.3515 0.9647 1>}
      uVRegion {bIsValid=false, max=<0 0>, min=<0 0>}
     }
     dropIndicator_Below {
      bIsDynamicallyLoaded=false
      drawAs=Box
      imageType=FullColor
      mirroring=NoMirror
      resourceName=../../../Engine/Content/Slate/Common/DropZoneIndicator_Below.png
      resourceObject=null
      tiling=NoTile
      imageSize=<0 0>
      margin=<ltrb 0.625 0 0 0.625>
      outlineSettings {
       bUseBrushTransparency=false
       roundingType=HalfHeightRadius
       width=0
       color {colorUseRule=UseColor_Specified, specifiedColor=<rgba 0 0 0 0>}
       cornerRadii=<0 0 0 0>
      }
      tintColor {colorUseRule=UseColor_Specified, specifiedColor=<rgba 0.0091 0.3515 0.9647 1>}
      uVRegion {bIsValid=false, max=<0 0>, min=<0 0>}
     }
     dropIndicator_Onto {
      bIsDynamicallyLoaded=false
      drawAs=Box
      imageType=FullColor
      mirroring=NoMirror
      resourceName=../../../Engine/Content/Slate/Common/DropZoneIndicator_Onto.png
      resourceObject=null
      tiling=NoTile
      imageSize=<0 0>
      margin=<ltrb 0.25 0.25 0.25 0.25>
      outlineSettings {
       bUseBrushTransparency=false
       roundingType=HalfHeightRadius
       width=0
       color {colorUseRule=UseColor_Specified, specifiedColor=<rgba 0 0 0 0>}
       cornerRadii=<0 0 0 0>
      }
      tintColor {colorUseRule=UseColor_Specified, specifiedColor=<rgba 0.0091 0.3515 0.9647 1>}
      uVRegion {bIsValid=false, max=<0 0>, min=<0 0>}
     }
     evenRowBackgroundBrush {
      bIsDynamicallyLoaded=false
      drawAs=Image
      imageType=NoImage
      mirroring=NoMirror
      resourceName=None
      resourceObject=null
      tiling=NoTile
      imageSize=<0 0>
      margin=<ltrb 0 0 0 0>
      outlineSettings {
       bUseBrushTransparency=false
       roundingType=HalfHeightRadius
       width=0
       color {colorUseRule=UseColor_Specified, specifiedColor=<rgba 0 0 0 0>}
       cornerRadii=<0 0 0 0>
      }
      tintColor {colorUseRule=UseColor_Specified, specifiedColor=<rgba 0.0103 0.0103 0.0103 1>}
      uVRegion {bIsValid=false, max=<0 0>, min=<0 0>}
     }
     evenRowBackgroundHoveredBrush {
      bIsDynamicallyLoaded=false
      drawAs=Image
      imageType=NoImage
      mirroring=NoMirror
      resourceName=None
      resourceObject=null
      tiling=NoTile
      imageSize=<0 0>
      margin=<ltrb 0 0 0 0>
      outlineSettings {
       bUseBrushTransparency=false
       roundingType=HalfHeightRadius
       width=0
       color {colorUseRule=UseColor_Specified, specifiedColor=<rgba 0 0 0 0>}
       cornerRadii=<0 0 0 0>
      }
      tintColor {colorUseRule=UseColor_Specified, specifiedColor=<rgba 0.0176 0.0176 0.0176 1>}
      uVRegion {bIsValid=false, max=<0 0>, min=<0 0>}
     }
     inactiveBrush {
      bIsDynamicallyLoaded=false
      drawAs=Image
      imageType=NoImage
      mirroring=NoMirror
      resourceName=None
      resourceObject=null
      tiling=NoTile
      imageSize=<0 0>
      margin=<ltrb 0 0 0 0>
      outlineSettings {
       bUseBrushTransparency=false
       roundingType=HalfHeightRadius
       width=0
       color {colorUseRule=UseColor_Specified, specifiedColor=<rgba 0 0 0 0>}
       cornerRadii=<0 0 0 0>
      }
      tintColor {colorUseRule=UseColor_Specified, specifiedColor=<rgba 0.0513 0.0953 0.159 1>}
      uVRegion {bIsValid=false, max=<0 0>, min=<0 0>}
     }
     inactiveHighlightedBrush {
      bIsDynamicallyLoaded=false
      drawAs=Image
      imageType=NoImage
      mirroring=NoMirror
      resourceName=None
      resourceObject=null
      tiling=NoTile
      imageSize=<0 0>
      margin=<ltrb 0 0 0 0>
      outlineSettings {
       bUseBrushTransparency=false
       roundingType=HalfHeightRadius
       width=0
       color {colorUseRule=UseColor_Specified, specifiedColor=<rgba 0 0 0 0>}
       cornerRadii=<0 0 0 0>
      }
      tintColor {colorUseRule=UseColor_Specified, specifiedColor=<rgba 0.0252 0.0319 0.0423 1>}
      uVRegion {bIsValid=false, max=<0 0>, min=<0 0>}
     }
     inactiveHoveredBrush {
      bIsDynamicallyLoaded=false
      drawAs=Image
      imageType=NoImage
      mirroring=NoMirror
      resourceName=None
      resourceObject=null
      tiling=NoTile
      imageSize=<0 0>
      margin=<ltrb 0 0 0 0>
      outlineSettings {
       bUseBrushTransparency=false
       roundingType=HalfHeightRadius
       width=0
       color {colorUseRule=UseColor_Specified, specifiedColor=<rgba 0 0 0 0>}
       cornerRadii=<0 0 0 0>
      }
      tintColor {colorUseRule=UseColor_Specified, specifiedColor=<rgba 0.0513 0.0953 0.159 1>}
      uVRegion {bIsValid=false, max=<0 0>, min=<0 0>}
     }
     oddRowBackgroundBrush {
      bIsDynamicallyLoaded=false
      drawAs=Image
      imageType=NoImage
      mirroring=NoMirror
      resourceName=None
      resourceObject=null
      tiling=NoTile
      imageSize=<0 0>
      margin=<ltrb 0 0 0 0>
      outlineSettings {
       bUseBrushTransparency=false
       roundingType=HalfHeightRadius
       width=0
       color {colorUseRule=UseColor_Specified, specifiedColor=<rgba 0 0 0 0>}
       cornerRadii=<0 0 0 0>
      }
      tintColor {colorUseRule=UseColor_Specified, specifiedColor=<rgba 0.0103 0.0103 0.0103 1>}
      uVRegion {bIsValid=false, max=<0 0>, min=<0 0>}
     }
     oddRowBackgroundHoveredBrush {
      bIsDynamicallyLoaded=false
      drawAs=Image
      imageType=NoImage
      mirroring=NoMirror
      resourceName=None
      resourceObject=null
      tiling=NoTile
      imageSize=<0 0>
      margin=<ltrb 0 0 0 0>
      outlineSettings {
       bUseBrushTransparency=false
       roundingType=HalfHeightRadius
       width=0
       color {colorUseRule=UseColor_Specified, specifiedColor=<rgba 0 0 0 0>}
       cornerRadii=<0 0 0 0>
      }
      tintColor {colorUseRule=UseColor_Specified, specifiedColor=<rgba 0.0176 0.0176 0.0176 1>}
      uVRegion {bIsValid=false, max=<0 0>, min=<0 0>}
     }
     parentRowBackgroundBrush {
      bIsDynamicallyLoaded=false
      drawAs=Image
      imageType=NoImage
      mirroring=NoMirror
      resourceName=None
      resourceObject=null
      tiling=NoTile
      imageSize=<32 32>
      margin=<ltrb 0 0 0 0>
      outlineSettings {
       bUseBrushTransparency=false
       roundingType=HalfHeightRadius
       width=0
       color {colorUseRule=UseColor_Specified, specifiedColor=<rgba 0 0 0 0>}
       cornerRadii=<0 0 0 0>
      }
      tintColor {colorUseRule=UseColor_Specified, specifiedColor=<rgba 1 1 1 1>}
      uVRegion {bIsValid=false, max=<0 0>, min=<0 0>}
     }
     parentRowBackgroundHoveredBrush {
      bIsDynamicallyLoaded=false
      drawAs=Image
      imageType=NoImage
      mirroring=NoMirror
      resourceName=None
      resourceObject=null
      tiling=NoTile
      imageSize=<32 32>
      margin=<ltrb 0 0 0 0>
      outlineSettings {
       bUseBrushTransparency=false
       roundingType=HalfHeightRadius
       width=0
       color {colorUseRule=UseColor_Specified, specifiedColor=<rgba 0 0 0 0>}
       cornerRadii=<0 0 0 0>
      }
      tintColor {colorUseRule=UseColor_Specified, specifiedColor=<rgba 1 1 1 1>}
      uVRegion {bIsValid=false, max=<0 0>, min=<0 0>}
     }
     selectedTextColor {colorUseRule=UseColor_Specified, specifiedColor=<rgba 1 1 1 1>}
     selectorFocusedBrush {
      bIsDynamicallyLoaded=false
      drawAs=Border
      imageType=FullColor
      mirroring=NoMirror
      resourceName=../../../Engine/Content/Slate/Common/Selector.png
      resourceObject=null
      tiling=Both
      imageSize=<0 0>
      margin=<ltrb 0.25 0.25 0.25 0.25>
      outlineSettings {
       bUseBrushTransparency=false
       roundingType=HalfHeightRadius
       width=0
       color {colorUseRule=UseColor_Specified, specifiedColor=<rgba 0 0 0 0>}
       cornerRadii=<0 0 0 0>
      }
      tintColor {colorUseRule=UseColor_Specified, specifiedColor=<rgba 0.701 0.225 0.003 1>}
      uVRegion {bIsValid=false, max=<0 0>, min=<0 0>}
     }
     textColor {colorUseRule=UseColor_Specified, specifiedColor=<rgba 0.5271 0.5271 0.5271 1>}
    }
    scrollBarStyle {
     thickness=8
     draggedThumbImage {
      bIsDynamicallyLoaded=false
      drawAs=RoundedBox
      imageType=NoImage
      mirroring=NoMirror
      resourceName=None
      resourceObject=null
      tiling=NoTile
      imageSize=<0 0>
      margin=<ltrb 0 0 0 0>
      outlineSettings {
       bUseBrushTransparency=false
       roundingType=FixedRadius
       width=0
       color {colorUseRule=UseColor_Specified, specifiedColor=<rgba 0 0 0 0>}
       cornerRadii=<4 4 4 4>
      }
      tintColor {colorUseRule=UseColor_Specified, specifiedColor=<rgba 0.2159 0.2159 0.2159 1>}
      uVRegion {bIsValid=false, max=<0 0>, min=<0 0>}
     }
     horizontalBackgroundImage {
      bIsDynamicallyLoaded=false
      drawAs=NoDrawType
      imageType=NoImage
      mirroring=NoMirror
      resourceName=None
      resourceObject=null
      tiling=NoTile
      imageSize=<0 0>
      margin=<ltrb 0 0 0 0>
      outlineSettings {
       bUseBrushTransparency=false
       roundingType=HalfHeightRadius
       width=0
       color {colorUseRule=UseColor_Specified, specifiedColor=<rgba 0 0 0 0>}
       cornerRadii=<0 0 0 0>
      }
      tintColor {colorUseRule=UseColor_Specified, specifiedColor=<rgba 1 1 1 1>}
      uVRegion {bIsValid=false, max=<0 0>, min=<0 0>}
     }
     horizontalBottomSlotImage {
      bIsDynamicallyLoaded=false
      drawAs=NoDrawType
      imageType=NoImage
      mirroring=NoMirror
      resourceName=None
      resourceObject=null
      tiling=NoTile
      imageSize=<0 0>
      margin=<ltrb 0 0 0 0>
      outlineSettings {
       bUseBrushTransparency=false
       roundingType=HalfHeightRadius
       width=0
       color {colorUseRule=UseColor_Specified, specifiedColor=<rgba 0 0 0 0>}
       cornerRadii=<0 0 0 0>
      }
      tintColor {colorUseRule=UseColor_Specified, specifiedColor=<rgba 1 1 1 1>}
      uVRegion {bIsValid=false, max=<0 0>, min=<0 0>}
     }
     horizontalTopSlotImage {
      bIsDynamicallyLoaded=false
      drawAs=NoDrawType
      imageType=NoImage
      mirroring=NoMirror
      resourceName=None
      resourceObject=null
      tiling=NoTile
      imageSize=<0 0>
      margin=<ltrb 0 0 0 0>
      outlineSettings {
       bUseBrushTransparency=false
       roundingType=HalfHeightRadius
       width=0
       color {colorUseRule=UseColor_Specified, specifiedColor=<rgba 0 0 0 0>}
       cornerRadii=<0 0 0 0>
      }
      tintColor {colorUseRule=UseColor_Specified, specifiedColor=<rgba 1 1 1 1>}
      uVRegion {bIsValid=false, max=<0 0>, min=<0 0>}
     }
     hoveredThumbImage {
      bIsDynamicallyLoaded=false
      drawAs=RoundedBox
      imageType=NoImage
      mirroring=NoMirror
      resourceName=None
      resourceObject=null
      tiling=NoTile
      imageSize=<0 0>
      margin=<ltrb 0 0 0 0>
      outlineSettings {
       bUseBrushTransparency=false
       roundingType=FixedRadius
       width=0
       color {colorUseRule=UseColor_Specified, specifiedColor=<rgba 0 0 0 0>}
       cornerRadii=<4 4 4 4>
      }
      tintColor {colorUseRule=UseColor_Specified, specifiedColor=<rgba 0.2159 0.2159 0.2159 1>}
      uVRegion {bIsValid=false, max=<0 0>, min=<0 0>}
     }
     normalThumbImage {
      bIsDynamicallyLoaded=false
      drawAs=RoundedBox
      imageType=NoImage
      mirroring=NoMirror
      resourceName=None
      resourceObject=null
      tiling=NoTile
      imageSize=<0 0>
      margin=<ltrb 0 0 0 0>
      outlineSettings {
       bUseBrushTransparency=false
       roundingType=FixedRadius
       width=0
       color {colorUseRule=UseColor_Specified, specifiedColor=<rgba 0 0 0 0>}
       cornerRadii=<4 4 4 4>
      }
      tintColor {colorUseRule=UseColor_Specified, specifiedColor=<rgba 0.0953 0.0953 0.0953 1>}
      uVRegion {bIsValid=false, max=<0 0>, min=<0 0>}
     }
     verticalBackgroundImage {
      bIsDynamicallyLoaded=false
      drawAs=NoDrawType
      imageType=NoImage
      mirroring=NoMirror
      resourceName=None
      resourceObject=null
      tiling=NoTile
      imageSize=<0 0>
      margin=<ltrb 0 0 0 0>
      outlineSettings {
       bUseBrushTransparency=false
       roundingType=HalfHeightRadius
       width=0
       color {colorUseRule=UseColor_Specified, specifiedColor=<rgba 0 0 0 0>}
       cornerRadii=<0 0 0 0>
      }
      tintColor {colorUseRule=UseColor_Specified, specifiedColor=<rgba 1 1 1 1>}
      uVRegion {bIsValid=false, max=<0 0>, min=<0 0>}
     }
     verticalBottomSlotImage {
      bIsDynamicallyLoaded=false
      drawAs=NoDrawType
      imageType=NoImage
      mirroring=NoMirror
      resourceName=None
      resourceObject=null
      tiling=NoTile
      imageSize=<0 0>
      margin=<ltrb 0 0 0 0>
      outlineSettings {
       bUseBrushTransparency=false
       roundingType=HalfHeightRadius
       width=0
       color {colorUseRule=UseColor_Specified, specifiedColor=<rgba 0 0 0 0>}
       cornerRadii=<0 0 0 0>
      }
      tintColor {colorUseRule=UseColor_Specified, specifiedColor=<rgba 1 1 1 1>}
      uVRegion {bIsValid=false, max=<0 0>, min=<0 0>}
     }
     verticalTopSlotImage {
      bIsDynamicallyLoaded=false
      drawAs=NoDrawType
      imageType=NoImage
      mirroring=NoMirror
      resourceName=None
      resourceObject=null
      tiling=NoTile
      imageSize=<0 0>
      margin=<ltrb 0 0 0 0>
      outlineSettings {
       bUseBrushTransparency=false
       roundingType=HalfHeightRadius
       width=0
       color {colorUseRule=UseColor_Specified, specifiedColor=<rgba 0 0 0 0>}
       cornerRadii=<0 0 0 0>
      }
      tintColor {colorUseRule=UseColor_Specified, specifiedColor=<rgba 1 1 1 1>}
      uVRegion {bIsValid=false, max=<0 0>, min=<0 0>}
     }
    }
    widgetStyle {
     comboButtonStyle {
      downArrowAlign=VAlign_Center
      buttonStyle {
       clickedSlateSound {resourceObject=null}
       disabled {
        bIsDynamicallyLoaded=false
        drawAs=RoundedBox
        imageType=NoImage
        mirroring=NoMirror
        resourceName=None
        resourceObject=null
        tiling=NoTile
        imageSize=<0 0>
        margin=<ltrb 0 0 0 0>
        outlineSettings {
         bUseBrushTransparency=false
         roundingType=FixedRadius
         width=1
         color {colorUseRule=UseColor_ColorTable, specifiedColor=<rgba 0 0 0 0>}
         cornerRadii=<4 4 4 4>
        }
        tintColor {colorUseRule=UseColor_Specified, specifiedColor=<rgba 0.0395 0.0395 0.0395 1>}
        uVRegion {bIsValid=false, max=<0 0>, min=<0 0>}
       }
       disabledForeground {colorUseRule=UseColor_ColorTable, specifiedColor=<rgba 0 0 0 0>}
       hovered {
        bIsDynamicallyLoaded=false
        drawAs=RoundedBox
        imageType=NoImage
        mirroring=NoMirror
        resourceName=None
        resourceObject=null
        tiling=NoTile
        imageSize=<0 0>
        margin=<ltrb 0 0 0 0>
        outlineSettings {
         bUseBrushTransparency=false
         roundingType=FixedRadius
         width=1
         color {colorUseRule=UseColor_ColorTable, specifiedColor=<rgba 0 0 0 0>}
         cornerRadii=<4 4 4 4>
        }
        tintColor {colorUseRule=UseColor_Specified, specifiedColor=<rgba 0.0048 0.0048 0.0048 1>}
        uVRegion {bIsValid=false, max=<0 0>, min=<0 0>}
       }
       hoveredForeground {colorUseRule=UseColor_ColorTable, specifiedColor=<rgba 0 0 0 0>}
       hoveredSlateSound {resourceObject=null}
       normal {
        bIsDynamicallyLoaded=false
        drawAs=RoundedBox
        imageType=NoImage
        mirroring=NoMirror
        resourceName=None
        resourceObject=null
        tiling=NoTile
        imageSize=<0 0>
        margin=<ltrb 0 0 0 0>
        outlineSettings {
         bUseBrushTransparency=false
         roundingType=FixedRadius
         width=1
         color {colorUseRule=UseColor_ColorTable, specifiedColor=<rgba 0 0 0 0>}
         cornerRadii=<4 4 4 4>
        }
        tintColor {colorUseRule=UseColor_Specified, specifiedColor=<rgba 0.0048 0.0048 0.0048 1>}
        uVRegion {bIsValid=false, max=<0 0>, min=<0 0>}
       }
       normalForeground {colorUseRule=UseColor_ColorTable, specifiedColor=<rgba 0 0 0 0>}
       normalPadding=<ltrb 8 1 8 1>
       pressed {
        bIsDynamicallyLoaded=false
        drawAs=RoundedBox
        imageType=NoImage
        mirroring=NoMirror
        resourceName=None
        resourceObject=null
        tiling=NoTile
        imageSize=<0 0>
        margin=<ltrb 0 0 0 0>
        outlineSettings {
         bUseBrushTransparency=false
         roundingType=FixedRadius
         width=1
         color {colorUseRule=UseColor_ColorTable, specifiedColor=<rgba 0 0 0 0>}
         cornerRadii=<4 4 4 4>
        }
        tintColor {colorUseRule=UseColor_Specified, specifiedColor=<rgba 0.0048 0.0048 0.0048 1>}
        uVRegion {bIsValid=false, max=<0 0>, min=<0 0>}
       }
       pressedForeground {colorUseRule=UseColor_ColorTable, specifiedColor=<rgba 0 0 0 0>}
       pressedPadding=<ltrb 8 2 8 0>
       pressedSlateSound {resourceObject=null}
      }
      contentPadding=<ltrb 0 0 0 0>
      downArrowImage {
       bIsDynamicallyLoaded=false
       drawAs=Image
       imageType=Vector
       mirroring=NoMirror
       resourceName=../../../Engine/Content/Slate/Starship/CoreWidgets/ComboBox/wide-chevron-down.svg
       resourceObject=null
       tiling=NoTile
       imageSize=<16 16>
       margin=<ltrb 0 0 0 0>
       outlineSettings {
        bUseBrushTransparency=false
        roundingType=HalfHeightRadius
        width=0
        color {colorUseRule=UseColor_Specified, specifiedColor=<rgba 0 0 0 0>}
        cornerRadii=<0 0 0 0>
       }
       tintColor {colorUseRule=UseColor_Specified, specifiedColor=<rgba 1 1 1 1>}
       uVRegion {bIsValid=false, max=<0 0>, min=<0 0>}
      }
      downArrowPadding=<ltrb 2 2 2 2>
      menuBorderBrush {
       bIsDynamicallyLoaded=false
       drawAs=RoundedBox
       imageType=NoImage
       mirroring=NoMirror
       resourceName=None
       resourceObject=null
       tiling=NoTile
       imageSize=<0 0>
       margin=<ltrb 0 0 0 0>
       outlineSettings {
        bUseBrushTransparency=false
        roundingType=FixedRadius
        width=1
        color {colorUseRule=UseColor_ColorTable, specifiedColor=<rgba 0 0 0 0>}
        cornerRadii=<0 0 0 0>
       }
       tintColor {colorUseRule=UseColor_Specified, specifiedColor=<rgba 0.0395 0.0395 0.0395 1>}
       uVRegion {bIsValid=false, max=<0 0>, min=<0 0>}
      }
      menuBorderPadding=<ltrb 0 0 0 0>
      shadowColorAndOpacity=<rgba 0 0 0 1>
      shadowOffset=<0 0>
     }
     contentPadding=<ltrb 0 0 0 0>
     menuRowPadding=<ltrb 8 3 8 3>
     pressedSlateSound {resourceObject=null}
     selectionChangeSlateSound {resourceObject=null}
    }
   }
   slot {class=VerticalBoxSlot, properties {verticalAlignment=VAlign_Bottom}}
  }
  SizeBox_83 {
   widgetClassPath=&SizeBox
   parent=VerticalBox_45
   widget=SizeBox_83
   properties {bOverride_WidthOverride=true, widthOverride=150}
   slot {class=VerticalBoxSlot}
  }
 ]
}
