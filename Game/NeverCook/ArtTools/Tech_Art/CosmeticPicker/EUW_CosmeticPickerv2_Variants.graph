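# Editor utility widget for the variants panel of the Cosmetic Picker v2 tech art tool, editing a cosmetic item's variant channels.
# updateVariantWidgets queries CosmeticPickerModule through Python, then fills a wrap box with type-specific child widgets, defaulting to a combobox.
# Update Custom Variant Channels deletes channels by cosmetic type, re-adds them from each child through a cast chain, then updates the mannequin via the parent ID widget.
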EditorUtilityWidgetBlueprint /Game/NeverCook/ArtTools/Tech_Art/CosmeticPicker/EUW_CosmeticPickerv2_Variants
graphs [
 Delete Custom Variant Channels {nodes [
   n1 = Delete Custom Variant Channels!() -> (then, CosmeticType:CosmeticPicker_CosmeticTypes Enum=NewEnumerator0) {is=title K2Node_FunctionEntry}
   > Execute Python Script!(
    execute
    PythonScript="import TechArt.CosmeticPicker.CosmeticPickerModule as CosmeticPicker\r\nfrom importlib import *\r\nreload(CosmeticPicker)\r\nCosmeticPicker.deleteCharacterVariantChannelsByType(type_index)"
    typeIndex:CosmeticPicker_CosmeticTypes Enum=@n1
   ) -> (then, ReturnValue) {inputs [typeIndex]}
  ]}
 EventGraph {nodes [
   Event PreConstruct()
   Event Construct()
   Event Tick()
  ]}
 Set Custom Variant Channels {nodes [
   Set Custom Variant Channels() {is=title K2Node_FunctionEntry}
   > n1 = For Each Loop(Array:Array of Widget Object References=GetAllChildren(self=Get WrapBox_variants() -> (WrapBox_variants:Wrap Box Object Reference))) -> (Array Element:Widget Object Reference)
   updateMannequin(execute=@n1.Completed)
   PrintString(
    execute=@n1.LoopBody
    InString=Append!(A="To String (Integer)"(InInt=@n1.Array Index), B=" ", C:String=GetDisplayName(Object=@n1)) -> (ReturnValue)
   )
   > n2 = Cast To EUW_CosmeticPickerv2_comboboxVariant!(execute, Object=@n1) -> (then, CastFailed, AsEUW Cosmetic Pickerv 2 Combobox Variant:EUW Cosmetic Pickerv 2 Combobox Variant Object Reference) {is=title K2Node_DynamicCast}
   > Add Custom Variant Channel(self:EUW Cosmetic Pickerv 2 Combobox Variant Object Reference=@n2)
   n3 = Cast To EUW_CosmeticPickerv2_RichColor!(execute=@n2.CastFailed, Object=@n1) -> (then, CastFailed, AsEUW Cosmetic Pickerv 2 Rich Color:EUW Cosmetic Pickerv 2 Rich Color Object Reference) {is=title K2Node_DynamicCast}
   > Add Custom Variant Channel(self:EUW Cosmetic Pickerv 2 Rich Color Object Reference=@n3)
   n4 = Cast To EUW_CosmeticPickerv2_FloatSlider!(execute=@n3.CastFailed, Object=@n1) -> (then, CastFailed, AsEUW Cosmetic Pickerv 2 Float Slider:EUW Cosmetic Pickerv 2 Float Slider Object Reference) {is=title K2Node_DynamicCast}
   > Add Custom Variant Channel(self:EUW Cosmetic Pickerv 2 Float Slider Object Reference=@n4)
   n5 = Cast To EUW_CosmeticPickerv2_ApplyWrap!(execute=@n4.CastFailed, Object=@n1) -> (then, CastFailed, AsEUW Cosmetic Pickerv 2 Apply Wrap:EUW Cosmetic Pickerv 2 Apply Wrap Object Reference) {is=title K2Node_DynamicCast}
   > Add Custom Variant Channel(self:EUW Cosmetic Pickerv 2 Apply Wrap Object Reference=@n5)
   n6 = Cast To EUW_CosmeticPickerv2_Numerical!(execute=@n5.CastFailed, Object=@n1) -> (then, CastFailed, AsEUW Cosmetic Pickerv 2 Numerical:EUW Cosmetic Pickerv 2 Numerical Object Reference) {is=title K2Node_DynamicCast}
   > Add Custom Variant Channel(self:EUW Cosmetic Pickerv 2 Numerical Object Reference=@n6)
   n7 = Cast To EUW_CosmeticPickerv2_CustomizableSpray!(execute=@n6.CastFailed, Object=@n1) -> (then, CastFailed, AsEUW Cosmetic Pickerv 2 Customizable Spray:EUW Cosmetic Pickerv 2 Customizable Spray Object Reference) {is=title K2Node_DynamicCast}
   > Add Custom Variant Channel(self:EUW Cosmetic Pickerv 2 Customizable Spray Object Reference=@n7)
   n8 = Cast To EUW_CosmeticPickerv2_TextBox!(execute=@n7.CastFailed, Object=@n1) -> (then, CastFailed, AsEUW Cosmetic Pickerv 2 Text Box:EUW Cosmetic Pickerv 2 Text Box Object Reference) {is=title K2Node_DynamicCast}
   > Add Custom Variant Channel(self:EUW Cosmetic Pickerv 2 Text Box Object Reference=@n8)
   n9 = Cast To EUW_CosmeticPickerv2_ItemDefRedirect!(execute=@n8.CastFailed, Object=@n1) -> (then, CastFailed, AsEUW Cosmetic Pickerv 2 Item Def Redirect:EUW Cosmetic Pickerv 2 Item Def Redirect Object Reference) {is=title K2Node_DynamicCast}
   > add Custom Variant Channel(self:EUW Cosmetic Pickerv 2 Item Def Redirect Object Reference=@n9)
  ]}
 Update Custom Variant Channels {nodes [
   Update Custom Variant Channels() {is=title K2Node_FunctionEntry}
   > PrintString(InString=Variant Changed)
   > Delete Custom Variant Channels!(
    execute
    self
    CosmeticType:CosmeticPicker_CosmeticTypes Enum=Get cosmeticType() -> (cosmeticType:CosmeticPicker_CosmeticTypes Enum=NewEnumerator0)
   ) -> (then) {is=title K2Node_CallFunction}
   > Set Custom Variant Channels() {is=title K2Node_CallFunction}
  ]}
 addDefaultWidget {nodes [
   n1 = addDefaultWidget!() -> (then, id:Fort Item Definition Object Reference, index:int, name:String) {is=title K2Node_FunctionEntry}
   > n2 = Create Widget!(
    execute:Exec
    WorldContextObject:Object Reference
    Class:User Widget Class Reference=&/EUW_CosmeticPickerv2_comboboxVariant.EUW_CosmeticPickerv2_comboboxVariant_C
    OwningPlayer:Player Controller Object Reference
   ) -> (then:Exec, ReturnValue:EUW Cosmetic Pickerv 2 Combobox Variant Object Reference)
   > setup!(
    execute
    self:EUW Cosmetic Pickerv 2 Combobox Variant Object Reference=@n2
    cosmeticID:Fort Item Definition Object Reference=@n1
    itemVariantIndex:int=@n1.index
    wrapBox:Wrap Box Object Reference=Get WrapBox_variants()
    name:String=@n1.name
    CosmeticType:CosmeticPicker_CosmeticTypes Enum=Get cosmeticType() -> (cosmeticType=NewEnumerator0)
   ) -> (then)
   > Return Node!(execute, widget:EUW Cosmetic Pickerv 2 Combobox Variant Object Reference=@n2)
  ]}
 addFloatSliderWidget {nodes [
   n1 = addFloatSliderWidget!() -> (then, id:Fort Item Definition Object Reference, index:int, name:String) {is=title K2Node_FunctionEntry}
   > n2 = Create Widget(Class=&/EUW_CosmeticPickerv2_FloatSlider.EUW_CosmeticPickerv2_FloatSlider_C) -> (ReturnValue:EUW Cosmetic Pickerv 2 Float Slider Object Reference)
   > setup!(
    execute
    self:EUW Cosmetic Pickerv 2 Float Slider Object Reference=@n2
    cosmeticID=@n1
    itemVariantIndex=@n1.index
    wrapBox=Get WrapBox_variants()
    name=@n1.name
   ) -> (then)
  ]}
 addItemRedirectWidget {nodes [
   n1 = addItemRedirectWidget!() -> (then, id:Fort Item Definition Object Reference, index:int, name:String) {is=title K2Node_FunctionEntry}
   > n2 = Create Widget(Class=&/EUW_CosmeticPickerv2_ItemDefRedirect.EUW_CosmeticPickerv2_ItemDefRedirect_C) -> (ReturnValue:EUW Cosmetic Pickerv 2 Item Def Redirect Object Reference)
   > setup!(
    execute
    self:EUW Cosmetic Pickerv 2 Item Def Redirect Object Reference=@n2
    cosmeticID=@n1
    itemVariantIndex=@n1.index
    wrapBox=Get WrapBox_variants()
    name=@n1.name
    cosmeticType:CosmeticPicker_CosmeticTypes Enum=Get cosmeticType() -> (cosmeticType=NewEnumerator0)
   ) -> (then)
  ]}
 addNumericalWidget {nodes [
   n1 = addNumericalWidget!() -> (then, id:Fort Item Definition Object Reference, index:int, name:String) {is=title K2Node_FunctionEntry}
   > n2 = Create Widget(Class=&/EUW_CosmeticPickerv2_Numerical.EUW_CosmeticPickerv2_Numerical_C) -> (ReturnValue:EUW Cosmetic Pickerv 2 Numerical Object Reference)
   > setup!(
    execute
    self:EUW Cosmetic Pickerv 2 Numerical Object Reference=@n2
    cosmeticID=@n1
    itemVariantIndex=@n1.index
    wrapBox=Get WrapBox_variants()
    name=@n1.name
   ) -> (then)
  ]}
 addRichColorWidget {nodes [
   n1 = addRichColorWidget!() -> (then, id:Fort Item Definition Object Reference, index:int, name:String) {is=title K2Node_FunctionEntry}
   > n2 = Create Widget(Class=&/EUW_CosmeticPickerv2_RichColor.EUW_CosmeticPickerv2_RichColor_C) -> (ReturnValue:EUW Cosmetic Pickerv 2 Rich Color Object Reference)
   > setup!(
    execute
    self:EUW Cosmetic Pickerv 2 Rich Color Object Reference=@n2
    cosmeticID=@n1
    itemVariantIndex=@n1.index
    wrapBox=Get WrapBox_variants()
    name=@n1.name
    CosmeticType:CosmeticPicker_CosmeticTypes Enum=Get cosmeticType() -> (cosmeticType=NewEnumerator0)
   ) -> (then)
   > Return Node(widget:EUW Cosmetic Pickerv 2 Rich Color Object Reference=@n2)
  ]}
 addSprayWidget {nodes [
   n1 = addSprayWidget!() -> (then, id:Fort Item Definition Object Reference, index:int, name:String) {is=title K2Node_FunctionEntry}
   > n2 = Create Widget(Class=&/EUW_CosmeticPickerv2_CustomizableSpray.EUW_CosmeticPickerv2_CustomizableSpray_C) -> (ReturnValue:EUW Cosmetic Pickerv 2 Customizable Spray Object Reference)
   > setup!(
    execute
    self:EUW Cosmetic Pickerv 2 Customizable Spray Object Reference=@n2
    cosmeticID=@n1
    itemVariantIndex=@n1.index
    wrapBox=Get WrapBox_variants()
    name=@n1.name
   ) -> (then)
  ]}
 addTextBoxWidget {nodes [
   n1 = addTextBoxWidget!() -> (then, id:Fort Item Definition Object Reference, index:int, name:String) {is=title K2Node_FunctionEntry}
   > n2 = Create Widget(Class=&/EUW_CosmeticPickerv2_TextBox.EUW_CosmeticPickerv2_TextBox_C) -> (ReturnValue:EUW Cosmetic Pickerv 2 Text Box Object Reference)
   > setup(
    self:EUW Cosmetic Pickerv 2 Text Box Object Reference=@n2
    cosmeticID=@n1
    itemVariantIndex=@n1.index
    wrapBox=Get WrapBox_variants()
    name=@n1.name
   )
  ]}
 addVariantApplyWrap {nodes [
   n1 = addVariantApplyWrap!() -> (then, id:Fort Item Definition Object Reference, index:int, name:String) {is=title K2Node_FunctionEntry}
   > n2 = Create Widget(Class=&/EUW_CosmeticPickerv2_ApplyWrap.EUW_CosmeticPickerv2_ApplyWrap_C) -> (ReturnValue:EUW Cosmetic Pickerv 2 Apply Wrap Object Reference)
   > setup(
    self:EUW Cosmetic Pickerv 2 Apply Wrap Object Reference=@n2
    cosmeticID=@n1
    itemVariantIndex=@n1.index
    wrapBox=Get WrapBox_variants()
    name=@n1.name
   )
  ]}
 clearVariants {nodes [
   clearVariants() {is=title K2Node_FunctionEntry}
   > ClearChildren(self=Get WrapBox_variants())
  ]}
 getComboBoxVariantValues {nodes [
   getComboBoxVariantValues() {is=title K2Node_FunctionEntry}
   > n1 = For Each Loop(Array=GetAllChildren(self=Get WrapBox_variants()))
   Return Node!(
    execute=@n1.Completed
    optionIndex:Array of Integers=Get optionIndices() -> (optionIndices:Array of Integers)
    defaultIndex:Array of Integers=Get defaulndices() -> (defaulndices:Array of Integers)
   )
   > n2 = Cast To EUW_CosmeticPickerv2_comboboxVariant(Object=@n1)
   > n3 = Get Current Variant!(execute, self:EUW Cosmetic Pickerv 2 Combobox Variant Object Reference=@n2) -> (then, name:String, optionName:String, optionIndex:int=0, numOptions:int=0, defaultIndex:int=0) {is=title K2Node_CallFunction}
   > PrintString(InString=Append!(
     A="To String (Integer)"(InInt=@n1.Array Index)
     B=" "
     C=@n3
     D:String=" "
     E:String=@n3.optionName
     F:String=optionIndex:
     G:String="To String (Integer)"(InInt=@n3.optionIndex)
     H:String=" defaultIndex: "
     I:String="To String (Integer)"(InInt=@n3.defaultIndex)
    ) -> (ReturnValue))
   > Add(TargetArray:Array of Integers=Get optionIndices(), NewItem:"Integer (by ref)"=@n3.optionIndex) {is=title K2Node_CallArrayFunction}
   > Add(TargetArray=Get defaulndices(), NewItem=@n3.defaultIndex)
  ]}
 getNumVariants {nodes [
   n1 = getNumVariants!() -> (then, id:Fort Item Definition Object Reference) {is=title K2Node_FunctionEntry}
   > n2 = Execute Python Script!(
    execute
    PythonScript="import TechArt.CosmeticPicker.CosmeticPickerModule as CosmeticPicker\r\nnum = CosmeticPicker.getNumVariants(id)"
    id:Fort Item Definition Object Reference=@n1
   ) -> (then, ReturnValue, num:int) {inputs [id], outputs [num]}
   > Return Node!(execute, num:int=@n2.num)
  ]}
 getVariantName {nodes [
   n1 = getVariantName!() -> (then, id:Fort Item Definition Object Reference, index:int) {is=title K2Node_FunctionEntry}
   > n2 = Execute Python Script!(
    execute
    PythonScript="import TechArt.CosmeticPicker.CosmeticPickerModule as CosmeticPicker\r\nname = CosmeticPicker.getVariantName(id, index)"
    id=@n1
    index:int=@n1.index
   ) -> (then, ReturnValue, name:String) {inputs [id,index], outputs [name]}
   > PrintString(InString=Append!(
     A="##########################################################################################################################\r\n"
     B=@n2.name
     C="\r\n##########################################################################################################"
    ) -> (ReturnValue))
   > Return Node!(execute, name:String=@n2.name)
  ]}
 getVariantType {nodes [
   n1 = getVariantType!() -> (then, id:Fort Item Definition Object Reference, index:int) {is=title K2Node_FunctionEntry}
   > n2 = Execute Python Script!(
    execute
    PythonScript="import TechArt.CosmeticPicker.CosmeticPickerModule as CosmeticPicker\r\ntype = CosmeticPicker.getVariantType(id, index)"
    id=@n1
    index=@n1.index
   ) -> (then, ReturnValue, type:String) {inputs [id,index], outputs [type]}
   > Return Node!(execute, type:String=@n2.type)
  ]}
 setAppliedWrap {nodes [
   n1 = setAppliedWrap!() -> (then, inWrap:Athena Item Wrap Definition Object Reference) {is=title K2Node_FunctionEntry}
   > n2 = For Each Loop(Array=GetAllChildren(self=Get WrapBox_variants()))
   > n3 = Cast To EUW_CosmeticPickerv2_ApplyWrap(Object=@n2)
   > setAppliedWrap!(execute, self:EUW Cosmetic Pickerv 2 Apply Wrap Object Reference=@n3, inWrap:Athena Item Wrap Definition Object Reference=@n1) -> (then) {is=title K2Node_CallFunction}
  ]}
 updateMannequin {nodes [
   updateMannequin() {is=title K2Node_FunctionEntry}
   > PrintString(InString=UpdateMannequin)
   > n1 = FindParentWidgetOfType!(
    execute
    StartingWidget:Widget Object Reference=Self-Reference()
    Type:Widget Class Reference=&/EUW_CosmeticPickerv2_ID.EUW_CosmeticPickerv2_ID_C
   ) -> (then, ReturnValue:EUW Cosmetic Pickerv 2 ID Object Reference)
   > n2 = Cast To EUW_CosmeticPickerv2_ID!(execute, Object=@n1) -> (then, CastFailed, AsEUW Cosmetic Pickerv 2 ID:EUW Cosmetic Pickerv 2 ID Object Reference) {is=title K2Node_DynamicCast}
   > updateMannequin(self:EUW Cosmetic Pickerv 2 ID Object Reference=@n2) {is=title K2Node_CallFunction}
  ]}
 updateVariantWidgets {nodes [
   n1 = updateVariantWidgets!() -> (then, id:Fort Item Definition Object Reference) {is=title K2Node_FunctionEntry}
   > Set localId(localId:Fort Item Definition Object Reference=@n1) -> (Output_Get:Fort Item Definition Object Reference)
   > ClearChildren(self=Get WrapBox_variants())
   > n2 = getNumVariants!(execute, self, id:Fort Item Definition Object Reference=Get localId() -> (localId:Fort Item Definition Object Reference)) -> (then, num:int=0) {is=title K2Node_CallFunction}
   > n3 = For Loop(LastIndex=int - int(A=@n2))
   > Set index(index:int=@n3) -> (Output_Get:int=0)
   > n4 = getVariantName!(execute, self, id:Fort Item Definition Object Reference=Get localId(), index:int=Get index() -> (index:int=0)) -> (then, name:String) {is=title K2Node_CallFunction}
   > n5 = Set name(name:String=@n4) -> (Output_Get:String)
   > PrintString(InString=@n5)
   > n6 = getVariantType!(execute, self, id:Fort Item Definition Object Reference=Get localId(), index:int=Get index() -> (index=0)) -> (then, type:String) {is=title K2Node_CallFunction}
   > n7 = Switch on String!(execute:Exec, Selection:String=@n6) -> (
    Default:Exec
    FortCosmeticRichColorVariant:Exec
    FortCosmeticItemTexture:Exec
    FortCosmeticFloatSliderVariant:Exec
    FortCosmeticLoadoutTagDrivenVariant:Exec
    FortCosmeticVariantApplyWrap:Exec
    FortCosmeticNumericalVariant:Exec
    FortCustomizableObjectSprayVariant:Exec
    FortCosmeticMaterialParameterSetVariant:Exec
    FortCosmeticTextVariant:Exec
    FortCosmeticItemDefRedirectVariant:Exec
   ) {pinNames [
     FortCosmeticRichColorVariant
     FortCosmeticItemTexture
     FortCosmeticFloatSliderVariant
     FortCosmeticLoadoutTagDrivenVariant
     FortCosmeticVariantApplyWrap
     FortCosmeticNumericalVariant
     FortCustomizableObjectSprayVariant
     FortCosmeticMaterialParameterSetVariant
     FortCosmeticTextVariant
     FortCosmeticItemDefRedirectVariant
    ]}
   addFloatSliderWidget!(
    execute=@n7.FortCosmeticFloatSliderVariant
    self
    id:Fort Item Definition Object Reference=Get localId()
    index:int=Get index() -> (index=0)
    name:String=Get name() -> (name:String)
   ) -> (then) {is=title K2Node_CallFunction}
   addItemRedirectWidget!(
    execute=@n7.FortCosmeticItemDefRedirectVariant
    self
    id:Fort Item Definition Object Reference=Get localId()
    index:int=Get index() -> (index=0)
    name:String=Get name()
   ) -> (then) {is=title K2Node_CallFunction}
   addNumericalWidget!(
    execute=@n7.FortCosmeticNumericalVariant
    self
    id:Fort Item Definition Object Reference=Get localId()
    index:int=Get index() -> (index=0)
    name:String=Get name()
   ) -> (then) {is=title K2Node_CallFunction}
   addTextBoxWidget!(
    execute=@n7.FortCosmeticTextVariant
    self
    id:Fort Item Definition Object Reference=Get localId()
    index:int=Get index() -> (index=0)
    name:String=Get name()
   ) -> (then) {is=title K2Node_CallFunction}
   addVariantApplyWrap!(
    execute=@n7.FortCosmeticVariantApplyWrap
    self
    id:Fort Item Definition Object Reference=Get localId()
    index:int=Get index() -> (index=0)
    name:String=Get name()
   ) -> (then) {is=title K2Node_CallFunction}
   addSprayWidget!(
    execute=@n7.FortCustomizableObjectSprayVariant
    self
    id:Fort Item Definition Object Reference=Get localId()
    index:int=Get index() -> (index=0)
    name:String=Get name()
   ) -> (then) {is=title K2Node_CallFunction}
   addDefaultWidget!(
    execute=[@n7.Default, @n7.FortCosmeticMaterialParameterSetVariant]
    self
    id:Fort Item Definition Object Reference=Get localId()
    index:int=Get index() -> (index=0)
    name:String=Get name()
   ) -> (then, widget:EUW Cosmetic Pickerv 2 Combobox Variant Object Reference) {is=title K2Node_CallFunction}
   addRichColorWidget!(
    execute=@n7.FortCosmeticRichColorVariant
    self
    id:Fort Item Definition Object Reference=Get localId()
    index:int=Get index() -> (index=0)
    name:String=Get name()
   ) -> (then, widget:EUW Cosmetic Pickerv 2 Rich Color Object Reference) {is=title K2Node_CallFunction}
  ]}
]
variables [cosmeticType {type=CosmeticPicker_CosmeticTypes, default=Character}]
widgets {
 info {rootWidgetClass=&CanvasPanel, widgetCount=4}
 widgets [
  CanvasPanel_0 {widgetClassPath=&CanvasPanel, bIsVariable=true, widget=CanvasPanel_0}
  Image_64 {
   widgetClassPath=&Image
   bIsVariable=true
   parent=CanvasPanel_0
   widget=Image_64
   properties {
    clipping=ClipToBounds
    brush {
     bIsDynamicallyLoaded=false
     drawAs=Image
     imageType=NoImage
     mirroring=NoMirror
     resourceName=None
     resourceObject=null
     tiling=NoTile
     imageSize=<1 1>
     margin=<ltrb 0 0 0 0>
     outlineSettings {
      bUseBrushTransparency=false
      roundingType=HalfHeightRadius
      width=0
      color {colorUseRule=UseColor_Specified, specifiedColor=<rgba 0 0 0 0>}
      cornerRadii=<0 0 0 0>
     }
     tintColor {colorUseRule=UseColor_Specified, specifiedColor=<rgba 1 1 1 1>}
     uVRegion {bIsValid=false, max=<0 0>, min=<0 0>}
    }
    colorAndOpacity=<rgba 0.0176 0.0176 0.0176 1>
   }
   slot {
    class=CanvasPanelSlot
    properties {bAutoSize=true, layoutData {alignment=<0 0>, anchors {maximum=<1 1>, minimum=<0 0>}, offsets=<ltrb 0 0 0 0>}}
   }
  }
  Overlay_86 {
   widgetClassPath=&Overlay
   parent=CanvasPanel_0
   widget=Overlay_86
   slot {
    class=CanvasPanelSlot
    properties {bAutoSize=true, layoutData {alignment=<0 0>, anchors {maximum=<1 1>, minimum=<0 0>}, offsets=<ltrb 0 0 0 0>}}
   }
  }
  WrapBox_variants {
   widgetClassPath=&WrapBox
   bIsVariable=true
   parent=Overlay_86
   widget=WrapBox_variants
   properties {bExplicitWrapSize=true, clipping=ClipToBounds, horizontalAlignment=HAlign_Center, wrapSize=600, innerSlotPadding=<10 10>}
   slot {class=OverlaySlot}
  }
 ]
}
