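# Editor utility widget for a cosmetic picker tool: a reusable list button that points the Content Browser at a folder.
# A single button wraps a small bold text label, with a GetText function that returns the BtnName variable as text.
# On click, it passes its Path variable as a one-element array to SyncBrowserToFolders, jumping the browser to that folder.
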EditorUtilityWidgetBlueprint /Game/NeverCook/ArtTools/Tech_Art/CosmeticPicker/EUW_Cosmetic_ListBtn
graphs [
 EventGraph {nodes [
   "On Clicked (EditorUtilityButton_0)"() {is=title K2Node_ComponentBoundEvent}
   > SyncBrowserToFolders!(execute, FolderList:Array of Strings=Make Array("[0]":String=Get Path() -> (Path:String)) -> (Array:Array of Strings)) -> (then)
   Event PreConstruct()
   Event Construct()
   Event Tick()
  ]}
 GetText {nodes [
   GetText() {is=title K2Node_FunctionEntry}
   > Return Node!(execute, ReturnValue:Text="To Text (String)"(InString=Get BtnName() -> (BtnName:String)))
  ]}
]
variables [
 BtnName {type=string, default=""}
 Path {type=string, default=""}
]
widgets {
 info {rootWidgetClass=&EditorUtilityButton, widgetCount=2}
 widgets [
  EditorUtilityButton_0 {widgetClassPath=&EditorUtilityButton, widget=EditorUtilityButton_0, slot=~}
  TextBlock_24 {
   widgetClassPath=&TextBlock
   parent=EditorUtilityButton_0
   widget=TextBlock_24
   properties {
    text=Text Block
    colorAndOpacity {colorUseRule=UseColor_Specified, specifiedColor=<rgba 0.5052 0.5052 0.5052 1>}
    font {
     bForceMonospaced=false
     bMaterialIsStencil=false
     fontMaterial=null
     fontObject=&Roboto
     letterSpacing=0
     monospacedWidth=1
     size=8.25
     skewAmount=0
     typefaceFontName=Bold
     outlineSettings {
      bApplyOutlineToDropShadows=false
      bMiteredCorners=false
      bSeparateFillAlpha=false
      outlineMaterial=null
      outlineSize=0
      outlineColor=<rgba 0 0 0 1>
     }
    }
   }
   slot {class=ButtonSlot}
  }
 ]
}
