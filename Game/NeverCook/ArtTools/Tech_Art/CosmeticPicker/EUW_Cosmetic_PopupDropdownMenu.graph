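# EUW_Cosmetic_PopupDropdownMenu is an editor utility widget that serves as a popup dropdown container in the CosmeticPicker art tool.
# Its only widget is a root VerticalBox, and nothing in this graph adds entries to it.
# On mouse leave it clears all children from the box, emptying the menu rather than removing the widget itself.
# PreConstruct, Construct and Tick are unimplemented.
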EditorUtilityWidgetBlueprint /Game/NeverCook/ArtTools/Tech_Art/CosmeticPicker/EUW_Cosmetic_PopupDropdownMenu
graphs [EventGraph {nodes [
   Event OnMouseLeave!() -> (OutputDelegate, then, MouseEvent:"Pointer Event Structure (by ref)") {is=title K2Node_Event}
   > ClearChildren(self=Get VerticalBox_0() -> (VerticalBox_0:Vertical Box Object Reference))
   Event PreConstruct()
   Event Construct()
   Event Tick()
  ]}]
variables []
widgets {
 info {rootWidgetClass=&VerticalBox, widgetCount=1}
 widgets [VerticalBox_0 {widgetClassPath=&VerticalBox, bIsVariable=true, widget=VerticalBox_0}]
}
