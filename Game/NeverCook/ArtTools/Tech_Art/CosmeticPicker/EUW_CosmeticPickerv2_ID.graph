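# Editor Utility Widget for one item-definition slot in the Cosmetic Picker tool, configured by cosmetic type (character, backpack, shoes, Sparks instruments, etc.).
# Setup loads saved overrides and filters a Details View to that type; edits push the override through Python and refresh the Variants widget.
# Shoe edits also compile the Customizable Object via Python.
# A folder button locates mesh folders, and a cancelled drag ray-traces from the cursor to spawn a preview actor.
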EditorUtilityWidgetBlueprint /Game/NeverCook/ArtTools/Tech_Art/CosmeticPicker/EUW_CosmeticPickerv2_ID
graphs [
 EventGraph {nodes [
   "On Clicked (FolderFinderBtn)"() {is=title K2Node_ComponentBoundEvent}
   > PrintString(InString=Searching)
   n1 = Get Shoes() -> (Shoes:Cosmetic Shoes Item Definition Soft Object Reference)
   n2 = Get Companion() -> (Companion:Cosmetic Companion Item Definition Soft Object Reference)
   > n3 = FindAssetFolder!(
    execute
    self
    Asset:Object Reference=Select!(
     NewEnumerator0:Object Reference=Get character() -> (character:Athena Character Item Definition Object Reference)
     NewEnumerator1:Object Reference=Get backpack() -> (backpack:Athena Backpack Item Definition Object Reference)
     NewEnumerator2:Object Reference=Get pickaxe() -> (pickaxe:Athena Pickaxe Item Definition Object Reference)
     NewEnumerator3:Object Reference=Get glider() -> (glider:Athena Glider Item Definition Object Reference)
     NewEnumerator4:Object Reference=Get Emote() -> (Emote:Athena Dance Item Definition Object Reference)
     NewEnumerator5:Object Reference=Get wrap() -> (wrap:Athena Item Wrap Definition Object Reference)
     NewEnumerator6:Object Reference
     NewEnumerator7:Object Reference=Get contrail() -> (contrail:Athena Sky Dive Contrail Item Definition Object Reference)
     NewEnumerator8:Object Reference=Resolve Soft Reference!(Input:Cosmetic Shoes Item Definition Soft Object Reference=@n1) -> (Output:Cosmetic Shoes Item Definition Object Reference)
     NewEnumerator9:Object Reference=Resolve Soft Reference(Input:Cosmetic Companion Item Definition Soft Object Reference=@n2) -> (Output:Cosmetic Companion Item Definition Object Reference)
     NewEnumerator10:Object Reference=Get Guitar() -> (Guitar:Sparks Guitar Item Definition Object Reference)
     NewEnumerator11:Object Reference=Get Bass() -> (Bass:Sparks Bass Item Definition Object Reference)
     NewEnumerator12:Object Reference=Get Drums() -> (Drums:Sparks Drum Item Definition Object Reference)
     NewEnumerator13:Object Reference=Get Mic() -> (Mic:Sparks Mic Item Definition Object Reference)
     NewEnumerator14:Object Reference=Get Keytar() -> (Keytar:Sparks Keyboard Item Definition Object Reference)
     NewEnumerator15:Object Reference=Get Aura() -> (Aura:Sparks Aura Item Definition Object Reference)
     Index:CosmeticPicker_CosmeticTypes Enum=Get CosmeticType() -> (CosmeticType:CosmeticPicker_CosmeticTypes Enum=NewEnumerator0)
    ) -> (ReturnValue:Object Reference)
    InputPin:int=Get CharacterPartIndex() -> (CharacterPartIndex:int=0)
   ) -> (then, PotentialPaths:Array of Strings)
   > n4 = Branch(Condition="Equal (Integer)"(A=Length(TargetArray:Array of Strings=@n3), B=1))
   > SyncBrowserToFolders!(execute, FolderList:Array of Strings=@n3) -> (then)
   Branch(execute=@n4.else, Condition=Is Not Empty!(TargetArray=@n3) -> (ReturnValue:bool))
   > Set Visibility(Visibility=SelfHitTestInvisible, self=Get CharacterPartPopup() -> (CharacterPartPopup:Vertical Box Object Reference))
   > n5 = For Each Loop(Array=@n3)
   > n6 = Construct Object from Class!(
    execute
    Class=&/EUW_Cosmetic_ListBtn.EUW_Cosmetic_ListBtn_C
    self
    BtnName:String=PathToName!(In:String=@n5) -> (Out:String)
    Path:String=@n5
   ) -> (then, ReturnValue:EUW Cosmetic List Btn Object Reference)
   > Add(
    TargetArray:Array of EUW Cosmetic List Btn Object References=Get BtnsList() -> (BtnsList:Array of EUW Cosmetic List Btn Object References)
    NewItem:"EUW Cosmetic List Btn Object Reference (by ref)"=@n6
   ) {is=title K2Node_CallArrayFunction}
   > AddChildToVerticalBox(
    self=Get VerticalBox_0!(self:EUW Cosmetic Popup Dropdown Menu Object Reference=Get EUW_PopupDropdownMenu() -> (EUW_PopupDropdownMenu:EUW Cosmetic Popup Dropdown Menu Object Reference)) -> (VerticalBox_0:Vertical Box Object Reference)
    Content=@n6
   )
   Event Construct()
   n7 = Get CosmeticType() -> (CosmeticType=NewEnumerator0)
   > n8 = Set cosmeticType!(
    execute
    cosmeticType:CosmeticPicker_CosmeticTypes Enum=@n7
    self:EUW Cosmetic Pickerv 2 Variants Object Reference=Get Variants() -> (Variants:EUW Cosmetic Pickerv 2 Variants Object Reference)
   ) -> (then, Output_Get:CosmeticPicker_CosmeticTypes Enum=NewEnumerator0)
   > Branch(Condition="Not Equal (Enum)"!(A:CosmeticPicker_CosmeticTypes Enum=@n8, B:CosmeticPicker_CosmeticTypes Enum=NewEnumerator6) -> (ReturnValue:bool))
   > Setup!(execute, self, CosmeticType:CosmeticPicker_CosmeticTypes Enum=@n7, loadCheatValues:bool=true) -> (then)
   Event OnDragCancelled!() -> (OutputDelegate, then, PointerEvent:"Pointer Event Structure (by ref)", Operation:Drag Drop Operation Object Reference) {is=title K2Node_Event}
   > Delay!(execute, WorldContextObject:Object Reference, Duration=0.2) -> (then)
   > PrintString(InString=Drag Cancelled)
   > PrintString(InString=GetDisplayName(Object=Get Item Definitiion() -> (Item Definitiion:Fort Item Definition Object Reference)))
   > n9 = Get Cursor World Ray!(execute) -> (then, CameraLocation:Vector, RayOrigin:Vector, RayDirection:Vector, ReturnValue:bool) {is=title K2Node_CallFunction}
   > n10 = Line Trace By Channel!(
    execute
    WorldContextObject:Object Reference
    Start:Vector=@n9.RayOrigin
    End:Vector=vector + vector(A=@n9.RayOrigin, B=vector * vector(A=@n9.RayDirection, B:Vector="9999,9999,9999") -> (ReturnValue=~))
    TraceChannel:ETraceTypeQuery Enum
    bTraceComplex:bool=false
    ActorsToIgnore:Array of Actor Object References
    DrawDebugType:EDrawDebugTrace Enum
    bIgnoreSelf:bool
    TraceColor:Linear Color Structure
    TraceHitColor:Linear Color Structure
    DrawTime:float
   ) -> (
    then
    OutHit_bBlockingHit:bool
    OutHit_bInitialOverlap:bool
    OutHit_Time:float
    OutHit_Distance:float
    OutHit_Location:Vector
    OutHit_ImpactPoint:Vector
    OutHit_Normal:Vector
    OutHit_ImpactNormal:Vector
    OutHit_PhysMat:Physical Material Object Reference
    OutHit_HitActor:Actor Object Reference
    OutHit_HitComponent:Primitive Component Object Reference
    OutHit_HitBoneName:Name=None
    OutHit_BoneName:Name=None
    OutHit_HitItem:int=0
    OutHit_ElementIndex:int=0
    OutHit_FaceIndex:int=0
    OutHit_TraceStart:Vector
    OutHit_TraceEnd:Vector
    ReturnValue:bool
   )
   > spawnItem!(execute, self, position:Vector=@n10.OutHit_Location) -> (then)
   Event PreConstruct()
   Event Tick()
  ]}
 FindAssetFolder {nodes [
   n1 = FindAssetFolder!() -> (then, Asset:Object Reference, InputPin:int) {is=title K2Node_FunctionEntry}
   > n2 = Cast To AthenaCharacterItemDefinition!(execute, Object=@n1) -> (then, CastFailed, AsAthena Character Item Definition:Athena Character Item Definition Object Reference) {is=title K2Node_DynamicCast}
   > n3 = For Each Loop with Break!(
    Exec:Exec
    Array:Array of Custom Character Part Soft Object References=Get BaseCharacterParts!(self:Athena Character Item Definition Object Reference=@n2) -> (BaseCharacterParts:Array of Custom Character Part Soft Object References)
    Break:Exec
   ) -> (LoopBody:Exec, Array Element:Custom Character Part Soft Object Reference, Array Index:int, Completed:Exec)
   n4 = BreakSoftObjectPath(InSoftObjectPath=To Soft Object Path!(SoftObjectReference:Soft Object Reference=@n3) -> (ReturnValue:Soft Object Path Structure))
   > LoadAsset!(
    execute
    self:Editor Asset Subsystem Object Reference=Get EditorAssetSubsystem() -> (ReturnValue:Editor Asset Subsystem Object Reference) {is=title K2Node_GetEditorSubsystem}
    AssetPath=Left!(SourceString:String=@n4, Count:int=FindSubstring(SearchIn=@n4, Substring=., bSearchFromEnd=true)) -> (ReturnValue:String)
   ) -> (then, ReturnValue)
   n5 = GetSkeletalMesh!(self:Custom Character Part Object Reference=Resolve Soft Reference(Input:Custom Character Part Soft Object Reference=@n3) -> (Output:Custom Character Part Object Reference)) -> (ReturnValue:Skeletal Mesh Object Reference)
   > GetPathNameForLoadedAsset(LoadedAsset=@n5)
   n6 = GetSoftObjectPath!(Object=@n5) -> (ReturnValue_PathString:String)
   > Add(
    TargetArray:Array of Strings=Get Paths() -> (Paths:Array of Strings)
    NewItem:"String (by ref)"=Left(SourceString=@n6, Count=FindSubstring(SearchIn=@n6, Substring=/Mesh, bSearchFromEnd=true))
   )
   n7 = For Each Loop(Exec=[@n19.then, @n11.then, @n15.then, @n3.Completed], Array=Get Paths())
   Return Node!(execute=@n7.Completed, PotentialPaths:Array of Strings=Get FilteredPaths() -> (FilteredPaths:Array of Strings))
   > Add Unique(TargetArray=Get FilteredPaths(), NewItem=@n7)
   n8 = Cast To AthenaPickaxeItemDefinition!(execute=@n2.CastFailed, Object=@n1) -> (then, CastFailed, AsAthena Pickaxe Item Definition:Athena Pickaxe Item Definition Object Reference) {is=title K2Node_DynamicCast}
   n9 = GetWeaponMeshOverride!(self:Fort Weapon Item Definition Object Reference=Get WeaponDefinition!(self:Athena Pickaxe Item Definition Object Reference=@n8) -> (WeaponDefinition:Fort Weapon Melee Item Definition Object Reference)) -> (ReturnValue:Skeletal Mesh Object Reference)
   > GetPathNameForLoadedAsset(LoadedAsset=@n9)
   n10 = GetSoftObjectPath(Object=@n9)
   > n11 = Add(TargetArray=Get Paths(), NewItem=Left(SourceString=@n10, Count=FindSubstring(SearchIn=@n10, Substring=/Mesh, bSearchFromEnd=true)))
   n12 = Cast To AthenaBackpackItemDefinition!(execute=@n8.CastFailed, Object=@n1) -> (then, CastFailed, AsAthena Backpack Item Definition:Athena Backpack Item Definition Object Reference) {is=title K2Node_DynamicCast}
   n13 = GetSkeletalMesh(self="Get (a copy)"(Array:Array of Custom Character Part Object References=GetCharacterParts!(self:Athena Character Part Item Definition Object Reference=@n12) -> (ReturnValue:Array of Custom Character Part Object References)) -> (Output:Custom Character Part Object Reference))
   > GetPathNameForLoadedAsset(LoadedAsset=@n13)
   n14 = GetSoftObjectPath(Object=@n13)
   > n15 = Add(TargetArray=Get Paths(), NewItem=Left(SourceString=@n14, Count=FindSubstring(SearchIn=@n14, Substring=/Mesh, bSearchFromEnd=true)))
   n16 = Cast To AthenaGliderItemDefinition!(execute=@n12.CastFailed, Object=@n1) -> (then, CastFailed, AsAthena Glider Item Definition:Athena Glider Item Definition Object Reference) {is=title K2Node_DynamicCast}
   > n17 = GetEditorProperty!(execute:Exec, Object:Object Reference=@n16, PropertyName:Name=SkeletalMesh) -> (then:Exec, PropertyValue:Soft Object Reference, ReturnValue:bool)
   n18 = BreakSoftObjectPath(InSoftObjectPath=To Soft Object Path(SoftObjectReference=@n17))
   > n19 = Add(TargetArray=Get Paths(), NewItem=Left(SourceString=@n18, Count=FindSubstring(SearchIn=@n18, Substring=/Mesh, bSearchFromEnd=true)))
   Select!(Option 0:String=FaceAcc, Option 1:String=Head, Option 2:String=Bod, Index:int=0) -> (ReturnValue:String)
  ]}
 IDChanged {nodes [
   n1 = IDChanged!() -> (then, PropertyName:Name) {is=title K2Node_FunctionEntry}
   > PrintString(InString=Append(A="IDChanged ", B="To String (Name)"(InName=@n1)))
   > n2 = getId!(execute, self, type:CosmeticPicker_CosmeticTypes Enum=Get CosmeticType() -> (CosmeticType=NewEnumerator0)) -> (then, id:Fort Item Definition Object Reference)
   > Set Item Definitiion(Item Definitiion:Fort Item Definition Object Reference=@n2) -> (Output_Get:Fort Item Definition Object Reference)
   > SetOverride!(
    execute
    self
    id:Fort Item Definition Object Reference=@n2
    cosmeticType:CosmeticPicker_CosmeticTypes Enum=Get CosmeticType() -> (CosmeticType=NewEnumerator0)
   ) -> (then)
   > n3 = updateVariants!(execute, self, id:Fort Item Definition Object Reference=@n2) -> (then, Variants:EUW Cosmetic Pickerv 2 Variants Object Reference)
   > Update Custom Variant Channels(self:EUW Cosmetic Pickerv 2 Variants Object Reference=@n3)
   > n4 = Switch on CosmeticPicker_CosmeticTypes!(execute, Selection:CosmeticPicker_CosmeticTypes Enum=Get CosmeticType() -> (CosmeticType=NewEnumerator0)) -> (
    NewEnumerator0:Exec
    NewEnumerator1:Exec
    NewEnumerator2:Exec
    NewEnumerator3:Exec
    NewEnumerator4:Exec
    NewEnumerator5:Exec
    NewEnumerator6:Exec
    NewEnumerator7:Exec
    NewEnumerator8:Exec
    NewEnumerator9:Exec
    NewEnumerator10:Exec
    NewEnumerator11:Exec
    NewEnumerator12:Exec
    NewEnumerator13:Exec
    NewEnumerator14:Exec
    NewEnumerator15:Exec
   ) {is=title K2Node_SwitchEnum}
   > n5 = FindParentWidgetOfType!(
    execute
    StartingWidget:Widget Object Reference=Self-Reference()
    Type:Widget Class Reference=&/EUW_CosmeticPickerv2.EUW_CosmeticPickerv2_C
   ) -> (then, ReturnValue:EUW Cosmetic Pickerv 2 Object Reference)
   > n6 = Cast To EUW_CosmeticPickerv2!(execute, Object=@n5) -> (then, CastFailed, AsEUW Cosmetic Pickerv 2:EUW Cosmetic Pickerv 2 Object Reference) {is=title K2Node_DynamicCast}
   > characterChanged(self:EUW Cosmetic Pickerv 2 Object Reference=@n6)
   Execute Python Script!(
    execute=@n4.NewEnumerator8
    PythonScript="## Get the Customizable Object CO, from a Shoe_ItemDef and Compile it\r\n\r\nimport unreal\r\n\r\nshoe_path = shoe_item_path\r\nasset_registry = unreal.AssetRegistryHelpers.get_asset_registry()\r\n\r\ndef get_dep(asset, dep_type=None, debug=False):    \r\n    if debug: print(f\"Getting '{asset.asset_name}' dependencies: {dep_type}\")\r\n    dependency_options = unreal.AssetRegistryDependencyOptions(include_soft_package_references=True, include_hard_package_references=True)    \r\n    dependencies = asset_registry.get_dependencies(asset.package_name, dependency_options)\r\n    if dependencies:\r\n        if dep_type:\r\n            filters = unreal.ARFilter(package_names=dependencies, class_names=[dep_type])\r\n            filtered_deps = asset_registry.get_assets(filters)\r\n            if filtered_deps:\r\n                item = filtered_deps[0]\r\n                if debug: print(f\"Filtered Deps: {item}\")\r\n                return item\r\n\r\n        else:\r\n            filters = unreal.ARFilter(package_names=dependencies)\r\n            if debug:\r\n                print(\"Dependencies\")            \r\n                for i in dependencies: \r\n                    print(f\" dep: {i}\")\r\n                    \r\n        return None\r\n    \r\ndef compile_shoe(item_path, debug=False):  \r\n    asset = asset_registry.get_asset_by_object_path(item_path)\r\n    if debug: print(f\"A asset: {asset}\")\r\n    item_def = unreal.load_asset(item_path)\r\n    if debug: print(item_def)\r\n    if item_def:\r\n        parts = item_def.get_editor_property(\"CharacterParts\")\r\n        if parts:\r\n            cp = parts[0]\r\n            if debug: print(cp)\r\n            part_data_list = cp.get_editor_property(\"cosmetic_part_data_list\")\r\n            if part_data_list:\r\n                if debug:\r\n                    for i in part_data_list:\r\n                        print(f\"Part Data: {i}\")\r\n    \r\n            asset = asset_registry.get_asset_by_object_path(cp.get_path_name())\r\n            print(asset)\r\n            customizable_data = get_dep(asset, dep_type=\"CosmeticSharedCustomizableData\")\r\n            if customizable_data:\r\n                if debug: print(f\"Customizable Data: {customizable_data.asset_name}\")\r\n                data_obj = unreal.load_asset(customizable_data.package_name)\r\n                if data_obj:\r\n                    co = data_obj.get_editor_property(\"CustomizableObject\")\r\n                    asset_path = co.get_path_name()\r\n                    print(asset_path)\r\n                    if debug: print(f\"AssetPath: {asset_path}\")\r\n                    asset_data = asset_registry.get_asset_by_object_path(asset_path)\r\n                    ##asset_data = unreal.EditorAssetLibrary.find_asset_data(asset_path)\r\n                    if not unreal.AssetRegistryHelpers.is_valid(asset_data):\r\n                        print(\"Invalid Asset Registry data\")\r\n                    else:\r\n                        print(\"Valid Assit Data\")\r\n                    co_asset = asset_data.get_asset()\r\n                    if debug: print(f\"CO Asset: {co_asset}\")\r\n                    if not co_asset.is_compiled():\r\n                        unreal.CustomizableObjectEditorFunctionLibrary.compile_customizable_object_synchronously(co_asset)\r\n                        print(f\"Customizable Object|CO is now Compiled: {co_asset.get_name()}\")\r\n                    else:\r\n                        print(f\"Customizable Object|CO is Already Compiled: {co_asset.get_name()}\")\r\n\r\ncompile_shoe(shoe_path)"
    shoe_item_path:String=To Soft Object Path!(SoftObjectReference=Get Shoes()) -> (ReturnValue_PathString:String)
   ) -> (then, ReturnValue) {inputs [shoe_item_path]}
  ]}
 OnDragDetected {nodes [
   OnDragDetected!() -> (then, MyGeometry:Geometry Structure, PointerEvent:"Pointer Event Structure (by ref)") {is=title K2Node_FunctionEntry}
   > PrintString(InString=Drag Detected)
   > n1 = Create Widget!(
    execute:Exec
    WorldContextObject:Object Reference
    Class:User Widget Class Reference=/Game/NeverCook/ArtTools/Tech_Art/AssetPlaylist/EUW_AssetPlaylist_DragDropWidget.EUW_AssetPlaylist_DragDropWidget_C
    OwningPlayer:Player Controller Object Reference
   ) -> (then:Exec, ReturnValue:EUW Asset Playlist Drag Drop Widget Object Reference)
   > n2 = "Create Drag && Drop Operation"!(
    execute:Exec
    WorldContextObject:Object Reference
    Class:Drag Drop Operation Class Reference=/Script/UMG.DragDropOperation
    Tag:String
    Payload:Object Reference
    DefaultDragVisual:Widget Object Reference=@n1
    Pivot:EDragPivot Enum=CenterCenter
    Offset:Vector 2D Structure=(X=0,Y=0)
   ) -> (then:Exec, ReturnValue:Drag Drop Operation Object Reference) {is=title K2Node_CreateDragDropOperation}
   > Return Node!(execute, Operation:Drag Drop Operation Object Reference=@n2)
  ]}
 OnMouseButtonDown {nodes [
   n1 = OnMouseButtonDown!() -> (then, MyGeometry:Geometry Structure, MouseEvent:"Pointer Event Structure (by ref)") {is=title K2Node_FunctionEntry}
   > n2 = Branch(Condition="Equal (Key)"(
     A:Key Structure=Get Effecting Button!(Input:"Pointer Event Structure (by ref)"=@n1.MouseEvent) -> (ReturnValue:Key Structure) {is=title K2Node_CallFunction}
     B:Key Structure=LeftMouseButton
    ) -> (ReturnValue:bool=~) {is=title K2Node_PromotableOperator})
   > PrintString(InString="Left  Down")
   > n3 = DetectDragIfPressed!(
    execute
    PointerEvent:"Pointer Event Structure (by ref)"=@n1.MouseEvent
    WidgetDetectingDrag:Widget Object Reference
    DragKey:Key Structure=LeftMouseButton
   ) -> (then, ReturnValue:Event Reply Structure)
   > Return Node!(execute, ReturnValue:Event Reply Structure=@n3)
   PrintString(execute=@n2.else, InString=Right Down)
   > Return Node(ReturnValue=Handled())
  ]}
 OnMouseButtonUp {nodes [
   n1 = OnMouseButtonUp!() -> (then, MyGeometry:Geometry Structure, MouseEvent:"Pointer Event Structure (by ref)") {is=title K2Node_FunctionEntry}
   > n2 = Branch(Condition="Equal (Key)"(A=Get Effecting Button(Input=@n1.MouseEvent), B=LeftMouseButton) -> (ReturnValue=~))
   > PrintString(InString=Left Up)
   n3 = Handled()
   > Return Node(ReturnValue=@n3)
   PrintString(execute=@n2.else, InString=Right Up)
   > Return Node(ReturnValue=@n3)
  ]}
 PathToName {nodes [
   n1 = PathToName!() -> (then, In:String) {is=title K2Node_FunctionEntry}
   > n2 = Sequence()
   > n3 = Branch(Condition=Contains(SearchIn=@n1, Substring=Head, bSearchFromEnd=true))
   > Set FinalString(FinalString:String=Head) -> (Output_Get:String)
   n4 = Branch(execute=@n3.else, Condition=Contains(SearchIn=@n1, Substring=FaceAcc, bSearchFromEnd=true))
   > Set FinalString(FinalString=FaceAcc)
   Branch(execute=@n4.else, Condition=Contains(SearchIn=@n1, Substring=Bod, bSearchFromEnd=true))
   > Set FinalString(FinalString=Body)
   Return Node!(execute=@n2.then_1, Out:String=Get FinalString() -> (FinalString:String))
  ]}
 SetId {nodes [
   n1 = SetId!() -> (then, id:Fort Item Definition Object Reference, cosmeticType:CosmeticPicker_CosmeticTypes Enum=NewEnumerator0) {is=title K2Node_FunctionEntry}
   > Set Item Definitiion(Item Definitiion=@n1)
   > n2 = Switch on CosmeticPicker_CosmeticTypes(Selection=@n1.cosmeticType)
   > n3 = Cast To AthenaCharacterItemDefinition(Object=@n1)
   > Set character(character:Athena Character Item Definition Object Reference=@n3) -> (Output_Get:Athena Character Item Definition Object Reference)
   Set character(execute=@n3.CastFailed)
   n4 = Cast To AthenaBackpackItemDefinition(execute=@n2.NewEnumerator1, Object=@n1)
   > Set backpack(backpack:Athena Backpack Item Definition Object Reference=@n4) -> (Output_Get:Athena Backpack Item Definition Object Reference)
   Set backpack(execute=@n4.CastFailed)
   n5 = Cast To SparksGuitarItemDefinition!(execute=@n2.NewEnumerator10, Object=@n1) -> (then, CastFailed, AsSparks Guitar Item Definition:Sparks Guitar Item Definition Object Reference) {is=title K2Node_DynamicCast}
   > Set Guitar(Guitar:Sparks Guitar Item Definition Object Reference=@n5) -> (Output_Get:Sparks Guitar Item Definition Object Reference)
   Set Guitar(execute=@n5.CastFailed)
   n6 = Cast To SparksBassItemDefinition!(execute=@n2.NewEnumerator11, Object=@n1) -> (then, CastFailed, AsSparks Bass Item Definition:Sparks Bass Item Definition Object Reference) {is=title K2Node_DynamicCast}
   > Set Bass(Bass:Sparks Bass Item Definition Object Reference=@n6) -> (Output_Get:Sparks Bass Item Definition Object Reference)
   Set Bass(execute=@n6.CastFailed)
   n7 = Cast To SparksDrumItemDefinition!(execute=@n2.NewEnumerator12, Object=@n1) -> (then, CastFailed, AsSparks Drum Item Definition:Sparks Drum Item Definition Object Reference) {is=title K2Node_DynamicCast}
   > Set Drums(Drums:Sparks Drum Item Definition Object Reference=@n7) -> (Output_Get:Sparks Drum Item Definition Object Reference)
   Set Drums(execute=@n7.CastFailed)
   n8 = Cast To SparksMicItemDefinition!(execute=@n2.NewEnumerator13, Object=@n1) -> (then, CastFailed, AsSparks Mic Item Definition:Sparks Mic Item Definition Object Reference) {is=title K2Node_DynamicCast}
   > Set Mic(Mic:Sparks Mic Item Definition Object Reference=@n8) -> (Output_Get:Sparks Mic Item Definition Object Reference)
   Set Mic(execute=@n8.CastFailed)
   n9 = Cast To SparksKeyboardItemDefinition!(execute=@n2.NewEnumerator14, Object=@n1) -> (then, CastFailed, AsSparks Keyboard Item Definition:Sparks Keyboard Item Definition Object Reference) {is=title K2Node_DynamicCast}
   > Set Keytar(Keytar:Sparks Keyboard Item Definition Object Reference=@n9) -> (Output_Get:Sparks Keyboard Item Definition Object Reference)
   Set Keytar(execute=@n9.CastFailed)
   n10 = Cast To SparksAuraItemDefinition!(execute=@n2.NewEnumerator15, Object=@n1) -> (then, CastFailed, AsSparks Aura Item Definition:Sparks Aura Item Definition Object Reference) {is=title K2Node_DynamicCast}
   > Set Aura(Aura:Sparks Aura Item Definition Object Reference=@n10) -> (Output_Get:Sparks Aura Item Definition Object Reference)
   Set Aura(execute=@n10.CastFailed)
   n11 = Cast To AthenaPickaxeItemDefinition(execute=@n2.NewEnumerator2, Object=@n1)
   > Set pickaxe(pickaxe:Athena Pickaxe Item Definition Object Reference=@n11) -> (Output_Get:Athena Pickaxe Item Definition Object Reference)
   Set pickaxe(execute=@n11.CastFailed)
   n12 = Cast To AthenaGliderItemDefinition(execute=@n2.NewEnumerator3, Object=@n1)
   > Set glider(glider:Athena Glider Item Definition Object Reference=@n12) -> (Output_Get:Athena Glider Item Definition Object Reference)
   Set glider(execute=@n12.CastFailed)
   n13 = Cast To AthenaDanceItemDefinition!(execute=@n2.NewEnumerator4, Object=@n1) -> (then, CastFailed, AsAthena Dance Item Definition:Athena Dance Item Definition Object Reference) {is=title K2Node_DynamicCast}
   > Set Emote(Emote:Athena Dance Item Definition Object Reference=@n13) -> (Output_Get:Athena Dance Item Definition Object Reference)
   Set Emote(execute=@n13.CastFailed)
   n14 = Cast To AthenaItemWrapDefinition!(execute=@n2.NewEnumerator5, Object=@n1) -> (then, CastFailed, AsAthena Item Wrap Definition:Athena Item Wrap Definition Object Reference) {is=title K2Node_DynamicCast}
   > Set wrap(wrap:Athena Item Wrap Definition Object Reference=@n14) -> (Output_Get:Athena Item Wrap Definition Object Reference)
   Set wrap(execute=@n14.CastFailed)
   n15 = Cast To AthenaSkyDiveContrailItemDefinition!(execute=@n2.NewEnumerator7, Object=@n1) -> (then, CastFailed, AsAthena Sky Dive Contrail Item Definition:Athena Sky Dive Contrail Item Definition Object Reference) {is=title K2Node_DynamicCast}
   > setContrail!(execute, self, inContrail:Athena Sky Dive Contrail Item Definition Object Reference=@n15) -> (then)
   Set contrail(execute=@n15.CastFailed, contrail:Athena Sky Dive Contrail Item Definition Object Reference) -> (Output_Get:Athena Sky Dive Contrail Item Definition Object Reference)
   n16 = Cast To CosmeticCompanionItemDefinition!(execute=@n2.NewEnumerator9, Object=@n1) -> (then, CastFailed, AsCosmetic Companion Item Definition:Cosmetic Companion Item Definition Object Reference) {is=title K2Node_DynamicCast}
   > Set Companion(Companion:Cosmetic Companion Item Definition Soft Object Reference=Make Soft Reference!(Input:Cosmetic Companion Item Definition Object Reference=@n16) -> (Output:Cosmetic Companion Item Definition Soft Object Reference)) -> (Output_Get:Cosmetic Companion Item Definition Soft Object Reference)
   Set Companion(execute=@n16.CastFailed)
  ]}
 SetOverride {nodes [
   n1 = SetOverride!() -> (then, id:Fort Item Definition Object Reference, cosmeticType:CosmeticPicker_CosmeticTypes Enum=NewEnumerator0) {is=title K2Node_FunctionEntry}
   > n2 = Switch on CosmeticPicker_CosmeticTypes(Selection=@n1.cosmeticType)
   n3 = Cast To AthenaSkyDiveContrailItemDefinition(execute=@n2.NewEnumerator7, Object=@n1)
   setContrail(execute=[@n3.then, @n4.Is Not Valid], inContrail=@n3)
   n4 = Is Valid(exec=@n3.CastFailed, InputObject=@n3)
   Execute Python Script!(
    execute=@n2.NewEnumerator9
    PythonScript="import TechArt.CosmeticPicker.CosmeticPickerModule as CosmeticPicker\r\nCosmeticPicker.setOverride(id_name, id)"
    idName:String=Append(A=Mimosa)
    id:Fort Item Definition Object Reference=@n1
   ) -> (then, ReturnValue) {inputs [idName,id]}
   n5 = Enum to String!(Enumerator:CosmeticPicker_CosmeticTypes Enum=@n1.cosmeticType) -> (ReturnValue:String)
   PrintString(
    execute=[
     @n2.NewEnumerator0
     @n2.NewEnumerator1
     @n2.NewEnumerator10
     @n2.NewEnumerator11
     @n2.NewEnumerator12
     @n2.NewEnumerator13
     @n2.NewEnumerator14
     @n2.NewEnumerator15
     @n2.NewEnumerator2
     @n2.NewEnumerator3
     @n2.NewEnumerator4
     @n2.NewEnumerator5
     @n2.NewEnumerator6
     @n2.NewEnumerator8
    ]
    InString=Append!(A=SetID:, B=@n5, C:String, D:String=GetDisplayName(Object=@n1)) -> (ReturnValue)
   )
   > Execute Python Script(
    PythonScript="import TechArt.CosmeticPicker.CosmeticPickerModule as CosmeticPicker\r\nCosmeticPicker.setOverride(id_name, id)"
    idName=@n5
    id=@n1
   ) {inputs [idName,id]}
  ]}
 Setup {nodes [
   n1 = Setup!() -> (then, CosmeticType:CosmeticPicker_CosmeticTypes Enum=NewEnumerator0, loadCheatValues:bool=true) {is=title K2Node_FunctionEntry}
   > Set CosmeticType(CosmeticType:CosmeticPicker_CosmeticTypes Enum=@n1) -> (Output_Get:CosmeticPicker_CosmeticTypes Enum=NewEnumerator0)
   > Set cosmeticType(cosmeticType=Get CosmeticType() -> (CosmeticType=NewEnumerator0), self=Get Variants()) -> (Output_Get=NewEnumerator0)
   n2 = Enum to String(Enumerator=@n1)
   > PrintString(InString=Append!(A="IDSetup ", B=@n2) -> (ReturnValue))
   > n3 = Branch(Condition=@n1.loadCheatValues)
   > n4 = loadCheats()
   SetEditorProperty(
    execute=[@n4.then, @n3.else]
    Object=Get DetailsView_ID() -> (DetailsView_ID:Details View Object Reference)
    PropertyName=CategoriesToShow
    PropertyValue:Array of Names=Make Array("[0]":Name=String To Name(InString=@n2)) -> (Array:Array of Names)
   )
   n5 = Get DetailsView_ID()
   > SetObject(self=@n5, NewObject=Self-Reference())
   > Bind Event to On Property Changed(self:Property View Base Object Reference=@n5, Delegate=Create Event()) {is=title K2Node_AddDelegate}
   > n6 = getId(type=Get CosmeticType() -> (CosmeticType=NewEnumerator0))
   > updateVariantWidgets!(execute, self:EUW Cosmetic Pickerv 2 Variants Object Reference=Get Variants(), id:Fort Item Definition Object Reference=@n6) -> (then)
  ]}
 clear {nodes [
   clear() {is=title K2Node_FunctionEntry}
   > Set backpack()
   > Set glider()
   > Set pickaxe()
   > Set character()
   > Set contrail()
   > Set Shoes(Shoes:Cosmetic Shoes Item Definition Soft Object Reference) -> (Output_Get:Cosmetic Shoes Item Definition Soft Object Reference)
   > Set Companion()
   > Set Guitar()
   > Set Bass()
   > Set Drums()
   > Set Mic()
   > Set Keytar()
   > Set Aura()
   > clearVariants(self:EUW Cosmetic Pickerv 2 Variants Object Reference=Get Variants())
  ]}
 getContrail {nodes [
   getContrail() {is=title K2Node_FunctionEntry}
   > n1 = Execute Python Script!(execute, PythonScript="import TechArt.CosmeticPicker.CosmeticPickerModule as CosmeticPicker\r\ncontrail = CosmeticPicker.getContrail()") -> (then, ReturnValue, contrail:Athena Sky Dive Contrail Item Definition Object Reference) {outputs [contrail]}
   > Return Node!(execute, name:Athena Sky Dive Contrail Item Definition Object Reference=@n1.contrail)
  ]}
 getId {nodes [
   n1 = getId!() -> (then, type:CosmeticPicker_CosmeticTypes Enum=NewEnumerator0) {is=title K2Node_FunctionEntry}
   > n2 = Switch on CosmeticPicker_CosmeticTypes(Selection=@n1)
   n3 = Set local_id(execute=@n2.NewEnumerator10, local_id:Fort Item Definition Object Reference=Get Guitar()) -> (Output_Get:Fort Item Definition Object Reference)
   Return Node!(
    execute=[
     @n21.then
     @n10.then
     @n14.then
     @n3.then
     @n9.then
     @n4.then
     @n5.then
     @n6.then
     @n12.then
     @n11.then
     @n13.then
     @n7.then
     @n15.then
     @n18.then
     @n8.then
    ]
    id:Fort Item Definition Object Reference=Get local_id() -> (local_id:Fort Item Definition Object Reference)
   )
   n4 = Set local_id(execute=@n2.NewEnumerator11, local_id=Get Bass())
   n5 = Set local_id(execute=@n2.NewEnumerator12, local_id=Get Drums())
   n6 = Set local_id(execute=@n2.NewEnumerator13, local_id=Get Mic())
   n7 = Set local_id(execute=@n2.NewEnumerator14, local_id=Get Keytar())
   n8 = Set local_id(execute=@n2.NewEnumerator15, local_id=Get Aura())
   n9 = Set local_id(execute=@n2.NewEnumerator0, local_id=Get character())
   n10 = Set local_id(execute=@n2.NewEnumerator3, local_id=Get glider())
   n11 = Set local_id(execute=@n2.NewEnumerator2, local_id=Get pickaxe())
   n12 = Set local_id(execute=@n2.NewEnumerator1, local_id=Get backpack())
   n13 = Set local_id(execute=@n2.NewEnumerator4, local_id=Get Emote())
   n14 = Set local_id(execute=@n2.NewEnumerator5, local_id=Get wrap())
   n15 = Set local_id(execute=@n2.NewEnumerator7, local_id=Get contrail())
   n16 = LoadAsset_Blocking(execute=@n2.NewEnumerator8, Asset=Get Shoes())
   > n17 = Cast To FortItemDefinition!(execute, Object=@n16) -> (then, CastFailed, AsFort Item Definition:Fort Item Definition Object Reference) {is=title K2Node_DynamicCast}
   > n18 = Set local_id(local_id=@n17)
   n19 = LoadAsset_Blocking(execute=@n2.NewEnumerator9, Asset=Get Companion())
   > n20 = Cast To FortItemDefinition(Object=@n19)
   > n21 = Set local_id(local_id=@n20)
  ]}
 getOverride {nodes [
   n1 = getOverride!() -> (then, cosmeticType:CosmeticPicker_CosmeticTypes Enum=NewEnumerator0) {is=title K2Node_FunctionEntry}
   > n2 = Execute Python Script!(
    execute
    PythonScript="import TechArt.CosmeticPicker.CosmeticPickerModule as CosmeticPicker\r\n\r\nid = CosmeticPicker.getOverride(id_name)"
    id_name:String=Enum to String(Enumerator=@n1)
   ) -> (then, ReturnValue, id:Fort Item Definition Object Reference) {inputs [id_name], outputs [id]}
   > n3 = Branch(Condition=@n2)
   > Return Node(id=@n2.id)
   Return Node(execute=@n3.else)
  ]}
 loadCheats {nodes [
   loadCheats() {is=title K2Node_FunctionEntry}
   > n1 = Switch on CosmeticPicker_CosmeticTypes(Selection=Get CosmeticType() -> (CosmeticType=NewEnumerator0))
   > n2 = getOverride!(execute, self, cosmeticType:CosmeticPicker_CosmeticTypes Enum=NewEnumerator0) -> (then, id:Fort Item Definition Object Reference) {is=title K2Node_CallFunction}
   > n3 = Cast To AthenaCharacterItemDefinition(Object=@n2)
   > Is Valid(InputObject=@n3)
   > Set character(character=@n3)
   n4 = getOverride(execute=@n1.NewEnumerator1, cosmeticType=NewEnumerator1)
   > n5 = Cast To AthenaBackpackItemDefinition(Object=@n4)
   > Set backpack(backpack=@n5)
   n6 = getOverride(execute=@n1.NewEnumerator10, cosmeticType=NewEnumerator10)
   > n7 = Cast To SparksGuitarItemDefinition(Object=@n6)
   > Set Guitar(Guitar=@n7)
   n8 = getOverride(execute=@n1.NewEnumerator11, cosmeticType=NewEnumerator11)
   > n9 = Cast To SparksBassItemDefinition(Object=@n8)
   > Set Bass(Bass=@n9)
   n10 = getOverride(execute=@n1.NewEnumerator12, cosmeticType=NewEnumerator12)
   > n11 = Cast To SparksDrumItemDefinition(Object=@n10)
   > Set Drums(Drums=@n11)
   n12 = getOverride(execute=@n1.NewEnumerator13, cosmeticType=NewEnumerator13)
   > n13 = Cast To SparksMicItemDefinition(Object=@n12)
   > Set Mic(Mic=@n13)
   n14 = getOverride(execute=@n1.NewEnumerator14, cosmeticType=NewEnumerator14)
   > n15 = Cast To SparksKeyboardItemDefinition(Object=@n14)
   > Set Keytar(Keytar=@n15)
   n16 = getOverride(execute=@n1.NewEnumerator15, cosmeticType=NewEnumerator15)
   > n17 = Cast To SparksAuraItemDefinition(Object=@n16)
   > Set Aura(Aura=@n17)
   n18 = getOverride(execute=@n1.NewEnumerator2, cosmeticType=NewEnumerator2)
   > n19 = Cast To AthenaPickaxeItemDefinition(Object=@n18)
   > Set pickaxe(pickaxe=@n19)
   n20 = getOverride(execute=@n1.NewEnumerator3, cosmeticType=NewEnumerator3)
   > n21 = Cast To AthenaGliderItemDefinition(Object=@n20)
   > Set glider(glider=@n21)
   n22 = getContrail!(execute=@n1.NewEnumerator7, self) -> (then, name:Athena Sky Dive Contrail Item Definition Object Reference) {is=title K2Node_CallFunction}
   > Set contrail(contrail=@n22)
   n23 = getOverride(execute=@n1.NewEnumerator9, cosmeticType=NewEnumerator9)
   > n24 = Cast To CosmeticCompanionItemDefinition(Object=@n23)
   > Set Companion(Companion=Make Soft Reference(Input=@n24))
  ]}
 setAppliedWrap {nodes [
   n1 = setAppliedWrap!() -> (then, inWrap:Athena Item Wrap Definition Object Reference) {is=title K2Node_FunctionEntry}
   > setAppliedWrap!(execute, self:EUW Cosmetic Pickerv 2 Variants Object Reference=Get Variants(), inWrap:Athena Item Wrap Definition Object Reference=@n1) -> (then) {is=title K2Node_CallFunction}
  ]}
 setContrail {nodes [
   n1 = setContrail!() -> (then, inContrail:Athena Sky Dive Contrail Item Definition Object Reference) {is=title K2Node_FunctionEntry}
   > Execute Python Script!(
    execute
    PythonScript="import TechArt.CosmeticPicker.CosmeticPickerModule as CosmeticPicker\r\nCosmeticPicker.setContrail(contrail)"
    contrail:Athena Sky Dive Contrail Item Definition Object Reference=@n1
   ) -> (then, ReturnValue) {inputs [contrail]}
   > Return Node!(execute)
  ]}
 setMimosa {nodes [
   n1 = setMimosa!() -> (then, InMimosa:Cosmetic Companion Item Definition Object Reference) {is=title K2Node_FunctionEntry}
   > Execute Python Script!(
    execute
    PythonScript="import TechArt.CosmeticPicker.CosmeticPickerModule as CosmeticPicker\r\nCosmeticPicker.setMimosa(mimosa)"
    mimosa:Cosmetic Companion Item Definition Object Reference=@n1
   ) -> (then, ReturnValue) {inputs [mimosa]}
   > Return Node()
  ]}
 spawnItem {nodes [
   n1 = FindParentWidgetOfType(
    StartingWidget=Self-Reference()
    Type=&/EUW_CosmeticPickerv2_Advanced.EUW_CosmeticPickerv2_Advanced_C
   ) -> (ReturnValue:EUW Cosmetic Pickerv 2 Advanced Object Reference)
   > n2 = Cast To EUW_CosmeticPickerv2_Advanced!(execute, Object=@n1) -> (then, CastFailed, AsEUW Cosmetic Pickerv 2 Advanced:EUW Cosmetic Pickerv 2 Advanced Object Reference) {is=title K2Node_DynamicCast}
   > SetCompanionMannequin!(
    execute
    self:EUW Cosmetic Pickerv 2 Companion Mannequin Object Reference=Get EUW_CosmeticPickerv2_CompanionMannequin!(self:EUW Cosmetic Pickerv 2 Advanced Object Reference=@n2) -> (EUW_CosmeticPickerv2_CompanionMannequin:EUW Cosmetic Pickerv 2 Companion Mannequin Object Reference)
    CompanionMannequin:Fort Companion Mannequin Object Reference=Get localCompanionMannequin() -> (localCompanionMannequin:Fort Companion Mannequin Object Reference)
   ) -> (then)
   n3 = spawnItem!() -> (then, position:Vector=~) {is=title K2Node_FunctionEntry}
   > Set pos(pos:Vector=@n3) -> (Output_Get:Vector)
   > n4 = Switch on CosmeticPicker_CosmeticTypes(Selection=Get cosmeticType() -> (cosmeticType:CosmeticPicker_CosmeticTypes Enum=NewEnumerator0))
   > n5 = SpawnActor Fort Player Mannequin!(
    execute:Exec
    WorldContextObject:Object Reference
    Class:Actor Class Reference=/Script/FortniteGame.FortPlayerMannequin
    SpawnTransform_Location:Vector=Get pos() -> (pos:Vector)
    SpawnTransform_Rotation:Rotator="0, 0, 0"
    SpawnTransform_Scale:Vector="1,1,1"
    CollisionHandlingOverride:ESpawnActorCollisionHandlingMethod Enum=Undefined
    TransformScaleMethod:ESpawnActorScaleMethod Enum=MultiplyWithRoot
    Owner:Actor Object Reference
    Instigator:Pawn Object Reference
   ) -> (then:Exec, ReturnValue:Fort Player Mannequin Object Reference) {is=title K2Node_SpawnActorFromClass}
   > Set mannequin(mannequin:Fort Player Mannequin Object Reference=@n5) -> (Output_Get:Fort Player Mannequin Object Reference)
   > SetEditorProperty(
    Object=Get mannequin() -> (mannequin:Fort Player Mannequin Object Reference)
    PropertyName=athenaCharacter
    PropertyValue:Athena Character Item Definition Object Reference=Get character()
   )
   n6 = GetCharacterParts(self=Get backpack())
   n7 = Set SkeletalMesh(execute=@n4.NewEnumerator1, SkeletalMesh:Skeletal Mesh Object Reference=GetSkeletalMesh(self="Get (a copy)"(Array=@n6))) -> (Output_Get:Skeletal Mesh Object Reference)
   n8 = Get SkeletalMesh() -> (SkeletalMesh:Skeletal Mesh Object Reference)
   PrintString(execute=[@n14.then, @n7.then, @n18.then], InString=GetDisplayName(Object=@n8))
   n9 = GetBounds!(self:Skeletal Mesh Object Reference=@n8) -> (ReturnValue:Box Sphere Bounds Structure)
   > n10 = SpawnActor Fort Wrap Skeletal Mesh Actor!(
    execute:Exec
    WorldContextObject:Object Reference
    Class:Actor Class Reference=/Script/FortniteGame.FortWrapSkeletalMeshActor
    SpawnTransform_Location:Vector=vector + vector!(
     A=Get pos()
     B_X:double=0
     B_Y:double=0
     B_Z:double=BreakBoxSphereBounds!(InBoxSphereBounds:"Box Sphere Bounds Structure (by ref)"=@n9) -> (Origin:Vector, BoxExtent:Vector, SphereRadius:float).SphereRadius
    ) -> (ReturnValue)
    SpawnTransform_Rotation:Rotator="0, 0, 0"
    SpawnTransform_Scale:Vector="1,1,1"
    CollisionHandlingOverride:ESpawnActorCollisionHandlingMethod Enum=Undefined
    TransformScaleMethod:ESpawnActorScaleMethod Enum=MultiplyWithRoot
    Owner:Actor Object Reference
    Instigator:Pawn Object Reference
   ) -> (then:Exec, ReturnValue:Fort Wrap Skeletal Mesh Actor Object Reference) {is=title K2Node_SpawnActorFromClass}
   > SetSkeletalMeshAsset!(
    execute
    self:Skeletal Mesh Component Object Reference=Get SkeletalMeshComponent!(self:Skeletal Mesh Actor Object Reference=@n10) -> (SkeletalMeshComponent:Skeletal Mesh Component Object Reference)
    NewMesh:Skeletal Mesh Object Reference=Get SkeletalMesh()
   ) -> (then)
   n11 = GetEditorProperty(execute=@n4.NewEnumerator3, Object=Get glider(), PropertyName=SkeletalMesh) -> (PropertyValue:"Soft Object Reference (by ref)")
   > n12 = LoadAsset!(execute, AssetPath="To String (SoftObjectReference)"!(SoftObjectReference:"Soft Object Reference (by ref)"=@n11) -> (ReturnValue:String)) -> (then, ReturnValue)
   > n13 = Cast To SkeletalMesh!(execute, Object=@n12) -> (then, CastFailed, AsSkeletal Mesh:Skeletal Mesh Object Reference) {is=title K2Node_DynamicCast}
   > n14 = Set SkeletalMesh(SkeletalMesh=@n13)
   n15 = SpawnActor Fort Companion Mannequin!(
    execute:Exec=@n4.NewEnumerator9
    WorldContextObject:Object Reference
    Class:Actor Class Reference=/Script/CosmeticCompanionsRuntime.FortCompanionMannequin
    SpawnTransform_Location:Vector=Get pos()
    SpawnTransform_Rotation:Rotator="0, 0, 0"
    SpawnTransform_Scale:Vector="1,1,1"
    CollisionHandlingOverride:ESpawnActorCollisionHandlingMethod Enum=Undefined
    TransformScaleMethod:ESpawnActorScaleMethod Enum=MultiplyWithRoot
    Owner:Actor Object Reference
    Instigator:Pawn Object Reference
   ) -> (then:Exec, ReturnValue:Fort Companion Mannequin Object Reference) {is=title K2Node_SpawnActorFromClass}
   > n16 = Set localCompanionMannequin(localCompanionMannequin:Fort Companion Mannequin Object Reference=@n15) -> (Output_Get:Fort Companion Mannequin Object Reference)
   > SetEditorProperty(
    Object=@n16
    PropertyName=CompanionItemDefinition
    PropertyValue:Cosmetic Companion Item Definition Object Reference=Resolve Soft Reference(Input:Cosmetic Companion Item Definition Soft Object Reference=Get Companion()) -> (Output:Cosmetic Companion Item Definition Object Reference)
   )
   n17 = Get pickaxe()
   n18 = Set SkeletalMesh(execute=@n4.NewEnumerator2, SkeletalMesh=GetWeaponMeshOverride(self=Get WeaponDefinition(self=@n17)))
  ]}
 updateMannequin {nodes [
   updateMannequin() {is=title K2Node_FunctionEntry}
   > PrintString(InString=UpdateMannequin)
   > n1 = Sequence()
   > n2 = FindParentWidgetOfType(StartingWidget=Self-Reference(), Type=&EUW_CosmeticPickerv2_C) -> (ReturnValue:EUW Cosmetic Pickerv 2 Object Reference)
   > n3 = Cast To EUW_CosmeticPickerv2(Object=@n2)
   > updateMannequin(self:EUW Cosmetic Pickerv 2 Mannequin Object Reference=Get EUW_CosmeticPickerv2_Mannequin!(self:EUW Cosmetic Pickerv 2 Object Reference=@n3) -> (EUW_CosmeticPickerv2_Mannequin:EUW Cosmetic Pickerv 2 Mannequin Object Reference)) {is=title K2Node_CallFunction}
   n4 = FindParentWidgetOfType(
    execute=@n1.then_1
    StartingWidget=Self-Reference()
    Type=&EUW_CosmeticPickerv2_Advanced_C
   ) -> (ReturnValue:EUW Cosmetic Pickerv 2 Advanced Object Reference)
   > n5 = Cast To EUW_CosmeticPickerv2_Advanced(Object=@n4)
   > updateMannequin(self:EUW Cosmetic Pickerv 2 Companion Mannequin Object Reference=Get EUW_CosmeticPickerv2_CompanionMannequin(self=@n5))
  ]}
 updateVariants {nodes [
   n1 = updateVariants!() -> (then, id:Fort Item Definition Object Reference) {is=title K2Node_FunctionEntry}
   n2 = Get Variants()
   > updateVariantWidgets(self=@n2, id=@n1)
   > Return Node!(execute, Variants:EUW Cosmetic Pickerv 2 Variants Object Reference=@n2)
  ]}
]
variables [
 Aura {type=SparksAuraItemDefinition, default=null}
 backpack {type=AthenaBackpackItemDefinition, default=/BRCosmetics/Athena/Items/Cosmetics/Backpacks/BID_001_BlueSquire}
 Bass {type=SparksBassItemDefinition, default=null}
 BtnsList {type="array<EUW_Cosmetic_ListBtn_C>", default []}
 character {type=AthenaCharacterItemDefinition, default=/Game/Athena/Items/Cosmetics/Characters/CID_001_Athena_Commando_F_Default}
 CharacterPartIndex {type=integer}
 Companion {type=CosmeticCompanionItemDefinition, default=null}
 contrail {type=AthenaSkyDiveContrailItemDefinition, default=null}
 CosmeticType {type=CosmeticPicker_CosmeticTypes, default=Unknown}
 Drums {type=SparksDrumItemDefinition, default=null}
 Emote {type=AthenaDanceItemDefinition, default=null}
 glider {type=AthenaGliderItemDefinition, default=/Game/Athena/Items/Cosmetics/Gliders/DefaultGlider}
 Guitar {type=SparksGuitarItemDefinition, default=null}
 Item Definitiion {type=FortItemDefinition, default=null}
 Keytar {type=SparksKeyboardItemDefinition, default=null}
 Mic {type=SparksMicItemDefinition, default=null}
 pickaxe {type=AthenaPickaxeItemDefinition, default=/Game/Athena/Items/Cosmetics/Pickaxes/DefaultPickaxe}
 Shoes {type=CosmeticShoesItemDefinition, default=null}
 wrap {type=AthenaItemWrapDefinition, default=null}
]
widgets {
 info {rootWidgetClass=&CanvasPanel, widgetCount=11}
 widgets [
  CanvasPanel_0 {widgetClassPath=&CanvasPanel, widget=CanvasPanel_0}
  Image_64 {
   widgetClassPath=&Image
   bIsVariable=true
   parent=CanvasPanel_0
   widget=Image_64
   properties {
    brush {
     bIsDynamicallyLoaded=false
     drawAs=Image
     imageType=NoImage
     mirroring=NoMirror
     resourceName=None
     resourceObject=null
     tiling=NoTile
     imageSize=<1 1>
     margin=<ltrb 0 0 0 0>
     outlineSettings {
      bUseBrushTransparency=false
      roundingType=HalfHeightRadius
      width=0
      color {colorUseRule=UseColor_Specified, specifiedColor=<rgba 0 0 0 0>}
      cornerRadii=<0 0 0 0>
     }
     tintColor {colorUseRule=UseColor_Specified, specifiedColor=<rgba 1 1 1 1>}
     uVRegion {bIsValid=false, max=<0 0>, min=<0 0>}
    }
    colorAndOpacity=<rgba 0.0176 0.0176 0.0176 1>
   }
   slot {class=CanvasPanelSlot, properties {layoutData {alignment=<0 0>, anchors {maximum=<1 1>, minimum=<0 0>}, offsets=<ltrb 0 0 0 25>}}}
  }
  VerticalBox_26 {
   widgetClassPath=&VerticalBox
   parent=CanvasPanel_0
   widget=VerticalBox_26
   properties {clipping=ClipToBounds}
   slot {
    class=CanvasPanelSlot
    properties {bAutoSize=true, layoutData {alignment=<0 0>, anchors {maximum=<1 0>, minimum=<0 0>}, offsets=<ltrb 0 0 0 50>}}
   }
  }
  HorizontalBox_0 {
   widgetClassPath=&HorizontalBox
   parent=VerticalBox_26
   widget=HorizontalBox_0
   slot {class=VerticalBoxSlot, properties {size {sizeRule=Fill, value=1}}}
  }
  FolderFinderBtn {
   widgetClassPath=&EditorUtilityButton
   parent=HorizontalBox_0
   widget=FolderFinderBtn
   properties {widgetStyle {
     clickedSlateSound {resourceObject=null}
     disabled {
      bIsDynamicallyLoaded=false
      drawAs=RoundedBox
      imageType=NoImage
      mirroring=NoMirror
      resourceName=None
      resourceObject=/Game/UI/Foundation/Textures/Icons/Replay/T-Icon-Replay-Folder-L
      tiling=NoTile
      imageSize=<64 64>
      margin=<ltrb 0 0 0 0>
      outlineSettings {
       bUseBrushTransparency=false
       roundingType=FixedRadius
       width=1
       color {colorUseRule=UseColor_ColorTable, specifiedColor=<rgba 0 0 0 0>}
       cornerRadii=<4 4 4 4>
      }
      tintColor {colorUseRule=UseColor_Specified, specifiedColor=<rgba 0.0365 0.0365 0.0365 1>}
      uVRegion {bIsValid=false, max=<0 0>, min=<0 0>}
     }
     disabledForeground {colorUseRule=UseColor_ColorTable, specifiedColor=<rgba 0 0 0 0>}
     hovered {
      bIsDynamicallyLoaded=false
      drawAs=Image
      imageType=NoImage
      mirroring=NoMirror
      resourceName=None
      resourceObject=&T-Icon-Replay-Folder-L
      tiling=NoTile
      imageSize=<16 16>
      margin=<ltrb 0 0 0 0>
      outlineSettings {
       bUseBrushTransparency=false
       roundingType=FixedRadius
       width=1
       color {colorUseRule=UseColor_ColorTable, specifiedColor=<rgba 0 0 0 0>}
       cornerRadii=<4 4 4 4>
      }
      tintColor {colorUseRule=UseColor_Specified, specifiedColor=<rgba 0.4167 0.4167 0.4167 1>}
      uVRegion {bIsValid=false, max=<0 0>, min=<0 0>}
     }
     hoveredForeground {colorUseRule=UseColor_ColorTable, specifiedColor=<rgba 0 0 0 0>}
     hoveredSlateSound {resourceObject=null}
     normal {
      bIsDynamicallyLoaded=false
      drawAs=Image
      imageType=NoImage
      mirroring=NoMirror
      resourceName=None
      resourceObject=&T-Icon-Replay-Folder-L
      tiling=NoTile
      imageSize=<16 16>
      margin=<ltrb 0 0 0 0>
      outlineSettings {
       bUseBrushTransparency=false
       roundingType=FixedRadius
       width=1
       color {colorUseRule=UseColor_ColorTable, specifiedColor=<rgba 0 0 0 0>}
       cornerRadii=<4 4 4 4>
      }
      tintColor {colorUseRule=UseColor_Specified, specifiedColor=<rgba 0.2604 0.2604 0.2604 1>}
      uVRegion {bIsValid=false, max=<0 0>, min=<0 0>}
     }
     normalForeground {colorUseRule=UseColor_ColorTable, specifiedColor=<rgba 0 0 0 0>}
     normalPadding=<ltrb 12 1.5 12 1.5>
     pressed {
      bIsDynamicallyLoaded=false
      drawAs=Image
      imageType=NoImage
      mirroring=NoMirror
      resourceName=None
      resourceObject=&T-Icon-Replay-Folder-L
      tiling=NoTile
      imageSize=<16 16>
      margin=<ltrb 0 0 0 0>
      outlineSettings {
       bUseBrushTransparency=false
       roundingType=FixedRadius
       width=1
       color {colorUseRule=UseColor_ColorTable, specifiedColor=<rgba 0 0 0 0>}
       cornerRadii=<4 4 4 4>
      }
      tintColor {colorUseRule=UseColor_Specified, specifiedColor=<rgba 0.2604 0.2604 0.2604 1>}
      uVRegion {bIsValid=false, max=<0 0>, min=<0 0>}
     }
     pressedForeground {colorUseRule=UseColor_ColorTable, specifiedColor=<rgba 0 0 0 0>}
     pressedPadding=<ltrb 12 2.5 12 0.5>
     pressedSlateSound {resourceObject=null}
    }}
   slot {class=HorizontalBoxSlot, properties {horizontalAlignment=HAlign_Left, verticalAlignment=VAlign_Top, padding=<ltrb 8 16 0 0>}}
  }
  DetailsView_ID {
   widgetClassPath=&DetailsView
   parent=HorizontalBox_0
   widget=DetailsView_ID
   properties {categoriesToShow [CID], renderTransform {angle=0, scale=<1 1>, shear=<0 0>, translation=<0 -30>}}
   slot {class=HorizontalBoxSlot, properties {size {sizeRule=Fill, value=1}}}
  }
  ScrollBox_52 {
   widgetClassPath=&ScrollBox
   parent=VerticalBox_26
   widget=ScrollBox_52
   properties {
    orientation=Orient_Horizontal
    renderTransform {angle=0, scale=<1 1>, shear=<0 0>, translation=<0 -30>}
    widgetBarStyle {
     thickness=8
     draggedThumbImage {
      bIsDynamicallyLoaded=false
      drawAs=RoundedBox
      imageType=NoImage
      mirroring=NoMirror
      resourceName=None
      resourceObject=null
      tiling=NoTile
      imageSize=<0 0>
      margin=<ltrb 0 0 0 0>
      outlineSettings {
       bUseBrushTransparency=false
       roundingType=FixedRadius
       width=0
       color {colorUseRule=UseColor_Specified, specifiedColor=<rgba 0 0 0 0>}
       cornerRadii=<4 4 4 4>
      }
      tintColor {colorUseRule=UseColor_Specified, specifiedColor=<rgba 0.2159 0.2159 0.2159 1>}
      uVRegion {bIsValid=false, max=<0 0>, min=<0 0>}
     }
     horizontalBackgroundImage {
      bIsDynamicallyLoaded=false
      drawAs=NoDrawType
      imageType=NoImage
      mirroring=NoMirror
      resourceName=None
      resourceObject=null
      tiling=NoTile
      imageSize=<0 0>
      margin=<ltrb 0 0 0 0>
      outlineSettings {
       bUseBrushTransparency=false
       roundingType=HalfHeightRadius
       width=0
       color {colorUseRule=UseColor_Specified, specifiedColor=<rgba 0 0 0 0>}
       cornerRadii=<0 0 0 0>
      }
      tintColor {colorUseRule=UseColor_Specified, specifiedColor=<rgba 1 1 1 1>}
      uVRegion {bIsValid=false, max=<0 0>, min=<0 0>}
     }
     horizontalBottomSlotImage {
      bIsDynamicallyLoaded=false
      drawAs=NoDrawType
      imageType=NoImage
      mirroring=NoMirror
      resourceName=None
      resourceObject=null
      tiling=NoTile
      imageSize=<0 0>
      margin=<ltrb 0 0 0 0>
      outlineSettings {
       bUseBrushTransparency=false
       roundingType=HalfHeightRadius
       width=0
       color {colorUseRule=UseColor_Specified, specifiedColor=<rgba 0 0 0 0>}
       cornerRadii=<0 0 0 0>
      }
      tintColor {colorUseRule=UseColor_Specified, specifiedColor=<rgba 1 1 1 1>}
      uVRegion {bIsValid=false, max=<0 0>, min=<0 0>}
     }
     horizontalTopSlotImage {
      bIsDynamicallyLoaded=false
      drawAs=NoDrawType
      imageType=NoImage
      mirroring=NoMirror
      resourceName=None
      resourceObject=null
      tiling=NoTile
      imageSize=<0 0>
      margin=<ltrb 0 0 0 0>
      outlineSettings {
       bUseBrushTransparency=false
       roundingType=HalfHeightRadius
       width=0
       color {colorUseRule=UseColor_Specified, specifiedColor=<rgba 0 0 0 0>}
       cornerRadii=<0 0 0 0>
      }
      tintColor {colorUseRule=UseColor_Specified, specifiedColor=<rgba 1 1 1 1>}
      uVRegion {bIsValid=false, max=<0 0>, min=<0 0>}
     }
     hoveredThumbImage {
      bIsDynamicallyLoaded=false
      drawAs=RoundedBox
      imageType=NoImage
      mirroring=NoMirror
      resourceName=None
      resourceObject=null
      tiling=NoTile
      imageSize=<0 0>
      margin=<ltrb 0 0 0 0>
      outlineSettings {
       bUseBrushTransparency=false
       roundingType=FixedRadius
       width=0
       color {colorUseRule=UseColor_Specified, specifiedColor=<rgba 0 0 0 0>}
       cornerRadii=<4 4 4 4>
      }
      tintColor {colorUseRule=UseColor_Specified, specifiedColor=<rgba 0.2159 0.2159 0.2159 1>}
      uVRegion {bIsValid=false, max=<0 0>, min=<0 0>}
     }
     normalThumbImage {
      bIsDynamicallyLoaded=false
      drawAs=RoundedBox
      imageType=NoImage
      mirroring=NoMirror
      resourceName=None
      resourceObject=null
      tiling=NoTile
      imageSize=<0 0>
      margin=<ltrb 0 0 0 0>
      outlineSettings {
       bUseBrushTransparency=false
       roundingType=FixedRadius
       width=0
       color {colorUseRule=UseColor_Specified, specifiedColor=<rgba 0 0 0 0>}
       cornerRadii=<4 4 4 4>
      }
      tintColor {colorUseRule=UseColor_Specified, specifiedColor=<rgba 0.0953 0.0953 0.0953 1>}
      uVRegion {bIsValid=false, max=<0 0>, min=<0 0>}
     }
     verticalBackgroundImage {
      bIsDynamicallyLoaded=false
      drawAs=NoDrawType
      imageType=NoImage
      mirroring=NoMirror
      resourceName=None
      resourceObject=null
      tiling=NoTile
      imageSize=<0 0>
      margin=<ltrb 0 0 0 0>
      outlineSettings {
       bUseBrushTransparency=false
       roundingType=HalfHeightRadius
       width=0
       color {colorUseRule=UseColor_Specified, specifiedColor=<rgba 0 0 0 0>}
       cornerRadii=<0 0 0 0>
      }
      tintColor {colorUseRule=UseColor_Specified, specifiedColor=<rgba 1 1 1 1>}
      uVRegion {bIsValid=false, max=<0 0>, min=<0 0>}
     }
     verticalBottomSlotImage {
      bIsDynamicallyLoaded=false
      drawAs=NoDrawType
      imageType=NoImage
      mirroring=NoMirror
      resourceName=None
      resourceObject=null
      tiling=NoTile
      imageSize=<0 0>
      margin=<ltrb 0 0 0 0>
      outlineSettings {
       bUseBrushTransparency=false
       roundingType=HalfHeightRadius
       width=0
       color {colorUseRule=UseColor_Specified, specifiedColor=<rgba 0 0 0 0>}
       cornerRadii=<0 0 0 0>
      }
      tintColor {colorUseRule=UseColor_Specified, specifiedColor=<rgba 1 1 1 1>}
      uVRegion {bIsValid=false, max=<0 0>, min=<0 0>}
     }
     verticalTopSlotImage {
      bIsDynamicallyLoaded=false
      drawAs=NoDrawType
      imageType=NoImage
      mirroring=NoMirror
      resourceName=None
      resourceObject=null
      tiling=NoTile
      imageSize=<0 0>
      margin=<ltrb 0 0 0 0>
      outlineSettings {
       bUseBrushTransparency=false
       roundingType=HalfHeightRadius
       width=0
       color {colorUseRule=UseColor_Specified, specifiedColor=<rgba 0 0 0 0>}
       cornerRadii=<0 0 0 0>
      }
      tintColor {colorUseRule=UseColor_Specified, specifiedColor=<rgba 1 1 1 1>}
      uVRegion {bIsValid=false, max=<0 0>, min=<0 0>}
     }
    }
    widgetStyle {
     barThickness=8
     bottomShadowBrush {
      bIsDynamicallyLoaded=false
      drawAs=Box
      imageType=FullColor
      mirroring=NoMirror
      resourceName=../../../Engine/Content/Slate/Common/ScrollBoxShadowBottom.png
      resourceObject=null
      tiling=NoTile
      imageSize=<16 8>
      margin=<ltrb 0.5 0 0.5 1>
      outlineSettings {
       bUseBrushTransparency=false
       roundingType=HalfHeightRadius
       width=0
       color {colorUseRule=UseColor_Specified, specifiedColor=<rgba 0 0 0 0>}
       cornerRadii=<0 0 0 0>
      }
      tintColor {colorUseRule=UseColor_Specified, specifiedColor=<rgba 1 1 1 1>}
      uVRegion {bIsValid=false, max=<0 0>, min=<0 0>}
     }
     horizontalScrolledContentPadding=<ltrb 0 0 1 0>
     leftShadowBrush {
      bIsDynamicallyLoaded=false
      drawAs=Box
      imageType=FullColor
      mirroring=NoMirror
      resourceName=../../../Engine/Content/Slate/Common/ScrollBoxShadowLeft.png
      resourceObject=null
      tiling=NoTile
      imageSize=<8 16>
      margin=<ltrb 1 0.5 0 0.5>
      outlineSettings {
       bUseBrushTransparency=false
       roundingType=HalfHeightRadius
       width=0
       color {colorUseRule=UseColor_Specified, specifiedColor=<rgba 0 0 0 0>}
       cornerRadii=<0 0 0 0>
      }
      tintColor {colorUseRule=UseColor_Specified, specifiedColor=<rgba 1 1 1 1>}
      uVRegion {bIsValid=false, max=<0 0>, min=<0 0>}
     }
     rightShadowBrush {
      bIsDynamicallyLoaded=false
      drawAs=Box
      imageType=FullColor
      mirroring=NoMirror
      resourceName=../../../Engine/Content/Slate/Common/ScrollBoxShadowRight.png
      resourceObject=null
      tiling=NoTile
      imageSize=<8 16>
      margin=<ltrb 0 0.5 1 0.5>
      outlineSettings {
       bUseBrushTransparency=false
       roundingType=HalfHeightRadius
       width=0
       color {colorUseRule=UseColor_Specified, specifiedColor=<rgba 0 0 0 0>}
       cornerRadii=<0 0 0 0>
      }
      tintColor {colorUseRule=UseColor_Specified, specifiedColor=<rgba 1 1 1 1>}
      uVRegion {bIsValid=false, max=<0 0>, min=<0 0>}
     }
     topShadowBrush {
      bIsDynamicallyLoaded=false
      drawAs=Box
      imageType=FullColor
      mirroring=NoMirror
      resourceName=../../../Engine/Content/Slate/Common/ScrollBoxShadowTop.png
      resourceObject=null
      tiling=NoTile
      imageSize=<16 8>
      margin=<ltrb 0.5 1 0.5 0>
      outlineSettings {
       bUseBrushTransparency=false
       roundingType=HalfHeightRadius
       width=0
       color {colorUseRule=UseColor_Specified, specifiedColor=<rgba 0 0 0 0>}
       cornerRadii=<0 0 0 0>
      }
      tintColor {colorUseRule=UseColor_Specified, specifiedColor=<rgba 1 1 1 1>}
      uVRegion {bIsValid=false, max=<0 0>, min=<0 0>}
     }
     verticalScrolledContentPadding=<ltrb 0 0 0 1>
    }
   }
   slot {class=VerticalBoxSlot, properties {horizontalAlignment=HAlign_Center}}
  }
  Variants {
   widgetClassPath=&/EUW_CosmeticPickerv2_Variants
   bIsVariable=true
   parent=ScrollBox_52
   widget=Variants
   slot {class=ScrollBoxSlot, properties {horizontalAlignment=HAlign_Center}}
  }
  CharacterPartPopup {
   widgetClassPath=&VerticalBox
   bIsVariable=true
   parent=CanvasPanel_0
   widget=CharacterPartPopup
   properties {visibility=Hidden}
   slot {
    class=CanvasPanelSlot
    properties {bAutoSize=true, zOrder=1, layoutData {alignment=<0 0>, anchors {maximum=<0 0>, minimum=<0 0>}, offsets=<ltrb 16 24 84.523 63.462>}}
   }
  }
  SizeBox_75 {widgetClassPath=&SizeBox, parent=CanvasPanel_0, widget=SizeBox_75, slot {class=CanvasPanelSlot, properties=~}}
  EUW_PopupDropdownMenu {
   widgetClassPath=&/EUW_Cosmetic_PopupDropdownMenu
   bIsVariable=true
   parent=CanvasPanel_0
   widget=EUW_PopupDropdownMenu
   slot {
    class=CanvasPanelSlot
    properties {bAutoSize=true, zOrder=1, layoutData {alignment=<0 0>, anchors {maximum=<0 0>, minimum=<0 0>}, offsets=<ltrb 12 24 100 30>}}
   }
  }
 ]
}
