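# Editor utility widget for the Mannequin panel of Cosmetic Picker v2, linking a Fort Player Mannequin to the picker.
# A details view exposes its Mannequin category and binds a property-change delegate.
# MannequinChanged clears the parent picker's variant channels, then copies the mannequin's character and backpack into the character ID widget.
# updateMannequin sends either selected mannequins or the stored one to a Python CosmeticPicker module, reloaded per call, to set override variants.
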EditorUtilityWidgetBlueprint /Game/NeverCook/ArtTools/Tech_Art/CosmeticPicker/EUW_CosmeticPickerv2_Mannequin
graphs [
 EventGraph {nodes [
   Event Construct()
   > SetEditorProperty(
    Object=Get DetailsView_mannequin() -> (DetailsView_mannequin:Details View Object Reference)
    PropertyName=CategoriesToShow
    PropertyValue:Array of Names=Make Array("[0]":Name=String To Name(InString=Mannequin)) -> (Array:Array of Names)
   )
   n1 = Get DetailsView_mannequin()
   > SetObject(self=@n1, NewObject=Self-Reference())
   > Bind Event to On Property Changed(self:Property View Base Object Reference=@n1, Delegate=Create Event()) {is=title K2Node_AddDelegate}
   Event PreConstruct()
   Event Tick()
  ]}
 MannequinChanged {nodes [
   MannequinChanged!() -> (then, PropertyName:Name) {is=title K2Node_FunctionEntry}
   > PrintString(InString=MannequinChanged)
   > n1 = FindParentWidgetOfType!(
    execute
    StartingWidget:Widget Object Reference=Self-Reference()
    Type:Widget Class Reference=&/EUW_CosmeticPickerv2.EUW_CosmeticPickerv2_C
   ) -> (then, ReturnValue:EUW Cosmetic Pickerv 2 Object Reference)
   > n2 = Cast To EUW_CosmeticPickerv2!(execute, Object=@n1) -> (then, CastFailed, AsEUW Cosmetic Pickerv 2:EUW Cosmetic Pickerv 2 Object Reference) {is=title K2Node_DynamicCast}
   > clearCharacterVariantChannels(self:EUW Cosmetic Pickerv 2 Object Reference=@n2)
   n3 = Get Mannequin() -> (Mannequin:Fort Player Mannequin Object Reference)
   n4 = GetAthenaCharacterItemDefinition!(self:Fort Player Mannequin Object Reference=@n3) -> (ReturnValue:Athena Character Item Definition Object Reference)
   > PrintString(InString=Append(A=mannequin character:, B=GetDisplayName(Object=@n4)))
   n5 = Get EUW_CosmeticPicker_ID_Character!(self:EUW Cosmetic Pickerv 2 Object Reference=@n2) -> (EUW_CosmeticPicker_ID_Character:EUW Cosmetic Pickerv 2 ID Object Reference)
   > Set character!(execute, character:Athena Character Item Definition Object Reference=@n4, self:EUW Cosmetic Pickerv 2 ID Object Reference=@n5) -> (then, Output_Get:Athena Character Item Definition Object Reference)
   n6 = Get AthenaBackBling!(self:Fort Player Mannequin Object Reference=@n3) -> (AthenaBackBling:Athena Backpack Item Definition Object Reference)
   > PrintString(InString=Append(A=mannequin backpack:, B=GetDisplayName(Object=@n6)))
   > Set backpack!(execute, backpack:Athena Backpack Item Definition Object Reference=@n6, self:EUW Cosmetic Pickerv 2 ID Object Reference=@n5) -> (then, Output_Get:Athena Backpack Item Definition Object Reference)
   > IDChanged!(execute, self:EUW Cosmetic Pickerv 2 ID Object Reference=@n5, PropertyName:Name=None) -> (then)
   SetAthenaCharacterDefAndVariants!(
    execute
    self:Fort Player Mannequin Object Reference=Get Mannequin()
    NewAthenaCharacter:Athena Character Item Definition Object Reference=Get character!(self:EUW Cosmetic Pickerv 2 ID Object Reference) -> (character:Athena Character Item Definition Object Reference)
    InOverrideVariants:Array of Mcp Variant Channel Info Structures=Make Array("[0]":Mcp Variant Channel Info Structure) -> (Array:Array of Mcp Variant Channel Info Structures)
   ) -> (then)
  ]}
 Set Mannequn Override Variants {nodes [
   n1 = Set Mannequn Override Variants!() -> (then, mannequin:Fort Player Mannequin Object Reference) {is=title K2Node_FunctionEntry}
   > Execute Python Script!(
    execute
    PythonScript="import TechArt.CosmeticPicker.CosmeticPickerModule as CosmeticPicker\r\n\r\nfrom importlib import *\r\nreload(CosmeticPicker)\r\n\r\nCosmeticPicker.setMannequinOverrideVariants(mannequin)"
    mannequin:Fort Player Mannequin Object Reference=@n1
   ) -> (then, ReturnValue) {inputs [mannequin]}
   > Return Node()
  ]}
 clearMannequinSlot {nodes [
   clearMannequinSlot() {is=title K2Node_FunctionEntry}
   > Set Mannequin(Mannequin:Fort Player Mannequin Object Reference=Get Mannequin_clear() -> (Mannequin_clear:Fort Player Mannequin Object Reference)) -> (Output_Get:Fort Player Mannequin Object Reference)
  ]}
 updateMannequin {nodes [
   updateMannequin() {is=title K2Node_FunctionEntry}
   > PrintString(InString=UpdateMannequin)
   > n1 = Branch(Condition=Get UpdateSelected?() -> (UpdateSelected?:bool))
   > n2 = GetSelectedActors()
   > n3 = For Each Loop(Array:Array of Actor Object References=@n2) -> (Array Element:Actor Object Reference)
   > n4 = Cast To FortPlayerMannequin!(execute, Object=@n3) -> (then, CastFailed, AsFort Player Mannequin:Fort Player Mannequin Object Reference) {is=title K2Node_DynamicCast}
   > Set Mannequn Override Variants!(execute, self, mannequin:Fort Player Mannequin Object Reference=@n4) -> (then) {is=title K2Node_CallFunction}
   n5 = Get Mannequin()
   Is Valid(exec=@n1.else, InputObject=@n5)
   > Set Mannequn Override Variants(mannequin=@n5)
  ]}
]
variables [
 Mannequin {type=FortPlayerMannequin, default=null}
 Mannequin_clear {type=FortPlayerMannequin, default=null}
 UpdateSelected? {type=boolean}
]
widgets {
 info {rootWidgetClass=&CanvasPanel, widgetCount=3}
 widgets [
  CanvasPanel_0 {widgetClassPath=&CanvasPanel, widget=CanvasPanel_0}
  VerticalBox_26 {
   widgetClassPath=&VerticalBox
   parent=CanvasPanel_0
   widget=VerticalBox_26
   properties {clipping=ClipToBounds}
   slot {
    class=CanvasPanelSlot
    properties {bAutoSize=true, layoutData {alignment=<0 0>, anchors {maximum=<1 0>, minimum=<0 0>}, offsets=<ltrb 0 0 0 100>}}
   }
  }
  DetailsView_mannequin {
   widgetClassPath=&DetailsView
   parent=VerticalBox_26
   widget=DetailsView_mannequin
   properties {renderTransform {angle=0, scale=<1 1>, shear=<0 0>, translation=<0 -25>}}
   slot {class=VerticalBoxSlot}
  }
 ]
}
