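# Editor utility widget (Cosmetic Picker v2) for previewing Fortnite cosmetics on player and companion mannequins in the editor.
# Selection-change events can auto-load the selected mannequin into its picker panel, while Playlist Update feeds items to per-slot ID widgets and refreshes variants.
# "Gang" matching parses the character's name and gender to find pickaxe, backpack, glider and contrail by Asset Registry path matching.
# Python scripts and console commands drive overrides, LOD bias and Sparks aura.
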EditorUtilityWidgetBlueprint /Game/NeverCook/ArtTools/Tech_Art/CosmeticPicker/EUW_CosmeticPickerv2
graphs [
 EventGraph {nodes [
   n1 = "On Check State Changed (CheckBox_MannSelected)"!() -> (OutputDelegate, then, bIsChecked:bool) {is=title K2Node_ComponentBoundEvent}
   > Set UpdateSelected?!(
    execute
    UpdateSelected?:bool=@n1.bIsChecked
    self:EUW Cosmetic Pickerv 2 Mannequin Object Reference=Get EUW_CosmeticPickerv2_Mannequin() -> (EUW_CosmeticPickerv2_Mannequin:EUW Cosmetic Pickerv 2 Mannequin Object Reference)
   ) -> (then, Output_Get:bool)
   > n2 = Cast To EUW_CosmeticPickerv2_Advanced!(execute, Object=Get Advanced Tab() -> (Advanced Tab:Widget Object Reference)) -> (then, CastFailed, AsEUW Cosmetic Pickerv 2 Advanced:EUW Cosmetic Pickerv 2 Advanced Object Reference) {is=title K2Node_DynamicCast}
   > Set UpdateSelected?(
    UpdateSelected?=@n1.bIsChecked
    self:EUW Cosmetic Pickerv 2 Companion Mannequin Object Reference=Get EUW_CosmeticPickerv2_CompanionMannequin!(self:EUW Cosmetic Pickerv 2 Advanced Object Reference=@n2) -> (EUW_CosmeticPickerv2_CompanionMannequin:EUW Cosmetic Pickerv 2 Companion Mannequin Object Reference)
   )
   n3 = "On Check State Changed (Disable_SK_LOD)"!() -> (OutputDelegate, then, bIsChecked:bool) {is=title K2Node_ComponentBoundEvent}
   > ExecuteConsoleCommand(Command=Select(Option 0:String=r.SkeletalMeshLODBias 0, Option 1:String=r.SkeletalMeshLODBias -100, Index=@n3.bIsChecked) -> (ReturnValue:String))
   n4 = "On Check State Changed (CheckBox_autoLoad)"!() -> (OutputDelegate, then, bIsChecked:bool) {is=title K2Node_ComponentBoundEvent}
   > n5 = Branch(Condition=@n4.bIsChecked)
   > n6 = GetSelectedActors()
   n7 = "Get (a copy)"(Array:Array of Actor Object References=@n6) -> (Output:Actor Object Reference)
   > n8 = Cast To FortCompanionMannequin!(execute, Object=@n7) -> (then, CastFailed, AsFort Companion Mannequin:Fort Companion Mannequin Object Reference) {is=title K2Node_DynamicCast}
   > Set CompanionMannequin!(
    execute
    CompanionMannequin:Fort Companion Mannequin Object Reference=@n8
    self:EUW Cosmetic Pickerv 2 Companion Mannequin Object Reference=GetCompanionMannequinPicker!() -> (CompanionMannequinPicker:EUW Cosmetic Pickerv 2 Companion Mannequin Object Reference)
   ) -> (then, Output_Get:Fort Companion Mannequin Object Reference)
   n9 = Cast To FortPlayerMannequin!(execute=@n8.CastFailed, Object=@n7) -> (then, CastFailed, AsFort Player Mannequin:Fort Player Mannequin Object Reference) {is=title K2Node_DynamicCast}
   n10 = Get EUW_CosmeticPickerv2_Mannequin()
   > Set Mannequin!(execute, Mannequin:Fort Player Mannequin Object Reference=@n9, self:EUW Cosmetic Pickerv 2 Mannequin Object Reference=@n10) -> (then, Output_Get:Fort Player Mannequin Object Reference)
   Set Mannequin(execute=@n5.else, self=@n10)
   > Set CompanionMannequin(self=GetCompanionMannequinPicker())
   n11 = "On Check State Changed (CheckBox_ForceAura)"!() -> (OutputDelegate, then, bIsChecked:bool) {is=title K2Node_ComponentBoundEvent}
   > n12 = Branch(Condition=@n11.bIsChecked)
   > Execute Python Script(PythonScript="import unreal\nues = unreal.get_editor_subsystem(unreal.UnrealEditorSubsystem)\nw = ues.get_game_world() if ues else None\nif w:\n    unreal.SystemLibrary.execute_console_command(w, \"Sparks.ForceAura 1\")\n")
   Execute Python Script(
    execute=@n12.else
    PythonScript="import unreal\nues = unreal.get_editor_subsystem(unreal.UnrealEditorSubsystem)\nw = ues.get_game_world() if ues else None\nif w:\n    unreal.SystemLibrary.execute_console_command(w, \"Sparks.ForceAura 0\")\n"
   )
   n13 = SelectionChange!() -> (OutputDelegate, then, SelectionSet:Typed Element Selection Set Object Reference) {is=title K2Node_CustomEvent}
   > checkSelectionSetForMannequin!(execute, self, SelectionSet:Typed Element Selection Set Object Reference=@n13.SelectionSet) -> (then)
   Event Construct()
   > n14 = GetSelectionSet(self=Get LevelEditorSubsystem())
   > Bind Event to On Selection Change(self:Typed Element Selection Set Object Reference=@n14, Delegate=@n13) {is=title K2Node_AddDelegate}
   > checkSelectionSetForMannequin(SelectionSet=@n14)
   Execute Python Script(PythonScript="import importlib\r\nimport TechArt.CosmeticPicker.CosmeticPickerModule as CosmeticPicker\r\nimportlib.reload (CosmeticPicker)\r\n")
  ]}
 GetCompanionMannequinPicker {nodes [
   GetCompanionMannequinPicker() {is=title K2Node_FunctionEntry}
   > n1 = Cast To EUW_CosmeticPickerv2_Advanced(Object=Get Advanced Tab())
   > Return Node!(
    execute
    CompanionMannequinPicker:EUW Cosmetic Pickerv 2 Companion Mannequin Object Reference=Get EUW_CosmeticPickerv2_CompanionMannequin(self=@n1)
   )
  ]}
 GetCosmeticPaths {nodes [
   n1 = GetCosmeticPaths!() -> (then, cosmeticType:CosmeticPicker_CosmeticTypes Enum=NewEnumerator0) {is=title K2Node_FunctionEntry}
   > n2 = Switch on CosmeticPicker_CosmeticTypes!(execute, Selection:CosmeticPicker_CosmeticTypes Enum=@n1) -> (
    NewEnumerator0:Exec
    NewEnumerator1:Exec
    NewEnumerator2:Exec
    NewEnumerator3:Exec
    NewEnumerator4:Exec
    NewEnumerator5:Exec
    NewEnumerator6:Exec
    NewEnumerator7:Exec
    NewEnumerator8:Exec
    NewEnumerator9:Exec
    NewEnumerator10:Exec
    NewEnumerator11:Exec
    NewEnumerator12:Exec
    NewEnumerator13:Exec
    NewEnumerator14:Exec
    NewEnumerator15:Exec
   ) {is=title K2Node_SwitchEnum}
   n3 = GetAssetRegistry()
   n4 = GetAssetsByClass!(
    execute=@n2.NewEnumerator1
    self:Asset Registry Interface=@n3
    ClassPathName_PackageName:String=/Script/FortniteGame
    ClassPathName_AssetName:String=AthenaBackpackItemDefinition
    bSearchSubClasses:bool=true
   ) -> (then, OutAssetData:Array of Asset Data Structures, ReturnValue:bool)
   > n5 = GetAssetsByClass(self=@n3, ClassPathName_PackageName=&FortniteGame, ClassPathName_AssetName=AthenaPetCarrierItemDefinition, bSearchSubClasses=true)
   > Append Array!(execute, TargetArray:Array of Asset Data Structures=@n4, SourceArray:Array of Asset Data Structures=@n5) -> (then)
   > n6 = Set Assets(Assets:Array of Asset Data Structures=@n4) -> (Output_Get:Array of Asset Data Structures)
   n7 = For Each Loop(
    Exec=[@n6.then, @n9.then, @n11.then, @n13.then]
    Array:Array of Asset Data Structures=Get Assets() -> (Assets:Array of Asset Data Structures)
   ) -> (Array Element:Asset Data Structure)
   Return Node!(execute=@n7.Completed, paths:Array of Strings=Get Ids() -> (Ids:Array of Strings))
   > Add(TargetArray=Get Ids(), NewItem=ToSoftObjectPath!(InAssetData:"Asset Data Structure (by ref)"=@n7) -> (ReturnValue_PathString:String)) {is=title K2Node_CallArrayFunction}
   n8 = GetAssetsByClass(
    execute=@n2.NewEnumerator2
    self=GetAssetRegistry()
    ClassPathName_PackageName=&FortniteGame
    ClassPathName_AssetName=AthenaPickaxeItemDefinition
    bSearchSubClasses=true
   )
   > n9 = Set Assets(Assets=@n8)
   n10 = GetAssetsByClass(
    execute=@n2.NewEnumerator3
    self=GetAssetRegistry()
    ClassPathName_PackageName=&FortniteGame
    ClassPathName_AssetName=AthenaGliderItemDefinition
    bSearchSubClasses=true
   )
   > n11 = Set Assets(Assets=@n10)
   n12 = GetAssetsByClass(
    execute=@n2.NewEnumerator7
    self=GetAssetRegistry()
    ClassPathName_PackageName=&FortniteGame
    ClassPathName_AssetName=AthenaSkyDiveContrailItemDefinition
    bSearchSubClasses=true
   )
   > n13 = Set Assets(Assets=@n12)
  ]}
 Playlist Update {nodes [
   n1 = Playlist Update!() -> (then, cosmeticType:CosmeticPicker_CosmeticTypes Enum=NewEnumerator0, item:Fort Item Definition Object Reference) {is=title K2Node_FunctionEntry}
   > n2 = Sequence()
   > PrintString(InString=Playlist Update ---------------------------------------------------)
   > Set cosmeticItem(cosmeticItem:Fort Item Definition Object Reference=@n1.item) -> (Output_Get:Fort Item Definition Object Reference)
   n3 = Self-Reference()
   > PrintString(InString=Append!(
     A="PlaylistUpdate "
     B=Enum to String!(Enumerator:CosmeticPicker_CosmeticTypes Enum=@n1) -> (ReturnValue:String)
     C:String=" '"
     D:String=GetDisplayName(Object=@n1.item)
     E:String="' "
     F:String=GetDisplayName(Object=@n3)
    ) -> (ReturnValue))
   > n4 = Switch on CosmeticPicker_CosmeticTypes(Selection=@n1)
   > n5 = Cast To AthenaCharacterItemDefinition!(execute, Object=Get cosmeticItem() -> (cosmeticItem:Fort Item Definition Object Reference)) -> (then, CastFailed, AsAthena Character Item Definition:Athena Character Item Definition Object Reference) {is=title K2Node_DynamicCast}
   n6 = Get EUW_CosmeticPicker_ID_Character() -> (EUW_CosmeticPicker_ID_Character:EUW Cosmetic Pickerv 2 ID Object Reference)
   > n7 = Set character!(execute, character:Athena Character Item Definition Object Reference=@n5, self:EUW Cosmetic Pickerv 2 ID Object Reference=@n6) -> (then, Output_Get:Athena Character Item Definition Object Reference)
   n8 = updateVariants!(
    execute=[@n9.then, @n7.then]
    self:EUW Cosmetic Pickerv 2 ID Object Reference=@n6
    id:Fort Item Definition Object Reference=Get character!(self:EUW Cosmetic Pickerv 2 ID Object Reference=@n6) -> (character:Athena Character Item Definition Object Reference)
   ) -> (then, Variants:EUW Cosmetic Pickerv 2 Variants Object Reference)
   > updateVariantWidgets!(execute, self:EUW Cosmetic Pickerv 2 Variants Object Reference=@n8, id:Fort Item Definition Object Reference=Get cosmeticItem()) -> (then)
   n9 = Set character(execute=@n5.CastFailed, self=@n6)
   n10 = Cast To AthenaBackpackItemDefinition!(execute=@n4.NewEnumerator1, Object=Get cosmeticItem()) -> (then, CastFailed, AsAthena Backpack Item Definition:Athena Backpack Item Definition Object Reference) {is=title K2Node_DynamicCast}
   n11 = Get EUW_CosmeticPicker_ID_Backpack() -> (EUW_CosmeticPicker_ID_Backpack:EUW Cosmetic Pickerv 2 ID Object Reference)
   > n12 = Set backpack!(execute, backpack:Athena Backpack Item Definition Object Reference=@n10, self:EUW Cosmetic Pickerv 2 ID Object Reference=@n11) -> (then, Output_Get:Athena Backpack Item Definition Object Reference)
   n13 = updateVariants(
    execute=[@n14.then, @n12.then]
    self=@n11
    id=Get backpack!(self:EUW Cosmetic Pickerv 2 ID Object Reference=@n11) -> (backpack:Athena Backpack Item Definition Object Reference)
   )
   > updateVariantWidgets(self=@n13, id=Get cosmeticItem())
   n14 = Set backpack(execute=@n10.CastFailed, self=@n11)
   n15 = Cast To AthenaPickaxeItemDefinition!(execute=@n4.NewEnumerator2, Object=Get cosmeticItem()) -> (then, CastFailed, AsAthena Pickaxe Item Definition:Athena Pickaxe Item Definition Object Reference) {is=title K2Node_DynamicCast}
   n16 = Get EUW_CosmeticPicker_ID_Pickaxe() -> (EUW_CosmeticPicker_ID_Pickaxe:EUW Cosmetic Pickerv 2 ID Object Reference)
   > n17 = Set pickaxe!(execute, pickaxe:Athena Pickaxe Item Definition Object Reference=@n15, self:EUW Cosmetic Pickerv 2 ID Object Reference=@n16) -> (then, Output_Get:Athena Pickaxe Item Definition Object Reference)
   updateVariants(
    execute=[@n18.then, @n17.then]
    self=@n16
    id=Get pickaxe!(self:EUW Cosmetic Pickerv 2 ID Object Reference=@n16) -> (pickaxe:Athena Pickaxe Item Definition Object Reference)
   )
   n18 = Set pickaxe(execute=@n15.CastFailed, self=@n16)
   n19 = Cast To AthenaGliderItemDefinition!(execute=@n4.NewEnumerator3, Object=Get cosmeticItem()) -> (then, CastFailed, AsAthena Glider Item Definition:Athena Glider Item Definition Object Reference) {is=title K2Node_DynamicCast}
   n20 = Get EUW_CosmeticPicker_ID_Glider() -> (EUW_CosmeticPicker_ID_Glider:EUW Cosmetic Pickerv 2 ID Object Reference)
   > n21 = Set glider!(execute, glider:Athena Glider Item Definition Object Reference=@n19, self:EUW Cosmetic Pickerv 2 ID Object Reference=@n20) -> (then, Output_Get:Athena Glider Item Definition Object Reference)
   updateVariants(
    execute=[@n22.then, @n21.then]
    self=@n20
    id=Get glider!(self:EUW Cosmetic Pickerv 2 ID Object Reference=@n20) -> (glider:Athena Glider Item Definition Object Reference)
   )
   n22 = Set glider(execute=@n19.CastFailed, self=@n20)
   ExecutePythonCommand(
    execute=@n2.then_1
    PythonCommand="import TechArt.CosmeticPicker.CosmeticPickerModule as CosmeticPicker\r\nCosmeticPicker.kickOverrides()\r\n"
   )
  ]}
 Search for Cosmetic {nodes [
   n1 = Search for Cosmetic!() -> (then, cosmeticType:CosmeticPicker_CosmeticTypes Enum=NewEnumerator0, name:String, gender:String) {is=title K2Node_FunctionEntry}
   > Set cosmeticName(cosmeticName:String=@n1.name) -> (Output_Get:String)
   > n2 = GetCosmeticPaths!(execute, self, cosmeticType:CosmeticPicker_CosmeticTypes Enum=@n1) -> (then, paths:Array of Strings) {is=title K2Node_CallFunction}
   > Set IDs(IDs:Array of Strings=@n2) -> (Output_Get:Array of Strings)
   > n3 = Sequence!(execute) -> (then_0, then_1, then_2:Exec)
   > n4 = Branch(Condition="Equal Exactly (String)"(A=@n1.gender, B=M))
   > n5 = Sequence!(execute) -> (then_0, then_1)
   n6 = Get cosmeticName() -> (cosmeticName:String)
   > n7 = Set newName(newName:String=Append!(A=@n6, B=Male) -> (ReturnValue)) -> (Output_Get:String)
   n8 = searchForEndsWithInArray!(
    execute=[@n9.then, @n12.then, @n13.then, @n7.then, @n16.then, @n17.then]
    self
    array:Array of Strings=Get IDs() -> (IDs:Array of Strings)
    name:String=Get newName() -> (newName:String)
   ) -> (then, cosmetic:String)
   > Branch(Condition="Not Equal Exactly (String)"(A=@n8, B=None))
   > Return Node!(execute, foundCosmetic:String=@n8, success:bool=true)
   n9 = Set newName(execute=@n5.then_1, newName=Append!(A=@n6, B=male1h, C:String) -> (ReturnValue))
   n10 = Sequence(execute=@n4.else)
   n11 = Get cosmeticName()
   > n12 = Set newName(newName=Append!(A=@n11, B=Female) -> (ReturnValue))
   n13 = Set newName(execute=@n10.then_1, newName=Append!(A=@n11, B=female1h, C:String) -> (ReturnValue))
   n14 = Sequence(execute=@n3.then_1)
   n15 = Get cosmeticName()
   > n16 = Set newName(newName=Append!(A=@n15, B=1h) -> (ReturnValue))
   n17 = Set newName(execute=@n14.then_1, newName=@n15)
   Return Node(execute=@n3.then_2, foundCosmetic=None, success=false)
  ]}
 characterChanged {nodes [
   characterChanged() {is=title K2Node_FunctionEntry}
   > PrintString(InString="CharacterChanged ")
   > n1 = Branch(Condition=IsChecked(self=Get CheckBox_gangAssets() -> (CheckBox_gangAssets:Check Box Object Reference)))
   > PrintString(InString=GangAssets checked)
   > gangCosmetics!(execute, self, AdvancedTab:Widget Object Reference=Get Advanced Tab()) -> (then)
   PrintString(execute=@n1.else, InString=GangAssets not checked)
  ]}
 checkSelectionSetForMannequin {nodes [
   n1 = checkSelectionSetForMannequin!() -> (then, SelectionSet:Typed Element Selection Set Object Reference) {is=title K2Node_FunctionEntry}
   > n2 = Branch(Condition=IsChecked(self=Get CheckBox_autoLoad() -> (CheckBox_autoLoad:Check Box Object Reference)))
   > PrintString(InString=checkSelectionSet)
   > n3 = GetElementHandles!(
    execute
    ElementList:Typed Element List Structure=GetNormalizedSelection!(
     SelectionSet:Typed Element Selection Set Object Reference=@n1
     NormalizationOptions_bExpandGroups:bool=true
     NormalizationOptions_bFollowAttachment:bool=False
    ) -> (ReturnValue:Typed Element List Structure)
    BaseInterfaceType:Interface Interface
   ) -> (then, ReturnValue:Array of Script Typed Element Handle Structures)
   > PrintString(InString="To String (Integer)"(InInt=Length(TargetArray:Array of Script Typed Element Handle Structures=@n3)))
   > n4 = Branch(Condition=Is Valid Index!(TargetArray:Array of Script Typed Element Handle Structures=@n3, IndexToTest:int) -> (ReturnValue:bool))
   > n5 = For Each Loop(Array:Array of Script Typed Element Handle Structures=@n3) -> (Array Element:Script Typed Element Handle Structure)
   > n6 = GetObject!(
    execute:Exec
    self:Object Reference=GetElementInterface!(
     self:Typed Element Registry Object Reference=Get Default Typed Element Registry!() -> (ReturnValue:Typed Element Registry Object Reference) {is=title K2Node_CallFunction}
     InElementHandle:"Script Typed Element Handle Structure (by ref)"=@n5
     InBaseInterfaceType:Interface Interface=/Script/TypedElementRuntime.TypedElementObjectInterface
    ) -> (ReturnValue:Object Reference)
    InElementHandle:"Script Typed Element Handle Structure (by ref)"=@n5
   ) -> (then:Exec, ReturnValue:Object Reference) {is=title K2Node_Message}
   > n7 = Is Valid(InputObject=@n6)
   > n8 = Cast To FortCompanionMannequin(Object=@n6)
   > PrintString(InString=GetDisplayName(Object=@n8))
   > PrintString(InString=euw_cosmeticPicker_v2 - about to call mannequin changed)
   n9 = GetCompanionMannequinPicker() {is=title K2Node_CallFunction}
   > SetCompanionMannequin!(
    execute
    self:EUW Cosmetic Pickerv 2 Companion Mannequin Object Reference=@n9
    CompanionMannequin:Fort Companion Mannequin Object Reference=@n8
   ) -> (then)
   > PrintString(InString=ShowMannequin)
   > SetVisibility(self=@n9, InVisibility=Visible)
   n10 = Cast To FortPlayerMannequin(execute=@n8.CastFailed, Object=@n6)
   > PrintString(InString=GetDisplayName(Object=@n10))
   n11 = Get EUW_CosmeticPickerv2_Mannequin()
   > Set Mannequin(Mannequin=@n10, self=@n11)
   > PrintString(InString=euw_cosmeticPicker_v2 - about to call mannequin changed)
   > MannequinChanged!(execute, self:EUW Cosmetic Pickerv 2 Mannequin Object Reference=@n11, PropertyName:Name=None) -> (then)
   > PrintString(InString=ShowMannequin)
   > SetVisibility(self=@n11, InVisibility=Visible)
   PrintString(execute=[@n10.CastFailed, @n7.Is Not Valid, @n4.else], InString=HideMannequin)
   > SetVisibility(self=Get EUW_CosmeticPickerv2_Mannequin(), InVisibility=Collapsed)
   > clearMannequinSlot(self:EUW Cosmetic Pickerv 2 Mannequin Object Reference=Get EUW_CosmeticPickerv2_Mannequin())
   SetVisibility(execute=@n2.else, self=Get EUW_CosmeticPickerv2_Mannequin(), InVisibility=Visible)
  ]}
 clear {nodes [
   clear() {is=title K2Node_FunctionEntry}
   > clear(self:EUW Cosmetic Pickerv 2 ID Object Reference=Get EUW_CosmeticPicker_ID_Character()) {is=title K2Node_CallFunction}
   > clear(self=Get EUW_CosmeticPicker_ID_Pickaxe())
   > clear(self=Get EUW_CosmeticPicker_ID_Backpack())
   > clear(self=Get EUW_CosmeticPicker_ID_Glider())
  ]}
 clearCharacterVariantChannels {nodes [
   clearCharacterVariantChannels() {is=title K2Node_FunctionEntry}
   > Execute Python Script(PythonScript="import TechArt.CosmeticPicker.CosmeticPickerModule as CosmeticPicker\r\nCosmeticPicker.clearCharacterVariantChannels()\r\n")
  ]}
 gangCosmetics {nodes [
   n1 = Search for Cosmetic!(
    execute=@n10.then_0
    self
    cosmeticType:CosmeticPicker_CosmeticTypes Enum=NewEnumerator2
    name:String=Get Name() -> (Name:String)
    gender:String=Get Gender() -> (Gender:String)
   ) -> (then, foundCosmetic:String, success:bool) {is=title K2Node_CallFunction}
   n2 = LoadAsset_Blocking(
    execute=@n18.then
    Asset=To Soft Object Reference(SoftObjectPath=MakeSoftObjectPath!(PathString:String=@n1) -> (ReturnValue:Soft Object Path Structure))
   )
   > n3 = Cast To AthenaPickaxeItemDefinition(Object=@n2)
   n4 = Get EUW_CosmeticPicker_ID_Pickaxe()
   SetOverride!(
    execute
    self:EUW Cosmetic Pickerv 2 ID Object Reference=@n4
    id:Fort Item Definition Object Reference=@n3
    cosmeticType:CosmeticPicker_CosmeticTypes Enum=NewEnumerator2
   ) -> (then)
   > updateVariants(self=@n4, id=@n3)
   n5 = Search for Cosmetic(execute=@n10.then_1, cosmeticType=NewEnumerator1, name=Get Name(), gender=Get Gender())
   n6 = LoadAsset_Blocking(execute=@n19.then, Asset=To Soft Object Reference(SoftObjectPath=MakeSoftObjectPath(PathString=@n5)))
   > n7 = Cast To AthenaBackpackItemDefinition(Object=@n6)
   n8 = Get EUW_CosmeticPicker_ID_Backpack()
   SetOverride(self=@n8, id=@n7, cosmeticType=NewEnumerator1)
   > updateVariants(self=@n8, id=@n7)
   gangCosmetics!() -> (then, AdvancedTab:Widget Object Reference) {is=title K2Node_FunctionEntry}
   > PrintString(InString=Gang Cosmetics)
   > n9 = getNameAndGenderOfCharacter!(execute, self) -> (then, name:String, gender:String)
   > Set Name(Name:String=@n9) -> (Output_Get:String)
   > Set Gender(Gender:String=@n9.gender) -> (Output_Get:String)
   > n10 = Sequence!(execute) -> (then_0, then_1, then_2:Exec, then_3:Exec)
   n11 = Search for Cosmetic(execute=@n10.then_2, cosmeticType=NewEnumerator3, name=Get Name(), gender=Get Gender())
   > PrintString(InString=Append(A="glider  match:", B=@n11))
   > Branch(Condition=@n11.success)
   > n12 = LoadAsset_Blocking(Asset=To Soft Object Reference(SoftObjectPath=MakeSoftObjectPath(PathString=@n11)))
   > n13 = Cast To AthenaGliderItemDefinition(Object=@n12)
   n14 = Get EUW_CosmeticPicker_ID_Glider()
   > Set glider(glider=@n13, self=@n14)
   > SetOverride(self=@n14, id=@n13, cosmeticType=NewEnumerator3)
   > updateVariants(self=@n14, id=@n13)
   n15 = Cast To EUW_CosmeticPickerv2_Advanced(execute=@n10.then_3, Object=Get Advanced Tab())
   > n16 = Search for Cosmetic(cosmeticType=NewEnumerator7, name=Get Name(), gender=Get Gender())
   > PrintString(InString=Append(A="glider  match:", B=@n16))
   > Branch(Condition=@n16.success)
   > n17 = LoadAsset_Blocking(Asset=To Soft Object Reference(SoftObjectPath=MakeSoftObjectPath(PathString=@n16)))
   > SetContrail!(execute, self:EUW Cosmetic Pickerv 2 Advanced Object Reference=@n15, Contrail:Object Reference=@n17) -> (then)
   IDChanged!(execute=@n21.then, self:EUW Cosmetic Pickerv 2 ID Object Reference=@n8, PropertyName:Name=None) -> (then)
   IDChanged(execute=@n20.then, self=@n4, PropertyName=None)
   PrintString(execute=@n1.then, InString=Append(A=pickaxe match:, B=@n1))
   > n18 = Branch(Condition=@n1.success)
   PrintString(execute=@n5.then, InString=Append(A=backpack match:, B=@n5))
   > n19 = Branch(Condition=@n5.success)
   n20 = Set pickaxe(execute=@n3.then, pickaxe=@n3, self=@n4)
   n21 = Set backpack(execute=@n7.then, backpack=@n7, self=@n8)
  ]}
 getNameAndGenderOfCharacter {nodes [
   getNameAndGenderOfCharacter() {is=title K2Node_FunctionEntry}
   n1 = Get EUW_CosmeticPicker_ID_Character()
   n2 = GetBaseFilename(InPath=Get Object Path String(Object=Get character(self=@n1)), bRemovePath=true)
   > PrintString(InString=@n2)
   n3 = ParseIntoArray(SourceString=@n2, Delimiter=_)
   > n4 = Branch(Condition="integer > integer"(A=Length(TargetArray:Array of Strings=@n3), B=1))
   > n5 = Branch(Condition="Equal, Case Insensitive (String)"(A="Get (a ref)"(Array:Array of Strings=@n3, Dimension 1=0) -> (Output:"String (by ref)"), B=character))
   > Set l_name(l_name:String="Get (a ref)"(Array=@n3, Dimension 1=1)) -> (Output_Get:String)
   > n6 = Set l_gender(l_gender:String) -> (Output_Get:String)
   Return Node!(execute=[@n9.then, @n6.then], name:String=Get l_name() -> (l_name:String), gender:String=Get l_gender() -> (l_gender:String))
   n7 = Split(SourceString=@n2, InStr=_, SearchDir=FromEnd)
   n8 = Split(SourceString=@n7, InStr=_, SearchDir=FromEnd)
   PrintString(execute=[@n4.else, @n5.else], InString=Append!(A=name:, B=@n7.RightS, C:String=" gender:", D:String=@n8.RightS) -> (ReturnValue))
   > Set l_name(l_name=@n7.RightS)
   > n9 = Set l_gender(l_gender=@n8.RightS)
  ]}
 searchForEndsWithInArray {nodes [
   n1 = searchForEndsWithInArray!() -> (then, array:Array of Strings, name:String) {is=title K2Node_FunctionEntry}
   > PrintString(InString=Append!(A="checking for ", B=@n1.name) -> (ReturnValue))
   > n2 = Sequence!(execute) -> (then_0, then_1, then_2)
   > n3 = For Each Loop(Array:Array of Strings=@n1) -> (Array Element:String)
   > Branch(Condition=EndsWith(SourceString=@n3, InSuffix=@n1.name))
   > Return Node!(execute, cosmetic:String=@n3)
   n4 = For Each Loop(Exec=@n2.then_1, Array=@n1)
   > Branch(Condition=Contains(SearchIn=@n4, Substring=@n1.name))
   > Return Node(cosmetic=@n4)
   Return Node(execute=@n2.then_2, cosmetic=None)
  ]}
 setAppliedWrapOnIDs {nodes [
   n1 = setAppliedWrapOnIDs!() -> (then, inWrap:Athena Item Wrap Definition Object Reference) {is=title K2Node_FunctionEntry}
   > setAppliedWrap!(
    execute
    self:EUW Cosmetic Pickerv 2 ID Object Reference=Get EUW_CosmeticPicker_ID_Character()
    inWrap:Athena Item Wrap Definition Object Reference=@n1
   ) -> (then)
   > setAppliedWrap(self=Get EUW_CosmeticPicker_ID_Backpack(), inWrap=@n1)
   > setAppliedWrap(self=Get EUW_CosmeticPicker_ID_Pickaxe(), inWrap=@n1)
  ]}
]
variables [Advanced Tab {type=Widget, default=null}]
widgets {
 info {rootWidgetClass=&CanvasPanel, widgetCount=51}
 widgets [
  CanvasPanel_0 {widgetClassPath=&CanvasPanel, bIsVariable=true, widget=CanvasPanel_0}
  Image_64 {
   widgetClassPath=&Image
   bIsVariable=true
   parent=CanvasPanel_0
   widget=Image_64
   properties {colorAndOpacity=<rgba 0.0103 0.0103 0.0103 1>}
   slot {class=CanvasPanelSlot, properties {layoutData {alignment=<0 0>, anchors {maximum=<1 1>, minimum=<0 0>}, offsets=<ltrb 0 0 0 0>}}}
  }
  Overlay_108 {
   widgetClassPath=&Overlay
   parent=CanvasPanel_0
   widget=Overlay_108
   slot {class=CanvasPanelSlot, properties {layoutData {alignment=<0 0>, anchors {maximum=<1 1>, minimum=<0 0>}, offsets=<ltrb 0 0 0 0>}}}
  }
  ScrollBox_176 {
   widgetClassPath=&ScrollBox
   parent=Overlay_108
   widget=ScrollBox_176
   properties {
    widgetBarStyle {
     thickness=8
     draggedThumbImage {
      bIsDynamicallyLoaded=false
      drawAs=RoundedBox
      imageType=NoImage
      mirroring=NoMirror
      resourceName=None
      resourceObject=null
      tiling=NoTile
      imageSize=<0 0>
      margin=<ltrb 0 0 0 0>
      outlineSettings {
       bUseBrushTransparency=false
       roundingType=FixedRadius
       width=0
       color {colorUseRule=UseColor_Specified, specifiedColor=<rgba 0 0 0 0>}
       cornerRadii=<4 4 4 4>
      }
      tintColor {colorUseRule=UseColor_Specified, specifiedColor=<rgba 0.2159 0.2159 0.2159 1>}
      uVRegion {bIsValid=false, max=<0 0>, min=<0 0>}
     }
     horizontalBackgroundImage {
      bIsDynamicallyLoaded=false
      drawAs=NoDrawType
      imageType=NoImage
      mirroring=NoMirror
      resourceName=None
      resourceObject=null
      tiling=NoTile
      imageSize=<0 0>
      margin=<ltrb 0 0 0 0>
      outlineSettings {
       bUseBrushTransparency=false
       roundingType=HalfHeightRadius
       width=0
       color {colorUseRule=UseColor_Specified, specifiedColor=<rgba 0 0 0 0>}
       cornerRadii=<0 0 0 0>
      }
      tintColor {colorUseRule=UseColor_Specified, specifiedColor=<rgba 1 1 1 1>}
      uVRegion {bIsValid=false, max=<0 0>, min=<0 0>}
     }
     horizontalBottomSlotImage {
      bIsDynamicallyLoaded=false
      drawAs=NoDrawType
      imageType=NoImage
      mirroring=NoMirror
      resourceName=None
      resourceObject=null
      tiling=NoTile
      imageSize=<0 0>
      margin=<ltrb 0 0 0 0>
      outlineSettings {
       bUseBrushTransparency=false
       roundingType=HalfHeightRadius
       width=0
       color {colorUseRule=UseColor_Specified, specifiedColor=<rgba 0 0 0 0>}
       cornerRadii=<0 0 0 0>
      }
      tintColor {colorUseRule=UseColor_Specified, specifiedColor=<rgba 1 1 1 1>}
      uVRegion {bIsValid=false, max=<0 0>, min=<0 0>}
     }
     horizontalTopSlotImage {
      bIsDynamicallyLoaded=false
      drawAs=NoDrawType
      imageType=NoImage
      mirroring=NoMirror
      resourceName=None
      resourceObject=null
      tiling=NoTile
      imageSize=<0 0>
      margin=<ltrb 0 0 0 0>
      outlineSettings {
       bUseBrushTransparency=false
       roundingType=HalfHeightRadius
       width=0
       color {colorUseRule=UseColor_Specified, specifiedColor=<rgba 0 0 0 0>}
       cornerRadii=<0 0 0 0>
      }
      tintColor {colorUseRule=UseColor_Specified, specifiedColor=<rgba 1 1 1 1>}
      uVRegion {bIsValid=false, max=<0 0>, min=<0 0>}
     }
     hoveredThumbImage {
      bIsDynamicallyLoaded=false
      drawAs=RoundedBox
      imageType=NoImage
      mirroring=NoMirror
      resourceName=None
      resourceObject=null
      tiling=NoTile
      imageSize=<0 0>
      margin=<ltrb 0 0 0 0>
      outlineSettings {
       bUseBrushTransparency=false
       roundingType=FixedRadius
       width=0
       color {colorUseRule=UseColor_Specified, specifiedColor=<rgba 0 0 0 0>}
       cornerRadii=<4 4 4 4>
      }
      tintColor {colorUseRule=UseColor_Specified, specifiedColor=<rgba 0.2159 0.2159 0.2159 1>}
      uVRegion {bIsValid=false, max=<0 0>, min=<0 0>}
     }
     normalThumbImage {
      bIsDynamicallyLoaded=false
      drawAs=RoundedBox
      imageType=NoImage
      mirroring=NoMirror
      resourceName=None
      resourceObject=null
      tiling=NoTile
      imageSize=<0 0>
      margin=<ltrb 0 0 0 0>
      outlineSettings {
       bUseBrushTransparency=false
       roundingType=FixedRadius
       width=0
       color {colorUseRule=UseColor_Specified, specifiedColor=<rgba 0 0 0 0>}
       cornerRadii=<4 4 4 4>
      }
      tintColor {colorUseRule=UseColor_Specified, specifiedColor=<rgba 0.0953 0.0953 0.0953 1>}
      uVRegion {bIsValid=false, max=<0 0>, min=<0 0>}
     }
     verticalBackgroundImage {
      bIsDynamicallyLoaded=false
      drawAs=NoDrawType
      imageType=NoImage
      mirroring=NoMirror
      resourceName=None
      resourceObject=null
      tiling=NoTile
      imageSize=<0 0>
      margin=<ltrb 0 0 0 0>
      outlineSettings {
       bUseBrushTransparency=false
       roundingType=HalfHeightRadius
       width=0
       color {colorUseRule=UseColor_Specified, specifiedColor=<rgba 0 0 0 0>}
       cornerRadii=<0 0 0 0>
      }
      tintColor {colorUseRule=UseColor_Specified, specifiedColor=<rgba 1 1 1 1>}
      uVRegion {bIsValid=false, max=<0 0>, min=<0 0>}
     }
     verticalBottomSlotImage {
      bIsDynamicallyLoaded=false
      drawAs=NoDrawType
      imageType=NoImage
      mirroring=NoMirror
      resourceName=None
      resourceObject=null
      tiling=NoTile
      imageSize=<0 0>
      margin=<ltrb 0 0 0 0>
      outlineSettings {
       bUseBrushTransparency=false
       roundingType=HalfHeightRadius
       width=0
       color {colorUseRule=UseColor_Specified, specifiedColor=<rgba 0 0 0 0>}
       cornerRadii=<0 0 0 0>
      }
      tintColor {colorUseRule=UseColor_Specified, specifiedColor=<rgba 1 1 1 1>}
      uVRegion {bIsValid=false, max=<0 0>, min=<0 0>}
     }
     verticalTopSlotImage {
      bIsDynamicallyLoaded=false
      drawAs=NoDrawType
      imageType=NoImage
      mirroring=NoMirror
      resourceName=None
      resourceObject=null
      tiling=NoTile
      imageSize=<0 0>
      margin=<ltrb 0 0 0 0>
      outlineSettings {
       bUseBrushTransparency=false
       roundingType=HalfHeightRadius
       width=0
       color {colorUseRule=UseColor_Specified, specifiedColor=<rgba 0 0 0 0>}
       cornerRadii=<0 0 0 0>
      }
      tintColor {colorUseRule=UseColor_Specified, specifiedColor=<rgba 1 1 1 1>}
      uVRegion {bIsValid=false, max=<0 0>, min=<0 0>}
     }
    }
    widgetStyle {
     barThickness=8
     bottomShadowBrush {
      bIsDynamicallyLoaded=false
      drawAs=Box
      imageType=FullColor
      mirroring=NoMirror
      resourceName=../../../Engine/Content/Slate/Common/ScrollBoxShadowBottom.png
      resourceObject=null
      tiling=NoTile
      imageSize=<16 8>
      margin=<ltrb 0.5 0 0.5 1>
      outlineSettings {
       bUseBrushTransparency=false
       roundingType=HalfHeightRadius
       width=0
       color {colorUseRule=UseColor_Specified, specifiedColor=<rgba 0 0 0 0>}
       cornerRadii=<0 0 0 0>
      }
      tintColor {colorUseRule=UseColor_Specified, specifiedColor=<rgba 1 1 1 1>}
      uVRegion {bIsValid=false, max=<0 0>, min=<0 0>}
     }
     horizontalScrolledContentPadding=<ltrb 0 0 1 0>
     leftShadowBrush {
      bIsDynamicallyLoaded=false
      drawAs=Box
      imageType=FullColor
      mirroring=NoMirror
      resourceName=../../../Engine/Content/Slate/Common/ScrollBoxShadowLeft.png
      resourceObject=null
      tiling=NoTile
      imageSize=<8 16>
      margin=<ltrb 1 0.5 0 0.5>
      outlineSettings {
       bUseBrushTransparency=false
       roundingType=HalfHeightRadius
       width=0
       color {colorUseRule=UseColor_Specified, specifiedColor=<rgba 0 0 0 0>}
       cornerRadii=<0 0 0 0>
      }
      tintColor {colorUseRule=UseColor_Specified, specifiedColor=<rgba 1 1 1 1>}
      uVRegion {bIsValid=false, max=<0 0>, min=<0 0>}
     }
     rightShadowBrush {
      bIsDynamicallyLoaded=false
      drawAs=Box
      imageType=FullColor
      mirroring=NoMirror
      resourceName=../../../Engine/Content/Slate/Common/ScrollBoxShadowRight.png
      resourceObject=null
      tiling=NoTile
      imageSize=<8 16>
      margin=<ltrb 0 0.5 1 0.5>
      outlineSettings {
       bUseBrushTransparency=false
       roundingType=HalfHeightRadius
       width=0
       color {colorUseRule=UseColor_Specified, specifiedColor=<rgba 0 0 0 0>}
       cornerRadii=<0 0 0 0>
      }
      tintColor {colorUseRule=UseColor_Specified, specifiedColor=<rgba 1 1 1 1>}
      uVRegion {bIsValid=false, max=<0 0>, min=<0 0>}
     }
     topShadowBrush {
      bIsDynamicallyLoaded=false
      drawAs=Box
      imageType=FullColor
      mirroring=NoMirror
      resourceName=../../../Engine/Content/Slate/Common/ScrollBoxShadowTop.png
      resourceObject=null
      tiling=NoTile
      imageSize=<16 8>
      margin=<ltrb 0.5 1 0.5 0>
      outlineSettings {
       bUseBrushTransparency=false
       roundingType=HalfHeightRadius
       width=0
       color {colorUseRule=UseColor_Specified, specifiedColor=<rgba 0 0 0 0>}
       cornerRadii=<0 0 0 0>
      }
      tintColor {colorUseRule=UseColor_Specified, specifiedColor=<rgba 1 1 1 1>}
      uVRegion {bIsValid=false, max=<0 0>, min=<0 0>}
     }
     verticalScrolledContentPadding=<ltrb 0 0 0 1>
    }
   }
   slot {class=OverlaySlot, properties {horizontalAlignment=HAlign_Fill, verticalAlignment=VAlign_Fill}}
  }
  VerticalBox_0 {
   widgetClassPath=&VerticalBox
   parent=ScrollBox_176
   widget=VerticalBox_0
   slot {class=ScrollBoxSlot, properties {horizontalAlignment=HAlign_Center, verticalAlignment=VAlign_Top}}
  }
  HorizontalBox_57 {widgetClassPath=&HorizontalBox, parent=VerticalBox_0, widget=HorizontalBox_57, slot {class=VerticalBoxSlot}}
  HorizontalBox_235 {
   widgetClassPath=&HorizontalBox
   parent=HorizontalBox_57
   widget=HorizontalBox_235
   slot {class=HorizontalBoxSlot, properties {horizontalAlignment=HAlign_Center, padding=<ltrb 0 10 0 0>}}
  }
  CheckBox_gangAssets {
   widgetClassPath=/Script/UMG.CheckBox
   bIsVariable=true
   parent=HorizontalBox_235
   widget=CheckBox_gangAssets
   properties {
    checkedState=Checked
    widgetStyle {
     checkBoxType=CheckBox
     backgroundHoveredImage {
      bIsDynamicallyLoaded=false
      drawAs=RoundedBox
      imageType=NoImage
      mirroring=NoMirror
      resourceName=None
      resourceObject=null
      tiling=NoTile
      imageSize=<18 18>
      margin=<ltrb 0 0 0 0>
      outlineSettings {
       bUseBrushTransparency=false
       roundingType=FixedRadius
       width=1
       color {colorUseRule=UseColor_ColorTable, specifiedColor=<rgba 0 0 0 0>}
       cornerRadii=<3 3 3 3>
      }
      tintColor {colorUseRule=UseColor_Specified, specifiedColor=<rgba 0.0048 0.0048 0.0048 1>}
      uVRegion {bIsValid=false, max=<0 0>, min=<0 0>}
     }
     backgroundImage {
      bIsDynamicallyLoaded=false
      drawAs=RoundedBox
      imageType=NoImage
      mirroring=NoMirror
      resourceName=None
      resourceObject=null
      tiling=NoTile
      imageSize=<18 18>
      margin=<ltrb 0 0 0 0>
      outlineSettings {
       bUseBrushTransparency=false
       roundingType=FixedRadius
       width=1
       color {colorUseRule=UseColor_ColorTable, specifiedColor=<rgba 0 0 0 0>}
       cornerRadii=<3 3 3 3>
      }
      tintColor {colorUseRule=UseColor_Specified, specifiedColor=<rgba 0.0048 0.0048 0.0048 1>}
      uVRegion {bIsValid=false, max=<0 0>, min=<0 0>}
     }
     backgroundPressedImage {
      bIsDynamicallyLoaded=false
      drawAs=RoundedBox
      imageType=NoImage
      mirroring=NoMirror
      resourceName=None
      resourceObject=null
      tiling=NoTile
      imageSize=<18 18>
      margin=<ltrb 0 0 0 0>
      outlineSettings {
       bUseBrushTransparency=false
       roundingType=FixedRadius
       width=1
       color {colorUseRule=UseColor_ColorTable, specifiedColor=<rgba 0 0 0 0>}
       cornerRadii=<3 3 3 3>
      }
      tintColor {colorUseRule=UseColor_Specified, specifiedColor=<rgba 0.0048 0.0048 0.0048 1>}
      uVRegion {bIsValid=false, max=<0 0>, min=<0 0>}
     }
     borderBackgroundColor {colorUseRule=UseColor_Specified, specifiedColor=<rgba 1 1 1 1>}
     checkedForeground {colorUseRule=UseColor_Specified, specifiedColor=<rgba 1 1 1 1>}
     checkedHoveredForeground {colorUseRule=UseColor_Specified, specifiedColor=<rgba 1 1 1 1>}
     checkedHoveredImage {
      bIsDynamicallyLoaded=false
      drawAs=Image
      imageType=Vector
      mirroring=NoMirror
      resourceName=../../../Engine/Content/Slate/Starship/CoreWidgets/CheckBox/check.svg
      resourceObject=null
      tiling=NoTile
      imageSize=<16 16>
      margin=<ltrb 0 0 0 0>
      outlineSettings {
       bUseBrushTransparency=false
       roundingType=HalfHeightRadius
       width=0
       color {colorUseRule=UseColor_Specified, specifiedColor=<rgba 0 0 0 0>}
       cornerRadii=<0 0 0 0>
      }
      tintColor {colorUseRule=UseColor_Specified, specifiedColor=<rgba 0.0044 0.2384 1 1>}
      uVRegion {bIsValid=false, max=<0 0>, min=<0 0>}
     }
     checkedImage {
      bIsDynamicallyLoaded=false
      drawAs=Image
      imageType=Vector
      mirroring=NoMirror
      resourceName=../../../Engine/Content/Slate/Starship/CoreWidgets/CheckBox/check.svg
      resourceObject=null
      tiling=NoTile
      imageSize=<16 16>
      margin=<ltrb 0 0 0 0>
      outlineSettings {
       bUseBrushTransparency=false
       roundingType=HalfHeightRadius
       width=0
       color {colorUseRule=UseColor_Specified, specifiedColor=<rgba 0 0 0 0>}
       cornerRadii=<0 0 0 0>
      }
      tintColor {colorUseRule=UseColor_Specified, specifiedColor=<rgba 0 0.162 0.7454 1>}
      uVRegion {bIsValid=false, max=<0 0>, min=<0 0>}
     }
     checkedPressedForeground {colorUseRule=UseColor_Specified, specifiedColor=<rgba 1 1 1 1>}
     checkedPressedImage {
      bIsDynamicallyLoaded=false
      drawAs=Image
      imageType=Vector
      mirroring=NoMirror
      resourceName=../../../Engine/Content/Slate/Starship/CoreWidgets/CheckBox/check.svg
      resourceObject=null
      tiling=NoTile
      imageSize=<16 16>
      margin=<ltrb 0 0 0 0>
      outlineSettings {
       bUseBrushTransparency=false
       roundingType=HalfHeightRadius
       width=0
       color {colorUseRule=UseColor_Specified, specifiedColor=<rgba 0 0 0 0>}
       cornerRadii=<0 0 0 0>
      }
      tintColor {colorUseRule=UseColor_Specified, specifiedColor=<rgba 0 0.0802 0.3515 1>}
      uVRegion {bIsValid=false, max=<0 0>, min=<0 0>}
     }
     checkedSlateSound {resourceObject=null}
     foregroundColor {colorUseRule=UseColor_Specified, specifiedColor=<rgba 1 1 1 1>}
     hoveredForeground {colorUseRule=UseColor_Specified, specifiedColor=<rgba 1 1 1 1>}
     hoveredSlateSound {resourceObject=null}
     padding=<ltrb 2 0 0 0>
     pressedForeground {colorUseRule=UseColor_Specified, specifiedColor=<rgba 1 1 1 1>}
     uncheckedHoveredImage {
      bIsDynamicallyLoaded=false
      drawAs=NoDrawType
      imageType=NoImage
      mirroring=NoMirror
      resourceName=None
      resourceObject=null
      tiling=NoTile
      imageSize=<0 0>
      margin=<ltrb 0 0 0 0>
      outlineSettings {
       bUseBrushTransparency=false
       roundingType=HalfHeightRadius
       width=0
       color {colorUseRule=UseColor_Specified, specifiedColor=<rgba 0 0 0 0>}
       cornerRadii=<0 0 0 0>
      }
      tintColor {colorUseRule=UseColor_Specified, specifiedColor=<rgba 1 1 1 1>}
      uVRegion {bIsValid=false, max=<0 0>, min=<0 0>}
     }
     uncheckedImage {
      bIsDynamicallyLoaded=false
      drawAs=NoDrawType
      imageType=NoImage
      mirroring=NoMirror
      resourceName=None
      resourceObject=null
      tiling=NoTile
      imageSize=<0 0>
      margin=<ltrb 0 0 0 0>
      outlineSettings {
       bUseBrushTransparency=false
       roundingType=HalfHeightRadius
       width=0
       color {colorUseRule=UseColor_Specified, specifiedColor=<rgba 0 0 0 0>}
       cornerRadii=<0 0 0 0>
      }
      tintColor {colorUseRule=UseColor_Specified, specifiedColor=<rgba 1 1 1 1>}
      uVRegion {bIsValid=false, max=<0 0>, min=<0 0>}
     }
     uncheckedPressedImage {
      bIsDynamicallyLoaded=false
      drawAs=NoDrawType
      imageType=NoImage
      mirroring=NoMirror
      resourceName=None
      resourceObject=null
      tiling=NoTile
      imageSize=<0 0>
      margin=<ltrb 0 0 0 0>
      outlineSettings {
       bUseBrushTransparency=false
       roundingType=HalfHeightRadius
       width=0
       color {colorUseRule=UseColor_Specified, specifiedColor=<rgba 0 0 0 0>}
       cornerRadii=<0 0 0 0>
      }
      tintColor {colorUseRule=UseColor_Specified, specifiedColor=<rgba 1 1 1 1>}
      uVRegion {bIsValid=false, max=<0 0>, min=<0 0>}
     }
     uncheckedSlateSound {resourceObject=null}
     undeterminedForeground {colorUseRule=UseColor_Specified, specifiedColor=<rgba 1 1 1 1>}
     undeterminedHoveredImage {
      bIsDynamicallyLoaded=false
      drawAs=Image
      imageType=Vector
      mirroring=NoMirror
      resourceName=../../../Engine/Content/Slate/Starship/CoreWidgets/CheckBox/indeterminate.svg
      resourceObject=null
      tiling=NoTile
      imageSize=<16 16>
      margin=<ltrb 0 0 0 0>
      outlineSettings {
       bUseBrushTransparency=false
       roundingType=HalfHeightRadius
       width=0
       color {colorUseRule=UseColor_Specified, specifiedColor=<rgba 0 0 0 0>}
       cornerRadii=<0 0 0 0>
      }
      tintColor {colorUseRule=UseColor_Specified, specifiedColor=<rgba 0.0044 0.2384 1 1>}
      uVRegion {bIsValid=false, max=<0 0>, min=<0 0>}
     }
     undeterminedImage {
      bIsDynamicallyLoaded=false
      drawAs=Image
      imageType=Vector
      mirroring=NoMirror
      resourceName=../../../Engine/Content/Slate/Starship/CoreWidgets/CheckBox/indeterminate.svg
      resourceObject=null
      tiling=NoTile
      imageSize=<16 16>
      margin=<ltrb 0 0 0 0>
      outlineSettings {
       bUseBrushTransparency=false
       roundingType=HalfHeightRadius
       width=0
       color {colorUseRule=UseColor_Specified, specifiedColor=<rgba 0 0 0 0>}
       cornerRadii=<0 0 0 0>
      }
      tintColor {colorUseRule=UseColor_Specified, specifiedColor=<rgba 0 0.162 0.7454 1>}
      uVRegion {bIsValid=false, max=<0 0>, min=<0 0>}
     }
     undeterminedPressedImage {
      bIsDynamicallyLoaded=false
      drawAs=Image
      imageType=Vector
      mirroring=NoMirror
      resourceName=../../../Engine/Content/Slate/Starship/CoreWidgets/CheckBox/indeterminate.svg
      resourceObject=null
      tiling=NoTile
      imageSize=<16 16>
      margin=<ltrb 0 0 0 0>
      outlineSettings {
       bUseBrushTransparency=false
       roundingType=HalfHeightRadius
       width=0
       color {colorUseRule=UseColor_Specified, specifiedColor=<rgba 0 0 0 0>}
       cornerRadii=<0 0 0 0>
      }
      tintColor {colorUseRule=UseColor_Specified, specifiedColor=<rgba 0 0.0802 0.3515 1>}
      uVRegion {bIsValid=false, max=<0 0>, min=<0 0>}
     }
    }
   }
   slot {class=HorizontalBoxSlot}
  }
  CommonTextBlock_187 {
   widgetClassPath=&CommonTextBlock
   parent=HorizontalBox_235
   widget=CommonTextBlock_187
   properties {
    text=Gang Assets
    toolTipText="If True: Changing the character will also attempt to change the other cosmetics to match.  \r\n\r\nDisable to set each individually."
    font {
     bForceMonospaced=false
     bMaterialIsStencil=false
     fontMaterial=null
     fontObject=&BurbankSmall
     letterSpacing=0
     monospacedWidth=1
     size=10
     skewAmount=0
     typefaceFontName=Medium
     outlineSettings {
      bApplyOutlineToDropShadows=false
      bMiteredCorners=false
      bSeparateFillAlpha=false
      outlineMaterial=null
      outlineSize=0
      outlineColor=<rgba 0 0 0 1>
     }
    }
    shadowOffset=<0 0>
   }
   slot {class=HorizontalBoxSlot}
  }
  Spacer_149 {widgetClassPath=&Spacer, parent=HorizontalBox_57, widget=Spacer_149, properties {size=<20 1>}, slot {class=HorizontalBoxSlot}}
  HorizontalBox {
   widgetClassPath=&HorizontalBox
   parent=HorizontalBox_57
   widget=HorizontalBox
   slot {class=HorizontalBoxSlot, properties {horizontalAlignment=HAlign_Center, padding=<ltrb 0 10 0 0>}}
  }
  CheckBox_autoLoad {
   widgetClassPath=&CheckBox
   bIsVariable=true
   parent=HorizontalBox
   widget=CheckBox_autoLoad
   properties {widgetStyle {
     checkBoxType=CheckBox
     backgroundHoveredImage {
      bIsDynamicallyLoaded=false
      drawAs=RoundedBox
      imageType=NoImage
      mirroring=NoMirror
      resourceName=None
      resourceObject=null
      tiling=NoTile
      imageSize=<18 18>
      margin=<ltrb 0 0 0 0>
      outlineSettings {
       bUseBrushTransparency=false
       roundingType=FixedRadius
       width=1
       color {colorUseRule=UseColor_ColorTable, specifiedColor=<rgba 0 0 0 0>}
       cornerRadii=<3 3 3 3>
      }
      tintColor {colorUseRule=UseColor_Specified, specifiedColor=<rgba 0.0048 0.0048 0.0048 1>}
      uVRegion {bIsValid=false, max=<0 0>, min=<0 0>}
     }
     backgroundImage {
      bIsDynamicallyLoaded=false
      drawAs=RoundedBox
      imageType=NoImage
      mirroring=NoMirror
      resourceName=None
      resourceObject=null
      tiling=NoTile
      imageSize=<18 18>
      margin=<ltrb 0 0 0 0>
      outlineSettings {
       bUseBrushTransparency=false
       roundingType=FixedRadius
       width=1
       color {colorUseRule=UseColor_ColorTable, specifiedColor=<rgba 0 0 0 0>}
       cornerRadii=<3 3 3 3>
      }
      tintColor {colorUseRule=UseColor_Specified, specifiedColor=<rgba 0.0048 0.0048 0.0048 1>}
      uVRegion {bIsValid=false, max=<0 0>, min=<0 0>}
     }
     backgroundPressedImage {
      bIsDynamicallyLoaded=false
      drawAs=RoundedBox
      imageType=NoImage
      mirroring=NoMirror
      resourceName=None
      resourceObject=null
      tiling=NoTile
      imageSize=<18 18>
      margin=<ltrb 0 0 0 0>
      outlineSettings {
       bUseBrushTransparency=false
       roundingType=FixedRadius
       width=1
       color {colorUseRule=UseColor_ColorTable, specifiedColor=<rgba 0 0 0 0>}
       cornerRadii=<3 3 3 3>
      }
      tintColor {colorUseRule=UseColor_Specified, specifiedColor=<rgba 0.0048 0.0048 0.0048 1>}
      uVRegion {bIsValid=false, max=<0 0>, min=<0 0>}
     }
     borderBackgroundColor {colorUseRule=UseColor_Specified, specifiedColor=<rgba 1 1 1 1>}
     checkedForeground {colorUseRule=UseColor_Specified, specifiedColor=<rgba 1 1 1 1>}
     checkedHoveredForeground {colorUseRule=UseColor_Specified, specifiedColor=<rgba 1 1 1 1>}
     checkedHoveredImage {
      bIsDynamicallyLoaded=false
      drawAs=Image
      imageType=Vector
      mirroring=NoMirror
      resourceName=../../../Engine/Content/Slate/Starship/CoreWidgets/CheckBox/check.svg
      resourceObject=null
      tiling=NoTile
      imageSize=<16 16>
      margin=<ltrb 0 0 0 0>
      outlineSettings {
       bUseBrushTransparency=false
       roundingType=HalfHeightRadius
       width=0
       color {colorUseRule=UseColor_Specified, specifiedColor=<rgba 0 0 0 0>}
       cornerRadii=<0 0 0 0>
      }
      tintColor {colorUseRule=UseColor_Specified, specifiedColor=<rgba 0.0044 0.2384 1 1>}
      uVRegion {bIsValid=false, max=<0 0>, min=<0 0>}
     }
     checkedImage {
      bIsDynamicallyLoaded=false
      drawAs=Image
      imageType=Vector
      mirroring=NoMirror
      resourceName=../../../Engine/Content/Slate/Starship/CoreWidgets/CheckBox/check.svg
      resourceObject=null
      tiling=NoTile
      imageSize=<16 16>
      margin=<ltrb 0 0 0 0>
      outlineSettings {
       bUseBrushTransparency=false
       roundingType=HalfHeightRadius
       width=0
       color {colorUseRule=UseColor_Specified, specifiedColor=<rgba 0 0 0 0>}
       cornerRadii=<0 0 0 0>
      }
      tintColor {colorUseRule=UseColor_Specified, specifiedColor=<rgba 0 0.162 0.7454 1>}
      uVRegion {bIsValid=false, max=<0 0>, min=<0 0>}
     }
     checkedPressedForeground {colorUseRule=UseColor_Specified, specifiedColor=<rgba 1 1 1 1>}
     checkedPressedImage {
      bIsDynamicallyLoaded=false
      drawAs=Image
      imageType=Vector
      mirroring=NoMirror
      resourceName=../../../Engine/Content/Slate/Starship/CoreWidgets/CheckBox/check.svg
      resourceObject=null
      tiling=NoTile
      imageSize=<16 16>
      margin=<ltrb 0 0 0 0>
      outlineSettings {
       bUseBrushTransparency=false
       roundingType=HalfHeightRadius
       width=0
       color {colorUseRule=UseColor_Specified, specifiedColor=<rgba 0 0 0 0>}
       cornerRadii=<0 0 0 0>
      }
      tintColor {colorUseRule=UseColor_Specified, specifiedColor=<rgba 0 0.0802 0.3515 1>}
      uVRegion {bIsValid=false, max=<0 0>, min=<0 0>}
     }
     checkedSlateSound {resourceObject=null}
     foregroundColor {colorUseRule=UseColor_Specified, specifiedColor=<rgba 1 1 1 1>}
     hoveredForeground {colorUseRule=UseColor_Specified, specifiedColor=<rgba 1 1 1 1>}
     hoveredSlateSound {resourceObject=null}
     padding=<ltrb 2 0 0 0>
     pressedForeground {colorUseRule=UseColor_Specified, specifiedColor=<rgba 1 1 1 1>}
     uncheckedHoveredImage {
      bIsDynamicallyLoaded=false
      drawAs=NoDrawType
      imageType=NoImage
      mirroring=NoMirror
      resourceName=None
      resourceObject=null
      tiling=NoTile
      imageSize=<0 0>
      margin=<ltrb 0 0 0 0>
      outlineSettings {
       bUseBrushTransparency=false
       roundingType=HalfHeightRadius
       width=0
       color {colorUseRule=UseColor_Specified, specifiedColor=<rgba 0 0 0 0>}
       cornerRadii=<0 0 0 0>
      }
      tintColor {colorUseRule=UseColor_Specified, specifiedColor=<rgba 1 1 1 1>}
      uVRegion {bIsValid=false, max=<0 0>, min=<0 0>}
     }
     uncheckedImage {
      bIsDynamicallyLoaded=false
      drawAs=NoDrawType
      imageType=NoImage
      mirroring=NoMirror
      resourceName=None
      resourceObject=null
      tiling=NoTile
      imageSize=<0 0>
      margin=<ltrb 0 0 0 0>
      outlineSettings {
       bUseBrushTransparency=false
       roundingType=HalfHeightRadius
       width=0
       color {colorUseRule=UseColor_Specified, specifiedColor=<rgba 0 0 0 0>}
       cornerRadii=<0 0 0 0>
      }
      tintColor {colorUseRule=UseColor_Specified, specifiedColor=<rgba 1 1 1 1>}
      uVRegion {bIsValid=false, max=<0 0>, min=<0 0>}
     }
     uncheckedPressedImage {
      bIsDynamicallyLoaded=false
      drawAs=NoDrawType
      imageType=NoImage
      mirroring=NoMirror
      resourceName=None
      resourceObject=null
      tiling=NoTile
      imageSize=<0 0>
      margin=<ltrb 0 0 0 0>
      outlineSettings {
       bUseBrushTransparency=false
       roundingType=HalfHeightRadius
       width=0
       color {colorUseRule=UseColor_Specified, specifiedColor=<rgba 0 0 0 0>}
       cornerRadii=<0 0 0 0>
      }
      tintColor {colorUseRule=UseColor_Specified, specifiedColor=<rgba 1 1 1 1>}
      uVRegion {bIsValid=false, max=<0 0>, min=<0 0>}
     }
     uncheckedSlateSound {resourceObject=null}
     undeterminedForeground {colorUseRule=UseColor_Specified, specifiedColor=<rgba 1 1 1 1>}
     undeterminedHoveredImage {
      bIsDynamicallyLoaded=false
      drawAs=Image
      imageType=Vector
      mirroring=NoMirror
      resourceName=../../../Engine/Content/Slate/Starship/CoreWidgets/CheckBox/indeterminate.svg
      resourceObject=null
      tiling=NoTile
      imageSize=<16 16>
      margin=<ltrb 0 0 0 0>
      outlineSettings {
       bUseBrushTransparency=false
       roundingType=HalfHeightRadius
       width=0
       color {colorUseRule=UseColor_Specified, specifiedColor=<rgba 0 0 0 0>}
       cornerRadii=<0 0 0 0>
      }
      tintColor {colorUseRule=UseColor_Specified, specifiedColor=<rgba 0.0044 0.2384 1 1>}
      uVRegion {bIsValid=false, max=<0 0>, min=<0 0>}
     }
     undeterminedImage {
      bIsDynamicallyLoaded=false
      drawAs=Image
      imageType=Vector
      mirroring=NoMirror
      resourceName=../../../Engine/Content/Slate/Starship/CoreWidgets/CheckBox/indeterminate.svg
      resourceObject=null
      tiling=NoTile
      imageSize=<16 16>
      margin=<ltrb 0 0 0 0>
      outlineSettings {
       bUseBrushTransparency=false
       roundingType=HalfHeightRadius
       width=0
       color {colorUseRule=UseColor_Specified, specifiedColor=<rgba 0 0 0 0>}
       cornerRadii=<0 0 0 0>
      }
      tintColor {colorUseRule=UseColor_Specified, specifiedColor=<rgba 0 0.162 0.7454 1>}
      uVRegion {bIsValid=false, max=<0 0>, min=<0 0>}
     }
     undeterminedPressedImage {
      bIsDynamicallyLoaded=false
      drawAs=Image
      imageType=Vector
      mirroring=NoMirror
      resourceName=../../../Engine/Content/Slate/Starship/CoreWidgets/CheckBox/indeterminate.svg
      resourceObject=null
      tiling=NoTile
      imageSize=<16 16>
      margin=<ltrb 0 0 0 0>
      outlineSettings {
       bUseBrushTransparency=false
       roundingType=HalfHeightRadius
       width=0
       color {colorUseRule=UseColor_Specified, specifiedColor=<rgba 0 0 0 0>}
       cornerRadii=<0 0 0 0>
      }
      tintColor {colorUseRule=UseColor_Specified, specifiedColor=<rgba 0 0.0802 0.3515 1>}
      uVRegion {bIsValid=false, max=<0 0>, min=<0 0>}
     }
    }}
   slot {class=HorizontalBoxSlot}
  }
  CommonTextBlock {
   widgetClassPath=&CommonTextBlock
   parent=HorizontalBox
   widget=CommonTextBlock
   properties {
    text=AutoLoad Selected Mannequins
    font {
     bForceMonospaced=false
     bMaterialIsStencil=false
     fontMaterial=null
     fontObject=&BurbankSmall
     letterSpacing=0
     monospacedWidth=1
     size=10
     skewAmount=0
     typefaceFontName=Medium
     outlineSettings {
      bApplyOutlineToDropShadows=false
      bMiteredCorners=false
      bSeparateFillAlpha=false
      outlineMaterial=null
      outlineSize=0
      outlineColor=<rgba 0 0 0 1>
     }
    }
    shadowOffset=<0 0>
   }
   slot {class=HorizontalBoxSlot}
  }
  Spacer_2 {widgetClassPath=&Spacer, parent=HorizontalBox, widget=Spacer_2, properties {size=<20 1>}, slot {class=HorizontalBoxSlot}}
  HorizontalBox_126 {widgetClassPath=&HorizontalBox, parent=VerticalBox_0, widget=HorizontalBox_126, slot {class=VerticalBoxSlot}}
  HorizontalBox_1 {
   widgetClassPath=&HorizontalBox
   parent=HorizontalBox_126
   widget=HorizontalBox_1
   slot {class=HorizontalBoxSlot, properties {horizontalAlignment=HAlign_Center, padding=<ltrb 0 10 0 0>}}
  }
  CheckBox_MannSelected {
   widgetClassPath=&CheckBox
   bIsVariable=true
   parent=HorizontalBox_1
   widget=CheckBox_MannSelected
   properties {widgetStyle {
     checkBoxType=CheckBox
     backgroundHoveredImage {
      bIsDynamicallyLoaded=false
      drawAs=RoundedBox
      imageType=NoImage
      mirroring=NoMirror
      resourceName=None
      resourceObject=null
      tiling=NoTile
      imageSize=<18 18>
      margin=<ltrb 0 0 0 0>
      outlineSettings {
       bUseBrushTransparency=false
       roundingType=FixedRadius
       width=1
       color {colorUseRule=UseColor_ColorTable, specifiedColor=<rgba 0 0 0 0>}
       cornerRadii=<3 3 3 3>
      }
      tintColor {colorUseRule=UseColor_Specified, specifiedColor=<rgba 0.0048 0.0048 0.0048 1>}
      uVRegion {bIsValid=false, max=<0 0>, min=<0 0>}
     }
     backgroundImage {
      bIsDynamicallyLoaded=false
      drawAs=RoundedBox
      imageType=NoImage
      mirroring=NoMirror
      resourceName=None
      resourceObject=null
      tiling=NoTile
      imageSize=<18 18>
      margin=<ltrb 0 0 0 0>
      outlineSettings {
       bUseBrushTransparency=false
       roundingType=FixedRadius
       width=1
       color {colorUseRule=UseColor_ColorTable, specifiedColor=<rgba 0 0 0 0>}
       cornerRadii=<3 3 3 3>
      }
      tintColor {colorUseRule=UseColor_Specified, specifiedColor=<rgba 0.0048 0.0048 0.0048 1>}
      uVRegion {bIsValid=false, max=<0 0>, min=<0 0>}
     }
     backgroundPressedImage {
      bIsDynamicallyLoaded=false
      drawAs=RoundedBox
      imageType=NoImage
      mirroring=NoMirror
      resourceName=None
      resourceObject=null
      tiling=NoTile
      imageSize=<18 18>
      margin=<ltrb 0 0 0 0>
      outlineSettings {
       bUseBrushTransparency=false
       roundingType=FixedRadius
       width=1
       color {colorUseRule=UseColor_ColorTable, specifiedColor=<rgba 0 0 0 0>}
       cornerRadii=<3 3 3 3>
      }
      tintColor {colorUseRule=UseColor_Specified, specifiedColor=<rgba 0.0048 0.0048 0.0048 1>}
      uVRegion {bIsValid=false, max=<0 0>, min=<0 0>}
     }
     borderBackgroundColor {colorUseRule=UseColor_Specified, specifiedColor=<rgba 1 1 1 1>}
     checkedForeground {colorUseRule=UseColor_Specified, specifiedColor=<rgba 1 1 1 1>}
     checkedHoveredForeground {colorUseRule=UseColor_Specified, specifiedColor=<rgba 1 1 1 1>}
     checkedHoveredImage {
      bIsDynamicallyLoaded=false
      drawAs=Image
      imageType=Vector
      mirroring=NoMirror
      resourceName=../../../Engine/Content/Slate/Starship/CoreWidgets/CheckBox/check.svg
      resourceObject=null
      tiling=NoTile
      imageSize=<16 16>
      margin=<ltrb 0 0 0 0>
      outlineSettings {
       bUseBrushTransparency=false
       roundingType=HalfHeightRadius
       width=0
       color {colorUseRule=UseColor_Specified, specifiedColor=<rgba 0 0 0 0>}
       cornerRadii=<0 0 0 0>
      }
      tintColor {colorUseRule=UseColor_Specified, specifiedColor=<rgba 0.0044 0.2384 1 1>}
      uVRegion {bIsValid=false, max=<0 0>, min=<0 0>}
     }
     checkedImage {
      bIsDynamicallyLoaded=false
      drawAs=Image
      imageType=Vector
      mirroring=NoMirror
      resourceName=../../../Engine/Content/Slate/Starship/CoreWidgets/CheckBox/check.svg
      resourceObject=null
      tiling=NoTile
      imageSize=<16 16>
      margin=<ltrb 0 0 0 0>
      outlineSettings {
       bUseBrushTransparency=false
       roundingType=HalfHeightRadius
       width=0
       color {colorUseRule=UseColor_Specified, specifiedColor=<rgba 0 0 0 0>}
       cornerRadii=<0 0 0 0>
      }
      tintColor {colorUseRule=UseColor_Specified, specifiedColor=<rgba 0 0.162 0.7454 1>}
      uVRegion {bIsValid=false, max=<0 0>, min=<0 0>}
     }
     checkedPressedForeground {colorUseRule=UseColor_Specified, specifiedColor=<rgba 1 1 1 1>}
     checkedPressedImage {
      bIsDynamicallyLoaded=false
      drawAs=Image
      imageType=Vector
      mirroring=NoMirror
      resourceName=../../../Engine/Content/Slate/Starship/CoreWidgets/CheckBox/check.svg
      resourceObject=null
      tiling=NoTile
      imageSize=<16 16>
      margin=<ltrb 0 0 0 0>
      outlineSettings {
       bUseBrushTransparency=false
       roundingType=HalfHeightRadius
       width=0
       color {colorUseRule=UseColor_Specified, specifiedColor=<rgba 0 0 0 0>}
       cornerRadii=<0 0 0 0>
      }
      tintColor {colorUseRule=UseColor_Specified, specifiedColor=<rgba 0 0.0802 0.3515 1>}
      uVRegion {bIsValid=false, max=<0 0>, min=<0 0>}
     }
     checkedSlateSound {resourceObject=null}
     foregroundColor {colorUseRule=UseColor_Specified, specifiedColor=<rgba 1 1 1 1>}
     hoveredForeground {colorUseRule=UseColor_Specified, specifiedColor=<rgba 1 1 1 1>}
     hoveredSlateSound {resourceObject=null}
     padding=<ltrb 2 0 0 0>
     pressedForeground {colorUseRule=UseColor_Specified, specifiedColor=<rgba 1 1 1 1>}
     uncheckedHoveredImage {
      bIsDynamicallyLoaded=false
      drawAs=NoDrawType
      imageType=NoImage
      mirroring=NoMirror
      resourceName=None
      resourceObject=null
      tiling=NoTile
      imageSize=<0 0>
      margin=<ltrb 0 0 0 0>
      outlineSettings {
       bUseBrushTransparency=false
       roundingType=HalfHeightRadius
       width=0
       color {colorUseRule=UseColor_Specified, specifiedColor=<rgba 0 0 0 0>}
       cornerRadii=<0 0 0 0>
      }
      tintColor {colorUseRule=UseColor_Specified, specifiedColor=<rgba 1 1 1 1>}
      uVRegion {bIsValid=false, max=<0 0>, min=<0 0>}
     }
     uncheckedImage {
      bIsDynamicallyLoaded=false
      drawAs=NoDrawType
      imageType=NoImage
      mirroring=NoMirror
      resourceName=None
      resourceObject=null
      tiling=NoTile
      imageSize=<0 0>
      margin=<ltrb 0 0 0 0>
      outlineSettings {
       bUseBrushTransparency=false
       roundingType=HalfHeightRadius
       width=0
       color {colorUseRule=UseColor_Specified, specifiedColor=<rgba 0 0 0 0>}
       cornerRadii=<0 0 0 0>
      }
      tintColor {colorUseRule=UseColor_Specified, specifiedColor=<rgba 1 1 1 1>}
      uVRegion {bIsValid=false, max=<0 0>, min=<0 0>}
     }
     uncheckedPressedImage {
      bIsDynamicallyLoaded=false
      drawAs=NoDrawType
      imageType=NoImage
      mirroring=NoMirror
      resourceName=None
      resourceObject=null
      tiling=NoTile
      imageSize=<0 0>
      margin=<ltrb 0 0 0 0>
      outlineSettings {
       bUseBrushTransparency=false
       roundingType=HalfHeightRadius
       width=0
       color {colorUseRule=UseColor_Specified, specifiedColor=<rgba 0 0 0 0>}
       cornerRadii=<0 0 0 0>
      }
      tintColor {colorUseRule=UseColor_Specified, specifiedColor=<rgba 1 1 1 1>}
      uVRegion {bIsValid=false, max=<0 0>, min=<0 0>}
     }
     uncheckedSlateSound {resourceObject=null}
     undeterminedForeground {colorUseRule=UseColor_Specified, specifiedColor=<rgba 1 1 1 1>}
     undeterminedHoveredImage {
      bIsDynamicallyLoaded=false
      drawAs=Image
      imageType=Vector
      mirroring=NoMirror
      resourceName=../../../Engine/Content/Slate/Starship/CoreWidgets/CheckBox/indeterminate.svg
      resourceObject=null
      tiling=NoTile
      imageSize=<16 16>
      margin=<ltrb 0 0 0 0>
      outlineSettings {
       bUseBrushTransparency=false
       roundingType=HalfHeightRadius
       width=0
       color {colorUseRule=UseColor_Specified, specifiedColor=<rgba 0 0 0 0>}
       cornerRadii=<0 0 0 0>
      }
      tintColor {colorUseRule=UseColor_Specified, specifiedColor=<rgba 0.0044 0.2384 1 1>}
      uVRegion {bIsValid=false, max=<0 0>, min=<0 0>}
     }
     undeterminedImage {
      bIsDynamicallyLoaded=false
      drawAs=Image
      imageType=Vector
      mirroring=NoMirror
      resourceName=../../../Engine/Content/Slate/Starship/CoreWidgets/CheckBox/indeterminate.svg
      resourceObject=null
      tiling=NoTile
      imageSize=<16 16>
      margin=<ltrb 0 0 0 0>
      outlineSettings {
       bUseBrushTransparency=false
       roundingType=HalfHeightRadius
       width=0
       color {colorUseRule=UseColor_Specified, specifiedColor=<rgba 0 0 0 0>}
       cornerRadii=<0 0 0 0>
      }
      tintColor {colorUseRule=UseColor_Specified, specifiedColor=<rgba 0 0.162 0.7454 1>}
      uVRegion {bIsValid=false, max=<0 0>, min=<0 0>}
     }
     undeterminedPressedImage {
      bIsDynamicallyLoaded=false
      drawAs=Image
      imageType=Vector
      mirroring=NoMirror
      resourceName=../../../Engine/Content/Slate/Starship/CoreWidgets/CheckBox/indeterminate.svg
      resourceObject=null
      tiling=NoTile
      imageSize=<16 16>
      margin=<ltrb 0 0 0 0>
      outlineSettings {
       bUseBrushTransparency=false
       roundingType=HalfHeightRadius
       width=0
       color {colorUseRule=UseColor_Specified, specifiedColor=<rgba 0 0 0 0>}
       cornerRadii=<0 0 0 0>
      }
      tintColor {colorUseRule=UseColor_Specified, specifiedColor=<rgba 0 0.0802 0.3515 1>}
      uVRegion {bIsValid=false, max=<0 0>, min=<0 0>}
     }
    }}
   slot {class=HorizontalBoxSlot}
  }
  CommonTextBlock_1 {
   widgetClassPath=&CommonTextBlock
   parent=HorizontalBox_1
   widget=CommonTextBlock_1
   properties {
    text=Mannequin Multi-Select
    font {
     bForceMonospaced=false
     bMaterialIsStencil=false
     fontMaterial=null
     fontObject=&BurbankSmall
     letterSpacing=0
     monospacedWidth=1
     size=10
     skewAmount=0
     typefaceFontName=Medium
     outlineSettings {
      bApplyOutlineToDropShadows=false
      bMiteredCorners=false
      bSeparateFillAlpha=false
      outlineMaterial=null
      outlineSize=0
      outlineColor=<rgba 0 0 0 1>
     }
    }
    shadowOffset=<0 0>
   }
   slot {class=HorizontalBoxSlot}
  }
  Spacer_3 {widgetClassPath=&Spacer, parent=HorizontalBox_126, widget=Spacer_3, properties {size=<20 1>}, slot {class=HorizontalBoxSlot}}
  HorizontalBox_2 {
   widgetClassPath=&HorizontalBox
   parent=HorizontalBox_126
   widget=HorizontalBox_2
   slot {class=HorizontalBoxSlot, properties {horizontalAlignment=HAlign_Center, padding=<ltrb 0 10 0 0>}}
  }
  Disable_SK_LOD {
   widgetClassPath=&CheckBox
   bIsVariable=true
   parent=HorizontalBox_2
   widget=Disable_SK_LOD
   properties {widgetStyle {
     checkBoxType=CheckBox
     backgroundHoveredImage {
      bIsDynamicallyLoaded=false
      drawAs=RoundedBox
      imageType=NoImage
      mirroring=NoMirror
      resourceName=None
      resourceObject=null
      tiling=NoTile
      imageSize=<18 18>
      margin=<ltrb 0 0 0 0>
      outlineSettings {
       bUseBrushTransparency=false
       roundingType=FixedRadius
       width=1
       color {colorUseRule=UseColor_ColorTable, specifiedColor=<rgba 0 0 0 0>}
       cornerRadii=<3 3 3 3>
      }
      tintColor {colorUseRule=UseColor_Specified, specifiedColor=<rgba 0.0048 0.0048 0.0048 1>}
      uVRegion {bIsValid=false, max=<0 0>, min=<0 0>}
     }
     backgroundImage {
      bIsDynamicallyLoaded=false
      drawAs=RoundedBox
      imageType=NoImage
      mirroring=NoMirror
      resourceName=None
      resourceObject=null
      tiling=NoTile
      imageSize=<18 18>
      margin=<ltrb 0 0 0 0>
      outlineSettings {
       bUseBrushTransparency=false
       roundingType=FixedRadius
       width=1
       color {colorUseRule=UseColor_ColorTable, specifiedColor=<rgba 0 0 0 0>}
       cornerRadii=<3 3 3 3>
      }
      tintColor {colorUseRule=UseColor_Specified, specifiedColor=<rgba 0.0048 0.0048 0.0048 1>}
      uVRegion {bIsValid=false, max=<0 0>, min=<0 0>}
     }
     backgroundPressedImage {
      bIsDynamicallyLoaded=false
      drawAs=RoundedBox
      imageType=NoImage
      mirroring=NoMirror
      resourceName=None
      resourceObject=null
      tiling=NoTile
      imageSize=<18 18>
      margin=<ltrb 0 0 0 0>
      outlineSettings {
       bUseBrushTransparency=false
       roundingType=FixedRadius
       width=1
       color {colorUseRule=UseColor_ColorTable, specifiedColor=<rgba 0 0 0 0>}
       cornerRadii=<3 3 3 3>
      }
      tintColor {colorUseRule=UseColor_Specified, specifiedColor=<rgba 0.0048 0.0048 0.0048 1>}
      uVRegion {bIsValid=false, max=<0 0>, min=<0 0>}
     }
     borderBackgroundColor {colorUseRule=UseColor_Specified, specifiedColor=<rgba 1 1 1 1>}
     checkedForeground {colorUseRule=UseColor_Specified, specifiedColor=<rgba 1 1 1 1>}
     checkedHoveredForeground {colorUseRule=UseColor_Specified, specifiedColor=<rgba 1 1 1 1>}
     checkedHoveredImage {
      bIsDynamicallyLoaded=false
      drawAs=Image
      imageType=Vector
      mirroring=NoMirror
      resourceName=../../../Engine/Content/Slate/Starship/CoreWidgets/CheckBox/check.svg
      resourceObject=null
      tiling=NoTile
      imageSize=<16 16>
      margin=<ltrb 0 0 0 0>
      outlineSettings {
       bUseBrushTransparency=false
       roundingType=HalfHeightRadius
       width=0
       color {colorUseRule=UseColor_Specified, specifiedColor=<rgba 0 0 0 0>}
       cornerRadii=<0 0 0 0>
      }
      tintColor {colorUseRule=UseColor_Specified, specifiedColor=<rgba 0.0044 0.2384 1 1>}
      uVRegion {bIsValid=false, max=<0 0>, min=<0 0>}
     }
     checkedImage {
      bIsDynamicallyLoaded=false
      drawAs=Image
      imageType=Vector
      mirroring=NoMirror
      resourceName=../../../Engine/Content/Slate/Starship/CoreWidgets/CheckBox/check.svg
      resourceObject=null
      tiling=NoTile
      imageSize=<16 16>
      margin=<ltrb 0 0 0 0>
      outlineSettings {
       bUseBrushTransparency=false
       roundingType=HalfHeightRadius
       width=0
       color {colorUseRule=UseColor_Specified, specifiedColor=<rgba 0 0 0 0>}
       cornerRadii=<0 0 0 0>
      }
      tintColor {colorUseRule=UseColor_Specified, specifiedColor=<rgba 0 0.162 0.7454 1>}
      uVRegion {bIsValid=false, max=<0 0>, min=<0 0>}
     }
     checkedPressedForeground {colorUseRule=UseColor_Specified, specifiedColor=<rgba 1 1 1 1>}
     checkedPressedImage {
      bIsDynamicallyLoaded=false
      drawAs=Image
      imageType=Vector
      mirroring=NoMirror
      resourceName=../../../Engine/Content/Slate/Starship/CoreWidgets/CheckBox/check.svg
      resourceObject=null
      tiling=NoTile
      imageSize=<16 16>
      margin=<ltrb 0 0 0 0>
      outlineSettings {
       bUseBrushTransparency=false
       roundingType=HalfHeightRadius
       width=0
       color {colorUseRule=UseColor_Specified, specifiedColor=<rgba 0 0 0 0>}
       cornerRadii=<0 0 0 0>
      }
      tintColor {colorUseRule=UseColor_Specified, specifiedColor=<rgba 0 0.0802 0.3515 1>}
      uVRegion {bIsValid=false, max=<0 0>, min=<0 0>}
     }
     checkedSlateSound {resourceObject=null}
     foregroundColor {colorUseRule=UseColor_Specified, specifiedColor=<rgba 1 1 1 1>}
     hoveredForeground {colorUseRule=UseColor_Specified, specifiedColor=<rgba 1 1 1 1>}
     hoveredSlateSound {resourceObject=null}
     padding=<ltrb 2 0 0 0>
     pressedForeground {colorUseRule=UseColor_Specified, specifiedColor=<rgba 1 1 1 1>}
     uncheckedHoveredImage {
      bIsDynamicallyLoaded=false
      drawAs=NoDrawType
      imageType=NoImage
      mirroring=NoMirror
      resourceName=None
      resourceObject=null
      tiling=NoTile
      imageSize=<0 0>
      margin=<ltrb 0 0 0 0>
      outlineSettings {
       bUseBrushTransparency=false
       roundingType=HalfHeightRadius
       width=0
       color {colorUseRule=UseColor_Specified, specifiedColor=<rgba 0 0 0 0>}
       cornerRadii=<0 0 0 0>
      }
      tintColor {colorUseRule=UseColor_Specified, specifiedColor=<rgba 1 1 1 1>}
      uVRegion {bIsValid=false, max=<0 0>, min=<0 0>}
     }
     uncheckedImage {
      bIsDynamicallyLoaded=false
      drawAs=NoDrawType
      imageType=NoImage
      mirroring=NoMirror
      resourceName=None
      resourceObject=null
      tiling=NoTile
      imageSize=<0 0>
      margin=<ltrb 0 0 0 0>
      outlineSettings {
       bUseBrushTransparency=false
       roundingType=HalfHeightRadius
       width=0
       color {colorUseRule=UseColor_Specified, specifiedColor=<rgba 0 0 0 0>}
       cornerRadii=<0 0 0 0>
      }
      tintColor {colorUseRule=UseColor_Specified, specifiedColor=<rgba 1 1 1 1>}
      uVRegion {bIsValid=false, max=<0 0>, min=<0 0>}
     }
     uncheckedPressedImage {
      bIsDynamicallyLoaded=false
      drawAs=NoDrawType
      imageType=NoImage
      mirroring=NoMirror
      resourceName=None
      resourceObject=null
      tiling=NoTile
      imageSize=<0 0>
      margin=<ltrb 0 0 0 0>
      outlineSettings {
       bUseBrushTransparency=false
       roundingType=HalfHeightRadius
       width=0
       color {colorUseRule=UseColor_Specified, specifiedColor=<rgba 0 0 0 0>}
       cornerRadii=<0 0 0 0>
      }
      tintColor {colorUseRule=UseColor_Specified, specifiedColor=<rgba 1 1 1 1>}
      uVRegion {bIsValid=false, max=<0 0>, min=<0 0>}
     }
     uncheckedSlateSound {resourceObject=null}
     undeterminedForeground {colorUseRule=UseColor_Specified, specifiedColor=<rgba 1 1 1 1>}
     undeterminedHoveredImage {
      bIsDynamicallyLoaded=false
      drawAs=Image
      imageType=Vector
      mirroring=NoMirror
      resourceName=../../../Engine/Content/Slate/Starship/CoreWidgets/CheckBox/indeterminate.svg
      resourceObject=null
      tiling=NoTile
      imageSize=<16 16>
      margin=<ltrb 0 0 0 0>
      outlineSettings {
       bUseBrushTransparency=false
       roundingType=HalfHeightRadius
       width=0
       color {colorUseRule=UseColor_Specified, specifiedColor=<rgba 0 0 0 0>}
       cornerRadii=<0 0 0 0>
      }
      tintColor {colorUseRule=UseColor_Specified, specifiedColor=<rgba 0.0044 0.2384 1 1>}
      uVRegion {bIsValid=false, max=<0 0>, min=<0 0>}
     }
     undeterminedImage {
      bIsDynamicallyLoaded=false
      drawAs=Image
      imageType=Vector
      mirroring=NoMirror
      resourceName=../../../Engine/Content/Slate/Starship/CoreWidgets/CheckBox/indeterminate.svg
      resourceObject=null
      tiling=NoTile
      imageSize=<16 16>
      margin=<ltrb 0 0 0 0>
      outlineSettings {
       bUseBrushTransparency=false
       roundingType=HalfHeightRadius
       width=0
       color {colorUseRule=UseColor_Specified, specifiedColor=<rgba 0 0 0 0>}
       cornerRadii=<0 0 0 0>
      }
      tintColor {colorUseRule=UseColor_Specified, specifiedColor=<rgba 0 0.162 0.7454 1>}
      uVRegion {bIsValid=false, max=<0 0>, min=<0 0>}
     }
     undeterminedPressedImage {
      bIsDynamicallyLoaded=false
      drawAs=Image
      imageType=Vector
      mirroring=NoMirror
      resourceName=../../../Engine/Content/Slate/Starship/CoreWidgets/CheckBox/indeterminate.svg
      resourceObject=null
      tiling=NoTile
      imageSize=<16 16>
      margin=<ltrb 0 0 0 0>
      outlineSettings {
       bUseBrushTransparency=false
       roundingType=HalfHeightRadius
       width=0
       color {colorUseRule=UseColor_Specified, specifiedColor=<rgba 0 0 0 0>}
       cornerRadii=<0 0 0 0>
      }
      tintColor {colorUseRule=UseColor_Specified, specifiedColor=<rgba 0 0.0802 0.3515 1>}
      uVRegion {bIsValid=false, max=<0 0>, min=<0 0>}
     }
    }}
   slot {class=HorizontalBoxSlot}
  }
  CommonTextBlock_2 {
   widgetClassPath=&CommonTextBlock
   parent=HorizontalBox_2
   widget=CommonTextBlock_2
   properties {
    text=Disable Skeletal Mesh LOD
    font {
     bForceMonospaced=false
     bMaterialIsStencil=false
     fontMaterial=null
     fontObject=&BurbankSmall
     letterSpacing=0
     monospacedWidth=1
     size=10
     skewAmount=0
     typefaceFontName=Medium
     outlineSettings {
      bApplyOutlineToDropShadows=false
      bMiteredCorners=false
      bSeparateFillAlpha=false
      outlineMaterial=null
      outlineSize=0
      outlineColor=<rgba 0 0 0 1>
     }
    }
    shadowOffset=<0 0>
   }
   slot {class=HorizontalBoxSlot}
  }
  Spacer_314 {widgetClassPath=&Spacer, parent=ScrollBox_176, widget=Spacer_314, properties {size=<1 10>}, slot {class=ScrollBoxSlot}}
  EUW_CosmeticPickerv2_Mannequin {
   widgetClassPath=&/EUW_CosmeticPickerv2_Mannequin
   bIsVariable=true
   parent=ScrollBox_176
   widget=EUW_CosmeticPickerv2_Mannequin
   slot {class=ScrollBoxSlot}
  }
  Spacer_407 {widgetClassPath=&Spacer, parent=ScrollBox_176, widget=Spacer_407, properties {size=<1 0>}, slot {class=ScrollBoxSlot}}
  SizeBox_105 {
   widgetClassPath=&SizeBox
   parent=ScrollBox_176
   widget=SizeBox_105
   slot {class=ScrollBoxSlot, properties {padding=<ltrb 10 0 10 0>}}
  }
  Image_78 {
   widgetClassPath=&Image
   bIsVariable=true
   parent=SizeBox_105
   widget=Image_78
   properties {
    brush {
     bIsDynamicallyLoaded=false
     drawAs=Image
     imageType=NoImage
     mirroring=NoMirror
     resourceName=None
     resourceObject=null
     tiling=NoTile
     imageSize=<8 8>
     margin=<ltrb 0 0 0 0>
     outlineSettings {
      bUseBrushTransparency=false
      roundingType=HalfHeightRadius
      width=0
      color {colorUseRule=UseColor_Specified, specifiedColor=<rgba 0 0 0 0>}
      cornerRadii=<0 0 0 0>
     }
     tintColor {colorUseRule=UseColor_Specified, specifiedColor=<rgba 1 1 1 1>}
     uVRegion {bIsValid=false, max=<0 0>, min=<0 0>}
    }
    colorAndOpacity=<rgba 0.0176 0.0176 0.0176 1>
   }
   slot {class=SizeBoxSlot}
  }
  EUW_CosmeticPicker_ID_Character {
   widgetClassPath=&/EUW_CosmeticPickerv2_ID
   bIsVariable=true
   parent=ScrollBox_176
   widget=EUW_CosmeticPicker_ID_Character
   properties {cosmeticType=Character}
   slot {class=ScrollBoxSlot, properties {padding=<ltrb 10 0 10 0>}}
  }
  Spacer_493 {widgetClassPath=&Spacer, parent=ScrollBox_176, widget=Spacer_493, properties {size=<1 0>}, slot {class=ScrollBoxSlot}}
  EUW_CosmeticPicker_ID_Pickaxe {
   widgetClassPath=&EUW_CosmeticPickerv2_ID
   bIsVariable=true
   parent=ScrollBox_176
   widget=EUW_CosmeticPicker_ID_Pickaxe
   properties {cosmeticType=Pickaxe}
   slot {class=ScrollBoxSlot, properties {padding=<ltrb 10 0 10 0>}}
  }
  Spacer {widgetClassPath=&Spacer, parent=ScrollBox_176, widget=Spacer, properties {size=<1 0>}, slot {class=ScrollBoxSlot}}
  EUW_CosmeticPicker_ID_Backpack {
   widgetClassPath=&EUW_CosmeticPickerv2_ID
   bIsVariable=true
   parent=ScrollBox_176
   widget=EUW_CosmeticPicker_ID_Backpack
   properties {cosmeticType=Backpack}
   slot {class=ScrollBoxSlot, properties {padding=<ltrb 10 0 10 0>}}
  }
  Spacer_1 {widgetClassPath=&Spacer, parent=ScrollBox_176, widget=Spacer_1, properties {size=<1 0>}, slot {class=ScrollBoxSlot}}
  EUW_CosmeticPicker_ID_Glider {
   widgetClassPath=&EUW_CosmeticPickerv2_ID
   bIsVariable=true
   parent=ScrollBox_176
   widget=EUW_CosmeticPicker_ID_Glider
   properties {cosmeticType=Glider}
   slot {class=ScrollBoxSlot, properties {padding=<ltrb 10 0 10 0>}}
  }
  Spacer_4 {widgetClassPath=&Spacer, parent=ScrollBox_176, widget=Spacer_4, properties {size=<1 0>}, slot {class=ScrollBoxSlot}}
  EUW_CosmeticPicker_ID_Shoes {
   widgetClassPath=&EUW_CosmeticPickerv2_ID
   bIsVariable=true
   parent=ScrollBox_176
   widget=EUW_CosmeticPicker_ID_Shoes
   properties {cosmeticType=Shoes}
   slot {class=ScrollBoxSlot, properties {padding=<ltrb 10 0 10 0>}}
  }
  Spacer_5 {widgetClassPath=&Spacer, parent=ScrollBox_176, widget=Spacer_5, properties {size=<1 0>}, slot {class=ScrollBoxSlot}}
  EUW_CosmeticPicker_ID_Guitar {
   widgetClassPath=&EUW_CosmeticPickerv2_ID
   bIsVariable=true
   parent=ScrollBox_176
   widget=EUW_CosmeticPicker_ID_Guitar
   properties {cosmeticType=Guitar}
   slot {class=ScrollBoxSlot, properties {padding=<ltrb 10 0 10 0>}}
  }
  Spacer_6 {widgetClassPath=&Spacer, parent=ScrollBox_176, widget=Spacer_6, properties {size=<1 0>}, slot {class=ScrollBoxSlot}}
  EUW_CosmeticPicker_ID_Bass {
   widgetClassPath=&EUW_CosmeticPickerv2_ID
   bIsVariable=true
   parent=ScrollBox_176
   widget=EUW_CosmeticPicker_ID_Bass
   properties {cosmeticType=Bass}
   slot {class=ScrollBoxSlot, properties {padding=<ltrb 10 0 10 0>}}
  }
  Spacer_7 {widgetClassPath=&Spacer, parent=ScrollBox_176, widget=Spacer_7, properties {size=<1 0>}, slot {class=ScrollBoxSlot}}
  EUW_CosmeticPicker_ID_Drums {
   widgetClassPath=&EUW_CosmeticPickerv2_ID
   bIsVariable=true
   parent=ScrollBox_176
   widget=EUW_CosmeticPicker_ID_Drums
   properties {cosmeticType=Drums}
   slot {class=ScrollBoxSlot, properties {padding=<ltrb 10 0 10 0>}}
  }
  Spacer_8 {widgetClassPath=&Spacer, parent=ScrollBox_176, widget=Spacer_8, properties {size=<1 0>}, slot {class=ScrollBoxSlot}}
  EUW_CosmeticPicker_ID_Mic {
   widgetClassPath=&EUW_CosmeticPickerv2_ID
   bIsVariable=true
   parent=ScrollBox_176
   widget=EUW_CosmeticPicker_ID_Mic
   properties {cosmeticType=Mic}
   slot {class=ScrollBoxSlot, properties {padding=<ltrb 10 0 10 0>}}
  }
  Spacer_9 {widgetClassPath=&Spacer, parent=ScrollBox_176, widget=Spacer_9, properties {size=<1 0>}, slot {class=ScrollBoxSlot}}
  EUW_CosmeticPicker_ID_Keytar {
   widgetClassPath=&EUW_CosmeticPickerv2_ID
   bIsVariable=true
   parent=ScrollBox_176
   widget=EUW_CosmeticPicker_ID_Keytar
   properties {cosmeticType=Keytar}
   slot {class=ScrollBoxSlot, properties {padding=<ltrb 10 0 10 0>}}
  }
  Spacer_10 {widgetClassPath=&Spacer, parent=ScrollBox_176, widget=Spacer_10, properties {size=<1 0>}, slot {class=ScrollBoxSlot}}
  HorizontalBox_0 {
   widgetClassPath=&HorizontalBox
   parent=ScrollBox_176
   widget=HorizontalBox_0
   slot {class=ScrollBoxSlot, properties {padding=<ltrb 10 0 10 0>}}
  }
  EUW_CosmeticPicker_ID_Aura {
   widgetClassPath=&EUW_CosmeticPickerv2_ID
   bIsVariable=true
   parent=HorizontalBox_0
   widget=EUW_CosmeticPicker_ID_Aura
   properties {cosmeticType=Aura}
   slot {class=HorizontalBoxSlot, properties {size {sizeRule=Fill, value=1}}}
  }
  CheckBox_ForceAura {
   widgetClassPath=&CheckBox
   bIsVariable=true
   parent=HorizontalBox_0
   widget=CheckBox_ForceAura
   properties {widgetStyle {
     checkBoxType=CheckBox
     backgroundHoveredImage {
      bIsDynamicallyLoaded=false
      drawAs=RoundedBox
      imageType=NoImage
      mirroring=NoMirror
      resourceName=None
      resourceObject=null
      tiling=NoTile
      imageSize=<18 18>
      margin=<ltrb 0 0 0 0>
      outlineSettings {
       bUseBrushTransparency=false
       roundingType=FixedRadius
       width=1
       color {colorUseRule=UseColor_ColorTable, specifiedColor=<rgba 0 0 0 0>}
       cornerRadii=<3 3 3 3>
      }
      tintColor {colorUseRule=UseColor_Specified, specifiedColor=<rgba 0.0048 0.0048 0.0048 1>}
      uVRegion {bIsValid=false, max=<0 0>, min=<0 0>}
     }
     backgroundImage {
      bIsDynamicallyLoaded=false
      drawAs=RoundedBox
      imageType=NoImage
      mirroring=NoMirror
      resourceName=None
      resourceObject=null
      tiling=NoTile
      imageSize=<18 18>
      margin=<ltrb 0 0 0 0>
      outlineSettings {
       bUseBrushTransparency=false
       roundingType=FixedRadius
       width=1
       color {colorUseRule=UseColor_ColorTable, specifiedColor=<rgba 0 0 0 0>}
       cornerRadii=<3 3 3 3>
      }
      tintColor {colorUseRule=UseColor_Specified, specifiedColor=<rgba 0.0048 0.0048 0.0048 1>}
      uVRegion {bIsValid=false, max=<0 0>, min=<0 0>}
     }
     backgroundPressedImage {
      bIsDynamicallyLoaded=false
      drawAs=RoundedBox
      imageType=NoImage
      mirroring=NoMirror
      resourceName=None
      resourceObject=null
      tiling=NoTile
      imageSize=<18 18>
      margin=<ltrb 0 0 0 0>
      outlineSettings {
       bUseBrushTransparency=false
       roundingType=FixedRadius
       width=1
       color {colorUseRule=UseColor_ColorTable, specifiedColor=<rgba 0 0 0 0>}
       cornerRadii=<3 3 3 3>
      }
      tintColor {colorUseRule=UseColor_Specified, specifiedColor=<rgba 0.0048 0.0048 0.0048 1>}
      uVRegion {bIsValid=false, max=<0 0>, min=<0 0>}
     }
     borderBackgroundColor {colorUseRule=UseColor_Specified, specifiedColor=<rgba 1 1 1 1>}
     checkedForeground {colorUseRule=UseColor_Specified, specifiedColor=<rgba 1 1 1 1>}
     checkedHoveredForeground {colorUseRule=UseColor_Specified, specifiedColor=<rgba 1 1 1 1>}
     checkedHoveredImage {
      bIsDynamicallyLoaded=false
      drawAs=Image
      imageType=Vector
      mirroring=NoMirror
      resourceName=../../../Engine/Content/Slate/Starship/CoreWidgets/CheckBox/check.svg
      resourceObject=null
      tiling=NoTile
      imageSize=<16 16>
      margin=<ltrb 0 0 0 0>
      outlineSettings {
       bUseBrushTransparency=false
       roundingType=HalfHeightRadius
       width=0
       color {colorUseRule=UseColor_Specified, specifiedColor=<rgba 0 0 0 0>}
       cornerRadii=<0 0 0 0>
      }
      tintColor {colorUseRule=UseColor_Specified, specifiedColor=<rgba 0.0044 0.2384 1 1>}
      uVRegion {bIsValid=false, max=<0 0>, min=<0 0>}
     }
     checkedImage {
      bIsDynamicallyLoaded=false
      drawAs=Image
      imageType=Vector
      mirroring=NoMirror
      resourceName=../../../Engine/Content/Slate/Starship/CoreWidgets/CheckBox/check.svg
      resourceObject=null
      tiling=NoTile
      imageSize=<16 16>
      margin=<ltrb 0 0 0 0>
      outlineSettings {
       bUseBrushTransparency=false
       roundingType=HalfHeightRadius
       width=0
       color {colorUseRule=UseColor_Specified, specifiedColor=<rgba 0 0 0 0>}
       cornerRadii=<0 0 0 0>
      }
      tintColor {colorUseRule=UseColor_Specified, specifiedColor=<rgba 0 0.162 0.7454 1>}
      uVRegion {bIsValid=false, max=<0 0>, min=<0 0>}
     }
     checkedPressedForeground {colorUseRule=UseColor_Specified, specifiedColor=<rgba 1 1 1 1>}
     checkedPressedImage {
      bIsDynamicallyLoaded=false
      drawAs=Image
      imageType=Vector
      mirroring=NoMirror
      resourceName=../../../Engine/Content/Slate/Starship/CoreWidgets/CheckBox/check.svg
      resourceObject=null
      tiling=NoTile
      imageSize=<16 16>
      margin=<ltrb 0 0 0 0>
      outlineSettings {
       bUseBrushTransparency=false
       roundingType=HalfHeightRadius
       width=0
       color {colorUseRule=UseColor_Specified, specifiedColor=<rgba 0 0 0 0>}
       cornerRadii=<0 0 0 0>
      }
      tintColor {colorUseRule=UseColor_Specified, specifiedColor=<rgba 0 0.0802 0.3515 1>}
      uVRegion {bIsValid=false, max=<0 0>, min=<0 0>}
     }
     checkedSlateSound {resourceObject=null}
     foregroundColor {colorUseRule=UseColor_Specified, specifiedColor=<rgba 1 1 1 1>}
     hoveredForeground {colorUseRule=UseColor_Specified, specifiedColor=<rgba 1 1 1 1>}
     hoveredSlateSound {resourceObject=null}
     padding=<ltrb 2 0 0 0>
     pressedForeground {colorUseRule=UseColor_Specified, specifiedColor=<rgba 1 1 1 1>}
     uncheckedHoveredImage {
      bIsDynamicallyLoaded=false
      drawAs=NoDrawType
      imageType=NoImage
      mirroring=NoMirror
      resourceName=None
      resourceObject=null
      tiling=NoTile
      imageSize=<0 0>
      margin=<ltrb 0 0 0 0>
      outlineSettings {
       bUseBrushTransparency=false
       roundingType=HalfHeightRadius
       width=0
       color {colorUseRule=UseColor_Specified, specifiedColor=<rgba 0 0 0 0>}
       cornerRadii=<0 0 0 0>
      }
      tintColor {colorUseRule=UseColor_Specified, specifiedColor=<rgba 1 1 1 1>}
      uVRegion {bIsValid=false, max=<0 0>, min=<0 0>}
     }
     uncheckedImage {
      bIsDynamicallyLoaded=false
      drawAs=NoDrawType
      imageType=NoImage
      mirroring=NoMirror
      resourceName=None
      resourceObject=null
      tiling=NoTile
      imageSize=<0 0>
      margin=<ltrb 0 0 0 0>
      outlineSettings {
       bUseBrushTransparency=false
       roundingType=HalfHeightRadius
       width=0
       color {colorUseRule=UseColor_Specified, specifiedColor=<rgba 0 0 0 0>}
       cornerRadii=<0 0 0 0>
      }
      tintColor {colorUseRule=UseColor_Specified, specifiedColor=<rgba 1 1 1 1>}
      uVRegion {bIsValid=false, max=<0 0>, min=<0 0>}
     }
     uncheckedPressedImage {
      bIsDynamicallyLoaded=false
      drawAs=NoDrawType
      imageType=NoImage
      mirroring=NoMirror
      resourceName=None
      resourceObject=null
      tiling=NoTile
      imageSize=<0 0>
      margin=<ltrb 0 0 0 0>
      outlineSettings {
       bUseBrushTransparency=false
       roundingType=HalfHeightRadius
       width=0
       color {colorUseRule=UseColor_Specified, specifiedColor=<rgba 0 0 0 0>}
       cornerRadii=<0 0 0 0>
      }
      tintColor {colorUseRule=UseColor_Specified, specifiedColor=<rgba 1 1 1 1>}
      uVRegion {bIsValid=false, max=<0 0>, min=<0 0>}
     }
     uncheckedSlateSound {resourceObject=null}
     undeterminedForeground {colorUseRule=UseColor_Specified, specifiedColor=<rgba 1 1 1 1>}
     undeterminedHoveredImage {
      bIsDynamicallyLoaded=false
      drawAs=Image
      imageType=Vector
      mirroring=NoMirror
      resourceName=../../../Engine/Content/Slate/Starship/CoreWidgets/CheckBox/indeterminate.svg
      resourceObject=null
      tiling=NoTile
      imageSize=<16 16>
      margin=<ltrb 0 0 0 0>
      outlineSettings {
       bUseBrushTransparency=false
       roundingType=HalfHeightRadius
       width=0
       color {colorUseRule=UseColor_Specified, specifiedColor=<rgba 0 0 0 0>}
       cornerRadii=<0 0 0 0>
      }
      tintColor {colorUseRule=UseColor_Specified, specifiedColor=<rgba 0.0044 0.2384 1 1>}
      uVRegion {bIsValid=false, max=<0 0>, min=<0 0>}
     }
     undeterminedImage {
      bIsDynamicallyLoaded=false
      drawAs=Image
      imageType=Vector
      mirroring=NoMirror
      resourceName=../../../Engine/Content/Slate/Starship/CoreWidgets/CheckBox/indeterminate.svg
      resourceObject=null
      tiling=NoTile
      imageSize=<16 16>
      margin=<ltrb 0 0 0 0>
      outlineSettings {
       bUseBrushTransparency=false
       roundingType=HalfHeightRadius
       width=0
       color {colorUseRule=UseColor_Specified, specifiedColor=<rgba 0 0 0 0>}
       cornerRadii=<0 0 0 0>
      }
      tintColor {colorUseRule=UseColor_Specified, specifiedColor=<rgba 0 0.162 0.7454 1>}
      uVRegion {bIsValid=false, max=<0 0>, min=<0 0>}
     }
     undeterminedPressedImage {
      bIsDynamicallyLoaded=false
      drawAs=Image
      imageType=Vector
      mirroring=NoMirror
      resourceName=../../../Engine/Content/Slate/Starship/CoreWidgets/CheckBox/indeterminate.svg
      resourceObject=null
      tiling=NoTile
      imageSize=<16 16>
      margin=<ltrb 0 0 0 0>
      outlineSettings {
       bUseBrushTransparency=false
       roundingType=HalfHeightRadius
       width=0
       color {colorUseRule=UseColor_Specified, specifiedColor=<rgba 0 0 0 0>}
       cornerRadii=<0 0 0 0>
      }
      tintColor {colorUseRule=UseColor_Specified, specifiedColor=<rgba 0 0.0802 0.3515 1>}
      uVRegion {bIsValid=false, max=<0 0>, min=<0 0>}
     }
    }}
   slot {class=HorizontalBoxSlot, properties {verticalAlignment=VAlign_Center, padding=<ltrb 10 0 10 0>}}
  }
  Text_ForceAura {
   widgetClassPath=&CommonTextBlock
   parent=HorizontalBox_0
   widget=Text_ForceAura
   properties {
    justification=Center
    text=Force Aura
    font {
     bForceMonospaced=false
     bMaterialIsStencil=false
     fontMaterial=null
     fontObject=&NotoSans
     letterSpacing=0
     monospacedWidth=1
     size=12
     skewAmount=0
     typefaceFontName=NotoSans-Regular
     outlineSettings {
      bApplyOutlineToDropShadows=false
      bMiteredCorners=false
      bSeparateFillAlpha=false
      outlineMaterial=null
      outlineSize=0
      outlineColor=<rgba 0 0 0 1>
     }
    }
    shadowOffset=<0 0>
   }
   slot {class=HorizontalBoxSlot, properties {verticalAlignment=VAlign_Center, padding=<ltrb 0 0 10 0>}}
  }
 ]
}
