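# Editor utility widget for posing Fortnite mannequins and building pose and prop galleries in the level.
# Clicking a data table pose applies gender-specific animation to selected mannequins.
# The gallery button duplicates a mannequin into two rotation variants per pose of the chosen type, using random ID:index tags to find and update them on reruns.
# The props button makes yaw-stepped copies of a wrap mesh; selection changes enable both buttons.
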EditorUtilityWidgetBlueprint /Game/NeverCook/EditorUtilities/CosmeticTools/MannequinPoseWidget/EUW_MannequinPosing
graphs [
 AddNewRandomTagToMannequin {nodes [
   n1 = AddNewRandomTagToMannequin!() -> (then, actor:Actor Object Reference) {is=title K2Node_FunctionEntry}
   n2 = RandomIntegerInRange!(Min:int=0, Max:int=99999999) -> (ReturnValue:int=0)
   > Set l_IDSection(l_IDSection:String=Append(A=Get TagPrefix() -> (TagPrefix:String), B="To String (Integer)"(InInt=@n2))) -> (Output_Get:String)
   n3 = Get l_IDSection() -> (l_IDSection:String)
   > n4 = Set l_finalTag(l_finalTag:Name=String To Name(InString=Append(A=@n3, B=:0))) -> (Output_Get:Name=None)
   > n5 = Set l_tags(l_tags:Array of Names=Get Tags!(self:Actor Object Reference=@n1) -> (Tags:Array of Names)) -> (Output_Get:Array of Names)
   > Add(TargetArray:Array of Names=@n5, NewItem:"Name (by ref)"=@n4) {is=title K2Node_CallArrayFunction}
   > Set Tags!(execute, Tags:Array of Names=Get l_tags() -> (l_tags:Array of Names), self:Actor Object Reference=@n1) -> (then, Output_Get:Array of Names)
   > Return Node!(execute, IDSection:String=Get l_IDSection())
  ]}
 CreatePropsGallery {nodes [
   n1 = CreatePropsGallery!() -> (then, Wrap SK Mesh:Fort Wrap Skeletal Mesh Actor Object Reference) {is=title K2Node_FunctionEntry}
   > n2 = For Loop(FirstIndex=1, LastIndex=3)
   > Is Valid(InputObject=@n1)
   n3 = float * float(A=GetMaxElement!(A:Vector=Get Actor Bounds(self=@n1, bOnlyCollidingComponents=false).BoxExtent) -> (ReturnValue:double), B=2) {is=title K2Node_PromotableOperator}
   > n4 = DuplicateActor!(
    execute
    self:Editor Actor Subsystem Object Reference=Get EditorActorSubsystem()
    ActorToDuplicate:Actor Object Reference=@n1
    ToWorld:World Object Reference
    Offset:Vector=MakeVector(X=float * float(A=@n3, B:int=@n2), Y=0)
   ) -> (then, ReturnValue:Actor Object Reference)
   > Add Actor Local Rotation!(
    execute
    self:Actor Object Reference=@n4
    DeltaRotation:Rotator=MakeRotator(Yaw=float * float(A=-90, B=@n2))
    bSweep:bool=false
    bTeleport:bool=false
   ) -> (then, SweepHitResult:Hit Result Structure)
  ]}
 DuplicateAndAssignAnims {nodes [
   n1 = DuplicateAndAssignAnims!() -> (then, mannequin:Fort Player Mannequin Object Reference) {is=title K2Node_FunctionEntry}
   > n2 = Set l_mannequin(l_mannequin:Fort Player Mannequin Object Reference=@n1) -> (Output_Get:Fort Player Mannequin Object Reference)
   > Is Valid(InputObject=@n2)
   > n3 = Sequence!(execute) -> (then_0, then_1, then_2:Exec, then_3:Exec)
   n4 = Get AnimDataTable() -> (AnimDataTable:Data Table Object Reference)
   > n5 = GetDataTableRowNames(Table=@n4)
   > n6 = For Each Loop(Array:Array of Names=@n5) -> (Array Element:Name)
   n7 = Get l_animDatas() -> (l_animDatas:Array of Struct Mannequin Posing Structures)
   Set l_numAnims(execute=@n6.Completed, l_numAnims:int=int * int(A=Length(TargetArray:Array of Struct Mannequin Posing Structures=@n7), B=2)) -> (Output_Get:int=0)
   n8 = Get Data Table Row!(execute:Exec=@n6.LoopBody, DataTable:Data Table Object Reference=@n4, RowName:Name=@n6) -> (then:Exec, RowNotFound:Exec, ReturnValue:Struct Mannequin Posing Structure)
   > Branch(Condition="Equal (Enum)"(
     A:Enum_MannequinPosing Enum=Break Struct Mannequin Posing!(Struct_MannequinPosing:"Struct Mannequin Posing Structure (by ref)"=@n8) -> (
      FriendlyName_2_#1:String
      AnimDataMale_10_#2:Single Animation Play Data Structure
      AnimDataFemale_12_#3:Single Animation Play Data Structure
      InvertedRotation_15_#4:Rotator=~
      Type_18_#5:Enum_MannequinPosing Enum
     ) {is=title K2Node_BreakStruct}.Type_18_#5
     B:Enum_MannequinPosing Enum=Get PoseGalleryType() -> (PoseGalleryType:Enum_MannequinPosing Enum=NewEnumerator0)
    ))
   > Add(TargetArray:Array of Struct Mannequin Posing Structures=Get l_animDatas(), NewItem:"Struct Mannequin Posing Structure (by ref)"=@n8)
   n9 = For Each Loop with Break!(
    Exec:Exec=@n3.then_1
    Array:Array of Names=Get Tags(self=Get l_mannequin() -> (l_mannequin:Fort Player Mannequin Object Reference))
    Break:Exec=@n16.then
   ) -> (LoopBody:Exec, Array Element:Name, Array Index:int, Completed:Exec)
   n10 = Branch(execute=@n9.Completed, Condition="Equal (Name)"(A=Get l_finalTag() -> (l_finalTag:Name=None)))
   > n11 = AddNewRandomTagToMannequin!(execute, self, actor:Actor Object Reference=Get l_mannequin()) -> (then, IDSection:String) {is=title K2Node_CallFunction}
   > Add(
    TargetArray:Array of Fort Player Mannequin Object References=Get l_mannequinCache() -> (l_mannequinCache:Array of Fort Player Mannequin Object References)
    NewItem:"Fort Player Mannequin Object Reference (by ref)"=Get l_mannequin()
   )
   > n12 = For Loop(FirstIndex=1, LastIndex=int - int(A=Get l_numAnims() -> (l_numAnims:int=0)))
   > n13 = DuplicateMannequin!(execute, self, Mannequin:Fort Player Mannequin Object Reference=Get l_mannequin(), offsetCount:int=@n12) -> (then, AsFort Player Mannequin:Fort Player Mannequin Object Reference)
   > Add(TargetArray=Get l_mannequinCache(), NewItem=@n13)
   n14 = Append!(A=@n11, B=:, C:String="To String (Integer)"(InInt=@n12)) -> (ReturnValue)
   > Set Tags(Tags=Make Array("[0]":Name=String To Name(InString=@n14)) -> (Array:Array of Names), self=@n13)
   n15 = GetMannequinsWithTagBase!(
    execute=@n10.else
    self
    parentMannequin:Fort Player Mannequin Object Reference=Get l_mannequin()
    tag:Name=Get l_finalTag() -> (l_finalTag=None)
    count:int=Get l_numAnims() -> (l_numAnims=0)
   ) -> (then, Mannequins:Array of Fort Player Mannequin Object References)
   > Set l_mannequinCache(l_mannequinCache:Array of Fort Player Mannequin Object References=@n15) -> (Output_Get:Array of Fort Player Mannequin Object References)
   Branch(execute=@n9.LoopBody, Condition=StartsWith(SourceString="To String (Name)"(InName=@n9), InPrefix=Get TagPrefix()))
   > n16 = Set l_finalTag(l_finalTag=@n9) -> (Output_Get=None)
   n17 = For Each Loop(Exec=@n3.then_2, Array:Array of Struct Mannequin Posing Structures=Get l_animDatas()) -> (Array Element:Struct Mannequin Posing Structure)
   > n18 = Set l_dataStructTemp(l_dataStructTemp:Struct Mannequin Posing Structure=@n17) -> (Output_Get:Struct Mannequin Posing Structure)
   > n19 = Set members in Struct Mannequin Posing!(
    execute:Exec
    StructRef:"Struct Mannequin Posing Structure (by ref)"=@n18
    InvertedRotation_15_#4:Rotator="0,0,0"
   ) -> (then:Exec, StructOut:"Struct Mannequin Posing Structure (by ref)") {is=title K2Node_SetFieldsInStruct}
   > Add(
    TargetArray:Array of Struct Mannequin Posing Structures=Get l_finalAnimData() -> (l_finalAnimData:Array of Struct Mannequin Posing Structures)
    NewItem:"Struct Mannequin Posing Structure (by ref)"=@n19
   )
   > Add(TargetArray=Get l_finalAnimData(), NewItem=@n17)
   n20 = For Each Loop(Exec=@n3.then_3, Array:Array of Fort Player Mannequin Object References=Get l_mannequinCache()) -> (Array Element:Fort Player Mannequin Object Reference)
   n21 = "Get (a ref)"(Array:Array of Struct Mannequin Posing Structures=Get l_finalAnimData(), Dimension 1=@n20.Array Index) -> (Output:"Struct Mannequin Posing Structure (by ref)")
   > setAnimationOnMannequin!(
    execute
    self
    Struct_MannequinPosing:"Struct Mannequin Posing Structure (by ref)"=@n21
    Mannequin:Fort Player Mannequin Object Reference=@n20
   ) -> (then)
   > Set Actor Rotation!(
    execute
    self:Actor Object Reference=@n20
    NewRotation:Rotator=Combine Rotators!(
     A:Rotator=Break Struct Mannequin Posing(Struct_MannequinPosing=@n21) -> (InvertedRotation_15_#4=~).InvertedRotation_15_#4
     B:Rotator="0,-90,0"
    ) -> (ReturnValue:Rotator)
    bTeleportPhysics:bool=false
   ) -> (then, ReturnValue:bool) {is=title K2Node_CallFunction}
   n22 = For Each Loop(Array=Get l_mannequinCache())
   > n23 = Branch(Condition="integer > integer"(A="% (Integer)"(A=@n22.Array Index, B=2) -> (ReturnValue=0)))
   n24 = int / int(A=@n22.Array Index, B=2)
   n25 = "Get (a ref)"(Array=Get l_animDatas(), Dimension 1=int - int(A=@n24))
   > setAnimationOnMannequin(Struct_MannequinPosing=@n25, Mannequin=@n22)
   > Set Actor Rotation(
    self=@n22
    NewRotation=Break Struct Mannequin Posing(Struct_MannequinPosing=@n25) -> (InvertedRotation_15_#4=~).InvertedRotation_15_#4
    bTeleportPhysics=false
   )
   setAnimationOnMannequin(execute=@n23.else, Struct_MannequinPosing="Get (a ref)"(Array=Get l_animDatas(), Dimension 1=@n24), Mannequin=@n22)
   > Set Actor Rotation(self=@n22, NewRotation="0, 0, 0", bTeleportPhysics=false)
  ]}
 DuplicateMannequin {nodes [
   n1 = DuplicateMannequin!() -> (then, Mannequin:Fort Player Mannequin Object Reference, offsetCount:int) {is=title K2Node_FunctionEntry}
   n2 = Make Literal Float(Value=150)
   > n3 = DuplicateActor(self=Get EditorActorSubsystem(), ActorToDuplicate=@n1, Offset=MakeVector(Y=float * float(A=@n2, B=@n1.offsetCount)))
   > n4 = Cast To FortPlayerMannequin!(execute, Object=@n3) -> (then, CastFailed, AsFort Player Mannequin:Fort Player Mannequin Object Reference) {is=title K2Node_DynamicCast}
   > Return Node!(execute, AsFort Player Mannequin:Fort Player Mannequin Object Reference=@n4)
  ]}
 EventGraph {nodes [
   n1 = "On Item Left Clicked (listPoses)"!() -> (OutputDelegate, then, isShiftDown:bool, isCtrlDown:bool, data:Object Reference, item:User Widget Object Reference) {is=title K2Node_ComponentBoundEvent}
   > n2 = getIndexOfItem!(
    execute
    self:TA Generic 01 List Object Reference=Get listPoses() -> (listPoses:TA Generic 01 List Object Reference)
    Content:Widget Object Reference=@n1.item
   ) -> (then, index:int=0)
   > n3 = GetDataTableRowNames(Table=Get AnimDataTable())
   > n4 = Get Data Table Row(DataTable=Get AnimDataTable(), RowName="Get (a copy)"(Array:Array of Names=@n3, Dimension 1=@n2) -> (Output:Name))
   > n5 = GetSelectedActors()
   > n6 = Filter by Class!(
    execute
    TargetArray:Array of Object References=@n5
    ObjectClass:Object Class Reference=/Script/FortniteGame.FortPlayerMannequin
    FilterType:EEditorScriptingFilterType Enum
   ) -> (then, ReturnValue:Array of Fort Player Mannequin Object References)
   > n7 = For Each Loop(Array=@n6)
   > setAnimationOnMannequin(Struct_MannequinPosing=@n4, Mannequin=@n7)
   "On Clicked (TA_Generic01_Button)"!() -> (OutputDelegate, then, widget:User Widget Object Reference, event:Pointer Event Structure) {is=title K2Node_ComponentBoundEvent}
   > n8 = GetSelectedActors()
   > n9 = Filter by Class(TargetArray=@n8, ObjectClass=&FortPlayerMannequin)
   > Branch(Condition=Is Valid Index!(TargetArray:Array of Fort Player Mannequin Object References=@n9, IndexToTest:int) -> (ReturnValue:bool))
   > DuplicateAndAssignAnims!(
    execute
    self
    mannequin:Fort Player Mannequin Object Reference="Get (a copy)"(Array:Array of Fort Player Mannequin Object References=@n9) -> (Output:Fort Player Mannequin Object Reference)
   ) -> (then) {is=title K2Node_CallFunction}
   "On Clicked (TA_Generic01_Button_Props)"!() -> (OutputDelegate, then, widget:User Widget Object Reference, event:Pointer Event Structure) {is=title K2Node_ComponentBoundEvent}
   > n10 = GetSelectedActors()
   > n11 = Filter by Class(TargetArray=@n10, ObjectClass=/Script/FortniteGame.FortWrapSkeletalMeshActor) -> (ReturnValue:Array of Fort Wrap Skeletal Mesh Actor Object References)
   > CreatePropsGallery!(
    execute
    self
    Wrap SK Mesh:Fort Wrap Skeletal Mesh Actor Object Reference="Get (a copy)"(Array:Array of Fort Wrap Skeletal Mesh Actor Object References=@n11) -> (Output:Fort Wrap Skeletal Mesh Actor Object Reference)
   ) -> (then) {is=title K2Node_CallFunction}
   Event Destruct()
   > n12 = GetSelectionSet(self=Get LevelEditorSubsystem())
   > Unbind Event from On Selection Change(self:Typed Element Selection Set Object Reference=@n12, Delegate=Create Event()) {is=title K2Node_RemoveDelegate}
   Event Construct()
   > initPoseList()
   > n13 = GetSelectionSet(self=Get LevelEditorSubsystem())
   > Bind Event to On Selection Change(self:Typed Element Selection Set Object Reference=@n13, Delegate=Create Event()) {is=title K2Node_AddDelegate}
   > handleEditorSelectionChanged!(execute, self, SelectionSet:Typed Element Selection Set Object Reference=@n13) -> (then)
   Event PreConstruct()
   > SetObject(self=Get DetailsView_60() -> (DetailsView_60:Details View Object Reference), NewObject=Self-Reference())
   DuplicateAndAssignAnims(mannequin=SpawnMannequinFromView!(execute, self) -> (then, ReturnValue:Fort Player Mannequin Object Reference))
  ]}
 GetMannequinsWithTagBase {nodes [
   n1 = GetAllActorsOfClass!(execute=@n13.then_0, WorldContextObject:Object Reference, ActorClass:Actor Class Reference=&FortPlayerMannequin) -> (then, OutActors:Array of Fort Player Mannequin Object References)
   n2 = Filter by Class(TargetArray=@n1, ObjectClass=&FortPlayerMannequin) -> (ReturnValue:Array of Fort Player Mannequin Object References)
   n3 = For Each Loop(Exec=[@n2.then, @n1.then], Array=@n1)
   > Set l_mann(l_mann:Fort Player Mannequin Object Reference=@n3) -> (Output_Get:Fort Player Mannequin Object Reference)
   > n4 = For Each Loop with Break(Array=Get Tags(self=@n3))
   n5 = "To String (Name)"(InName=@n4)
   n6 = Get l_IDSection()
   > PrintString(InString=Append(A=@n5, B="Starts with:  ", C=@n6))
   > Branch(Condition=StartsWith(SourceString=@n5, InPrefix=@n6))
   > n7 = Set l_foundIndex(l_foundIndex:int=String To Integer(InString=Split(SourceString=@n5, InStr=:, SearchDir=FromEnd).RightS)) -> (Output_Get:int=0)
   > PrintString(InString=Append!(A="Found Index: ", B="To String (Integer)"(InInt=@n7)) -> (ReturnValue))
   > Branch(Condition="integer > integer"(A=@n7, B=-1))
   n8 = Get l_foundIndex() -> (l_foundIndex:int=0)
   > Set Array Elem!(
    execute
    TargetArray:Array of Fort Player Mannequin Object References=Get l_mannequinsWithTag() -> (l_mannequinsWithTag:Array of Fort Player Mannequin Object References)
    Index:int=@n8
    Item:"Fort Player Mannequin Object Reference (by ref)"=Get l_mann() -> (l_mann:Fort Player Mannequin Object Reference)
    bSizeToFit:bool=true
   ) -> (then)
   n9 = Get l_mann()
   > PrintString(InString=Append!(A="Adding ", B=GetDisplayName(Object=@n9), C:String=" to ", D:String="To String (Integer)"(InInt=@n8)) -> (ReturnValue))
   n10 = GetMannequinsWithTagBase!() -> (then, parentMannequin:Fort Player Mannequin Object Reference, tag:Name, count:int) {is=title K2Node_FunctionEntry}
   > Set l_parent(l_parent:Fort Player Mannequin Object Reference=@n10) -> (Output_Get:Fort Player Mannequin Object Reference)
   > n11 = Set l_tag(l_tag:Name=@n10.tag) -> (Output_Get:Name=None)
   > Resize!(execute, TargetArray:Array of Fort Player Mannequin Object References=Get l_mannequinsWithTag(), Size:int=@n10.count) -> (then) {is=title K2Node_CallArrayFunction}
   n12 = Split(SourceString="To String (Name)"(InName=@n11), InStr=:, SearchDir=FromEnd)
   > Set l_IDSection(l_IDSection=@n12)
   > Set l_parentIndex(l_parentIndex:int=String To Integer(InString=@n12.RightS)) -> (Output_Get:int=0)
   > n13 = Sequence!(execute) -> (then_0, then_1, then_2)
   n14 = For Each Loop(Exec=@n13.then_1, Array=Get l_mannequinsWithTag())
   > n15 = Is Valid(InputObject=@n14)
   n16 = DuplicateMannequin(
    execute=@n15.Is Not Valid
    Mannequin=Get l_parent() -> (l_parent:Fort Player Mannequin Object Reference)
    offsetCount=int - int(A=@n14.Array Index, B=Get l_parentIndex() -> (l_parentIndex:int=0))
   ) {is=title K2Node_CallFunction}
   n17 = Append!(A=Get l_IDSection(), B=:, C="To String (Integer)"(InInt=@n14.Array Index)) -> (ReturnValue)
   > Set Tags(Tags=Make Array("[0]"=String To Name(InString=@n17)), self=@n16)
   > Set Array Elem(TargetArray=Get l_mannequinsWithTag(), Index=@n14.Array Index, Item=@n16, bSizeToFit=false)
   Return Node!(execute=@n13.then_2, Mannequins:Array of Fort Player Mannequin Object References=Get l_mannequinsWithTag())
  ]}
 SpawnMannequinFromView {nodes [
   SpawnMannequinFromView() {is=title K2Node_FunctionEntry}
   > n1 = Sequence!(execute) -> (then_0, then_1)
   > n2 = Get Editor Camera Info!(execute) -> (then, CameraLocation:Vector, ForwardDirection:Vector, FOV:float, ReturnValue:bool) {is=title K2Node_CallFunction}
   > Set l_noHitLoc(l_noHitLoc:Vector=vector + vector(A=@n2, B=vector * vector(A=@n2.ForwardDirection, B:float=1500) -> (ReturnValue=~))) -> (Output_Get:Vector)
   > n3 = Line Trace By Channel!(
    execute
    WorldContextObject:Object Reference
    Start:Vector=@n2
    End:Vector=vector + vector(A=@n2, B=vector * vector(A=@n2.ForwardDirection, B=50000) -> (ReturnValue=~))
    TraceChannel:ETraceTypeQuery Enum
    bTraceComplex:bool=false
    ActorsToIgnore:Array of Actor Object References
    DrawDebugType:EDrawDebugTrace Enum
    bIgnoreSelf:bool
    TraceColor:Linear Color Structure
    TraceHitColor:Linear Color Structure
    DrawTime:float
   ) -> (then, OutHit:Hit Result Structure, ReturnValue:bool)
   > n4 = Branch(Condition=@n3.ReturnValue)
   n5 = BreakHitResult!(Hit:"Hit Result Structure (by ref)"=@n3) -> (
    bBlockingHit:bool
    bInitialOverlap:bool
    Time:float
    Distance:float
    Location:Vector
    ImpactPoint:Vector
    Normal:Vector
    ImpactNormal:Vector
    PhysMat:Physical Material Object Reference
    HitActor:Actor Object Reference
    HitComponent:Primitive Component Object Reference
    HitBoneName:Name
    BoneName:Name
    HitItem:int
    ElementIndex:int
    FaceIndex:int
    TraceStart:Vector
    TraceEnd:Vector
   )
   > Set l_loc(l_loc:Vector=@n5.Location) -> (Output_Get:Vector)
   > Set l_normal(l_normal:Vector=@n5.Normal) -> (Output_Get:Vector)
   Set l_loc(execute=@n4.else, l_loc=Get l_noHitLoc() -> (l_noHitLoc:Vector))
   > Set l_normal(l_normal="0,0,1")
   n6 = SpawnActorFromClass!(
    execute=@n1.then_1
    self:Editor Actor Subsystem Object Reference=Get EditorActorSubsystem()
    ActorClass:Actor Class Reference=&FortPlayerMannequin
    Location:Vector=Get l_loc() -> (l_loc:Vector)
    Rotation:Rotator
    bTransient:bool
   ) -> (then, ReturnValue:Fort Player Mannequin Object Reference)
   > SetEditorProperty(Object=@n6, PropertyName=bIsEditorOnlyActor, PropertyValue=MakeLiteralBool(Value=true))
   > Execute Python Script!(
    execute
    PythonScript="import TechArt.CosmeticPicker.CosmeticPickerModule as CosmeticPicker\r\nimport importlib\r\n\r\nimportlib.reload(CosmeticPicker)\r\n\r\nCosmeticPicker.setMannequinOverrideVariants(mannequin)"
    mannequin:Fort Player Mannequin Object Reference=@n6
   ) -> (then, ReturnValue) {inputs [mannequin]}
   > Return Node!(execute, ReturnValue:Fort Player Mannequin Object Reference=@n6)
  ]}
 handleEditorSelectionChanged {nodes [
   n1 = handleEditorSelectionChanged!() -> (then, SelectionSet:Typed Element Selection Set Object Reference) {is=title K2Node_FunctionEntry}
   > n2 = Sequence()
   > n3 = Set l_mannequin(l_mannequin:Object Reference=GetTopSelectedObject!(self:Typed Element Selection Set Object Reference=@n1, InRequiredClass:Object Class Reference=&FortPlayerMannequin) -> (ReturnValue:Object Reference)) -> (Output_Get:Object Reference)
   > n4 = Is Valid(InputObject=@n3)
   n5 = Get TA_Generic01_Button() -> (TA_Generic01_Button:TA Generic 01 Button Object Reference)
   SetIsEnabled(execute=@n4.Is Not Valid, self=@n5, bInIsEnabled=false)
   > SetToolTipText!(
    execute
    self:Widget Object Reference=@n5
    InToolTipText:"Text (by ref)"=MakeLiteralText(Value="INVTEXT(\"Must select a Mannequin in level to create a gallery\")")
   ) -> (then)
   > Set bIsEnabled!(execute, bIsEnabled:bool=false, self:Widget Object Reference=@n5) -> (then, Output_Get:bool)
   SetIsEnabled(execute=@n4.Is Valid, self=@n5, bInIsEnabled=true)
   > SetToolTipText(
    self=@n5
    InToolTipText=MakeLiteralText(Value="INVTEXT(\"Build a gallery of mannequins, one for each pose in the pose list above.\\r\\n  \\r\\nIf this has been run before it will find the previous mannequins and update them to use the poses in the list\\r\\n\\r\\nIf no gallery has been made before this will duplicate the selected mannequin to build the pose gallery.\")")
   )
   > Set bIsEnabled(bIsEnabled=true, self=@n5)
   n6 = Set l_prop(execute=@n2.then_1, l_prop:Object Reference=GetTopSelectedObject(self=@n1, InRequiredClass=&FortWrapSkeletalMeshActor)) -> (Output_Get:Object Reference)
   > n7 = Is Valid(InputObject=@n6)
   n8 = Get TA_Generic01_Button_Props() -> (TA_Generic01_Button_Props:TA Generic 01 Button Object Reference)
   SetIsEnabled(execute=@n7.Is Not Valid, self=@n8, bInIsEnabled=false)
   > SetToolTipText(self=@n8, InToolTipText=MakeLiteralText(Value="INVTEXT(\"Must select a Wrap Skeletal Mesh in level to create a gallery\")"))
   > Set bIsEnabled(bIsEnabled=false, self=@n8)
   SetIsEnabled(execute=@n7.Is Valid, self=@n8, bInIsEnabled=true)
   > SetToolTipText(
    self=@n8
    InToolTipText=MakeLiteralText(Value="INVTEXT(\"Build a gallery of props with different angles\\r\\n  \\r\\nThis will always create a new set of meshes\")")
   )
   > Set bIsEnabled(bIsEnabled=true, self=@n8)
  ]}
 initPoseList {nodes [
   initPoseList() {is=title K2Node_FunctionEntry}
   n1 = Get AnimDataTable()
   > n2 = GetDataTableRowNames(Table=@n1)
   > n3 = For Each Loop(Array:Array of Names=@n2) -> (Array Element:Name)
   > n4 = Get Data Table Row(DataTable=@n1, RowName=@n3)
   > n5 = addItemGeneric!(
    execute
    self:TA Generic 01 List Object Reference=Get listPoses()
    displayString:String=Break Struct Mannequin Posing(Struct_MannequinPosing=@n4) -> (InvertedRotation_15_#4=~)
   ) -> (then, widget:User Widget Object Reference, data:Object Reference)
   > n6 = Cast To TA_Generic01_ListItem!(execute, Object=@n5) -> (then, CastFailed, AsTA Generic 01 List Item:TA Generic 01 List Item Object Reference) {is=title K2Node_DynamicCast}
   > setHorizontalAlignmentSet!(execute, self:TA Base Editor Button Widget Object Reference=@n6, alignment:EHorizontalAlignment Enum=HAlign_Left) -> (then)
  ]}
 setAnimationOnMannequin {nodes [
   n1 = setAnimationOnMannequin!() -> (then, Struct_MannequinPosing:"Struct Mannequin Posing Structure (by ref)", Mannequin:Fort Player Mannequin Object Reference) {is=title K2Node_FunctionEntry}
   > Set l_animPoseStruct(l_animPoseStruct:Struct Mannequin Posing Structure=@n1) -> (Output_Get:Struct Mannequin Posing Structure)
   > n2 = Sequence()
   n3 = Break Struct Mannequin Posing(Struct_MannequinPosing=Get l_animPoseStruct() -> (l_animPoseStruct:Struct Mannequin Posing Structure)) -> (InvertedRotation_15_#4=~)
   > Set l_finalAnimData(l_finalAnimData:Single Animation Play Data Structure=Select!(
     Invalid:Single Animation Play Data Structure=@n3.AnimDataFemale_12_#3
     Male:Single Animation Play Data Structure=@n3.AnimDataMale_10_#2
     Female:Single Animation Play Data Structure=@n3.AnimDataFemale_12_#3
     Both:Single Animation Play Data Structure=@n3.AnimDataFemale_12_#3
     Index:EFortCustomGender Enum=GetGender!(self:Fort Player Mannequin Object Reference=@n1.Mannequin) -> (ReturnValue:EFortCustomGender Enum=Invalid)
    ) -> (ReturnValue:Single Animation Play Data Structure)) -> (Output_Get:Single Animation Play Data Structure)
   n4 = Get SkeletalMeshComponent!(self:Skeletal Mesh Actor Object Reference=@n1.Mannequin) -> (SkeletalMeshComponent:Skeletal Mesh Component Object Reference)
   Set AnimationData!(
    execute=@n2.then_1
    AnimationData:Single Animation Play Data Structure=Get l_finalAnimData() -> (l_finalAnimData:Single Animation Play Data Structure)
    self:Skeletal Mesh Component Object Reference=@n4
   ) -> (then, Output_Get:Single Animation Play Data Structure)
   > SetAnimationMode!(
    execute
    self:Skeletal Mesh Component Object Reference=@n4
    InAnimationMode:EAnimationMode Enum=AnimationSingleNode
    bForceInitAnimScriptInstance:bool=true
   ) -> (then)
   > SetEditorProperty(
    Object=@n4
    PropertyName=bAutoActivate
    PropertyValue=NOT Boolean(A=Get bAutoActivate!(self:Actor Component Object Reference=@n4) -> (bAutoActivate:bool))
   )
  ]}
 setAnimationOnMannequins {nodes [
   n1 = setAnimationOnMannequins!() -> (then, Struct_MannequinPosing:"Struct Mannequin Posing Structure (by ref)") {is=title K2Node_FunctionEntry}
   > Set l_animPoseStruct(l_animPoseStruct=@n1)
   > n2 = GetSelectedActors()
   > n3 = Filter by Class(TargetArray=@n2, ObjectClass=&FortPlayerMannequin)
   > n4 = For Each Loop(Array:Array of Fort Player Mannequin Object References=@n3) -> (Array Element:Fort Player Mannequin Object Reference)
   > n5 = Sequence()
   n6 = Break Struct Mannequin Posing(Struct_MannequinPosing=Get l_animPoseStruct()) -> (InvertedRotation_15_#4=~)
   > Set l_finalAnimData(l_finalAnimData=Select(
     Invalid=@n6.AnimDataFemale_12_#3
     Male=@n6.AnimDataMale_10_#2
     Female=@n6.AnimDataFemale_12_#3
     Both=@n6.AnimDataFemale_12_#3
     Index=GetGender(self=@n4) -> (ReturnValue=Invalid)
    ))
   n7 = Get SkeletalMeshComponent(self=@n4)
   Set AnimationData(execute=@n5.then_1, AnimationData=Get l_finalAnimData(), self=@n7)
   > SetAnimationMode(self=@n7, InAnimationMode=AnimationSingleNode, bForceInitAnimScriptInstance=true)
   > SetEditorProperty(Object=@n7, PropertyName=bAutoActivate, PropertyValue=NOT Boolean(A=Get bAutoActivate(self=@n7)))
  ]}
]
variables [
 AnimDataTable {type=DataTable, default=&/DataTable_MannequinPoses}
 NewVar {type="array<FortPlayerMannequin>", default []}
 PoseGalleryType {type=Enum_MannequinPosing, default=Default}
 TagPrefix {type=string, default=AutoMannquinSetupID:}
]
widgets {
 info {rootWidgetClass=&StackBox, widgetCount=8}
 widgets [
  StackBox_0 {widgetClassPath=&StackBox, widget=StackBox_0, properties {orientation=Orient_Vertical}}
  TextBlock_54 {
   widgetClassPath=&TextBlock
   parent=StackBox_0
   widget=TextBlock_54
   properties {
    text="Pose List (Sets pose on all selected Mannequins)"
    colorAndOpacity {colorUseRule=UseColor_Specified, specifiedColor=<rgba 1 1 1 0.733>}
    font {
     bForceMonospaced=false
     bMaterialIsStencil=false
     fontMaterial=null
     fontObject=&Roboto
     letterSpacing=0
     monospacedWidth=1
     size=13.5
     skewAmount=0
     typefaceFontName=Bold
     outlineSettings {
      bApplyOutlineToDropShadows=false
      bMiteredCorners=false
      bSeparateFillAlpha=false
      outlineMaterial=null
      outlineSize=0
      outlineColor=<rgba 0 0 0 1>
     }
    }
   }
   slot {class=StackBoxSlot, properties {padding=<ltrb 20 10 5 5>}}
  }
  Image_133 {
   widgetClassPath=&Image
   bIsVariable=true
   parent=StackBox_0
   widget=Image_133
   properties {
    brush {
     bIsDynamicallyLoaded=false
     drawAs=Image
     imageType=NoImage
     mirroring=NoMirror
     resourceName=None
     resourceObject=null
     tiling=NoTile
     imageSize=<32 2>
     margin=<ltrb 0 0 0 0>
     outlineSettings {
      bUseBrushTransparency=false
      roundingType=HalfHeightRadius
      width=0
      color {colorUseRule=UseColor_Specified, specifiedColor=<rgba 0 0 0 0>}
      cornerRadii=<0 0 0 0>
     }
     tintColor {colorUseRule=UseColor_Specified, specifiedColor=<rgba 1 1 1 1>}
     uVRegion {bIsValid=false, max=<0 0>, min=<0 0>}
    }
    colorAndOpacity=<rgba 1 1 1 0.806>
   }
   slot {class=StackBoxSlot}
  }
  Border_0 {
   widgetClassPath=&Border
   parent=StackBox_0
   widget=Border_0
   properties {brushColor=<rgba 0.1198 0.1198 0.1198 0.319>, padding=<ltrb 5 5 5 5>}
   slot {class=StackBoxSlot, properties {size {sizeRule=Fill, value=1}}}
  }
  listPoses {
   widgetClassPath=/Game/Blueprints/Blutilities/ForTechArt/Widgets/List/TA_Generic01_List
   bIsVariable=true
   parent=Border_0
   widget=listPoses
   properties {allowMultiSelect=false, enableSelection=false}
   slot {class=BorderSlot, properties {padding=<ltrb 5 5 5 5>}}
  }
  DetailsView_60 {
   widgetClassPath=&DetailsView
   parent=StackBox_0
   widget=DetailsView_60
   properties {categoriesToShow [NoCategory]}
   slot {class=StackBoxSlot}
  }
  TA_Generic01_Button {
   widgetClassPath=&TA_Generic01_Button
   parent=StackBox_0
   widget=TA_Generic01_Button
   properties {
    label=Setup Mannequin Pose Gallery
    toolTipText="Build a gallery of mannequins, one for each pose in the pose list above.\r\n  \r\nIf this has been run before it will find the previous mannequins and update them to use the poses in the list\r\n\r\nIf no gallery has been made before this will duplicate the selected mannequin to build the pose gallery."
   }
   slot {class=StackBoxSlot}
  }
  TA_Generic01_Button_Props {
   widgetClassPath=&TA_Generic01_Button
   parent=StackBox_0
   widget=TA_Generic01_Button_Props
   properties {label=Setup Props Gallery, toolTipText="Build a gallery of props with different angles\r\n  \r\nThis will always create a new set of meshes"}
   slot {class=StackBoxSlot}
  }
 ]
}
