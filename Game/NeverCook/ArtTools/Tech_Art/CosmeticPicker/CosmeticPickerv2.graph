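# Editor utility widget that serves as the tech-art Cosmetic Picker shell, placing an asset playlist stack beside Cosmetics, Advanced, Global and mannequin Poses tabs.
# On construct it sets viewport size, runs the ItemAccess.ForcedMontageAccess console command, and passes the second tab's widget to the first, cast as EUW_CosmeticPickerv2, as its Advanced Tab.
# An overlaid Refresh button revalidates the tab container, and tab selection prints the index and widget name.
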EditorUtilityWidgetBlueprint /Game/NeverCook/ArtTools/Tech_Art/CosmeticPicker/CosmeticPickerv2
graphs [EventGraph {nodes [
   "On Clicked (Refresh)"() {is=title K2Node_ComponentBoundEvent}
   > revalidateTabs(self:Base Tab Container Widget Object Reference=Get Generic01_TabContainer() -> (Generic01_TabContainer:Generic 01 Tab Container Object Reference))
   n1 = "On Tab Selected (Generic01_TabContainer)"!() -> (OutputDelegate, then, tabIndex:int=0, tabWidget:Widget Object Reference) {is=title K2Node_ComponentBoundEvent}
   > PrintString(InString=Append!(A="Tab selected ", B="To String (Integer)"(InInt=@n1.tabIndex), C:String=" ", D:String=GetDisplayName(Object=@n1.tabWidget)) -> (ReturnValue))
   Event Construct()
   > SetDesiredSizeInViewport!(execute, self:User Widget Object Reference, Size:Vector 2D Structure=(X=200,Y=300)) -> (then)
   > ExecuteConsoleCommand(Command=ItemAccess.ForcedMontageAccess 1)
   n2 = Get Generic01_TabContainer()
   > n3 = getTabWidgetByIndex!(execute, self:Base Tab Container Widget Object Reference=@n2, index:int=0) -> (then, widget:Widget Object Reference)
   > n4 = Cast To EUW_CosmeticPickerv2!(execute, Object=@n3) -> (then, CastFailed, AsEUW Cosmetic Pickerv 2:EUW Cosmetic Pickerv 2 Object Reference) {is=title K2Node_DynamicCast}
   > n5 = getTabWidgetByIndex(self=@n2, index=1)
   > Set Advanced Tab!(execute, Advanced Tab:Widget Object Reference=@n5, self:EUW Cosmetic Pickerv 2 Object Reference=@n4) -> (then, Output_Get:Widget Object Reference)
   Setup UI()
   Event PreConstruct()
   Event Tick()
  ]}]
variables []
widgets {
 info {rootWidgetClass=&CanvasPanel, widgetCount=5}
 widgets [
  CanvasPanel_0 {widgetClassPath=&CanvasPanel, widget=CanvasPanel_0}
  HorizontalBox_21 {
   widgetClassPath=&HorizontalBox
   parent=CanvasPanel_0
   widget=HorizontalBox_21
   slot {class=CanvasPanelSlot, properties {layoutData {alignment=<0 0>, anchors {maximum=<1 1>, minimum=<0 0>}, offsets=<ltrb 0 0 0 0>}}}
  }
  EUW_CosmeticPicker_playlistStack {
   widgetClassPath=/Game/NeverCook/ArtTools/Tech_Art/AssetPlaylist/EUW_AssetPlaylist_verticalGroupList
   bIsVariable=true
   parent=HorizontalBox_21
   widget=EUW_CosmeticPicker_playlistStack
   slot {class=HorizontalBoxSlot}
  }
  Generic01_TabContainer {
   widgetClassPath=/Game/Blueprints/Blutilities/ForTechArt/Widgets/TabContainer/Generic01_TabContainer
   bIsVariable=true
   parent=HorizontalBox_21
   widget=Generic01_TabContainer
   properties {initialTabs {
     Advanced=&/EUW_CosmeticPickerv2_Advanced.EUW_CosmeticPickerv2_Advanced_C
     Cosmetics=&/EUW_CosmeticPickerv2.EUW_CosmeticPickerv2_C
     Global=&/EUW_CosmeticPickerV2_GlobalTab.EUW_CosmeticPickerV2_GlobalTab_C
     Poses=/Game/NeverCook/EditorUtilities/CosmeticTools/MannequinPoseWidget/EUW_MannequinPosing.EUW_MannequinPosing_C
    }}
   slot {class=HorizontalBoxSlot, properties {padding=<ltrb 5 0 5 5>, size {sizeRule=Fill, value=1}}}
  }
  Refresh {
   widgetClassPath=&EditorUtilityButton
   parent=CanvasPanel_0
   widget=Refresh
   properties {widgetStyle {
     clickedSlateSound {resourceObject=null}
     disabled {
      bIsDynamicallyLoaded=false
      drawAs=Image
      imageType=NoImage
      mirroring=NoMirror
      resourceName=None
      resourceObject=null
      tiling=NoTile
      imageSize=<32 32>
      margin=<ltrb 0 0 0 0>
      outlineSettings {
       bUseBrushTransparency=false
       roundingType=FixedRadius
       width=1
       color {colorUseRule=UseColor_ColorTable, specifiedColor=<rgba 0 0 0 0>}
       cornerRadii=<4 4 4 4>
      }
      tintColor {colorUseRule=UseColor_Specified, specifiedColor=<rgba 1 1 1 1>}
      uVRegion {bIsValid=false, max=<0 0>, min=<0 0>}
     }
     disabledForeground {colorUseRule=UseColor_ColorTable, specifiedColor=<rgba 0 0 0 0>}
     hovered {
      bIsDynamicallyLoaded=false
      drawAs=Image
      imageType=NoImage
      mirroring=NoMirror
      resourceName=None
      resourceObject=null
      tiling=NoTile
      imageSize=<32 32>
      margin=<ltrb 0 0 0 0>
      outlineSettings {
       bUseBrushTransparency=false
       roundingType=FixedRadius
       width=1
       color {colorUseRule=UseColor_ColorTable, specifiedColor=<rgba 0 0 0 0>}
       cornerRadii=<4 4 4 4>
      }
      tintColor {colorUseRule=UseColor_Specified, specifiedColor=<rgba 1 1 1 1>}
      uVRegion {bIsValid=false, max=<0 0>, min=<0 0>}
     }
     hoveredForeground {colorUseRule=UseColor_ColorTable, specifiedColor=<rgba 0 0 0 0>}
     hoveredSlateSound {resourceObject=null}
     normal {
      bIsDynamicallyLoaded=false
      drawAs=Image
      imageType=NoImage
      mirroring=NoMirror
      resourceName=None
      resourceObject=null
      tiling=NoTile
      imageSize=<32 32>
      margin=<ltrb 0 0 0 0>
      outlineSettings {
       bUseBrushTransparency=false
       roundingType=FixedRadius
       width=1
       color {colorUseRule=UseColor_ColorTable, specifiedColor=<rgba 0 0 0 0>}
       cornerRadii=<4 4 4 4>
      }
      tintColor {colorUseRule=UseColor_Specified, specifiedColor=<rgba 0.651 0.651 0.651 1>}
      uVRegion {bIsValid=false, max=<0 0>, min=<0 0>}
     }
     normalForeground {colorUseRule=UseColor_ColorTable, specifiedColor=<rgba 0 0 0 0>}
     normalPadding=<ltrb 12 1.5 12 1.5>
     pressed {
      bIsDynamicallyLoaded=false
      drawAs=Image
      imageType=NoImage
      mirroring=NoMirror
      resourceName=None
      resourceObject=null
      tiling=NoTile
      imageSize=<32 32>
      margin=<ltrb 0 0 0 0>
      outlineSettings {
       bUseBrushTransparency=false
       roundingType=FixedRadius
       width=1
       color {colorUseRule=UseColor_ColorTable, specifiedColor=<rgba 0 0 0 0>}
       cornerRadii=<4 4 4 4>
      }
      tintColor {colorUseRule=UseColor_Specified, specifiedColor=<rgba 0.5729 0.5729 0.5729 1>}
      uVRegion {bIsValid=false, max=<0 0>, min=<0 0>}
     }
     pressedForeground {colorUseRule=UseColor_ColorTable, specifiedColor=<rgba 0 0 0 0>}
     pressedPadding=<ltrb 12 2.5 12 0.5>
     pressedSlateSound {resourceObject=null}
    }}
   slot {
    class=CanvasPanelSlot
    properties {zOrder=1, layoutData {alignment=<0 0>, anchors {maximum=<1 0>, minimum=<1 0>}, offsets=<ltrb -36 4 20 20>}}
   }
  }
 ]
}
